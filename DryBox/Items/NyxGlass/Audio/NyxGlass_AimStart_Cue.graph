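# Sound cue for the start of aiming with the NyxGlass item, judging by its name.
# A Random node picks one of three AimStart wave variations with equal weight on each play.
# It plays through the Fort_UI sound class at a reduced volume multiplier, with no modulation or other processing nodes in the graph.
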SoundCue /DryBox/Items/NyxGlass/Audio/NyxGlass_AimStart_Cue
root=Random_0
nodes [
 {id=Random_0, children [WavePlayer_0,WavePlayer_1,WavePlayer_2], properties {weights [1,1,1]}}
 {id=WavePlayer_0, properties {soundWaveAssetPtr=&/NyxGlass_AimStart_01}}
 {id=WavePlayer_1, properties {soundWaveAssetPtr=&/NyxGlass_AimStart_02}}
 {id=WavePlayer_2, properties {soundWaveAssetPtr=&/NyxGlass_AimStart_03}}
]
properties {maxDistance=2097152, soundClassObject=&Fort_UI, volumeMultiplier=0.25}
