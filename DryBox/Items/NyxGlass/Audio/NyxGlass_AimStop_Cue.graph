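# Sound cue for the NyxGlass item's aim-stop sound, presumably played when aiming with it ends.
# A Random node picks one of four equally weighted wave variations, then a Modulator applies a slight random downward pitch shift.
# The cue plays at reduced volume in the Fort_UI sound class.
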SoundCue /DryBox/Items/NyxGlass/Audio/NyxGlass_AimStop_Cue
root=Modulator_0
nodes [
 {id=Modulator_0, children [Random_1], properties {pitchMax=1, pitchMin=0.9}}
 {id=Random_1, children [WavePlayer_3,WavePlayer_4,WavePlayer_5,WavePlayer_6], properties {weights [1,1,1,1]}}
 {id=WavePlayer_3, properties {soundWaveAssetPtr=&/NyxGlass_AimStop_01}}
 {id=WavePlayer_4, properties {soundWaveAssetPtr=&/NyxGlass_AimStop_02}}
 {id=WavePlayer_5, properties {soundWaveAssetPtr=&/NyxGlass_AimStop_03}}
 {id=WavePlayer_6, properties {soundWaveAssetPtr=&/NyxGlass_AimStop_04}}
]
properties {maxDistance=2097152, soundClassObject=&Fort_UI, volumeMultiplier=0.3}
