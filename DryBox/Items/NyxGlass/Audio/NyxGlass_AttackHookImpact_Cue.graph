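# Sound cue for the NyxGlass attack hook impact.
# It mixes three randomized layers (hook impact, debris, attack hook player) with pitch variation, plus volume variation on the latter two.
# IsLocal routes the mix to first-person FeyCrab attenuation when true, otherwise to IsTeam, where teammates get a quieter third-person NyxGlass path and others get it unreduced.
# Both attenuation paths share the mixer and use the Gadgets sound class.
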SoundCue /DryBox/Items/NyxGlass/Audio/NyxGlass_AttackHookImpact_Cue
root=Branch_4
nodes [
 {id=Attenuation_4, children [SoundClass_4], properties {attenuationSettings=/Game/Sounds/Attenuation/Fort_FeyCrab_Impact_1P_Att}}
 {id=Attenuation_5, children [SoundClass_5], properties {attenuationSettings=/Game/Sounds/Attenuation/Fort_NyxGlass_Impact_3P_Att}}
 {id=Branch_4, children [Attenuation_4, Branch_5, Branch_5], properties {boolParameterName=IsLocal}}
 {id=Branch_5, children [Modulator_3, Attenuation_5, Attenuation_5], properties {boolParameterName=IsTeam}}
 {id=Mixer_1, children [Modulator_0,Modulator_2,Modulator_4], properties {inputVolume [0.6,0.5,1.5]}}
 {id=Modulator_0, children [Random_2], properties {pitchMax=1.2}}
 {id=Modulator_2, children [Random_1], properties {pitchMax=1.2, pitchMin=0.8, volumeMax=0.9, volumeMin=0.4}}
 {id=Modulator_3, children [Attenuation_5], properties {pitchMax=1, pitchMin=1, volumeMax=0.4, volumeMin=0.4}}
 {id=Modulator_4, children [Random_3], properties {pitchMax=1.2, pitchMin=0.8, volumeMax=0.9, volumeMin=0.4}}
 {
  id=Random_1
  children [WavePlayer_6, WavePlayer_7, WavePlayer_8, WavePlayer_9, WavePlayer_10, WavePlayer_11]
  properties {bShouldExcludeFromBranchCulling=true, weights [1,1,1,1,1,1]}
 }
 {
  id=Random_2
  children [WavePlayer_0,WavePlayer_1,WavePlayer_2,WavePlayer_3,WavePlayer_4,WavePlayer_5]
  properties {bShouldExcludeFromBranchCulling=true, weights [1,1,1,1,1,1]}
 }
 {
  id=Random_3
  children [WavePlayer_13, WavePlayer_14, WavePlayer_15, WavePlayer_16]
  properties {bShouldExcludeFromBranchCulling=true, weights [1,1,1,1]}
 }
 {id=SoundClass_4, children [Mixer_1], properties {soundClassOverride=&Fort_Gadgets}}
 {id=SoundClass_5, children [Mixer_1], properties {soundClassOverride=&Fort_Gadgets}}
 {id=WavePlayer_0, properties {soundWaveAssetPtr=&/NyxGlass_HookImpact_01}}
 {id=WavePlayer_1, properties {soundWaveAssetPtr=&/NyxGlass_HookImpact_02}}
 {id=WavePlayer_10, properties {soundWaveAssetPtr=&/NyxGlass_HookImpact_Debris_05}}
 {id=WavePlayer_11, properties {soundWaveAssetPtr=&/NyxGlass_HookImpact_Debris_06}}
 {id=WavePlayer_13, properties {soundWaveAssetPtr=&/NyxGlass_AttackHookPlayer_01}}
 {id=WavePlayer_14, properties {soundWaveAssetPtr=&/NyxGlass_AttackHookPlayer_02}}
 {id=WavePlayer_15, properties {soundWaveAssetPtr=&/NyxGlass_AttackHookPlayer_03}}
 {id=WavePlayer_16, properties {soundWaveAssetPtr=&/NyxGlass_AttackHookPlayer_04}}
 {id=WavePlayer_2, properties {soundWaveAssetPtr=&/NyxGlass_HookImpact_03}}
 {id=WavePlayer_3, properties {soundWaveAssetPtr=&/NyxGlass_HookImpact_04}}
 {id=WavePlayer_4, properties {soundWaveAssetPtr=&/NyxGlass_HookImpact_05}}
 {id=WavePlayer_5, properties {soundWaveAssetPtr=&/NyxGlass_HookImpact_06}}
 {id=WavePlayer_6, properties {soundWaveAssetPtr=&/NyxGlass_HookImpact_Debris_01}}
 {id=WavePlayer_7, properties {soundWaveAssetPtr=&/NyxGlass_HookImpact_Debris_02}}
 {id=WavePlayer_8, properties {soundWaveAssetPtr=&/NyxGlass_HookImpact_Debris_03}}
 {id=WavePlayer_9, properties {soundWaveAssetPtr=&/NyxGlass_HookImpact_Debris_04}}
]
properties {maxDistance=11000, volumeMultiplier=0.8}
