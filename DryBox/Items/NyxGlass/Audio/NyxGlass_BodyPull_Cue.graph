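# Sound cue for the NyxGlass item's body-pull sound, mixed differently for the local player, teammates and others.
# All paths share an equal-weight random pick of three wave players, with randomized lowered pitch and volume and a short random delay.
# IsLocal routes to the gadget sound class without an attenuation node; otherwise audio uses the enemy third-person foley class and boost attenuation.
# IsTeam sharply cuts the volume for teammates.
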SoundCue /DryBox/Items/NyxGlass/Audio/NyxGlass_BodyPull_Cue
root=Branch_4
nodes [
 {id=Attenuation_5, children [SoundClass_5], properties {attenuationSettings=/Game/Sounds/Attenuation/Fort_NyxGlass_Boost_3P_Att}}
 {id=Branch_4, children [SoundClass_4,Branch_5,Branch_5], properties {boolParameterName=IsLocal}}
 {id=Branch_5, children [Modulator_3, Attenuation_5, Attenuation_5], properties {boolParameterName=IsTeam}}
 {id=Delay_0, children [Modulator_9], properties {delayMax=0.1, delayMin=0.075}}
 {id=Modulator_3, children [Attenuation_5], properties {pitchMax=1, pitchMin=1, volumeMax=0.1, volumeMin=0.1}}
 {id=Modulator_9, children [Random_0], properties {pitchMax=0.85, pitchMin=0.7, volumeMax=0.8, volumeMin=0.6}}
 {id=Random_0, children [WavePlayer_0,WavePlayer_1,WavePlayer_2], properties {weights [1,1,1]}}
 {id=SoundClass_4, children [Delay_0], properties {soundClassOverride=&Fort_Gadgets_Items}}
 {id=SoundClass_5, children [Delay_0], properties {soundClassOverride=&Fort_Player_Foley_3P_Enemy_Athena}}
 {id=WavePlayer_0, properties {soundWaveAssetPtr=&/NyxGlass_BodyPull_01}}
 {id=WavePlayer_1, properties {soundWaveAssetPtr=&/NyxGlass_BodyPull_03}}
 {id=WavePlayer_2, properties {soundWaveAssetPtr=&NyxGlass_BodyPull_03}}
]
properties {
 maxDistance=11000
 soundClassObject=&Fort_Gadgets_Items
 attenuationOverrides {falloffDistance=5000, stereoSpread=2000, attenuationShapeExtents {x=800}}
}
