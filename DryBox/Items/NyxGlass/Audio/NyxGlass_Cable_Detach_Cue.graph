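# Sound cue for the NyxGlass item's cable detach sound, routed by listener relationship.
# An IsLocal branch picks first-person attenuation; otherwise an IsTeam branch feeds grapple-glove retract third-person attenuation, with teammates' sounds played much quieter.
# Both paths use the Gadgets sound class and share slight pitch and volume variation.
# A quality-level switch selects a random wave from either a ten-variant pool or a five-variant subset.
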SoundCue /DryBox/Items/NyxGlass/Audio/NyxGlass_Cable_Detach_Cue
root=Branch_4
nodes [
 {id=Attenuation_4, children [SoundClass_4], properties {attenuationSettings=/Game/Sounds/Attenuation/Fort_NyxGlass_1P_Detach_Athena_Att}}
 {id=Attenuation_5, children [SoundClass_5], properties {attenuationSettings=/Game/Sounds/Attenuation/Fort_GrappleGloves_Retract_3P_Att}}
 {id=Branch_4, children [Attenuation_4, Branch_5, Branch_5], properties {boolParameterName=IsLocal}}
 {id=Branch_5, children [Modulator_0, Attenuation_5, Attenuation_5], properties {boolParameterName=IsTeam}}
 {id=Modulator_0, children [Attenuation_5], properties {pitchMax=1, pitchMin=1, volumeMax=0.3, volumeMin=0.3}}
 {id=Modulator_1, children [QualityLevel_0], properties {pitchMax=1, pitchMin=0.85, volumeMin=0.8}}
 {id=QualityLevel_0, children [Random_0,Random_0,Random_1,Random_1]}
 {
  id=Random_0
  children [WavePlayer_0,WavePlayer_1,WavePlayer_2,WavePlayer_3,WavePlayer_4,WavePlayer_5,WavePlayer_6,WavePlayer_7,WavePlayer_8,WavePlayer_9]
  properties {weights [1,1,1,1,1,1,1,1,1,1]}
 }
 {
  id=Random_1
  children [WavePlayer_10, WavePlayer_11, WavePlayer_12, WavePlayer_13, WavePlayer_14]
  properties {bShouldExcludeFromBranchCulling=true, weights [1,1,1,1,1]}
 }
 {id=SoundClass_4, children [Modulator_1], properties {soundClassOverride=&Fort_Gadgets}}
 {id=SoundClass_5, children [Modulator_1], properties {soundClassOverride=&Fort_Gadgets}}
 {id=WavePlayer_0, properties {soundWaveAssetPtr=&/NyxGlass_Cable_Detach_01}}
 {id=WavePlayer_1, properties {soundWaveAssetPtr=&/NyxGlass_Cable_Detach_02}}
 {id=WavePlayer_10, properties {soundWaveAssetPtr=&/NyxGlass_Cable_Detach_07}}
 {id=WavePlayer_11, properties {soundWaveAssetPtr=&/NyxGlass_Cable_Detach_09}}
 {id=WavePlayer_12, properties {soundWaveAssetPtr=&/NyxGlass_Cable_Detach_05}}
 {id=WavePlayer_13, properties {soundWaveAssetPtr=&/NyxGlass_Cable_Detach_03}}
 {id=WavePlayer_14, properties {soundWaveAssetPtr=&NyxGlass_Cable_Detach_01}}
 {id=WavePlayer_2, properties {soundWaveAssetPtr=&NyxGlass_Cable_Detach_03}}
 {id=WavePlayer_3, properties {soundWaveAssetPtr=&/NyxGlass_Cable_Detach_04}}
 {id=WavePlayer_4, properties {soundWaveAssetPtr=&NyxGlass_Cable_Detach_05}}
 {id=WavePlayer_5, properties {soundWaveAssetPtr=&/NyxGlass_Cable_Detach_06}}
 {id=WavePlayer_6, properties {soundWaveAssetPtr=&NyxGlass_Cable_Detach_07}}
 {id=WavePlayer_7, properties {soundWaveAssetPtr=&/NyxGlass_Cable_Detach_08}}
 {id=WavePlayer_8, properties {soundWaveAssetPtr=&NyxGlass_Cable_Detach_09}}
 {id=WavePlayer_9, properties {soundWaveAssetPtr=&/NyxGlass_Cable_Detach_10}}
]
properties {maxDistance=9000, volumeMultiplier=0.5}
