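# Sound cue for the NyxGlass item's cannot-fire sound, with separate local, teammate and other-player mixes.
# IsLocal selects gadget attenuation and class, otherwise third-person grappling attenuation with enemy foley class, which IsTeam lowers in volume.
# All paths share a short fixed delay, volume variation and quality switch between random pools of eight or five waves.
# Per-owner concurrency blocks new plays beyond two and sets a retrigger time.
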SoundCue /DryBox/Items/NyxGlass/Audio/NyxGlass_CannotFire_Cue
root=Branch_4
nodes [
 {id=Attenuation_4, children [SoundClass_4], properties {attenuationSettings=/Game/Sounds/Attenuation/Fort_GadgetGrappling_Att}}
 {id=Attenuation_5, children [SoundClass_5], properties {attenuationSettings=/Game/Sounds/Attenuation/Fort_Grappling_Movement_3P_Att}}
 {id=Branch_4, children [Attenuation_4, Branch_5, Branch_5], properties {boolParameterName=IsLocal}}
 {id=Branch_5, children [Modulator_3, Attenuation_5, Attenuation_5], properties {boolParameterName=IsTeam}}
 {id=Delay_0, children [QualityLevel_0], properties {delayMax=0.3, delayMin=0.3}}
 {id=Modulator_0, children [Delay_0], properties {pitchMax=1, pitchMin=1, volumeMin=0.7}}
 {id=Modulator_3, children [Attenuation_5], properties {pitchMax=1, pitchMin=1, volumeMax=0.2, volumeMin=0.2}}
 {id=QualityLevel_0, children [Random_4,Random_4,Random_5,Random_5]}
 {
  id=Random_4
  children [WavePlayer_31, WavePlayer_32, WavePlayer_33, WavePlayer_34, WavePlayer_35, WavePlayer_36, WavePlayer_37, WavePlayer_38]
  properties {weights [1,1,1,1,1,1,1,1]}
 }
 {
  id=Random_5
  children [WavePlayer_39, WavePlayer_40, WavePlayer_41, WavePlayer_42, WavePlayer_43]
  properties {bShouldExcludeFromBranchCulling=true, weights [1,1,1,1,1]}
 }
 {id=SoundClass_4, children [Modulator_0], properties {soundClassOverride=&Fort_Gadgets_Items}}
 {
  id=SoundClass_5
  children [Modulator_0]
  properties {soundClassOverride=&Fort_Player_Foley_3P_Enemy_Athena}
 }
 {id=WavePlayer_31, properties {soundWaveAssetPtr=&/NyxGlass_CannotFire_01}}
 {id=WavePlayer_32, properties {soundWaveAssetPtr=&/NyxGlass_CannotFire_02}}
 {id=WavePlayer_33, properties {soundWaveAssetPtr=&/NyxGlass_CannotFire_03}}
 {id=WavePlayer_34, properties {soundWaveAssetPtr=&/NyxGlass_CannotFire_04}}
 {id=WavePlayer_35, properties {soundWaveAssetPtr=&/NyxGlass_CannotFire_05}}
 {id=WavePlayer_36, properties {soundWaveAssetPtr=&/NyxGlass_CannotFire_06}}
 {id=WavePlayer_37, properties {soundWaveAssetPtr=&/NyxGlass_CannotFire_07}}
 {id=WavePlayer_38, properties {soundWaveAssetPtr=&/NyxGlass_CannotFire_08}}
 {id=WavePlayer_39, properties {soundWaveAssetPtr=&NyxGlass_CannotFire_01}}
 {id=WavePlayer_40, properties {soundWaveAssetPtr=&NyxGlass_CannotFire_03}}
 {id=WavePlayer_41, properties {soundWaveAssetPtr=&NyxGlass_CannotFire_05}}
 {id=WavePlayer_42, properties {soundWaveAssetPtr=&NyxGlass_CannotFire_07}}
 {id=WavePlayer_43, properties {soundWaveAssetPtr=&NyxGlass_CannotFire_08}}
]
properties {
 attenuationSettings=&Fort_GadgetGrappling_Att
 bOverrideConcurrency=true
 maxDistance=8500
 soundClassObject=&Fort_Gadgets
 volumeMultiplier=0.25
 concurrencyOverrides {bLimitToOwner=true, maxCount=2, resolutionRule=PreventNew, retriggerTime=0.3, volumeScaleAttackTime=0.1}
}
