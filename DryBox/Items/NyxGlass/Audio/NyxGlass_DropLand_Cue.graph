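# Sound cue for the NyxGlass item's drop-landing sound, assigned to the Fort_Loot_Collect sound class.
# A weighted Random node, excluded from branch culling, picks one of four DropLand waves, with variation 01 weighted slightly lower.
# A root Modulator adds slight pitch and volume variation to each play.
# The cue overrides attenuation with its own falloff distance and slightly lowers overall volume.
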SoundCue /DryBox/Items/NyxGlass/Audio/NyxGlass_DropLand_Cue
root=Modulator_0
nodes [
 {id=Modulator_0, children [Random_0], properties {pitchMax=1.1, pitchMin=0.9, volumeMax=1.1, volumeMin=0.9}}
 {
  id=Random_0
  children [WavePlayer_0,WavePlayer_1,WavePlayer_4,WavePlayer_5]
  properties {bShouldExcludeFromBranchCulling=true, weights [1,0.7,1,1]}
 }
 {id=WavePlayer_0, properties {soundWaveAssetPtr=&/NyxGlass_DropLand_04}}
 {id=WavePlayer_1, properties {soundWaveAssetPtr=&/NyxGlass_DropLand_01}}
 {id=WavePlayer_4, properties {soundWaveAssetPtr=&/NyxGlass_DropLand_02}}
 {id=WavePlayer_5, properties {soundWaveAssetPtr=&/NyxGlass_DropLand_03}}
]
properties {
 bOverrideAttenuation=true
 maxDistance=1000
 soundClassObject=&Fort_Loot_Collect
 volumeMultiplier=0.9
 attenuationOverrides {falloffDistance=1000, attenuationShapeExtents {x=0}}
}
