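# Sound cue for the NyxGlass item's drop sound, using weapon drop attenuation and the loot collect sound class.
# A Random node picks one of four evenly weighted drop variations and is excluded from branch culling.
# A Mixer layers that pick with a quieter fixed drop layer from the FromGrappleGlove folder.
# A root Modulator at default settings processes the mix, and the cue plays at reduced volume.
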SoundCue /DryBox/Items/NyxGlass/Audio/NyxGlass_Drop_Cue
root=Modulator_0
nodes [
 {id=Mixer_0, children [Random_0,WavePlayer_5], properties {inputVolume [1,0.6]}}
 {id=Modulator_0, children [Mixer_0]}
 {
  id=Random_0
  children [WavePlayer_0,WavePlayer_1,WavePlayer_6,WavePlayer_7]
  properties {bShouldExcludeFromBranchCulling=true, weights [1,1,1,1]}
 }
 {id=WavePlayer_0, properties {soundWaveAssetPtr=&/NyxGlass_Drop_01}}
 {id=WavePlayer_1, properties {soundWaveAssetPtr=&/NyxGlass_Drop_02}}
 {id=WavePlayer_5, properties {soundWaveAssetPtr=/DryBox/Items/NyxGlass/Audio/FromGrappleGlove/NyxGlass_GG_Drop_Layer_01}}
 {id=WavePlayer_6, properties {soundWaveAssetPtr=&/NyxGlass_Drop_03}}
 {id=WavePlayer_7, properties {soundWaveAssetPtr=&/NyxGlass_Drop_04}}
]
properties {
 attenuationSettings=&Fort_Weapon_Drop_Att
 maxDistance=3000
 soundClassObject=&Fort_Loot_Collect
 volumeMultiplier=0.4
}
