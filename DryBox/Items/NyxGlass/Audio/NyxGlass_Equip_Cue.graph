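# Sound cue for equipping the NyxGlass item, routed through the weapon foley sound class with weapon reload attenuation.
# A Random node picks one of five equally weighted equip wave files, then a Modulator applies slight pitch variation.
# Volume is reduced, and overridden concurrency allows one instance per owner, stopping the oldest when a new one plays.
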SoundCue /DryBox/Items/NyxGlass/Audio/NyxGlass_Equip_Cue
root=Modulator_0
nodes [
 {id=Modulator_0, children [Random_0], properties {pitchMax=1.1, pitchMin=0.9}}
 {id=Random_0, children [WavePlayer_0,WavePlayer_1,WavePlayer_5,WavePlayer_6,WavePlayer_7], properties {weights [1,1,1,1,1]}}
 {id=WavePlayer_0, properties {soundWaveAssetPtr=&/NyxGlass_Equip_01}}
 {id=WavePlayer_1, properties {soundWaveAssetPtr=&/NyxGlass_Equip_02}}
 {id=WavePlayer_5, properties {soundWaveAssetPtr=&/NyxGlass_Equip_03}}
 {id=WavePlayer_6, properties {soundWaveAssetPtr=&/NyxGlass_Equip_04}}
 {id=WavePlayer_7, properties {soundWaveAssetPtr=&/NyxGlass_Equip_05}}
]
properties {
 attenuationSettings=&Fort_Weapon_Reload_Att
 bOverrideConcurrency=true
 maxDistance=2250
 soundClassObject=&Fort_WeaponFoley
 volumeMultiplier=0.301
 concurrencyOverrides {bLimitToOwner=true, maxCount=1, resolutionRule=StopOldest}
}
