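# Sound cue for the NyxGlass item's fire sound, with different playback for local, teammate and other players.
# An IsLocal branch selects close projectile attenuation, otherwise an IsTeam branch selects third-person boost attenuation, with teammate volume sharply reduced.
# All paths share a sound class override and a mixer layering a random fire wave with a quieter, slightly delayed random rope wave varied in pitch and volume.
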SoundCue /DryBox/Items/NyxGlass/Audio/NyxGlass_Fire_Cue
root=Branch_4
nodes [
 {id=Attenuation_4, children [SoundClass_4], properties {attenuationSettings=&Fort_Weapon_Projectile_Close_Athena_Att}}
 {id=Attenuation_5, children [SoundClass_5], properties {attenuationSettings=/Game/Sounds/Attenuation/Fort_NyxGlass_Boost_3P_Att}}
 {id=Branch_4, children [Attenuation_4, Branch_5, Branch_5], properties {boolParameterName=IsLocal}}
 {id=Branch_5, children [Modulator_3, Attenuation_5, Attenuation_5], properties {boolParameterName=IsTeam}}
 {id=Delay_0, children [Random_1], properties {delayMax=0.1, delayMin=0.05}}
 {id=Mixer_0, children [Modulator_1,Modulator_2], properties {inputVolume [1,0.4]}}
 {id=Modulator_1, children [Random_0]}
 {id=Modulator_2, children [Delay_0], properties {pitchMax=1.2, pitchMin=0.8, volumeMax=0.95, volumeMin=0.6}}
 {id=Modulator_3, children [Attenuation_5], properties {pitchMax=1, pitchMin=1, volumeMax=0.2, volumeMin=0.2}}
 {
  id=Random_0
  children [WavePlayer_0,WavePlayer_1,WavePlayer_2,WavePlayer_3,WavePlayer_4,WavePlayer_5]
  properties {bShouldExcludeFromBranchCulling=true, weights [1,1,1,1,1,1]}
 }
 {
  id=Random_1
  children [WavePlayer_6, WavePlayer_7, WavePlayer_8, WavePlayer_9, WavePlayer_10, WavePlayer_11]
  properties {weights [1,1,1,1,1,1]}
 }
 {id=SoundClass_4, children [Mixer_0], properties {soundClassOverride=&Fort_WeaponRanged_Fire_3P_PSM}}
 {id=SoundClass_5, children [Mixer_0], properties {soundClassOverride=&Fort_WeaponRanged_Fire_3P_PSM}}
 {id=WavePlayer_0, properties {soundWaveAssetPtr=&/NyxGlass_Fire_06}}
 {id=WavePlayer_1, properties {soundWaveAssetPtr=&/NyxGlass_Fire_01}}
 {id=WavePlayer_10, properties {soundWaveAssetPtr=&/NyxGlass_Fire_Rope_04}}
 {id=WavePlayer_11, properties {soundWaveAssetPtr=&/NyxGlass_Fire_Rope_05}}
 {id=WavePlayer_2, properties {soundWaveAssetPtr=&/NyxGlass_Fire_02}}
 {id=WavePlayer_3, properties {soundWaveAssetPtr=&/NyxGlass_Fire_03}}
 {id=WavePlayer_4, properties {soundWaveAssetPtr=&/NyxGlass_Fire_04}}
 {id=WavePlayer_5, properties {soundWaveAssetPtr=&/NyxGlass_Fire_05}}
 {id=WavePlayer_6, properties {soundWaveAssetPtr=&/NyxGlass_Fire_Rope_06}}
 {id=WavePlayer_7, properties {soundWaveAssetPtr=&/NyxGlass_Fire_Rope_01}}
 {id=WavePlayer_8, properties {soundWaveAssetPtr=&/NyxGlass_Fire_Rope_02}}
 {id=WavePlayer_9, properties {soundWaveAssetPtr=&/NyxGlass_Fire_Rope_03}}
]
properties {attenuationSettings=/Game/Sounds/Attenuation/Fort_Weapon_Projectile_Grappling_Att, maxDistance=11000, volumeMultiplier=0.6}
