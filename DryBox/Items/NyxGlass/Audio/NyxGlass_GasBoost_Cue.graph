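# Sound cue for the NyxGlass gas boost, branching on IsLocal and then IsTeam to separate local and third-person playback.
# The local path crossfades randomized boost and empty-tank samples by the GasLevel parameter, with a quality-level node choosing larger or smaller variation pools.
# Other players hear a short-delayed, pitch-varied boost from a five-sample pool with third-person attenuation and no gas response, heavily quieted for teammates.
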SoundCue /DryBox/Items/NyxGlass/Audio/NyxGlass_GasBoost_Cue
root=Branch_4
nodes [
 {id=Attenuation_4, children [SoundClass_4], properties {attenuationSettings=/Game/Sounds/Attenuation/Fort_GadgetGrappling_Att}}
 {id=Attenuation_5, children [SoundClass_5], properties {attenuationSettings=/Game/Sounds/Attenuation/Fort_NyxGlass_Boost_3P_Att}}
 {id=Branch_4, children [Attenuation_4, Branch_5, Branch_5], properties {boolParameterName=IsLocal}}
 {id=Branch_5, children [Modulator_3, Attenuation_5, Attenuation_5], properties {boolParameterName=IsTeam}}
 {id=Delay_0, children [QualityLevel_0], properties {delayMax=0.3, delayMin=0.3}}
 {id=Delay_1, children [Modulator_1], properties {delayMax=0.2, delayMin=0.2}}
 {id=Mixer_0, children [ModulatorContinuous_2, ModulatorContinuous_3], properties {inputVolume [0.8,1.7]}}
 {id=Mixer_1, children [ModulatorContinuous_4, ModulatorContinuous_5], properties {inputVolume [0.8,1.7]}}
 {
  id=ModulatorContinuous_2
  children [Random_0]
  properties {pitchModulationParams {parameterName=GasLevel}, volumeModulationParams {maxInput=100, parameterName=GasLevel}}
 }
 {
  id=ModulatorContinuous_3
  children [Random_1]
  properties {
   pitchModulationParams {parameterName=GasLevel}
   volumeModulationParams {maxInput=65, maxOutput=0, minInput=5, minOutput=1, parameterName=GasLevel}
  }
 }
 {
  id=ModulatorContinuous_4
  children [Random_2]
  properties {pitchModulationParams {parameterName=GasLevel}, volumeModulationParams {maxInput=100, parameterName=GasLevel}}
 }
 {
  id=ModulatorContinuous_5
  children [Random_3]
  properties {
   pitchModulationParams {parameterName=GasLevel}
   volumeModulationParams {maxInput=65, maxOutput=0, minInput=5, minOutput=1, parameterName=GasLevel}
  }
 }
 {id=Modulator_0, children [Delay_0], properties {volumeMax=0.45, volumeMin=0.4}}
 {id=Modulator_1, children [Random_5], properties {pitchMax=1.1, pitchMin=0.9, volumeMax=0.8, volumeMin=0.7}}
 {id=Modulator_3, children [Attenuation_5], properties {pitchMax=1, pitchMin=1, volumeMax=0.1, volumeMin=0.1}}
 {id=QualityLevel_0, children [Mixer_0,Mixer_0,Mixer_1,Mixer_1]}
 {
  id=Random_0
  children [
   WavePlayer_0
   WavePlayer_1
   WavePlayer_2
   WavePlayer_3
   WavePlayer_4
   WavePlayer_5
   WavePlayer_6
   WavePlayer_7
   WavePlayer_8
   WavePlayer_9
   WavePlayer_10
   WavePlayer_11
  ]
  properties {weights [1,1,1,1,1,1,1,1,1,1,1,1]}
 }
 {
  id=Random_1
  children [WavePlayer_12, WavePlayer_16, WavePlayer_17, WavePlayer_18, WavePlayer_19, WavePlayer_20, WavePlayer_23, WavePlayer_24, WavePlayer_25]
  properties {weights [1,1,1,1,1,1,1,1,1]}
 }
 {
  id=Random_2
  children [WavePlayer_13, WavePlayer_14, WavePlayer_15, WavePlayer_21, WavePlayer_22]
  properties {bShouldExcludeFromBranchCulling=true, weights [1,1,1,1,1]}
 }
 {
  id=Random_3
  children [WavePlayer_30, WavePlayer_29, WavePlayer_28, WavePlayer_27, WavePlayer_26]
  properties {bShouldExcludeFromBranchCulling=true, weights [1,1,1,1,1]}
 }
 {
  id=Random_5
  children [WavePlayer_43, WavePlayer_44, WavePlayer_45, WavePlayer_46, WavePlayer_47]
  properties {bShouldExcludeFromBranchCulling=true, weights [1,1,1,1,1]}
 }
 {id=SoundClass_4, children [Modulator_0], properties {soundClassOverride=&Fort_Gadgets_Items}}
 {id=SoundClass_5, children [Delay_1], properties {soundClassOverride=&Fort_Player_Foley_3P_Enemy_Athena}}
 {id=WavePlayer_0, properties {soundWaveAssetPtr=&/NyxGlass_GasBoost_12}}
 {id=WavePlayer_1, properties {soundWaveAssetPtr=&/NyxGlass_GasBoost_01}}
 {id=WavePlayer_10, properties {soundWaveAssetPtr=&/NyxGlass_GasBoost_10}}
 {id=WavePlayer_11, properties {soundWaveAssetPtr=&/NyxGlass_GasBoost_11}}
 {id=WavePlayer_12, properties {soundWaveAssetPtr=&/NyxGlass_GasBoost_Empty_01}}
 {id=WavePlayer_13, properties {soundWaveAssetPtr=&NyxGlass_GasBoost_01}}
 {id=WavePlayer_14, properties {soundWaveAssetPtr=&/NyxGlass_GasBoost_02}}
 {id=WavePlayer_15, properties {soundWaveAssetPtr=&/NyxGlass_GasBoost_03}}
 {id=WavePlayer_16, properties {soundWaveAssetPtr=&/NyxGlass_GasBoost_Empty_02}}
 {id=WavePlayer_17, properties {soundWaveAssetPtr=&/NyxGlass_GasBoost_Empty_03}}
 {id=WavePlayer_18, properties {soundWaveAssetPtr=&/NyxGlass_GasBoost_Empty_04}}
 {id=WavePlayer_19, properties {soundWaveAssetPtr=&/NyxGlass_GasBoost_Empty_05}}
 {id=WavePlayer_2, properties {soundWaveAssetPtr=&NyxGlass_GasBoost_02}}
 {id=WavePlayer_20, properties {soundWaveAssetPtr=&/NyxGlass_GasBoost_Empty_06}}
 {id=WavePlayer_21, properties {soundWaveAssetPtr=&/NyxGlass_GasBoost_09}}
 {id=WavePlayer_22, properties {soundWaveAssetPtr=&NyxGlass_GasBoost_10}}
 {id=WavePlayer_23, properties {soundWaveAssetPtr=&/NyxGlass_GasBoost_Empty_07}}
 {id=WavePlayer_24, properties {soundWaveAssetPtr=&/NyxGlass_GasBoost_Empty_08}}
 {id=WavePlayer_25, properties {soundWaveAssetPtr=&/NyxGlass_GasBoost_Empty_09}}
 {id=WavePlayer_26, properties {soundWaveAssetPtr=&NyxGlass_GasBoost_Empty_08}}
 {id=WavePlayer_27, properties {soundWaveAssetPtr=&NyxGlass_GasBoost_Empty_06}}
 {id=WavePlayer_28, properties {soundWaveAssetPtr=&NyxGlass_GasBoost_Empty_04}}
 {id=WavePlayer_29, properties {soundWaveAssetPtr=&NyxGlass_GasBoost_Empty_02}}
 {id=WavePlayer_3, properties {soundWaveAssetPtr=&NyxGlass_GasBoost_03}}
 {id=WavePlayer_30, properties {soundWaveAssetPtr=&NyxGlass_GasBoost_Empty_01}}
 {id=WavePlayer_4, properties {soundWaveAssetPtr=&/NyxGlass_GasBoost_04}}
 {id=WavePlayer_43, properties {soundWaveAssetPtr=&NyxGlass_GasBoost_01}}
 {id=WavePlayer_44, properties {soundWaveAssetPtr=&NyxGlass_GasBoost_04}}
 {id=WavePlayer_45, properties {soundWaveAssetPtr=&/NyxGlass_GasBoost_06}}
 {id=WavePlayer_46, properties {soundWaveAssetPtr=&/NyxGlass_GasBoost_08}}
 {id=WavePlayer_47, properties {soundWaveAssetPtr=&NyxGlass_GasBoost_09}}
 {id=WavePlayer_5, properties {soundWaveAssetPtr=&/NyxGlass_GasBoost_05}}
 {id=WavePlayer_6, properties {soundWaveAssetPtr=&NyxGlass_GasBoost_06}}
 {id=WavePlayer_7, properties {soundWaveAssetPtr=&/NyxGlass_GasBoost_07}}
 {id=WavePlayer_8, properties {soundWaveAssetPtr=&NyxGlass_GasBoost_08}}
 {id=WavePlayer_9, properties {soundWaveAssetPtr=&NyxGlass_GasBoost_09}}
]
properties {
 attenuationSettings=&Fort_GadgetGrappling_Att
 maxDistance=11000
 soundClassObject=&Fort_Gadgets
 volumeMultiplier=0.7
}
