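# Sound cue for the NyxGlass item that, by its name, serves as the UI sound for gas being replenished.
# A single wave player is the root and plays the NyxGlass_GasReplenished_UI_01 wave, with no randomization or modulation.
# The cue routes through the Fort_UI sound class at reduced volume.
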SoundCue /DryBox/Items/NyxGlass/Audio/NyxGlass_GasReplenished_UI_01_Cue
root=WavePlayer_0
nodes [{id=WavePlayer_0, properties {soundWaveAssetPtr=&/NyxGlass_GasReplenished_UI_01}}]
properties {maxDistance=2097152, soundClassObject=&Fort_UI, volumeMultiplier=0.5}
