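# Sound cue for the NyxGlass item's hover, playing one of three equally weighted looping waves.
# An IsLocal branch gives the local player a gadget-class path with slight volume randomization and no attenuation node.
# Otherwise, an IsTeam branch feeds a shared grappling-movement 3P attenuation and enemy-foley sound class, with teammates turned well down.
# Unset parameters follow the non-local, non-team route, and all paths share the random node, which is excluded from branch culling.
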SoundCue /DryBox/Items/NyxGlass/Audio/NyxGlass_Hover_Cue
root=Branch_4
nodes [
 {id=Attenuation_5, children [SoundClass_5], properties {attenuationSettings=/Game/Sounds/Attenuation/Fort_Grappling_Movement_3P_Att}}
 {id=Branch_4, children [SoundClass_4,Branch_5,Branch_5], properties {boolParameterName=IsLocal}}
 {id=Branch_5, children [Modulator_3, Attenuation_5, Attenuation_5], properties {boolParameterName=IsTeam}}
 {id=Modulator_1, children [Random_1], properties {volumeMax=0.7, volumeMin=0.6}}
 {id=Modulator_3, children [Attenuation_5], properties {pitchMax=1, pitchMin=1, volumeMax=0.2, volumeMin=0.2}}
 {id=Random_1, children [WavePlayer_5,WavePlayer_6,WavePlayer_7], properties {bShouldExcludeFromBranchCulling=true, weights [1,1,1]}}
 {id=SoundClass_4, children [Modulator_1], properties {soundClassOverride=&Fort_Gadgets}}
 {
  id=SoundClass_5
  children [Random_1]
  properties {soundClassOverride=&Fort_Player_Foley_3P_Enemy_Athena}
 }
 {id=WavePlayer_5, properties {bLooping=true, soundWaveAssetPtr=&/NyxGlass_Hover_Loop_01}}
 {id=WavePlayer_6, properties {bLooping=true, soundWaveAssetPtr=&/NyxGlass_Hover_Loop_02}}
 {id=WavePlayer_7, properties {bLooping=true, soundWaveAssetPtr=&/NyxGlass_Hover_Loop_03}}
]
properties {maxDistance=6000, volumeMultiplier=1.5}
