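# Sound cue for picking up the NyxGlass item.
# A Random node picks one of four equally weighted pickup waves and feeds it through a Modulator at the root.
# It plays at reduced volume with the shared weapon pickup attenuation and the loot collect sound class.
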SoundCue /DryBox/Items/NyxGlass/Audio/NyxGlass_Pickup_Cue
root=Modulator_0
nodes [
 {id=Modulator_0, children [Random_0]}
 {id=Random_0, children [WavePlayer_0,WavePlayer_1,WavePlayer_5,WavePlayer_6], properties {weights [1,1,1,1]}}
 {id=WavePlayer_0, properties {soundWaveAssetPtr=&/NyxGlass_Pickup_01}}
 {id=WavePlayer_1, properties {soundWaveAssetPtr=&/NyxGlass_Pickup_02}}
 {id=WavePlayer_5, properties {soundWaveAssetPtr=&/NyxGlass_Pickup_03}}
 {id=WavePlayer_6, properties {soundWaveAssetPtr=&/NyxGlass_Pickup_04}}
]
properties {
 attenuationSettings=&Fort_Weapon_Pickup_Att
 maxDistance=3500
 soundClassObject=&Fort_Loot_Collect
 volumeMultiplier=0.398
}
