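# Sound cue for the NyxGlass item's release spin, playing one of seven equally weighted wave variations.
# A root modulator varies volume, then an IsLocal branch picks first-person attenuation and foley class, or a third-person path.
# The third-person path branches on IsTeam, playing teammates much quieter than others, with both sharing third-person attenuation and the enemy foley class.
# Concurrency is limited per owner, stopping the oldest instance.
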SoundCue /DryBox/Items/NyxGlass/Audio/NyxGlass_Release_Spin_Cue
root=Modulator_0
nodes [
 {
  id=Attenuation_4
  children [SoundClass_4]
  properties {attenuationSettings=/Game/Sounds/Fort_Player/Footsteps/Settings/Fort_Player_FS_1P_Att}
 }
 {
  id=Attenuation_5
  children [SoundClass_5]
  properties {attenuationSettings=/Game/Sounds/Fort_Player/Footsteps/Settings/Fort_Player_FS_3P_Att}
 }
 {id=Branch_4, children [Attenuation_4, Branch_5, Branch_5], properties {boolParameterName=IsLocal}}
 {id=Branch_5, children [Modulator_3,Modulator_1,Modulator_1], properties {boolParameterName=IsTeam}}
 {id=Modulator_0, children [Branch_4], properties {volumeMin=0.6}}
 {id=Modulator_1, children [Attenuation_5], properties {pitchMax=1, pitchMin=1, volumeMax=0.6, volumeMin=0.6}}
 {id=Modulator_3, children [Attenuation_5], properties {pitchMax=1, pitchMin=1, volumeMax=0.1, volumeMin=0.1}}
 {
  id=Random_1
  children [WavePlayer_8, WavePlayer_9, WavePlayer_10, WavePlayer_11, WavePlayer_12, WavePlayer_13, WavePlayer_14]
  properties {weights [1,1,1,1,1,1,1]}
 }
 {id=SoundClass_4, children [Random_1], properties {soundClassOverride=&Fort_Player_Foley_1P_Athena}}
 {
  id=SoundClass_5
  children [Random_1]
  properties {soundClassOverride=&Fort_Player_Foley_3P_Enemy_Athena}
 }
 {id=WavePlayer_10, properties {soundWaveAssetPtr=&/NyxGlass_Release_Spin_03}}
 {id=WavePlayer_11, properties {soundWaveAssetPtr=&/NyxGlass_Release_Spin_04}}
 {id=WavePlayer_12, properties {soundWaveAssetPtr=&/NyxGlass_Release_Spin_05}}
 {id=WavePlayer_13, properties {soundWaveAssetPtr=&/NyxGlass_Release_Spin_06}}
 {id=WavePlayer_14, properties {soundWaveAssetPtr=&/NyxGlass_Release_Spin_07}}
 {id=WavePlayer_8, properties {soundWaveAssetPtr=&/NyxGlass_Release_Spin_01}}
 {id=WavePlayer_9, properties {soundWaveAssetPtr=&/NyxGlass_Release_Spin_02}}
]
properties {
 maxDistance=3584
 volumeMultiplier=2
 concurrencyOverrides {bLimitToOwner=true, maxCount=4, resolutionRule=StopOldest, volumeScaleAttackTime=0.2}
}
