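# Sound cue for the short spin sound on release of the NyxGlass item, picked from seven equally weighted wave variants.
# An IsLocal branch sends the local player to first-person foley class and footstep attenuation, everyone else to third-person enemy class and attenuation.
# For non-local sources, an IsTeam branch plays teammates much quieter than enemies or unset cases.
# A root modulator adds variation, and concurrency caps instances per owner, stopping the oldest.
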SoundCue /DryBox/Items/NyxGlass/Audio/NyxGlass_Release_Spin_Short_Cue
root=Modulator_0
nodes [
 {
  id=Attenuation_4
  children [SoundClass_4]
  properties {attenuationSettings=/Game/Sounds/Fort_Player/Footsteps/Settings/Fort_Player_FS_1P_Att}
 }
 {
  id=Attenuation_5
  children [SoundClass_5]
  properties {attenuationSettings=/Game/Sounds/Fort_Player/Footsteps/Settings/Fort_Player_FS_3P_Att}
 }
 {id=Branch_4, children [Attenuation_4, Branch_5, Branch_5], properties {boolParameterName=IsLocal}}
 {id=Branch_5, children [Modulator_3,Modulator_1,Modulator_1], properties {boolParameterName=IsTeam}}
 {id=Modulator_0, children [Branch_4], properties {volumeMin=0.6}}
 {id=Modulator_1, children [Attenuation_5], properties {pitchMax=1, pitchMin=1, volumeMax=0.6, volumeMin=0.6}}
 {id=Modulator_3, children [Attenuation_5], properties {pitchMax=1, pitchMin=1, volumeMax=0.1, volumeMin=0.1}}
 {
  id=Random_0
  children [WavePlayer_0,WavePlayer_1,WavePlayer_2,WavePlayer_3,WavePlayer_4,WavePlayer_5,WavePlayer_6]
  properties {weights [1,1,1,1,1,1,1]}
 }
 {id=SoundClass_4, children [Random_0], properties {soundClassOverride=&Fort_Player_Foley_1P_Athena}}
 {
  id=SoundClass_5
  children [Random_0]
  properties {soundClassOverride=&Fort_Player_Foley_3P_Enemy_Athena}
 }
 {id=WavePlayer_0, properties {soundWaveAssetPtr=&/NyxGlass_Release_Spin_Short_07}}
 {id=WavePlayer_1, properties {soundWaveAssetPtr=&/NyxGlass_Release_Spin_Short_01}}
 {id=WavePlayer_2, properties {soundWaveAssetPtr=&/NyxGlass_Release_Spin_Short_02}}
 {id=WavePlayer_3, properties {soundWaveAssetPtr=&/NyxGlass_Release_Spin_Short_03}}
 {id=WavePlayer_4, properties {soundWaveAssetPtr=&/NyxGlass_Release_Spin_Short_04}}
 {id=WavePlayer_5, properties {soundWaveAssetPtr=&/NyxGlass_Release_Spin_Short_05}}
 {id=WavePlayer_6, properties {soundWaveAssetPtr=&/NyxGlass_Release_Spin_Short_06}}
]
properties {
 maxDistance=3584
 volumeMultiplier=2
 concurrencyOverrides {bLimitToOwner=true, maxCount=4, resolutionRule=StopOldest, volumeScaleAttackTime=0.2}
}
