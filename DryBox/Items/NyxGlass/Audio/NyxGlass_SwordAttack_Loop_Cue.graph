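# Sound cue for the NyxGlass sword attack loop, layering a random start one-shot over looping pull-energy, delayed energy-field and falling-wind sounds.
# Envelopers fade the loops in and raise their pitch over several seconds, building intensity as the attack continues.
# IsLocal plays the mix unattenuated; otherwise it uses third-person attenuation, with IsTeam lowering teammates' volume.
# All paths use the melee weapon sound class, with attack concurrency and virtualization disabled.
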SoundCue /DryBox/Items/NyxGlass/Audio/NyxGlass_SwordAttack_Loop_Cue
root=Branch_4
nodes [
 {id=Attenuation_5, children [SoundClass_5], properties {attenuationSettings=/Game/Sounds/Attenuation/Fort_NyxGlass_Impact_3P_Att}}
 {id=Branch_4, children [SoundClass_4,Branch_5,Branch_5], properties {boolParameterName=IsLocal}}
 {id=Branch_5, children [Modulator_3, Attenuation_5, Attenuation_5], properties {boolParameterName=IsTeam}}
 {id=Delay_0, children [WavePlayer_10], properties {delayMax=0.5, delayMin=0.2}}
 {
  id=Enveloper_0
  children [Modulator_0]
  properties {
   pitchCurve {editorCurveData {keys [{time=0}, {time=2.5}, <at 10.5355 1.167>, <at 18.8591 1.4287>]}}
   volumeCurve {editorCurveData {keys [<at 0 0>,<at 3 0.8>,<at 4 1>]}}
  }
 }
 {
  id=Enveloper_2
  children [WavePlayer_6]
  properties {pitchCurve {editorCurveData {keys [{time=0}, <at 10 1.5>]}}, volumeCurve {editorCurveData {keys [<at 0 0>,<at 2 0.7>,<at 5 1>]}}}
 }
 {
  id=Enveloper_4
  children [Delay_0]
  properties {
   pitchCurve {editorCurveData {keys [{time=0}, {time=2.5}, <at 9.2848 1.1276>, <at 14.005 1.253>]}}
   volumeCurve {editorCurveData {keys [<at 0 0>,<at 1.2375 0.0802>,<at 2.9322 0.6007>,<at 4.2055 1.0164>]}}
  }
 }
 {id=Mixer_1, children [Enveloper_0,Enveloper_4,Enveloper_2,Random_0], properties {inputVolume [3,0.1,0.4,0.7]}}
 {id=Modulator_0, children [WavePlayer_5], properties {volumeMax=0.95, volumeMin=0.9}}
 {id=Modulator_3, children [Attenuation_5], properties {pitchMax=1, pitchMin=1, volumeMax=0.4, volumeMin=0.4}}
 {
  id=Random_0
  children [WavePlayer_0,WavePlayer_1,WavePlayer_2,WavePlayer_3,WavePlayer_4]
  properties {bShouldExcludeFromBranchCulling=true, weights [1,1,1,1,1]}
 }
 {id=SoundClass_4, children [Mixer_1], properties {soundClassOverride=&Fort_WeaponMelee}}
 {id=SoundClass_5, children [Mixer_1], properties {soundClassOverride=&Fort_WeaponMelee}}
 {id=WavePlayer_0, properties {soundWaveAssetPtr=&/NyxGlass_SwordAttackStart_Long_01}}
 {id=WavePlayer_1, properties {soundWaveAssetPtr=&/NyxGlass_SwordAttackStart_Long_02}}
 {id=WavePlayer_10, properties {bLooping=true, soundWaveAssetPtr=/Game/Sounds/Apollo/Biomes/Skyfire/Energy_Field_Large_Close_Loop}}
 {id=WavePlayer_2, properties {soundWaveAssetPtr=&/NyxGlass_SwordAttackStart_Long_03}}
 {id=WavePlayer_3, properties {soundWaveAssetPtr=&/NyxGlass_SwordAttackStart_Long_04}}
 {id=WavePlayer_4, properties {soundWaveAssetPtr=&/NyxGlass_SwordAttackStart_Long_05}}
 {
  id=WavePlayer_5
  properties {bLooping=true, soundWaveAssetPtr=/FNGameplayCues/Sounds/Fort_GamePlay_Sounds/BRSword/AB_Pull_Energy_Loop_01_distant}
 }
 {id=WavePlayer_6, properties {bLooping=true, soundWaveAssetPtr=/Game/Athena/Sounds/Wind_Falling/AMB_Wind_Skydive_Vertical_Loop_01}}
]
properties {
 maxDistance=10000
 virtualizationMode=Disabled
 volumeMultiplier=0.7
 concurrencySet [/DryBox/Items/NyxGlass/Audio/Settings/NyxGlass_Attack_Conc]
}
