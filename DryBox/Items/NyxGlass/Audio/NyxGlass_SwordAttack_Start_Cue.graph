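# Sound cue for the start of a NyxGlass sword attack, randomly playing one of five equally weighted wave variations.
# An IsLocal branch sends local playback through the melee weapon sound class.
# Other playback uses grappling-movement 3P attenuation and enemy foley sound class, with an IsTeam branch lowering volume for teammates.
# All paths share the random node, which is excluded from branch culling, and a concurrency asset limits the cue.
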SoundCue /DryBox/Items/NyxGlass/Audio/NyxGlass_SwordAttack_Start_Cue
root=Branch_4
nodes [
 {id=Attenuation_5, children [SoundClass_5], properties {attenuationSettings=/Game/Sounds/Attenuation/Fort_Grappling_Movement_3P_Att}}
 {id=Branch_4, children [SoundClass_4,Branch_5,Branch_5], properties {boolParameterName=IsLocal}}
 {id=Branch_5, children [Modulator_3, Attenuation_5, Attenuation_5], properties {boolParameterName=IsTeam}}
 {id=Modulator_3, children [Attenuation_5], properties {pitchMax=1, pitchMin=1, volumeMax=0.4, volumeMin=0.4}}
 {
  id=Random_0
  children [WavePlayer_0,WavePlayer_1,WavePlayer_2,WavePlayer_3,WavePlayer_4]
  properties {bShouldExcludeFromBranchCulling=true, weights [1,1,1,1,1]}
 }
 {id=SoundClass_4, children [Random_0], properties {soundClassOverride=&Fort_WeaponMelee}}
 {
  id=SoundClass_5
  children [Random_0]
  properties {soundClassOverride=&Fort_Player_Foley_3P_Enemy_Athena}
 }
 {id=WavePlayer_0, properties {soundWaveAssetPtr=&/NyxGlass_SwordAttack_Start_04}}
 {id=WavePlayer_1, properties {soundWaveAssetPtr=&/NyxGlass_SwordAttack_Start_01}}
 {id=WavePlayer_2, properties {soundWaveAssetPtr=&/NyxGlass_SwordAttack_Start_02}}
 {id=WavePlayer_3, properties {soundWaveAssetPtr=&/NyxGlass_SwordAttack_Start_03}}
 {id=WavePlayer_4, properties {soundWaveAssetPtr=&/NyxGlass_SwordAttack_Start_05}}
]
properties {
 maxDistance=6000
 virtualizationMode=Disabled
 volumeMultiplier=0.7
 concurrencySet [/DryBox/Items/NyxGlass/Audio/Settings/NyxGlass_Attack_Conc]
}
