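# Sound cue for the NyxGlass item's timeout, built from its "cannot fire" sound variations and routed to the gadgets sound class.
# A Random node picks one of five equally weighted waves and feeds an Enveloper whose curve bends pitch downward over the first half second.
# A Modulator at the root adds slight per-play variation.
# Borrowed prop-search attenuation and a per-owner concurrency limit that stops the oldest instance constrain playback.
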SoundCue /DryBox/Items/NyxGlass/Audio/NyxGlass_Timeout_Cue
root=Modulator_0
nodes [
 {
  id=Enveloper_0
  children [Random_4]
  properties {
   pitchCurve {editorCurveData {keys [<at 0.0005 1.0724>,<at 0.1752 0.7839>,<at 0.4275 0.6368>]}}
   volumeCurve {editorCurveData {keys [<at 0 1>,<at 0.2023 1.0803>]}}
  }
 }
 {id=Modulator_0, children [Enveloper_0], properties {pitchMax=1}}
 {
  id=Random_4
  children [WavePlayer_31, WavePlayer_32, WavePlayer_34, WavePlayer_35, WavePlayer_37]
  properties {bShouldExcludeFromBranchCulling=true, weights [1,1,1,1,1]}
 }
 {id=WavePlayer_31, properties {soundWaveAssetPtr=&/NyxGlass_CannotFire_01}}
 {id=WavePlayer_32, properties {soundWaveAssetPtr=&/NyxGlass_CannotFire_02}}
 {id=WavePlayer_34, properties {soundWaveAssetPtr=&/NyxGlass_CannotFire_04}}
 {id=WavePlayer_35, properties {soundWaveAssetPtr=&/NyxGlass_CannotFire_05}}
 {id=WavePlayer_37, properties {soundWaveAssetPtr=&/NyxGlass_CannotFire_07}}
]
properties {
 attenuationSettings=/Game/Sounds/Attenuation/Fort_Prop_Search_Att
 bOverrideConcurrency=true
 maxDistance=1300
 soundClassObject=&Fort_Gadgets
 volumeMultiplier=0.3
 concurrencyOverrides {bLimitToOwner=true, maxCount=2, resolutionRule=StopOldest, volumeScaleAttackTime=0.1}
}
