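; Waist gear actor for the NyxGlass grapple-swing weapon, driving its client-side rope and gas visuals.
; Two timers maintain Niagara physics ropes linking backpack sockets to the player's hands.
; Rope updates stretch two meshes toward attach points along a draw curve, feeding material parameters and triggering shot and link effects.
; Release spawns shrink ropes tuned by the current release animation, plus smoke and hover gas, with a delayed-release workaround for replays.
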(event Custom|TickExecute_ODMGearValidCheck
  (bind _playerfromweaponbp (Variables|Default|GetPlayerFromWeaponBP))
  (bind _returnvalue (Customization|GetSkeletalMeshforPartType _playerfromweaponbp "Backpack" :!class /Script/FortniteGame.FortPlayerPawn))
  (bind _tmpphysicsrope_r (Variables|Default|GettmpPhysicsRope_R))
  (bind _currenthoverparticle (Variables|Default|GetCurrentHoverParticle))
  ;[1 1 1] This check need high frequency
  (Utilities|IsValid _playerfromweaponbp
    (:Is%20Valid
      (Utilities|IsValid _returnvalue
        (:Is%20Valid
          (if (Transformation|DoesSocketExist _returnvalue "FX_SwordRopeRoot_r")
            _
            (else
              _)))
        (:Is%20Not%20Valid
          _)))
    (:Is%20Not%20Valid
      _))
  (Utilities|FlowControl|Sequence
    (:then_0
      (Variables|Default|SetCanPhysicsRopeSpawn false))
    (:then_1
      (Utilities|IsValid (Variables|Default|GettmpPhysicsRope_L)
        (:Is%20Valid
          (Components|Activation|Deactivate _tmpphysicsrope_r)
          (Components|DestroyComponent _tmpphysicsrope_r))))
    (:then_2
      (Utilities|IsValid _currenthoverparticle
        (:Is%20Valid
          (Components|Activation|Deactivate _currenthoverparticle)
          (Components|DestroyComponent _currenthoverparticle))))))

(event EventBeginPlay
  (bind _playerfromweaponbp (Variables|Default|GetPlayerFromWeaponBP))
  (bind _returnvalue (Customization|GetSkeletalMeshforPartType _playerfromweaponbp "Backpack" :!class /Script/FortniteGame.FortPlayerPawn))
  (bind _playerfromweaponbp_2 (Variables|Default|GetPlayerFromWeaponBP))
  (bind _mesh (Variables|/Script/Engine.Character|GetMesh _playerfromweaponbp_2))
  (bind _tmpphysicsrope_r (Variables|Default|GettmpPhysicsRope_R))
  (bind _tmpphysicsrope_r_2 (Variables|Default|GettmpPhysicsRope_R))
  (bind _tmpphysicsrope_l (Variables|Default|GettmpPhysicsRope_L))
  (bind _roperight (Variables|Default|GetRopeRight))
  (Networking|SwitchHasAuthority
    (:Authority
      (Components|Activation|Deactivate (Variables|Default|GetNS_GasParticles_NyxGlass)))
    (:Remote
      (bind _asfort_player_pawn (Utilities|Casting|CastToFortPlayerPawn (Variables|Actor|GetInstigator))
        (:then
          (Variables|Default|SetPlayerFromWeaponBP _asfort_player_pawn))
        (:CastFailed
          _))
      (bind _asb_weapon_nyx_glass (Utilities|Casting|CastToB_Weapon_NyxGlass (Weapon|GetCurrentWeapon (Variables|Default|GetPlayerFromWeaponBP)))
        (:then
          (Variables|Default|SetWeaponBP _asb_weapon_nyx_glass))
        (:CastFailed
          _))
      (Rendering|Material|SetScalarParameterValueonMaterials _roperight "Random" 123.598)
      (Rendering|Material|SetScalarParameterValueonMaterials (Variables|Default|GetRopeLeft) "Random" 687.218)
      (Rendering|Material|SetScalarParameterValueonMaterials _roperight "IsRight" 1.0)
      (Utilities|FlowControl|Delay 0.1)
      (Variables|Default|SetCanPhysicsRopeSpawn true)
      (bind _returnvalue_2 (Utilities|Time|SetTimerbyEvent (delegate Custom|TickExecute_FXSpawn) 0.4 true false 0.1 0.1))
      (Variables|Default|SetHandle _returnvalue_2)
      (bind _returnvalue_3 (Utilities|Time|SetTimerbyEvent (delegate Custom|TickExecute_ODMGearValidCheck) 0.2 true false 0.1 0.1))
      (Variables|Default|SetHandleB _returnvalue_3)
      ;[1 1 1] SpawnFX
      (Utilities|IsValid (Variables|Default|GetPlayerFromWeaponBP)
        (:Is%20Valid
          (if (Utilities|IsValid (Variables|Default|GettmpPhysicsRope_R) :!class /Script/Engine.KismetSystemLibrary)
            _
            (elif (Variables|Default|GetCanPhysicsRopeSpawn)
              (bind _returnvalue_4 (Niagara|SpawnSystemAttached "/DryBox/Items/NyxGlass/FX/Rope/NS_SwordPhysicsRope_Remade.NS_SwordPhysicsRope_Remade" _returnvalue "FX_SwordRopeRoot_r" "0, 0, 0" "0, 0, 0" "SnapToTarget" false false))
              (Variables|Default|SettmpPhysicsRope_R _returnvalue_4)
              (bind _returnvalue_5 (Niagara|SpawnSystemAttached "/DryBox/Items/NyxGlass/FX/Rope/NS_SwordPhysicsRope_Remade.NS_SwordPhysicsRope_Remade" _returnvalue "FX_SwordRopeRoot_l" "0, 0, 0" "0, 0, 0" "SnapToTarget" false false))
              (Variables|Default|SettmpPhysicsRope_L _returnvalue_5)
              (Effects|Components|ParticleSystem|SetBoolParameter _tmpphysicsrope_r_2 "bIsRight" true)
              (Niagara|SetNiagaraSkeletalMeshComponent _tmpphysicsrope_r_2 "PlayerSM" _mesh)
              (Niagara|SetNiagaraSkeletalMeshComponent _tmpphysicsrope_l "PlayerSM" _mesh)
              (Niagara|SetSkeletalMeshDataInterfaceFilteredBones _tmpphysicsrope_l "PlayerSM" (Utilities|Array|MakeArray "hand_l"))
              (Niagara|SetSkeletalMeshDataInterfaceFilteredBones _tmpphysicsrope_r_2 "PlayerSM" (Utilities|Array|MakeArray "hand_r"))
              (Components|Activation|SetActive (Variables|Default|GettmpPhysicsRope_L) true true))))
        (:Is%20Not%20Valid
          (Components|Activation|Deactivate _tmpphysicsrope_r)
          (Components|DestroyComponent _tmpphysicsrope_r))))))

(event Custom|TickExecute_FXSpawn
  (bind _playerfromweaponbp (Variables|Default|GetPlayerFromWeaponBP))
  (bind _returnvalue (Customization|GetSkeletalMeshforPartType _playerfromweaponbp "Backpack" :!class /Script/FortniteGame.FortPlayerPawn))
  (bind _playerfromweaponbp_2 (Variables|Default|GetPlayerFromWeaponBP))
  (bind _playerfromweaponbp_3 (Variables|Default|GetPlayerFromWeaponBP))
  (bind _mesh (Variables|/Script/Engine.Character|GetMesh _playerfromweaponbp_3))
  (bind _weaponbp (Variables|Default|GetWeaponBP))
  (bind _tmpphysicsrope_r (Variables|Default|GettmpPhysicsRope_R))
  (bind _tmpphysicsrope_r_2 (Variables|Default|GettmpPhysicsRope_R))
  (bind _tmpphysicsrope_l (Variables|Default|GettmpPhysicsRope_L))
  (Utilities|FlowControl|Sequence
    (:then_0
      ;[1 1 1] SpawnFX
      (Utilities|IsValid (Variables|Default|GetPlayerFromWeaponBP)
        (:Is%20Valid
          (if (Utilities|IsValid (Variables|Default|GettmpPhysicsRope_R) :!class /Script/Engine.KismetSystemLibrary)
            _
            (elif (Variables|Default|GetCanPhysicsRopeSpawn)
              (bind _returnvalue_2 (Niagara|SpawnSystemAttached "/DryBox/Items/NyxGlass/FX/Rope/NS_SwordPhysicsRope_Remade.NS_SwordPhysicsRope_Remade" _returnvalue "FX_SwordRopeRoot_r" "0, 0, 0" "0, 0, 0" "SnapToTarget" false false))
              (Variables|Default|SettmpPhysicsRope_R _returnvalue_2)
              (bind _returnvalue_3 (Niagara|SpawnSystemAttached "/DryBox/Items/NyxGlass/FX/Rope/NS_SwordPhysicsRope_Remade.NS_SwordPhysicsRope_Remade" _returnvalue "FX_SwordRopeRoot_l" "0, 0, 0" "0, 0, 0" "SnapToTarget" false false))
              (Variables|Default|SettmpPhysicsRope_L _returnvalue_3)
              (Effects|Components|ParticleSystem|SetBoolParameter _tmpphysicsrope_r_2 "bIsRight" true)
              (Niagara|SetNiagaraSkeletalMeshComponent _tmpphysicsrope_r_2 "PlayerSM" _mesh)
              (Niagara|SetNiagaraSkeletalMeshComponent _tmpphysicsrope_l "PlayerSM" _mesh)
              (Niagara|SetSkeletalMeshDataInterfaceFilteredBones _tmpphysicsrope_l "PlayerSM" (Utilities|Array|MakeArray "hand_l"))
              (Niagara|SetSkeletalMeshDataInterfaceFilteredBones _tmpphysicsrope_r_2 "PlayerSM" (Utilities|Array|MakeArray "hand_r"))
              (Components|Activation|SetActive (Variables|Default|GettmpPhysicsRope_L) true true))))
        (:Is%20Not%20Valid
          (Components|Activation|Deactivate _tmpphysicsrope_r)
          (Components|DestroyComponent _tmpphysicsrope_r))))
    (:then_1
      ;[1 1 1] Hide rope logic
      (Variables|Default|SetCanPhysicsRopeSpawn true)
      (Utilities|IsValid _playerfromweaponbp_2
        (:Is%20Valid
          (Utilities|FlowControl|Sequence
            (:then_0
              (if (Math|Boolean|ORBoolean (Parachute|IsParachuteOpenandVisible _playerfromweaponbp_2) (Parachute|IsSkydiving _playerfromweaponbp_2) (Health|IsDBNO _playerfromweaponbp_2))
                _))
            (:then_1
              (bind _asfort_player_anim_instance (Utilities|Casting|CastToFortPlayerAnimInstance (Components|SkeletalMesh|GetAnimInstance (Variables|/Script/Engine.Character|GetMesh _playerfromweaponbp_2))))
              (if (Math|Boolean|ORBoolean (Variables|/Script/FortniteGame.FortPlayerAnimInstance|GetbIsTacticalSprint _asfort_player_anim_instance) (Variables|/Script/FortniteGame.FortPlayerAnimInstance|GetbIsSurfaceSwimming _asfort_player_anim_instance) (Variables|/Script/FortniteGame.FortPlayerAnimInstance|GetbIsWaterJump _asfort_player_anim_instance))
                _))
            (:then_2
              ;;[1 1 1] If the parent BP destroyed
              (Utilities|IsValid _weaponbp
                (:Is%20Valid
                  (if (== (Variables|/DryBox/Items/NyxGlass/B_Weapon_NyxGlass.B_Weapon_NyxGlass_C|GetWaistGear _weaponbp) self)
                    _
                    (else
                      _)))
                (:Is%20Not%20Valid
                  _)))))
        (:Is%20Not%20Valid
          _))
      (CallFunction|RopeVisualDestroy))))

(event EventDestroyed
  (Utilities|Time|ClearandInvalidateTimerbyHandle (Variables|Default|GetHandleB))
  (Utilities|Time|ClearandInvalidateTimerbyHandle (Variables|Default|GetHandle)))

(event Custom|RopeVisual_RopeReleaseVisuals
  (bind _currenthoverparticle (Variables|Default|GetCurrentHoverParticle))
  (bind _playerfromweaponbp (Variables|Default|GetPlayerFromWeaponBP))
  ;[1 1 1] RopeReleaseVisuals
  (Networking|SwitchHasAuthority
    (:Remote
      (Utilities|FlowControl|Sequence
        (:then_0
          (Variables|Default|SetIsRandomSwing? (Math|Conversions|ToInteger(Boolean) (>= (Math|Random|RandomFloatinRangefromStream (Math|Random|MakeRandomStream (Math|Float|Truncate (* (Utilities|Time|GetGameTimeinSeconds) 1e+02))) 0.0 1.0) 0.8)))
          (if (or (Rendering|IsVisible (Variables|Default|GetRopeRight)) (Rendering|IsVisible (Variables|Default|GetRopeLeft)))
            (Variables|Default|SetCacheLeftEndPos (+ (Transformation|GetWorldLocation (Variables|Default|GetRopeLeft)) (Utilities|Operators|Multiply (Transformation|GetUpVector (Variables|Default|GetRopeLeft)) (.z (Variables|/Script/Engine.SceneComponent|GetRelativeScale3D (Variables|Default|GetRopeLeft))) 5.0)))
            (Variables|Default|SetCacheRightEndPos (+ (Transformation|GetWorldLocation (Variables|Default|GetRopeRight)) (Utilities|Operators|Multiply (Transformation|GetUpVector (Variables|Default|GetRopeRight)) (.z (Variables|/Script/Engine.SceneComponent|GetRelativeScale3D (Variables|Default|GetRopeRight))) 5.0)))
            (Utilities|FlowControl|DelayUntilNextTick)
            (Default|RopeVisualSpawnReleaseFX (Math|Vector|SelectVector (Variables|Default|GetREPLAY_CacheLeftEndPos) (Variables|Default|GetCacheLeftEndPos) (Replay|GetIsPlayingReplay)) (Math|Vector|SelectVector (Variables|Default|GetREPLAY_CacheRightEndPos) (Variables|Default|GetCacheRightEndPos) (Replay|GetIsPlayingReplay))))
          (Default|RopeVisualSetARopeVisibility false)
          (Default|RopeVisualSetARopeVisibility true))
        (:then_1
          (Utilities|IsValid _currenthoverparticle
            (:Is%20Valid
              (Components|Activation|Deactivate _currenthoverparticle)
              (Components|DestroyComponent _currenthoverparticle))
            (:Is%20Not%20Valid
              _))
          (if (not (Math|Boolean|ORBoolean (Utilities|IsValid (WaterBody|GetCurrentWaterBody _playerfromweaponbp) :!class /Script/Engine.KismetSystemLibrary) (AI|Components|NavMovement|IsSwimming (Pawn|GetMovementComponent _playerfromweaponbp) :!class /Script/Engine.NavMovementComponent) (Water|IsWaterSprintBoosting _playerfromweaponbp)))
            (bind _returnvalue (Niagara|SpawnSystemAttached "/DryBox/Items/NyxGlass/FX/Air/NS_HoverGas.NS_HoverGas" (Customization|GetSkeletalMeshforPartType _playerfromweaponbp "Backpack" :!class /Script/FortniteGame.FortPlayerPawn) "None" "0, 0, 0" "0, 0, 0" "KeepRelativeOffset" true))
            (Variables|Default|SetCurrentHoverParticle _returnvalue)))))))

(event Custom|REPLAY_FixReleaseDelay
  (bind _currenthoverparticle (Variables|Default|GetCurrentHoverParticle))
  (bind _playerfromweaponbp (Variables|Default|GetPlayerFromWeaponBP))
  ;[1 1 1] RopeReleaseVisuals
  (Utilities|FlowControl|RetriggerableDelay (+ (Utilities|Time|GetWorldDeltaSeconds) 0.01))
  (Networking|SwitchHasAuthority
    (:Remote
      (Utilities|FlowControl|Sequence
        (:then_0
          (Variables|Default|SetIsRandomSwing? (Math|Conversions|ToInteger(Boolean) (>= (Math|Random|RandomFloatinRangefromStream (Math|Random|MakeRandomStream (Math|Float|Truncate (* (Utilities|Time|GetGameTimeinSeconds) 1e+02))) 0.0 1.0) 0.8)))
          (if (or (Rendering|IsVisible (Variables|Default|GetRopeRight)) (Rendering|IsVisible (Variables|Default|GetRopeLeft)))
            (Variables|Default|SetCacheLeftEndPos (+ (Transformation|GetWorldLocation (Variables|Default|GetRopeLeft)) (Utilities|Operators|Multiply (Transformation|GetUpVector (Variables|Default|GetRopeLeft)) (.z (Variables|/Script/Engine.SceneComponent|GetRelativeScale3D (Variables|Default|GetRopeLeft))) 5.0)))
            (Variables|Default|SetCacheRightEndPos (+ (Transformation|GetWorldLocation (Variables|Default|GetRopeRight)) (Utilities|Operators|Multiply (Transformation|GetUpVector (Variables|Default|GetRopeRight)) (.z (Variables|/Script/Engine.SceneComponent|GetRelativeScale3D (Variables|Default|GetRopeRight))) 5.0)))
            (Utilities|FlowControl|DelayUntilNextTick)
            (Default|RopeVisualSpawnReleaseFX (Math|Vector|SelectVector (Variables|Default|GetREPLAY_CacheLeftEndPos) (Variables|Default|GetCacheLeftEndPos) (Replay|GetIsPlayingReplay)) (Math|Vector|SelectVector (Variables|Default|GetREPLAY_CacheRightEndPos) (Variables|Default|GetCacheRightEndPos) (Replay|GetIsPlayingReplay))))
          (Default|RopeVisualSetARopeVisibility false)
          (Default|RopeVisualSetARopeVisibility true))
        (:then_1
          (Utilities|IsValid _currenthoverparticle
            (:Is%20Valid
              (Components|Activation|Deactivate _currenthoverparticle)
              (Components|DestroyComponent _currenthoverparticle))
            (:Is%20Not%20Valid
              _))
          (if (not (Math|Boolean|ORBoolean (Utilities|IsValid (WaterBody|GetCurrentWaterBody _playerfromweaponbp) :!class /Script/Engine.KismetSystemLibrary) (AI|Components|NavMovement|IsSwimming (Pawn|GetMovementComponent _playerfromweaponbp) :!class /Script/Engine.NavMovementComponent) (Water|IsWaterSprintBoosting _playerfromweaponbp)))
            (bind _returnvalue (Niagara|SpawnSystemAttached "/DryBox/Items/NyxGlass/FX/Air/NS_HoverGas.NS_HoverGas" (Customization|GetSkeletalMeshforPartType _playerfromweaponbp "Backpack" :!class /Script/FortniteGame.FortPlayerPawn) "None" "0, 0, 0" "0, 0, 0" "KeepRelativeOffset" true))
            (Variables|Default|SetCurrentHoverParticle _returnvalue)))))))

(event Custom|Turn%20On%20Rope%20Visuals
  ;[1 1 1] RopeReleaseVisuals
  (Default|RopeVisualSetARopeVisibility false true)
  (Default|RopeVisualSetARopeVisibility true true))

(fn RopeVisual_DeactivePushForwardFX
  (bind _ns_pushforwardfx (Variables|Default|GetNS_PushForwardFX))
  (Utilities|IsValid _ns_pushforwardfx
    (:Is%20Valid
      (Components|Activation|Deactivate _ns_pushforwardfx)
      (Components|DestroyComponent _ns_pushforwardfx)
      (Variables|Default|SetNS_PushForwardFX 0))
    (:Is%20Not%20Valid
      (Variables|Default|SetNS_PushForwardFX 0))))

(fn RopeVisual_Destroy
  (bind _tmpphysicsrope_r (Variables|Default|GettmpPhysicsRope_R))
  (bind _currenthoverparticle (Variables|Default|GetCurrentHoverParticle))
  (Utilities|FlowControl|Sequence
    (:then_0
      (Variables|Default|SetCanPhysicsRopeSpawn false))
    (:then_1
      (Utilities|IsValid (Variables|Default|GettmpPhysicsRope_L)
        (:Is%20Valid
          (Components|Activation|Deactivate _tmpphysicsrope_r)
          (Components|DestroyComponent _tmpphysicsrope_r))))
    (:then_2
      (Utilities|IsValid _currenthoverparticle
        (:Is%20Valid
          (Components|Activation|Deactivate _currenthoverparticle)
          (Components|DestroyComponent _currenthoverparticle))))))

(fn RopeVisual_SetARopeVisibility
  (params (bIsRight :type Boolean) (bVisibility :type Boolean))
  (Rendering|SetVisibility (select bIsRight (Variables|Default|GetRopeRight) (Variables|Default|GetRopeLeft)) bVisibility false))

(fn RopeVisual_SpawnPushForwardFX
  (bind _playerfromweaponbp (Variables|Default|GetPlayerFromWeaponBP))
  (Networking|SwitchHasAuthority
    (:Remote
      (if (Pawn|IsLocallyControlled (Variables|Default|GetPlayerFromWeaponBP))
        (Utilities|IsValid _playerfromweaponbp
          (:Is%20Valid
            (bind _returnvalue (Niagara|SpawnSystemAttached "/DryBox/Items/NyxGlass/FX/Air/NS_AirBoost.NS_AirBoost" (Variables|/Script/Engine.Character|GetMesh _playerfromweaponbp) "pelvis" "0, 0, 0" "0, 0, 0" "KeepRelativeOffset" true))
            (Transformation|SetRelativeTransform _returnvalue (Math|Transform|MakeTransform "4.000000,-13.000000,0.000000" "0.000000,130.000000,270.000000") false false)
            (Variables|Default|SetNS_PushForwardFX _returnvalue)))))))

(fn RopeVisual_SpawnReleaseFX
  (params (LeftEndPos :type Vector) (RightEndPos :type Vector))
  (bind _playerfromweaponbp (Variables|Default|GetPlayerFromWeaponBP))
  (bind _mesh (Variables|/Script/Engine.Character|GetMesh _playerfromweaponbp))
  (bind _returnvalue (Components|SkeletalMesh|AnimationBlueprintLinking|GetLinkedAnimLayerInstancebyClass _mesh "/DryBox/Items/NyxGlass/Animation/NyxGlass_AnimBP.NyxGlass_AnimBP_C"))
  (bind _rightshrinkfx (LocalVariables|GetRightShrinkFX))
  (bind _rightshrinkfx_2 (LocalVariables|GetRightShrinkFX))
  (bind _leftshrinkfx (LocalVariables|GetLeftShrinkFX))
  (bind _leftshrinkfx_2 (LocalVariables|GetLeftShrinkFX))
  (bind _leftshrinkfx_3 (LocalVariables|GetLeftShrinkFX))
  (bind _rightshrinkfx_3 (LocalVariables|GetRightShrinkFX))
  (Networking|SwitchHasAuthority
    (:Remote
      (CallFunction|RopeVisualSpawnReleaseSmokeFX)
      (Utilities|FlowControl|Sequence
        (:then_0
          (if (Rendering|IsVisible (Variables|Default|GetRopeLeft))
            (bind _returnvalue_2 (Niagara|SpawnSystemAttached "/DryBox/Items/NyxGlass/FX/Rope/NS_ShrinkRopeTest_NyxGlass_HookVer.NS_ShrinkRopeTest_NyxGlass_HookVer" (Variables|Default|GetDefaultSceneRoot) "None" "0, 0, 0" "0, 0, 0" "SnapToTarget" false false))
            (LocalVariables|SetLeftShrinkFX _returnvalue_2)
            (Niagara|SetNiagaraVariable(Position) _returnvalue_2 "Target" (Variables|GetLeftEndPos))))
        (:then_1
          (if (Rendering|IsVisible (Variables|Default|GetRopeRight))
            (bind _returnvalue_3 (Niagara|SpawnSystemAttached "/DryBox/Items/NyxGlass/FX/Rope/NS_ShrinkRopeTest_NyxGlass_HookVer.NS_ShrinkRopeTest_NyxGlass_HookVer" (Variables|Default|GetDefaultSceneRoot) "None" "0, 0, 0" "0, 0, 0" "SnapToTarget" false false))
            (bind _output_get (LocalVariables|SetRightShrinkFX _returnvalue_3))
            (Niagara|SetNiagaraVariable(Position) _output_get "Target" (Variables|GetRightEndPos))))
        (:then_2
          (Utilities|IsValid _returnvalue
            (:Is%20Valid
              (bind _asnyx_glass_anim_bp (Utilities|Casting|CastToNyxGlass_AnimBP _returnvalue))
              (Utilities|FlowControl|Sequence
                (:then_0
                  (Utilities|IsValid (LocalVariables|GetLeftShrinkFX)
                    (:Is%20Valid
                      (Transformation|SetRelativeLocation _leftshrinkfx "13.500000,-14.000000,6.500000" false false)
                      (Transformation|SetRelativeRotation _leftshrinkfx (Utilities|Select :Index (Variables|/DryBox/Items/NyxGlass/Animation/NyxGlass_AnimBP.NyxGlass_AnimBP_C|GetCurrentZipReleaseAnim _asnyx_glass_anim_bp) :Option%200 "10.000000,-110.000000,-70.000000" :Option%201 "90.000000,0.000000,0.000000" :Option%202 "90.000000,0.000000,0.000000" :Option%203 "-10.000000,100.000000,-70.000000" :Option%204 "90.000000,0.000000,0.000000" :Option%205 "10.000000,-110.000000,-70.000000" :Option%206 "90.000000,0.000000,0.000000" :Option%207 "90.000000,0.000000,0.000000" :Option%208 "90.000000,0.000000,0.000000") false false)
                      (Niagara|SetNiagaraVariable(Float) _leftshrinkfx "A" (Utilities|Select :Index (Variables|/DryBox/Items/NyxGlass/Animation/NyxGlass_AnimBP.NyxGlass_AnimBP_C|GetCurrentZipReleaseAnim _asnyx_glass_anim_bp) :Option%200 0.0 :Option%201 0.6 :Option%202 0.6 :Option%203 0.0 :Option%204 0.6 :Option%205 0.0 :Option%206 0.6 :Option%207 0.6 :Option%208 0.6))
                      (Niagara|SetNiagaraVariable(Float) _leftshrinkfx_2 "Radius" (Utilities|Select :Index (Variables|/DryBox/Items/NyxGlass/Animation/NyxGlass_AnimBP.NyxGlass_AnimBP_C|GetCurrentZipReleaseAnim _asnyx_glass_anim_bp) :Option%200 5e+01 :Option%201 6e+01 :Option%202 6e+01 :Option%203 5e+01 :Option%204 6e+01 :Option%205 5e+01 :Option%206 6e+01 :Option%207 6e+01 :Option%208 6e+01))
                      (Niagara|SetNiagaraVariable(Float) _leftshrinkfx_2 "Extend" (Utilities|Select :Index (Variables|/DryBox/Items/NyxGlass/Animation/NyxGlass_AnimBP.NyxGlass_AnimBP_C|GetCurrentZipReleaseAnim _asnyx_glass_anim_bp) :Option%200 6e+02 :Option%201 3e+02 :Option%202 3e+02 :Option%203 6e+02 :Option%204 3e+02 :Option%205 6e+02 :Option%206 6e+02 :Option%207 6e+02 :Option%208 6e+02))
                      (Niagara|SetNiagaraVariable(Float) _leftshrinkfx_2 "ScaleAge" (Utilities|Select :Index (Variables|/DryBox/Items/NyxGlass/Animation/NyxGlass_AnimBP.NyxGlass_AnimBP_C|GetCurrentZipReleaseAnim _asnyx_glass_anim_bp) :Option%200 1.3 :Option%201 1.0 :Option%202 1.0 :Option%203 1.0 :Option%204 1.0 :Option%205 1.3 :Option%206 1.0 :Option%207 0.9 :Option%208 1.0))
                      (Niagara|SetNiagaraVariable(Int32) _leftshrinkfx_2 "InverseSineDir" (Utilities|Select :Index (Variables|/DryBox/Items/NyxGlass/Animation/NyxGlass_AnimBP.NyxGlass_AnimBP_C|GetCurrentZipReleaseAnim _asnyx_glass_anim_bp) :Option%200 1 :Option%201 0 :Option%202 0 :Option%203 0 :Option%204 0 :Option%205 0 :Option%206 0 :Option%207 0 :Option%208 1)))))
                (:then_1
                  (Utilities|IsValid (LocalVariables|GetRightShrinkFX)
                    (:Is%20Valid
                      (Transformation|SetRelativeLocation (LocalVariables|GetRightShrinkFX) "13.500000,14.000000,6.500000" false false)
                      (Transformation|SetRelativeRotation _rightshrinkfx (Utilities|Select :Index (Variables|/DryBox/Items/NyxGlass/Animation/NyxGlass_AnimBP.NyxGlass_AnimBP_C|GetCurrentZipReleaseAnim _asnyx_glass_anim_bp) :Option%200 "-10.000000,-70.000000,-70.000000" :Option%201 "90.000000,0.000000,0.000000" :Option%202 "90.000000,0.000000,0.000000" :Option%203 "10.000000,80.000000,-70.000000" :Option%204 "90.000000,0.000000,0.000000" :Option%205 "-10.000000,-70.000000,-70.000000" :Option%206 "90.000000,0.000000,0.000000" :Option%207 "90.000000,0.000000,0.000000" :Option%208 "90.000000,0.000000,0.000000") false false)
                      (Niagara|SetNiagaraVariable(Float) _rightshrinkfx "A" (Utilities|Select :Index (Variables|/DryBox/Items/NyxGlass/Animation/NyxGlass_AnimBP.NyxGlass_AnimBP_C|GetCurrentZipReleaseAnim _asnyx_glass_anim_bp) :Option%200 0.0 :Option%201 0.0 :Option%202 0.0 :Option%203 0.0 :Option%204 0.0 :Option%205 0.0 :Option%206 0.0 :Option%207 0.0 :Option%208 0.0))
                      (Niagara|SetNiagaraVariable(Float) _rightshrinkfx_2 "Radius" (Utilities|Select :Index (Variables|/DryBox/Items/NyxGlass/Animation/NyxGlass_AnimBP.NyxGlass_AnimBP_C|GetCurrentZipReleaseAnim _asnyx_glass_anim_bp) :Option%200 5e+01 :Option%201 6e+01 :Option%202 6e+01 :Option%203 5e+01 :Option%204 6e+01 :Option%205 5e+01 :Option%206 6e+01 :Option%207 6e+01 :Option%208 6e+01))
                      (Niagara|SetNiagaraVariable(Float) _rightshrinkfx_2 "Extend" (Utilities|Select :Index (Variables|/DryBox/Items/NyxGlass/Animation/NyxGlass_AnimBP.NyxGlass_AnimBP_C|GetCurrentZipReleaseAnim _asnyx_glass_anim_bp) :Option%200 6e+02 :Option%201 3e+02 :Option%202 3e+02 :Option%203 6e+02 :Option%204 3e+02 :Option%205 6e+02 :Option%206 3e+02 :Option%207 3e+02 :Option%208 3e+02))
                      (Niagara|SetNiagaraVariable(Float) _rightshrinkfx_2 "ScaleAge" (Utilities|Select :Index (Variables|/DryBox/Items/NyxGlass/Animation/NyxGlass_AnimBP.NyxGlass_AnimBP_C|GetCurrentZipReleaseAnim _asnyx_glass_anim_bp) :Option%200 1.3 :Option%201 1.0 :Option%202 1.0 :Option%203 1.0 :Option%204 1.0 :Option%205 1.3 :Option%206 1.0 :Option%207 0.9 :Option%208 1.0))
                      (Niagara|SetNiagaraVariable(Int32) _rightshrinkfx_2 "InverseSineDir" (Utilities|Select :Index (Variables|/DryBox/Items/NyxGlass/Animation/NyxGlass_AnimBP.NyxGlass_AnimBP_C|GetCurrentZipReleaseAnim _asnyx_glass_anim_bp) :Option%200 1 :Option%201 0 :Option%202 0 :Option%203 0 :Option%204 0 :Option%205 0 :Option%206 0 :Option%207 0 :Option%208 1)))))
                (:then_2)))))
        (:then_3
          (Utilities|FlowControl|Sequence
            (:then_0
              (Utilities|IsValid _leftshrinkfx_3
                (:Is%20Valid
                  (Components|Activation|SetActive _leftshrinkfx_3 true true)
                  (Niagara|SetAutoDestroy _leftshrinkfx_3 true))))
            (:then_1
              (Utilities|IsValid _rightshrinkfx_3
                (:Is%20Valid
                  (Components|Activation|SetActive _rightshrinkfx_3 true true)
                  (Niagara|SetAutoDestroy _rightshrinkfx_3 true))))))))))

(fn RopeVisual_SpawnReleaseSmokeFX
  (Networking|SwitchHasAuthority
    (:Remote
      (if (Pawn|IsLocallyControlled (Variables|Default|GetPlayerFromWeaponBP))
        (bind _returnvalue (Niagara|SpawnSystemAttached "/DryBox/Items/NyxGlass/FX/Air/NS_ReleaseSmoke.NS_ReleaseSmoke" (Variables|Default|GetDefaultSceneRoot) "None" "0, 0, 0" "0, 0, 0" "SnapToTarget" true))
        (Transformation|SetAbsolute _returnvalue false true true)
        (Transformation|SetWorldScale3D _returnvalue "1.000000,1.000000,1.000000")))))

(fn RopeVisual_SpawnShotAirBreakFX
  (params (WeaponBP :type B%20Weapon%20Nyx%20Glass%20Object%20Reference))
  (bind (_leftactor _rightactor _leftattachpoint _rightattachpoint _bfoundleftattachpoint _bfoundrightattachpoint) (Utilities|Struct|BreakNyxGlassTargetingData (Variables|/Script/DryBoxRuntime.NyxGlassWeaponComponent_Swinging|GetCurTargetingData (Variables|/DryBox/Items/NyxGlass/B_Weapon_NyxGlass.B_Weapon_NyxGlass_C|GetNyxGlass%20Weapon%20Swinging%20BP%20Comp WeaponBP))))
  (Networking|SwitchHasAuthority
    (:Remote
      (if (Pawn|IsLocallyControlled (Variables|Default|GetPlayerFromWeaponBP))
        (if (<= (Math|Vector|DotProduct (Math|Vector|Normalize (/ (+ (- (Math|Vector|SelectVector _leftattachpoint (Transformation|GetActorLocation (Variables|Default|GetPlayerFromWeaponBP)) _bfoundleftattachpoint) (Transformation|GetActorLocation (Variables|Default|GetPlayerFromWeaponBP))) (- (Math|Vector|SelectVector _rightattachpoint (Transformation|GetActorLocation (Variables|Default|GetPlayerFromWeaponBP)) _bfoundrightattachpoint) (Transformation|GetActorLocation (Variables|Default|GetPlayerFromWeaponBP)))) "2.000000,2.000000,2.000000")) (Math|Vector|Normalize (Transformation|GetVelocity (Variables|Default|GetPlayerFromWeaponBP)))) -0.1)
          (bind _returnvalue (Niagara|SpawnSystemAttached "/DryBox/Items/NyxGlass/FX/Air/NS_ShotRopeAirBreak.NS_ShotRopeAirBreak" (Variables|Default|GetDefaultSceneRoot) "None" "0, 0, 0" (Math|Rotator|MakeRotfromX (Transformation|GetVelocity (Variables|Default|GetPlayerFromWeaponBP))) "SnapToTarget" true))
          (Transformation|SetAbsolute _returnvalue false true true)
          (Transformation|SetWorldScale3D _returnvalue "1.000000,1.000000,1.000000"))))))

(fn RopeVisual_SpawnShotBurstFX
  (params (Rotation :type Rotator) (bIsRight :type Boolean))
  (bind _currenthoverparticle (Variables|Default|GetCurrentHoverParticle))
  (bind _shotburstfx (LocalVariables|GetShotBurstFX))
  (Networking|SwitchHasAuthority
    (:Remote
      (if (Pawn|IsLocallyControlled (Variables|Default|GetPlayerFromWeaponBP))
        (if bIsRight
          (if (Variables|Default|GetFixShotFXIssue_R)
            (Variables|Default|SetFixShotFXIssue_R false))
          (elif (Variables|Default|GetFixShotFXIssue_L)
            (Variables|Default|SetFixShotFXIssue_L false)))
        (Utilities|FlowControl|Sequence
          (:then_0
            (bind _returnvalue (Niagara|SpawnSystemAttached "/DryBox/Items/NyxGlass/FX/Air/NS_ShotRopeTest.NS_ShotRopeTest" (Utilities|Casting|CastToSceneComponent (Utilities|SelectObject (Variables|Default|GetRopeRightAnchorPoint) (Variables|Default|GetRopeLeftAnchorPoint) bIsRight)) "None" "0, 0, 0" "0, 0, 0" "SnapToTarget" true))
            (LocalVariables|SetShotBurstFX _returnvalue))
          (:then_1
            (Utilities|IsValid _shotburstfx
              (:Is%20Valid
                (Transformation|SetWorldScale3D _shotburstfx "1.000000,1.000000,1.000000")
                (Transformation|SetAbsolute (LocalVariables|GetShotBurstFX) false true false)
                (Transformation|SetWorldRotation (LocalVariables|GetShotBurstFX) Rotation false false))))
          (:then_2
            ;[1 1 1] DirtyCheck
            (Utilities|IsValid _currenthoverparticle
              (:Is%20Valid
                (Components|Activation|Deactivate _currenthoverparticle)
                (Components|DestroyComponent _currenthoverparticle)))))))))

(fn RopeVisual_StageChecker
  (params (Arrive :type Boolean) (bIsRight :type Boolean) (AttachPoint :type Vector) (bIsAttacking :type Boolean))
  (bind _ropeleft (Variables|Default|GetRopeLeft))
  (bind _roperight (Variables|Default|GetRopeRight))
  (bind _returnvalue (Utilities|SelectObject _roperight _ropeleft bIsRight))
  (bind _returnvalue_2 (Utilities|Time|GetGameTimeinSeconds))
  (bind _asstatic_mesh_component (Utilities|Casting|CastToStaticMeshComponent _returnvalue))
  (Networking|SwitchHasAuthority
    (:Remote
      (if (Variables|/Script/Engine.SceneComponent|GetbVisible _asstatic_mesh_component)
        (if (Variables|GetbIsRight)
          (Variables|Default|SetFixShotFXIssue_R true)
          (else
            (Variables|Default|SetFixShotFXIssue_L true)))
        (else
          (Default|RopeVisualSpawnShotBurstFX (Math|Rotator|MakeRotfromZ (Math|Vector|Normalize (- AttachPoint (Transformation|GetWorldLocation (Utilities|Casting|CastToSceneComponent (Utilities|SelectObject (Variables|Default|GetRopeRightAnchorPoint) (Variables|Default|GetRopeLeftAnchorPoint) bIsRight)))))) bIsRight)
          (Rendering|Material|SetScalarParameterValueonMaterials _asstatic_mesh_component "ActiveTime" _returnvalue_2)
          (Rendering|Material|SetScalarParameterValueonMaterials _asstatic_mesh_component "Shine_AttackTime" (Math|Float|SelectFloat _returnvalue_2 -1e+02 (Variables|GetbIsAttacking)))))
      (if (Variables|GetArrive)
        (Utilities|FlowControl|Sequence
          (:then_0
            (if (Variables|GetbIsRight)
              (if (Variables|Default|GetWhenAttached_RightPointArrive)
                _
                (else
                  _))
              (elif (Variables|Default|GetWhenAttached_LeftPointArrive)
                _)))
          (:then_1
            (if (Variables|GetbIsAttacking)
              _
              (elif (Variables|GetbIsRight)
                (if (not (Variables|Default|GetWhenAttached_RightPointArrive))
                  (Effects|Components|ParticleSystem|SpawnFXAtLocation "/DryBox/Items/NyxGlass/FX/Rope/NS_RopeLinkFX.NS_RopeLinkFX" (Variables|GetAttachPoint)))
                (elif (Variables|Default|GetWhenAttached_LeftPointArrive)
                  _
                  (else
                    _)))))
          (:then_2
            (if (Variables|GetbIsRight)
              (Variables|Default|SetWhenAttached_RightPointArrive true)
              (else
                (Variables|Default|SetWhenAttached_LeftPointArrive true)))))
        (elif (Variables|GetbIsRight)
          (Variables|Default|SetWhenAttached_RightPointArrive false)
          (else
            (Variables|Default|SetWhenAttached_LeftPointArrive false))))
      (Rendering|Material|SetScalarParameterValueonMaterials _asstatic_mesh_component "ArriveTime" (Utilities|Time|GetGameTimeinSeconds)))))

(fn RopeVisual_TickExecute)

(fn RopeVisual_UpdateARopeVisual
  (params (AttachPoint :type Vector) (WeaponBP :type B%20Weapon%20Nyx%20Glass%20Object%20Reference) (bIsRight :type Boolean) (bIsAttacking :type Boolean))
  (bind _roperightanchorpoint (Variables|Default|GetRopeRightAnchorPoint))
  (bind _ropeleftanchorpoint (Variables|Default|GetRopeLeftAnchorPoint))
  (bind _returnvalue (select bIsRight _roperightanchorpoint _ropeleftanchorpoint))
  (bind _returnvalue_2 (Transformation|GetWorldLocation _returnvalue))
  (bind _israndomswing (Variables|Default|GetIsRandomSwing?))
  (bind _returnvalue_3 (Math|Conversions|ToFloat(Integer) _israndomswing))
  (bind _sf_timetofullyattachswinging (Variables|/DryBox/Items/NyxGlass/B_Weapon_NyxGlass.B_Weapon_NyxGlass_C|GetSF_TimeToFullyAttachSwinging WeaponBP))
  (bind _returnvalue_4 (ScalableFloat|GetValueatLevel _sf_timetofullyattachswinging))
  (bind _isswinging (Class|BWeaponGrappleParent|IsSwinging WeaponBP))
  (bind _sf_timetofullyattachgrounded (Variables|/DryBox/Items/NyxGlass/B_Weapon_NyxGlass.B_Weapon_NyxGlass_C|GetSF_TimeToFullyAttachGrounded WeaponBP))
  (bind _returnvalue_5 (ScalableFloat|GetValueatLevel _sf_timetofullyattachgrounded))
  (bind _returnvalue_6 (Math|Float|SelectFloat _returnvalue_4 _returnvalue_5 _isswinging))
  (bind _lastropeattachtime (Variables|/GrappleParent/GrappleParent/B_Weapon_GrappleParent.B_Weapon_GrappleParent_C|GetLastRopeAttachTime WeaponBP))
  (bind _returnvalue_7 (Utilities|Time|GetGameTimeinSeconds))
  (bind _returnvalue_8 (- _returnvalue_7 _lastropeattachtime))
  (bind _output (Utilities|NormalizeFloat _returnvalue_8 0 _returnvalue_6 true))
  (bind _ropedrawcurve (Variables|Default|GetRopeDrawCurve))
  (bind _returnvalue_9 (Math|Curves|GetFloatValue _ropedrawcurve _output))
  (bind _returnvalue_10 (- 1.0 _returnvalue_9))
  (bind _roperight (Variables|Default|GetRopeRight))
  (bind _ropeleft (Variables|Default|GetRopeLeft))
  (bind _returnvalue_11 (select bIsRight _roperight _ropeleft))
  (bind _finalattachpoint (LocalVariables|GetFinalAttachPoint))
  (bind _bisattacking (Variables|GetbIsAttacking))
  (bind _attachpointbga (Variables|/GrappleParent/GrappleParent/B_Weapon_GrappleParent.B_Weapon_GrappleParent_C|GetAttachPointBGA WeaponBP)
    (:then
      (bind _output_get (LocalVariables|SetFinalAttachPoint (Math|Vector|SelectVector (Transformation|GetActorLocation _attachpointbga) AttachPoint bIsAttacking))))
    (:else
      (LocalVariables|SetFinalAttachPoint AttachPoint)))
  (Networking|SwitchHasAuthority
    (:Remote
      (Niagara|SetNiagaraVariable(Bool) (Variables|Default|GetNS_GasParticles_NyxGlass) "bIsAtk" bIsAttacking)
      (Transformation|SetWorldRotation _returnvalue_11 (Math|Rotator|MakeRotfromZ (Math|Vector|Normalize (- (+ _output_get (* (- _returnvalue_2 _output_get) _returnvalue_10)) _returnvalue_2))) false false)
      (Transformation|SetWorldScale3D _returnvalue_11 (Math|Vector|MakeVector (Variables|Default|GetRopeWidth) (Variables|Default|GetRopeWidth) (* (Math|Vector|VectorLength (- (+ _output_get (* (- _returnvalue_2 _output_get) _returnvalue_10)) _returnvalue_2)) 0.01)))
      (Rendering|Material|SetVectorParameterValueonMaterials _returnvalue_11 "RopeEndPos" (+ _output_get (* (- _returnvalue_2 _output_get) _returnvalue_10)))
      (Rendering|Material|SetVectorParameterValueonMaterials _returnvalue_11 "RopeStartPos" _returnvalue_2)
      (Rendering|Material|SetScalarParameterValueonMaterials _returnvalue_11 "EnableHelixSwing" (Math|Float|SelectFloat 1.0 (- 1.0 _returnvalue_3) _bisattacking))
      (Rendering|Material|SetScalarParameterValueonMaterials _returnvalue_11 "EnableRandomSwing" (Math|Float|SelectFloat 0.0 _returnvalue_3 _bisattacking))
      (Rendering|Material|SetScalarParameterValueonMaterials _returnvalue_11 "Shine_TotalE" (Math|Float|SelectFloat 1.0 0.0 _bisattacking))
      (CallFunction|RopeVisualStageChecker (Utilities|Operators|Equal(==) (+ _output_get (* (- _returnvalue_2 _output_get) _returnvalue_10)) _output_get 2.5) (Variables|GetbIsRight) (LocalVariables|GetFinalAttachPoint) (Variables|GetbIsAttacking))
      ;[1 1 1] JIRA FORT-573507
      (if (Replay|GetIsPlayingReplay)
        (CallFunction|TurnonRopeVisuals)
        (CallFunction|REPLAYFixReleaseDelay)
        (if (Variables|GetbIsRight)
          (Variables|Default|SetREPLAY_CacheRightEndPos _finalattachpoint)
          (else
            (Variables|Default|SetREPLAY_CacheLeftEndPos _finalattachpoint)))))))

(fn ConstructionScript)
