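; Nyx Glass swinging weapon component that bridges nativized swing, targeting and velocity logic to its owning B_Weapon_NyxGlass Blueprint.
; Events cast the owner and forward callbacks: reticle and targeting completion on remote only, forgiveness timer and received targeting data on authority, ground exit stops skidding.
; Query functions translate the weapon's Blueprint state and slash-substate enums into native equivalents, report attacking and targeting-update checks, and velocity results pass back to the weapon.
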(event EventOnExitGround
  (bind _asb_weapon_grapple_parent (Utilities|Casting|CastToB_Weapon_GrappleParent (Components|GetOwner)))
  (Class|BWeaponGrappleParent|StopSkidding _asb_weapon_grapple_parent))

(event EventUpdateReticle
  (params (bHitSomething :type Boolean) (bUseBlocked :type Boolean) (bIsWeaponAimingAtTarget :type Boolean) (TargetPos :type Vector%20%28by%20ref%29 :default "0, 0, 0"))
  (Utilities|SwitchHasAuthority
    (:Remote
      (bind _asb_weapon_grapple_parent (Utilities|Casting|CastToB_Weapon_GrappleParent (Components|GetOwner)))
      (Class|BWeaponGrappleParent|UpdateReticlefromNative _asb_weapon_grapple_parent bHitSomething bUseBlocked bIsWeaponAimingAtTarget TargetPos))))

(event EventOnTargetingComplete
  (params (TargetingData :type Nyx%20Glass%20Targeting%20Data%20Structure%20%28by%20ref%29))
  (Utilities|SwitchHasAuthority
    (:Remote
      (bind _asb_weapon_nyx_glass (Utilities|Casting|CastToB_Weapon_NyxGlass (Components|GetOwner)))
      (Class|BWeaponNyxGlass|OnNativeTargetingComplete _asb_weapon_nyx_glass TargetingData))))

(event EventOnForgivenessTimerFinished
  (Utilities|SwitchHasAuthority
    (:Authority
      (bind _asb_weapon_nyx_glass (Utilities|Casting|CastToB_Weapon_NyxGlass (Components|GetOwner)))
      (ForgivenessTimer|OnNativeForgivenessTimerEnded _asb_weapon_nyx_glass))))

(event EventOnForgivenessTimerCanceled
  (Utilities|SwitchHasAuthority
    (:Authority
      (bind _asb_weapon_nyx_glass (Utilities|Casting|CastToB_Weapon_NyxGlass (Components|GetOwner)))
      (ForgivenessTimer|OnNativeForgivenessTimerCanceled _asb_weapon_nyx_glass))))

(event EventOnServerTargetingDataReceived
  (Utilities|SwitchHasAuthority
    (:Authority
      (bind _asb_weapon_nyx_glass (Utilities|Casting|CastToB_Weapon_NyxGlass (Components|GetOwner)))
      (Class|BWeaponNyxGlass|ServeronTargetDataReceived _asb_weapon_nyx_glass))))

(fn BPNyxGlassStateToNativeState
  (params (BPState :type State_NyxGlass%20Enum :default "NewEnumerator0"))
  (switch Utilities|FlowControl|Switch|SwitchonState_NyxGlass BPState
    (:NewEnumerator0
      (LocalVariables|SetNativizedState "Off"))
    (:NewEnumerator2
      (LocalVariables|SetNativizedState "Swinging"))
    (:NewEnumerator7
      (LocalVariables|SetNativizedState "AimingFromGround"))
    (:NewEnumerator5
      (LocalVariables|SetNativizedState "Hovering"))
    (:NewEnumerator6
      (LocalVariables|SetNativizedState "Attacking"))
    (:NewEnumerator9
      (LocalVariables|SetNativizedState "Slashing")))
  (return (LocalVariables|GetNativizedState)))

(fn GetNyxGlassSlashSubstate
  (bind _asb_weapon_nyx_glass (Utilities|Casting|CastToB_Weapon_NyxGlass (Components|GetOwner))
    (:then
      (switch Utilities|FlowControl|Switch|SwitchonSubState_Slash_NyxGlass (Variables|/DryBox/Items/NyxGlass/B_Weapon_NyxGlass.B_Weapon_NyxGlass_C|GetSubState_Slashing _asb_weapon_nyx_glass)
        (:NewEnumerator0
          (LocalVariables|SetNativizedSubstate "MovingToTarget"))
        (:NewEnumerator10
          (LocalVariables|SetNativizedSubstate "StoppedAtTarget"))
        (:NewEnumerator11
          (LocalVariables|SetNativizedSubstate "LeapingBack")))
      (return (LocalVariables|GetNativizedSubstate)))
    (:CastFailed
      (return "MovingToTarget"))))

(fn GetNyxGlassState
  (bind _asb_weapon_nyx_glass (Utilities|Casting|CastToB_Weapon_NyxGlass (Components|GetOwner))
    (:then
      (return (CallFunction|BPNyxGlassStatetoNativeState (Variables|/DryBox/Items/NyxGlass/B_Weapon_NyxGlass.B_Weapon_NyxGlass_C|GetState _asb_weapon_nyx_glass))))
    (:CastFailed
      (return "Off"))))

(fn IsAttacking
  (bind _asb_weapon_nyx_glass (Utilities|Casting|CastToB_Weapon_NyxGlass (Components|GetOwner)))
  (return (Class|BWeaponNyxGlass|IsAttackingBool _asb_weapon_nyx_glass)))

(fn OnVelocityCalculated
  (params (VelocityMetadata :type Nyx%20Glass%20Velocity%20Metadata%20Structure%20%28by%20ref%29))
  (bind _asb_weapon_nyx_glass (Utilities|Casting|CastToB_Weapon_NyxGlass (Components|GetOwner))
    (:then
      (Class|BWeaponNyxGlass|OnNativizedVelocityCalculated _asb_weapon_nyx_glass (Variables|GetVelocityMetadata)))
    (:CastFailed
      _))
  (return))

(fn ShouldUpdateTargeting
  (bind _asb_weapon_nyx_glass (Utilities|Casting|CastToB_Weapon_NyxGlass (Components|GetOwner)))
  (bind _updatetargeting (Class|BWeaponNyxGlass|ShouldUpdateTargeting _asb_weapon_nyx_glass))
  (return _updatetargeting))
