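; Fortnite NyxGlass weapon Blueprint: a fuel-powered swinging grapple with a waist gear rope attachment, ADS hovering and slash attacks.
; An enum state machine and native swinging component drive attach, detach and reticle updates, with server, client and multicast event paths.
; GameplayEffects consume LocalFuel and block firing during refueling, while a forgiveness timer gates cooldown.
; Timed failsafes exit stuck attack, slash or holster states, and timers drive timeout audio.
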(event EventFall_Client
  ;[1 1 1] Player Shot
  (|CallParentFunction)
  (CallFunction|ShowErrorMessage "NSLOCTEXT(\"[E2079C60299E575D39FFDE4CE19D00C1]\", \"E518342F43FF35358F7A0FACC921DBF8\", \"Got Shot\")")
  (Variables|Cache|UI|SetLastFreeFallTime (Utilities|Time|GetGameTimeinSeconds))
  (State|SetStateOff)
  (bind _swingingcontroller (Variables|Default|GetSwingingController)
    (:then
      (Swinging|EndSwingingMovement _swingingcontroller))
    (:else)))

(event EventBeginPlay
  (bind (_returnvalue_location _returnvalue_rotation) (Math|Transform|BreakTransform (Transformation|GetActorTransform (Variables|Default|GetPlayer))))
  ;[1 1 1] Begin Play
  (|CallParentFunction)
  (Utilities|FlowControl|Sequence
    (:then_0
      (CallFunction|SetupCachedScalableFloats))
    (:then_1
      (FortPawn|BindEventtoOnWeaponEquippedDelegate (Variables|Default|GetPlayer) (EventDispatchers|CreateEvent :!event On%20Any%20Weapon%20Equip)))
    (:then_2
      (Ability|Tasks|WaitGameplayTagRemoveFromActor (Variables|Default|GetPlayer) "(TagName=\"Granted.Cooldown.NyxGlass\")"
        (:then
          (CallFunction|SetState "NewEnumerator0"))
        (:Removed
          (Variables|Cache|SetDisableFiring false)
          (CallFunction|HandleTimeoutAudioStopped))))
    (:then_3
      (Ability|GameplayEffect|AddLooseGameplayTags (Variables|Default|GetPlayer) (GameplayTags|MakeLiteralGameplayTagContainer "(GameplayTags=((TagName=\"Gameplay.Action.Player.Swinging.Nyx\")))") true))
    (:then_4
      (bind _returnvalue (Building|SpawnBuildingGameplayActor "/DryBox/Items/NyxGlass/BGA_WaistAttachment_NyxGlass.BGA_WaistAttachment_NyxGlass_C" (Math|Transform|MakeTransform _returnvalue_location _returnvalue_rotation) (Variables|Default|GetPlayer) false false))
      (Transformation|AttachActorToComponent _returnvalue (Variables|/Script/Engine.Character|GetMesh (Variables|Default|GetPlayer)) "pelvis" "SnapToTarget" "KeepWorld" "KeepRelative" true)
      (bind _asbga_waist_attachment_nyx_glass (Utilities|Casting|CastToBGA_WaistAttachment_NyxGlass _returnvalue))
      (Variables|Cache|SetWaistGear _asbga_waist_attachment_nyx_glass)
      (Game|BindEventtoOnDestroyed (delegate Custom|OnDestroy)))
    (:then_5
      (if (Class|AthenaFunctionLibrary|RowtoBool (Variables|ADS|GetSF_FixADSSpriteSwap))
        (EventDispatchers|BindEventtoOnLocalTargetingChanged (EventDispatchers|CreateEvent :!event Custom|OnClientTargetingChanged))))
    (:then_6
      (Rendering|SetVisibility (Variables|Rope|GetRope) false false)
      (Networking|SwitchHasAuthority
        (:Remote
          (if (Math|Boolean|ANDBoolean (Class|AthenaFunctionLibrary|RowtoBool (Variables|Default|GetSF_FixMobileReplication)) (TouchPlatform|IsUsingTouch (CallFunction|GetPlayerController)) (TouchPlatform|IsMobileGame))
            (CallFunction|EventIsTouchControlsServer)))))))

(event EventFall
  ;[1 1 1] Player Shot
  (|CallParentFunction)
  (Class|FortWeaponComponentSwinging|ForceEndForgivenessTimer (Variables|Default|GetNyxGlass%20Weapon%20Swinging%20BP%20Comp))
  (State|SetStateOff)
  (bind _swingingcontroller (Variables|Default|GetSwingingController)
    (:then
      (Swinging|EndSwingingMovement _swingingcontroller))
    (:else)))

(event Custom|OnDestroy
  (params (DestroyedActor :type Actor%20Object%20Reference))
  (bind _waistgear (Variables|Cache|GetWaistGear))
  ;[1 1 1] Begin Play
  (Class|BGAWaistAttachmentNyxGlass|RopeVisualDestroy _waistgear)
  (Actor|DestroyActor _waistgear))

(event Custom|Event%20IsTouchControlsServer
  ;[1 1 1] Begin Play
  (Variables|Default|SetbIsTouchControls true))

(event EventApplyCooldown
  (params (Force :type Boolean))
  ;[1 1 1] Cooldown
  (Variables|Cache|SpeedCalculation|SetTimeSpentInCurrentSwingingSession 0.0)
  (Class|NyxGlassWeaponComponentSwinging|EndSwingingSession (Variables|Default|GetNyxGlass%20Weapon%20Swinging%20BP%20Comp))
  (Networking|SwitchHasAuthority
    (:Authority
      (if (Variables|Cache|FuelAndCooldown|GetRefueling)
        _
        (elif (or Force (<= (FortWeapon|GetMagazineAmmoCount) 0))
          (Class|FortWeaponComponentSwinging|ForceEndForgivenessTimer (Variables|Default|GetNyxGlass%20Weapon%20Swinging%20BP%20Comp))
          (elif (Class|FortWeaponComponentSwinging|IsForgivenessTimerActive (Variables|Default|GetNyxGlass%20Weapon%20Swinging%20BP%20Comp))
            _
            (elif (AI|Components|NavMovement|IsFalling (Variables|/Script/Engine.Character|GetCharacterMovement (Variables|Default|GetPlayer)))
              (Utilities|FlowControl|Delay 0.1)
              (else
                (ForgivenessTimer|StartForgivenessTimer)))))))))

(event EventSlowerUpdate
  (bind (_bsuccessfullyfoundattribute _returnvalue) (Ability|Attribute|GetFloatAttribute (Variables|Default|GetPlayer) "(AttributeName=\"LocalFuel\",Attribute=/Script/DryBoxRuntime.NyxGlassFuelAttributeSet:LocalFuel,AttributeOwner=\"/Script/CoreUObject.Class'/Script/DryBoxRuntime.NyxGlassFuelAttributeSet'\")"))
  (bind _state (Variables|Cache|GetState))
  (bind _returnvalue_2 (Utilities|Enum|Equal(Enum) _state "NewEnumerator9"))
  (bind _state_2 (Variables|Cache|GetState))
  (bind _returnvalue_3 (Utilities|Enum|Equal(Enum) _state_2 "NewEnumerator9"))
  ;[1 1 1] Refueling
  (|CallParentFunction)
  (Utilities|FlowControl|Sequence
    (:then_0
      (Networking|SwitchHasAuthority
        (:Remote
          (CallFunction|HandleDangleAudio))))
    (:then_1
      (Utilities|FlowControl|Sequence
        (:then_0
          (Networking|SwitchHasAuthority
            (:Authority
              (if (>= _returnvalue (ScalableFloat|GetValueatLevel (Variables|Design|FuelandCooldown|GetSF_MaxLocalFuel)))
                (if (Variables|Cache|FuelAndCooldown|GetRefueling)
                  (CallFunction|StopRefuelingMulticast))))))
        (:then_1
          ;;[1 1 1] Maybe start refueling.
          (CallFunction|UpdateRefueling))
        (:then_2
          ;;[1 1 1] Failsafe to ensure player can't get stuck in holstered state
          (if (and (not (CallFunction|IsSwinging)) (Weapon|IsWeaponHolstered (Variables|Default|GetPlayer)))
            (Weapon|SetHolsterWeapon (Variables|Default|GetPlayer) 0 false false "GrappleParent") ; Holster ID ensures this won't get\nin the way of holstering from anything else
          ))
        (:then_3
          ;;[1 1 1] Failsafe for if player is stuck in attack or slash for too long
          (Networking|SwitchHasAuthority
            (:Authority
              (Utilities|FlowControl|Sequence
                (:then_0
                  (if (Variables|Default|GetCached_SF_TimedFailsafe_Attack_Enabled)
                    (if (Utilities|Enum|Equal(Enum) (Variables|Cache|GetState) "NewEnumerator6")
                      (if (> (- (Utilities|Time|GetGameTimeinSeconds) (Variables|Default|GetLastStartedAttackingTime)) (ScalableFloat|GetValueatLevel (Variables|Default|GetTimedFailSafe_Attack_MaxTime)))
                        (CallFunction|Detach false) ; only needs to run on server
                        (Development|PrintString "NYXGLASS_DETACH_attackFailsafe")
                        (CallFunction|ExitSwinging(Multicast))))))
                (:then_1
                  (if (Variables|Default|GetCached_SF_TimedFailsafe_Slash_Enabled)
                    (if _returnvalue_3
                      (if (> (- (Utilities|Time|GetGameTimeinSeconds) (Variables|Default|GetLastStartedSlashingTime)) (+ (ScalableFloat|GetValueatLevel (Variables|Default|GetTimedFailSafe_Slash_MaxTime)) (ScalableFloat|GetValueatLevel (Variables|Design|Offense|Slash|GetSF_SlashLeapBackTime))))
                        (CallFunction|ExitSwinging(Multicast))))))
                (:then_2
                  (if (Variables|Default|GetCached_SF_TimedFailsafe_SlashMoveTowardTarget_Enabled)
                    (if (and _returnvalue_3 (Utilities|Enum|Equal(Enum) (Variables|Cache|Attack|GetSubState_Slashing) "NewEnumerator0"))
                      (if (> (- (Utilities|Time|GetGameTimeinSeconds) (Variables|Default|GetLastStartedSlashingTime)) (ScalableFloat|GetValueatLevel (Variables|Default|GetTimedFailSafe_SlashMoveTowardTarget_MaxTime)))
                        (CallFunction|ExitSwinging(Multicast))))))))
            (:Remote
              ;;[1 1 1] Client Failsafe
              (Utilities|FlowControl|Sequence
                (:then_0
                  (if (Variables|Default|GetCached_SF_TimedFailsafe_Attack_EnabledOnClient)
                    (if (Utilities|Enum|Equal(Enum) (Variables|Cache|GetState) "NewEnumerator6")
                      (if (> (- (Utilities|Time|GetGameTimeinSeconds) (Variables|Default|GetLastStartedAttackingTime)) (ScalableFloat|GetValueatLevel (Variables|Default|GetTimedFailSafe_Attack_ClientTime)))
                        (CallFunction|ExitSwinging(toServer))))))
                (:then_1
                  (if (Variables|Default|GetCached_SF_TimedFailsafe_Slash_EnabledOnClient)
                    (if _returnvalue_2
                      (if (> (- (Utilities|Time|GetGameTimeinSeconds) (Variables|Default|GetLastStartedSlashingTime)) (+ (ScalableFloat|GetValueatLevel (Variables|Default|GetTimedFailSafe_Slash_ClientTimeBuffer)) (ScalableFloat|GetValueatLevel (Variables|Design|Offense|Slash|GetSF_SlashLeapBackTime))))
                        (CallFunction|ExitSwinging(toServer))))))
                (:then_2
                  (if (Variables|Default|GetCached_SF_TimedFailsafe_SlashMoveTowardTarget_EnabledOnClient)
                    (if (and _returnvalue_2 (Utilities|Enum|Equal(Enum) (Variables|Cache|Attack|GetSubState_Slashing) "NewEnumerator0"))
                      (if (> (- (Utilities|Time|GetGameTimeinSeconds) (Variables|Default|GetLastStartedSlashingTime)) (ScalableFloat|GetValueatLevel (Variables|Default|GetTimedFailSafe_SlashMovingTowardTarget_ClientTime)))
                        (CallFunction|ExitSwinging(toServer))))))))))))))

(event FortWeapon|EventOnUnEquippedAndDetached
  (bind _player (Variables|Default|GetPlayer))
  (bind _returnvalue (Ability|GetFortAbilitySystemComponent _player))
  ;[1 1 1] Cooldown Edge Cases
  (|CallParentFunction)
  (Utilities|FlowControl|Sequence
    (:then_0
      (CallFunction|CancelAttackifAttacking (Class|AthenaFunctionLibrary|RowtoBool (Variables|Default|GetCancelSwingingMovementOnUnequip_WhenAttacking)))
      (Variables|Default|SetLastUnequippedTime (Utilities|Time|GetGameTimeinSeconds))
      (GameplayEffects|RemoveActiveEffectswithGrantedTags _returnvalue "(GameplayTags=((TagName=\"Granted.Athena.NyxGlass.UsingGas\")))")
      (GameplayEffects|RemoveActiveEffectswithGrantedTags _returnvalue "(GameplayTags=((TagName=\"Granted.Athena.NyxGlass.BlockAttackUntilReady\")))")
      (Rendering|SetActorHiddenInGame (Variables|Cache|GetWaistGear) true)
      (Audio|Components|Audio|FadeOut (Variables|Default|GetTimeoutAudioComponent) 0.1 0.0 "Linear" :!class /Script/Engine.AudioComponent)
      (Variables|ADS|SetADSPressed false)
      (if (ScalableFloat|GetValueasBool (Variables|Design|FuelAndCooldown|GetSF_TriggerCooldownOnUnequip))
        (bind _isswinging (Class|PlayerPawnAthena|IsSwinging (Variables|Default|GetPlayer)))
        (if (and (not _isswinging) (not (Utilities|OnGround)))
          (Class|FortWeaponComponentSwinging|ForceEndForgivenessTimer (Variables|Default|GetNyxGlass%20Weapon%20Swinging%20BP%20Comp)))))
    (:then_1
      (Cosmetics|UnhideBuiltinCapes))
    (:then_2
      ;;[1 1 1] Ensure detachment occurs if weapon is unequipped and attached
      (if (and (not (FortWeapon|IsEquipped)) (Variables|Default|GetbHasAttached))
        (CallFunction|Detach true true)))))

(event Custom|UpdateKeepHoldingTip
  (Utilities|FlowControl|Sequence
    (:then_0
      (if (== (Variables|Default|GetTimesAttackedSinceLanded) 1)
        (Variables|Default|SetShowKeepHoldingTipTimer (Utilities|Time|GetGameTimeinSeconds))))
    (:then_1
      (Variables|Default|SetTimesAttackedSinceLanded 0))))

(event Custom|StopRefueling_Multicast
  ;[1 1 1] Refueling
  (Variables|Cache|FuelAndCooldown|SetRefueling false)
  (GameplayEffects|RemoveActiveEffectswithGrantedTags (Ability|GetFortAbilitySystemComponent (Variables|Default|GetPlayer)) "(GameplayTags=((TagName=\"Granted.Athena.NyxGlass.BlockUse\"),(TagName=\"Granted.Athena.GrappleParent\")))")
  (Variables|Cache|SetDisableFiring false)
  (CallFunction|PlayReplenishedAudio)
  (if (Variables|ADS|GetADSPressed)
    (ADS|OnADSEnter true)))

(event Custom|ApplyCooldown_Server
  ;[1 1 1] Cooldown
  (bind _player (Variables|Default|GetPlayer)
    (:then
      (if (< (.ReturnValue (Ability|Attribute|GetFloatAttribute _player "(AttributeName=\"LocalFuel\",Attribute=/Script/DryBoxRuntime.NyxGlassFuelAttributeSet:LocalFuel,AttributeOwner=\"/Script/CoreUObject.Class'/Script/DryBoxRuntime.NyxGlassFuelAttributeSet'\")")) (Utilities|ScalableFloattoFloat (Variables|Design|FuelandCooldown|GetSF_MaxLocalFuel)))
        (Variables|Cache|FuelAndCooldown|SetRefueling true)
        (Variables|Cache|FuelAndCooldown|SetTimeStartedRefueling (Utilities|Time|GetGameTimeinSeconds))
        (GameplayEffects|ApplyGameplayEffectToSelf :self (Ability|GetFortAbilitySystemComponent (Variables|Default|GetPlayer)) :GameplayEffectClass "/DryBox/Items/NyxGlass/GE_RefuelingBlockFire_NyxGlass.GE_RefuelingBlockFire_NyxGlass_C" :Level 0.0)
        (CallFunction|UpdateKeepHoldingTip)
        (else
          ;;[1 0.5052 0.5052] Warn: tried to refuel when already full.
          (Fortnite|FortShippingLog self "B_Weapon_NyxGlass::ApplyCooldown_Server(): failed, attempted to refuel when NyxGlassFuelAttributeSet.LocalFuel is already full." true))))
    (:else)))

(event EventSpawnAttachment
  ;[1 1 1] Cooldown Edge Cases
  (|CallParentFunction)
  (Networking|SwitchHasAuthority
    (:Authority
      (Class|FortWeaponComponentSwinging|CancelForgivenessTimer (Variables|Default|GetNyxGlass%20Weapon%20Swinging%20BP%20Comp)))))

(event EventOnCooldownApplied
  ;[1 1 1] Cooldown
  (|CallParentFunction))

(event FortWeapon|EventOnSpawnProjectile
  (params (SpawnProjectile :type Fort%20Projectile%20Base%20Object%20Reference))
  ;[1 1 1] Cooldown Edge Cases
  (|CallParentFunction SpawnProjectile))

(event Custom|OnForgivenessTimerElapsed_BPClientOnly
  ;[1 1 1] Cooldown
  (Variables|Cache|ForgivenessTimer|SetClientForgivenessTimerElapsed true)
  (ForgivenessTimer|StopForgivenessTimerClient))

(event Custom|On%20Native%20Targeting%20Complete
  (params (TargetingData :type Nyx%20Glass%20Targeting%20Data%20Structure))
  (bind (_leftactor _rightactor _leftattachpoint _rightattachpoint _bfoundleftattachpoint _bfoundrightattachpoint _busingdirectraycast) (Utilities|Struct|BreakNyxGlassTargetingData TargetingData))
  (bind (_left_point _rightpoint _leftpointvalid _rightpointvalid) (CallFunction|CheckWhichSidePointsAreOn _leftattachpoint _rightattachpoint _bfoundleftattachpoint _bfoundrightattachpoint))
  ;[1 1 1] Native On Targeting Data calculated
  (if _busingdirectraycast
    _
    (elif (and _bfoundleftattachpoint _bfoundrightattachpoint)
      (Variables|/Script/DryBoxRuntime.NyxGlassWeaponComponent_Swinging|SetCurTargetingData (Utilities|Struct|MakeNyxGlassTargetingData _leftactor _rightactor _left_point _rightpoint _leftpointvalid _rightpointvalid _busingdirectraycast) (Variables|Default|GetNyxGlass%20Weapon%20Swinging%20BP%20Comp)))))

(event EventOnAttach
  (bind _attachpointbga (Variables|Attach|GetAttachPointBGA))
  (bind _returnvalue (Actor|GetAttachParentActor _attachpointbga))
  ;[1 1 1] Attach and Detach
  (Networking|SwitchHasAuthority
    (:Authority
      (if (Variables|Default|GetbServerWaitingForTargetData)
        (Variables|Default|SetServerTriedToAttachWhileWaiting true)
        (else
          _)))
    (:Remote
      _))
  (Variables|Default|SetShouldShowAttackAgainPrompt false)
  (Variables|Cache|SpeedCalculation|SetCurrentGasVec (- (Transformation|GetActorLocation (Variables|Attach|GetAttachPointBGA)) (Transformation|GetActorLocation (Variables|Default|GetPlayer))))
  (Class|NyxGlassWeaponComponentSwinging|OnAttached (Variables|Default|GetNyxGlass%20Weapon%20Swinging%20BP%20Comp))
  (Variables|Default|SetbHasAttached true)
  (bind _returnvalue_2 (Utilities|Time|InvalidateTimerHandle (Variables|Cache|Attach|GetAttachTimerHandle)))
  (Variables|Cache|Attach|SetAttachTimerHandle _returnvalue_2)
  (CallFunction|CleanUpTimeOutAudio)
  (Utilities|FlowControl|Sequence
    (:then_0
      (if (Variables|Cache|Attack|GetAttackOnNextAttach)
        (State|SetStateAttacking "NewEnumerator0")
        ;;[1 1 1] Attachment hook can damage other pawns
        (Utilities|IsValid _returnvalue
          (:Is%20Valid
            (if (!= _returnvalue (Variables|Default|GetPlayer))
              (bind _aspawn (Utilities|Casting|CastToPawn _returnvalue))
              (CallFunction|ApplyDamageGE _aspawn (Variables|Default|GetGE_AttachDamage)))))
        (else
          (CallFunction|SetState "NewEnumerator2")
          (if (Pawn|IsLocallyViewed (Variables|Default|GetPlayer))
            (Audio|PlaySound2D (Variables|Sound|GetSwingSound))
            (Networking|SwitchHasAuthority
              (:Remote
                (if (and (Variables|Status|GetHoldingFire) (> (- (Utilities|Time|GetGameTimeinSeconds) (Variables|Cache|GetLastFirePressedTime)) 0.2))
                  (Variables|Cache|Attach|SetbAttachTimerNotSetDueToInput true)
                  (Variables|Cache|Attach|SetAttachTime (Utilities|Time|GetGameTimeinSeconds))
                  (else
                    (bind _returnvalue_3 (Utilities|Time|SetTimerbyEvent (EventDispatchers|CreateEvent :!event Custom|AttachTimerElapsed) (Utilities|ScalableFloattoFloat (Variables|Design|Attachment|GetSF_MinAttachTime))))
                    (Variables|Cache|Attach|SetAttachTimerHandle _returnvalue_3)))))))))
    (:then_1
      (Class|BGAWaistAttachmentNyxGlass|TurnonRopeVisuals (Variables|Cache|GetWaistGear)))))

(event Custom|Attach%20After%20Server%20Finished%20Waiting%20For%20Target%20Data
  (bind _attachpointbga (Variables|Attach|GetAttachPointBGA))
  (bind _returnvalue (Actor|GetAttachParentActor _attachpointbga))
  ;[1 1 1] Attach and Detach
  (Variables|Default|SetServerTriedToAttachWhileWaiting false)
  (Variables|Default|SetShouldShowAttackAgainPrompt false)
  (Variables|Cache|SpeedCalculation|SetCurrentGasVec (- (Transformation|GetActorLocation (Variables|Attach|GetAttachPointBGA)) (Transformation|GetActorLocation (Variables|Default|GetPlayer))))
  (Class|NyxGlassWeaponComponentSwinging|OnAttached (Variables|Default|GetNyxGlass%20Weapon%20Swinging%20BP%20Comp))
  (Variables|Default|SetbHasAttached true)
  (bind _returnvalue_2 (Utilities|Time|InvalidateTimerHandle (Variables|Cache|Attach|GetAttachTimerHandle)))
  (Variables|Cache|Attach|SetAttachTimerHandle _returnvalue_2)
  (CallFunction|CleanUpTimeOutAudio)
  (Utilities|FlowControl|Sequence
    (:then_0
      (if (Variables|Cache|Attack|GetAttackOnNextAttach)
        (State|SetStateAttacking "NewEnumerator0")
        ;;[1 1 1] Attachment hook can damage other pawns
        (Utilities|IsValid _returnvalue
          (:Is%20Valid
            (if (!= _returnvalue (Variables|Default|GetPlayer))
              (bind _aspawn (Utilities|Casting|CastToPawn _returnvalue))
              (CallFunction|ApplyDamageGE _aspawn (Variables|Default|GetGE_AttachDamage)))))
        (else
          (CallFunction|SetState "NewEnumerator2")
          (if (Pawn|IsLocallyViewed (Variables|Default|GetPlayer))
            (Audio|PlaySound2D (Variables|Sound|GetSwingSound))
            (Networking|SwitchHasAuthority
              (:Remote
                (if (and (Variables|Status|GetHoldingFire) (> (- (Utilities|Time|GetGameTimeinSeconds) (Variables|Cache|GetLastFirePressedTime)) 0.2))
                  (Variables|Cache|Attach|SetbAttachTimerNotSetDueToInput true)
                  (Variables|Cache|Attach|SetAttachTime (Utilities|Time|GetGameTimeinSeconds))
                  (else
                    (bind _returnvalue_3 (Utilities|Time|SetTimerbyEvent (EventDispatchers|CreateEvent :!event Custom|AttachTimerElapsed) (Utilities|ScalableFloattoFloat (Variables|Design|Attachment|GetSF_MinAttachTime))))
                    (Variables|Cache|Attach|SetAttachTimerHandle _returnvalue_3)))))))))
    (:then_1
      (Class|BGAWaistAttachmentNyxGlass|TurnonRopeVisuals (Variables|Cache|GetWaistGear)))))

(event Custom|AttachTimerElapsed
  ;[1 1 1] Attach and Detach
  (Networking|SwitchHasAuthority
    (:Remote
      (if (Default|IsSwingAttached)
        (CallFunction|RequestDetachfromClient false)
        (Development|PrintString "NYXGLASS_DETACH_AttachTimerElapsed")))))

(event EventOnDetachedFromTooSlow(OwningClientOnly)
  ;[1 1 1] On Detached From Too Slow
  (CallFunction|OnDetachedfromTooSlow(Server))
  (Development|PrintString "NYXGLASS_DETACH_tooslow_owningclient"))

(event EventOnDetach
  ;[1 1 1] Attach and Detach
  (Variables|/Script/DryBoxRuntime.NyxGlassWeaponComponent_Swinging|SetCurTargetingData (Utilities|Struct|MakeNyxGlassTargetingData 0) (Variables|Default|GetNyxGlass%20Weapon%20Swinging%20BP%20Comp))
  (bind _returnvalue (Utilities|Time|InvalidateTimerHandle (Variables|Cache|Attach|GetAttachTimerHandle)))
  (Variables|Cache|Attach|SetAttachTimerHandle _returnvalue)
  (Variables|Default|SetbHasAttached false)
  (Variables|Cache|Attach|SetbAttachTimerNotSetDueToInput false)
  (Variables|Cache|Attack|SetAttackOnNextAttach false)
  (Variables|Cache|SpeedCalculation|SetCurrentGasVec "0, 0, 0")
  (Class|NyxGlassWeaponComponentSwinging|OnDetached (Variables|Default|GetNyxGlass%20Weapon%20Swinging%20BP%20Comp))
  (GameplayEffects|RemoveActiveEffectswithGrantedTags (Ability|GetFortAbilitySystemComponent (Variables|Default|GetPlayer)) "(GameplayTags=((TagName=\"Granted.Athena.NyxGlass.UsingGas.Attached\")))")
  (Variables|Cache|SpeedCalculation|SetUsingGas false)
  (Utilities|FlowControl|Sequence
    (:then_0
      (if (Variables|Cache|Attack|GetAutoAttacking)
        (|CallParentFunction)
        (Variables|Cache|Attack|SetAutoAttacking false)
        (else
          (switch Utilities|FlowControl|Switch|SwitchonState_NyxGlass (Variables|Cache|GetState)
            (:NewEnumerator0
              _)
            (:NewEnumerator2
              _))
          (CallFunction|RemoveGearifOutOfFuel)
          (Variables|Cache|SetLastSlingshotTime (Utilities|Time|GetGameTimeinSeconds)))))
    (:then_1
      ;;[1 1 1] Update Rope Visuals
      (Networking|SwitchHasAuthority
        (:Authority
          (CallFunction|RopeReleaseVisuals(MulticastRemoteClients)))
        (:Remote
          (Class|BGAWaistAttachmentNyxGlass|RopeVisualRopeReleaseVisuals (Variables|Cache|GetWaistGear)))))))

(event Custom|On%20Detached%20From%20Too%20Slow%20%28Server%29
  ;[1 1 1] On Detached From Too Slow
  (CallFunction|OnDetachedfromTooSlow(Multicast))
  (Development|PrintString "NYXGLASS_DETACH_tooslow_server"))

(event Custom|On%20Detached%20From%20Too%20Slow%20%28Multicast%29
  ;[1 1 1] On Detached From Too Slow
  (CallFunction|CancelAttackifAttacking (Class|AthenaFunctionLibrary|RowtoBool (Variables|Default|GetCancelSwingingMovementOnStuck_TooSlow)))
  (Development|PrintString "NYXGLASS_DETACH_tooslow_multicast"))

(event EventOnDetachWhenHitWall(Server)
  ;[1 1 1] On Detached From Too Slow
  (CallFunction|OnDetachwhenHitWall(Multicast))
  (Development|PrintString "NYXGLASS_DETACH_hitwall_server"))

(event Custom|ExitSwinging%20%28To%20Server%29
  ;[1 1 1] Refueling
  ;;[1 1 1] Client Failsafe
  (CallFunction|ExitSwinging(Multicast)))

(event Custom|On%20Detach%20When%20Hit%20Wall%20%28Multicast%29
  ;[1 1 1] On Detached From Too Slow
  (Development|PrintString "NYXGLASS_DETACH_hitwall_multicast")
  (if (Class|AthenaFunctionLibrary|RowtoBool (Variables|Default|GetApplyFixToExitAttackWhenDetachedFromHittingWall))
    (CallFunction|CancelAttackifAttacking (Class|AthenaFunctionLibrary|RowtoBool (Variables|Default|GetCancelSwingingMovementOnStuck_HitWall)))))

(event Custom|ExitSwinging%20%28Multicast%29
  ;[1 1 1] Refueling
  ;;[1 1 1] Failsafe for if player is stuck in attack or slash for too long
  (CallFunction|ForceExitSwinging))

(event EventOnSwingingEnded
  (params (SkipRemovingSwingingGE :type Boolean))
  ;[1 1 1] Attach and Detach
  ;;[1 1 1] On Exit Swinging
  (if (Default|IsSwingAttached)
    (CallFunction|Detach false true) ; Anytime exit swinging is requested, make sure detachment occurs, or else parent will fail to exit swinging
    (Development|PrintString "NYXGLASS_DETACH_fromexitswinging")
    (else
      _))
  (|CallParentFunction SkipRemovingSwingingGE)
  (Variables|Default|SetShouldShowAttackAgainPrompt false)
  (Components|Activation|Deactivate (Variables|/DryBox/Items/NyxGlass/BGA_WaistAttachment_NyxGlass.BGA_WaistAttachment_NyxGlass_C|GetNS_GasParticles_NyxGlass (Variables|Cache|GetWaistGear)))
  (Class|BGAWaistAttachmentNyxGlass|RopeVisualDeactivePushForwardFX (Variables|Cache|GetWaistGear))
  (CallFunction|EndLandingBlock)
  (switch Utilities|FlowControl|Switch|SwitchonState_NyxGlass (Variables|Cache|GetState)
    (:NewEnumerator0
      _)
    (:NewEnumerator2
      _)
    (:NewEnumerator7
      (CallFunction|SetState "NewEnumerator7")
      (Networking|SwitchHasAuthority
        (:Remote
          (if (== (Variables|Default|GetTimesAttackedBeforeLanded) 1)
            (Ability|SendGameplayEventtoActor (Variables|Default|GetPlayer) "(TagName=\"Event.Ability.NyxGlass.SingleAttackAndLanded\")")
            (else
              _))
          (Variables|Default|SetTimesAttackedBeforeLanded 0))))
    (:NewEnumerator5
      _)
    (:NewEnumerator6
      _)
    (:NewEnumerator9
      _))
  (CallFunction|SetState "NewEnumerator0"))

(event Rope|EventOnRopeUpdated
  (params (RopeDirection :type Vector :default "0, 0, 0") (RopeLength :type Float%20%28single-precision%29))
  (CallFunction|UpdateRopeFX))

(event Custom|Rope%20Release%20Visuals%20%28Multicast%20Remote%20Clients%29
  ;[1 1 1] Attach and Detach
  ;;[1 1 1] Update Rope Visuals
  (Networking|SwitchHasAuthority
    (:Remote
      (if (not (Pawn|IsLocallyControlled (Variables|Default|GetPlayer)))
        (Class|BGAWaistAttachmentNyxGlass|RopeVisualRopeReleaseVisuals (Variables|Cache|GetWaistGear))))))

(event EventOnSetTargeting
  (params (bNewIsTargeting :type Boolean))
  ;[1 1 1] Input Handling
  (if (Variables|ADS|GetCached_SF_UseADSOverhaul)
    (if bNewIsTargeting
      (ADS|OnADSEnter true)
      (else
        (ADS|OnADSExit)))))

(event Custom|OnClientTargetingChanged
  (params (bIsTargeting :type Boolean))
  (bind _returnvalue (GameplayTags|MakeLiteralGameplayTagContainer "(GameplayTags=((TagName=\"Granted.Athena.NyxGlass.ADS\")))"))
  (bind _player (Variables|Default|GetPlayer))
  ;[1 1 1] Input Handling
  (if bIsTargeting
    (Ability|GameplayEffect|AddLooseGameplayTags _player _returnvalue)
    (else
      (Ability|GameplayEffect|RemoveLooseGameplayTags _player _returnvalue))))

(event Custom|CleanUpTimeOutAudio
  ;[1 1 1] Timeout Audio
  (Utilities|Time|ClearandInvalidateTimerbyHandle (Variables|Audio|GetTimeoutTickAudioTimerHandle))
  (Utilities|Time|ClearandInvalidateTimerbyHandle (Variables|Audio|GetTimeoutAudioDelayTimerHandle))
  (Audio|Components|Audio|FadeOut (Variables|Default|GetTimeoutAudioComponent) 0.1 0.0 "Linear" :!class /Script/Engine.AudioComponent)
  (Variables|Audio|SetTimeoutBeepNumber 0))

(event Custom|ADS_Pressed
  (params (ConsumeFuel :type Boolean :default true))
  ;[1 1 1] Input Handling
  ;;[1 0.09507 0.1914] Old ADS behavior -- used if SF_UseADSOverhaul is false
  (if (not (Variables|ADS|GetCached_SF_UseADSOverhaul))
    (ADS|OnADSEnter ConsumeFuel)))

(event Custom|ADS_Released
  ;[1 1 1] Input Handling
  ;;[1 0.09507 0.1914] Old ADS behavior -- used if SF_UseADSOverhaul is false
  (if (not (Variables|ADS|GetCached_SF_UseADSOverhaul))
    (ADS|OnADSExit)))

(event Custom|HandleTimeoutAudio
  ;[1 1 1] Timeout Audio
  (if (and (Pawn|IsLocallyControlled (Game|GetInstigator)) (not (CallFunction|IsSwinging)))
    (bind _returnvalue (Utilities|Time|SetTimerbyEvent (delegate Custom|StartTimeoutAudio) (ScalableFloat|GetValueatLevel (Variables|Design|FuelAndCooldown|GetSF_WaitTimeBeforeCooldownTrigger))))
    (Variables|Audio|SetTimeoutAudioDelayTimerHandle _returnvalue)
    (if (FortWeapon|IsEquipped)
      (Audio|PlaySound2D (Variables|Audio|GetTimeout_StartWarning)))))

(event EventUpdateReticleFromNative
  (params (HitSomething :type Boolean) (BlockedUse :type Boolean) (IsTargetAtAimLocation :type Boolean) (TargetPosition :type Vector :default "0, 0, 0"))
  (bind (_leftactor _rightactor _leftattachpoint _rightattachpoint _bfoundleftattachpoint _bfoundrightattachpoint _busingdirectraycast) (Utilities|Struct|BreakNyxGlassTargetingData (Variables|/Script/DryBoxRuntime.NyxGlassWeaponComponent_Swinging|GetCurTargetingData (Variables|Default|GetNyxGlass%20Weapon%20Swinging%20BP%20Comp))))
  (bind _targetinghitresult (CallFunction|GetTargetingResult false))
  (bind _bblockinghit (Collision|BreakHitResult _targetinghitresult))
  (bind (_blockedbytag _blockedbynoammo) (CallFunction|IsFiringBlocked))
  (bind (_bsuccessfullyfoundattribute _returnvalue) (Ability|Attribute|GetFloatAttribute (Variables|Default|GetPlayer) "(AttributeName=\"LocalFuel\",Attribute=/Script/DryBoxRuntime.NyxGlassFuelAttributeSet:LocalFuel,AttributeOwner=\"/Script/CoreUObject.Class'/Script/DryBoxRuntime.NyxGlassFuelAttributeSet'\")"))
  ;[1 1 1] Update Reticle
  (Utilities|FlowControl|Sequence
    (:then_0
      (Default|CallUpdateGraceUI (Utilities|Time|IsTimerActivebyHandle (Variables|Cache|ForgivenessTimer|GetForgivenessTimer_BP)) (Math|Float|SelectFloat (Utilities|RemapFloat (/ (Utilities|Time|GetTimerElapsedTimebyHandle (Variables|Cache|ForgivenessTimer|GetForgivenessTimer_BP)) (ScalableFloat|GetValueatLevel (Variables|/Script/FortniteGame.FortWeaponComponent_Swinging|GetDefaultForgivenessTimer (Variables|Default|GetNyxGlass%20Weapon%20Swinging%20BP%20Comp)))) "(X=0.000000,Y=1.000000)" "(X=0.000000,Y=0.750000)" true) 0.0 (Utilities|Time|IsTimerActivebyHandle (Variables|Cache|ForgivenessTimer|GetForgivenessTimer_BP)))))
    (:then_1
      (Variables|Targeting|SetTargetingSomething _bblockinghit)
      (Variables|Default|GetPlayer
        (:then
          (Default|CallDoubleReticleUpdate _bblockinghit (Default|IsSwingAttached) (or (Math|Boolean|ANDBoolean (not BlockedUse) (not (Variables|Cache|GetDisableFiring)) (CallFunction|FuelIsUsable)) (GameplayTags|HasMatchingGameplayTag (Variables|Default|GetPlayer) "(TagName=\"Granted.Athena.GrappleParent.BlockUseAfterDetach\")")) (Math|Boolean|ORBoolean _blockedbytag _blockedbynoammo (Variables|Default|GetDisableFiringAfterSlash)) _leftattachpoint _bfoundleftattachpoint _rightattachpoint _bfoundrightattachpoint _busingdirectraycast (CallFunction|IsAttackingBool) _returnvalue (Variables|Cache|SpeedCalculation|GetUsingGas) (GameplayTags|HasAnyMatchingGameplayTags (Variables|Default|GetPlayer) (GameplayTags|MakeLiteralGameplayTagContainer "(GameplayTags=((TagName=\"Granted.Athena.NyxGlass.UsingGas.Attached.Attacking\"),(TagName=\"Granted.Athena.NyxGlass.UsingGas.Attached.Pulling\"),(TagName=\"Granted.Athena.NyxGlass.UsingGas.Hovering\")))")))
          (Utilities|FlowControl|Sequence
            (:then_0
              (if (not (Variables|Status|GetHoldingFire))
                (Variables|Status|SetHasDetachedSinceFireWasPressed false)))))
        (:else)))))

(event Custom|HoverFromThunderSpears
  ;[1 1 1] Input Handling
  (ADS|OnADSEnter false))

(event Custom|StartTimeoutAudio
  ;[1 1 1] Timeout Audio
  (if (not (CallFunction|IsSwinging))
    (bind _returnvalue (Utilities|Time|SetTimerbyEvent (delegate Custom|TickTimeoutSound) 1.0 true false (- 1.0 (ScalableFloat|GetValueatLevel (Variables|Design|FuelAndCooldown|GetSF_WaitTimeBeforeCooldownTrigger)))))
    (Variables|Audio|SetTimeoutTickAudioTimerHandle _returnvalue)
    (if (FortWeapon|IsEquipped)
      (bind _returnvalue_2 (Audio|SpawnSoundAttached :Sound (Variables|Audio|GetTimeout_Start) :AttachToComponent (Variables|Visual|GetWeaponMesh) :AttachPointName "None" :LocationType "KeepRelativeOffset" :bStopWhenAttachedToDestroyed true :VolumeMultiplier 1.0 :PitchMultiplier 1.0 :StartTime 0.0 :bAutoDestroy true))
      (Variables|Default|SetTimeoutAudioComponent _returnvalue_2)
      (bind _returnvalue_3 (Audio|SpawnSoundAttached (Variables|Audio|GetTimeout_Loop) (Variables|Visual|GetWeaponMesh)))
      (Variables|Default|SetTimeoutLoopComponent _returnvalue_3))))

(event Custom|StartHoveringAndConsumeFuel
  ;[1 1 1] Input Handling
  (State|SetStateHovering true))

(event EventFirePressed
  ;[1 1 1] Input Handling
  (Variables|Cache|SetLastFirePressedTime (Utilities|Time|GetGameTimeinSeconds))
  (if (Math|Boolean|ANDBoolean (not (Variables|Cache|FuelAndCooldown|GetRefueling)) (or (CallFunction|HasLocalFuel) (CallFunction|IsAttackingBool)) (not (Variables|Status|GetFreeFalling)))
    (Variables|Status|SetHoldingFire true)
    (Class|NyxGlassWeaponComponentSwinging|OnBPWeaponFired (Variables|Default|GetNyxGlass%20Weapon%20Swinging%20BP%20Comp))
    (if (CallFunction|IsAimingAttack)
      (CallFunction|StartAttack)
      (if (CallFunction|IsSwinging)
        (Variables|Default|SetPlayerLearnedToAttackFromAir true)
        (Networking|SwitchHasAuthority
          (:Remote
            _))
        (else
          _))
      (Math|Integer|IncrementInt (Variables|Default|GetTimesAttackedBeforeLanded))
      (else
        (switch Utilities|FlowControl|Switch|SwitchonState_NyxGlass (Variables|Cache|GetState)
          (:NewEnumerator0
            _)
          (:NewEnumerator2
            _)
          (:NewEnumerator7
            _)
          (:NewEnumerator5
            _))
        (|CallParentFunction)))))

(event Custom|ForceFirePressed
  ;[1 1 1] Input Handling
  (|CallParentFunction))

(event Custom|TickTimeoutSound
  ;[1 1 1] Timeout Audio
  (if (and (GameplayTags|HasMatchingGameplayTag (Variables|Default|GetPlayer) "(TagName=\"Granted.Cooldown.NyxGlass\")") (not (CallFunction|IsSwinging)))
    (Math|Integer|IncrementInt (Variables|Audio|GetTimeoutBeepNumber)
      (:%20%20
        (if (FortWeapon|IsEquipped)
          (Audio|PlaySound2D (Variables|Audio|GetTimeout_Tick) (ScalableFloat|GetValueatLevel (Variables|Audio|GetSF_BeepMultiplier) (Math|Conversions|ToFloat(Integer) (Variables|Audio|GetTimeoutBeepNumber)))))))
    (else
      (CallFunction|CleanUpTimeOutAudio))))

(event EventFireReleased
  (bind _returnvalue (Utilities|Time|GetGameTimeinSeconds))
  (bind _sf_minattachtime (Variables|Design|Attachment|GetSF_MinAttachTime))
  (bind _returnvalue_2 (Utilities|ScalableFloattoFloat _sf_minattachtime))
  (bind _attachtime (Variables|Cache|Attach|GetAttachTime))
  (bind _returnvalue_3 (- _returnvalue _attachtime))
  ;[1 1 1] Input Handling
  (Development|PrintString "NYXGLASS_normalDet_firereleased")
  (Variables|Status|SetHoldingFire false)
  (if (Variables|Cache|Attach|GetbAttachTimerNotSetDueToInput)
    (Variables|Cache|Attach|SetbAttachTimerNotSetDueToInput false)
    (if (> _returnvalue_3 _returnvalue_2)
      (CallFunction|AttachTimerElapsed)
      (else
        (bind _returnvalue_4 (Utilities|Time|SetTimerbyEvent (EventDispatchers|CreateEvent :!event Custom|AttachTimerElapsed) (- _returnvalue_2 _returnvalue_3)))
        (Variables|Cache|Attach|SetAttachTimerHandle _returnvalue_4)))))

(event Custom|HandleTimeoutAudioStopped
  ;[1 1 1] Timeout Audio
  (if (and (FortWeapon|IsEquipped) (Utilities|Time|IsTimerActivebyHandle (Variables|Audio|GetTimeoutTickAudioTimerHandle)))
    (CallFunction|CleanUpTimeOutAudio)
    (if (and (FortWeapon|IsEquipped) (Pawn|IsLocallyControlled (Game|GetInstigator)))
      (Audio|PlaySound2D (Variables|Audio|GetTimeout_Stop)))))

(event Custom|HandleDangleAudio
  (bind _danglingloop (Variables|Default|GetDanglingLoop))
  ;[1 1 1] Timeout Audio
  (if (Variables|Status|GetbIsAttached)
    (if (Audio|Components|Audio|IsPlaying (Variables|Default|GetDanglingLoop))
      (Audio|Parameter|SetFloatParameter _danglingloop "DangleSpeed" (Math|Vector|VectorLength (Variables|/Script/Engine.MovementComponent|GetVelocity (Pawn|GetMovementComponent (Variables|Default|GetPlayer)))) :!class /Script/AudioExtensions.AudioParameterControllerInterface)
      (else
        (Audio|Components|Audio|FadeIn _danglingloop 0.5 1.0 0.0 "Linear" :!class /Script/Engine.AudioComponent)))
    (elif (Audio|Components|Audio|IsPlaying (Variables|Default|GetDanglingLoop))
      (Audio|Components|Audio|Stop (Variables|Default|GetDanglingLoop)))))

(event Custom|ShowErrorMessage
  (params (ErrorMessage :type Text))
  ;[1 1 1] Input Handling
  (Variables|Cache|UI|SetCurrentErrorMessage ErrorMessage)
  (Variables|Cache|UI|SetErrorMessageSetTime (Utilities|Time|GetGameTimeinSeconds)))

(event FortWeapon|EventOnEquippedAndAttached
  ;[1 1 1] Timeout Audio
  (|CallParentFunction)
  (Utilities|FlowControl|Sequence
    (:then_0
      (Rendering|SetActorHiddenInGame (Variables|Cache|GetWaistGear) false)
      (if (Utilities|Time|IsTimerActivebyHandle (Variables|Audio|GetTimeoutTickAudioTimerHandle))
        (bind _timeoutloopcomponent (Variables|Default|GetTimeoutLoopComponent)
          (:then
            (Audio|Components|Audio|FadeIn _timeoutloopcomponent 0.1 1.0 0.0 "Linear" :!class /Script/Engine.AudioComponent))
          (:else
            (bind _returnvalue (Audio|SpawnSoundAttached (Variables|Audio|GetTimeout_Loop) (Variables|Visual|GetWeaponMesh)))
            (Variables|Default|SetTimeoutLoopComponent _returnvalue)))))
    (:then_1
      (Cosmetics|HideBuiltinCapes))))

(event Custom|BlockSwingingEnd
  (Variables|Default|SetBlockSwingEndOnLand true)
  (Utilities|Time|SetTimerbyEvent (delegate Custom|EndLandingBlock) 4.0) ; Failsafe
)

(event EventConsumeResourcesFromFiring
  (bind _attacking (CallFunction|IsAttackingBool))
  (bind _player (Variables|Default|GetPlayer))
  (bind _returnvalue (Ability|GetFortAbilitySystemComponent _player))
  ;[1 1 1] Consume Resources From Firing
  (GameplayEffects|ApplyGameplayEffectToSelf :self _returnvalue :GameplayEffectClass (select _attacking "/DryBox/Items/NyxGlass/GE_UseGas_Attack_Instant.GE_UseGas_Attack_Instant_C" "/DryBox/Items/NyxGlass/GE_UseGas_Swing_Instant.GE_UseGas_Swing_Instant_C") :Level 0.0)
  (GameplayEffects|ApplyGameplayEffectToSelf :self _returnvalue :GameplayEffectClass (select _attacking "/DryBox/Items/NyxGlass/GE_UseGas_Attack_OverTime.GE_UseGas_Attack_OverTime_C" "/DryBox/Items/NyxGlass/GE_UseGas_Swing_OverTime.GE_UseGas_Swing_OverTime_C") :Level 0.0)
  (Variables|Cache|SpeedCalculation|SetUsingGas true))

(event Custom|EndLandingBlockAfterDelay
  (Utilities|Time|SetTimerbyEvent (delegate Custom|EndLandingBlock) 0.1))

(event HitCheck|EventOnPlayerHitWhileSwinging
  (params (Hit :type Hit%20Result%20Structure%20%28by%20ref%29)))

(event Custom|EndLandingBlock
  (if (Variables|Default|GetBlockSwingEndOnLand)
    (Variables|Default|SetBlockSwingEndOnLand false)
    (CallFunction|RecheckLanding)))

(event Custom|SlashDamageAndFX
  (CallFunction|LandAttack)
  (GameplayCue|ExecuteGameplayCueOnActor(Burst) (Variables|Default|GetPlayer) (Variables|Default|GetGC_AttackSlash) (Ability|GameplayCue|MakeGameplayCueParameters :NormalizedMagnitude 0.0 :RawMagnitude 0.0 :Location (Variables|Cache|Attack|GetSlashingTarget) :Normal "0,0,0" :Instigator (Variables|Default|GetPlayer) :EffectCauser self :SourceObject (Variables|Default|GetPlayer) :GameplayEffectLevel 1 :AbilityLevel 1 :bReplicateLocationWhenUsingMinimalRepProxy false))
  (Development|PrintString "ShouldExecuteAttackSlashGC" true true "(R=1.000000,G=0.121433,B=0.000000,A=1.000000)"))

(event Custom|CancelAttackIfAttacking
  (params (CancelAttackByExitingSwingingMovement :type Boolean))
  ;[1 1 1] Cancel Attack
  (if (Variables|Default|GetCachedFixedStoredAttackState)
    (Variables|Cache|Attack|SetAttackOnNextAttach false)
    (else
      _))
  (switch Utilities|FlowControl|Switch|SwitchonState_NyxGlass (Variables|Cache|GetState)
    (:NewEnumerator6
      (CallFunction|SetState "NewEnumerator2")
      (CallFunction|EndLandingBlock)
      (if CancelAttackByExitingSwingingMovement
        (CallFunction|ForceExitSwinging)))
    (:NewEnumerator9
      (CallFunction|ForceExitSwinging))))

(event Custom|LandAttack
  ;[1 1 1] Attack
  ;;[1 1 1] Land Attack
  (Networking|SwitchHasAuthority
    (:Authority
      (Variables|Cache|SetLaunchBackwardOnNextDetach true)
      (bind _outhits (Collision|MultiSphereTraceForObjects (Transformation|GetActorLocation (Variables|Default|GetPlayer)) (+ (Transformation|GetActorLocation (Variables|Default|GetPlayer)) (Math|Vector|VectorClampSizeMax (- (Transformation|GetActorLocation (Variables|Attach|GetAttachPointBGA)) (Transformation|GetActorLocation (Variables|Default|GetPlayer))) (Utilities|ScalableFloattoFloat (Variables|Design|Offense|Attack|GetSF_AttackStartSlashDistance)))) (Utilities|ScalableFloattoFloat (Variables|Design|Offense|Attack|GetSF_AttackRadius)) (Utilities|Array|MakeArray "ObjectTypeQuery3" "ObjectTypeQuery1" "ObjectTypeQuery5" "ObjectTypeQuery2") false (Utilities|Array|MakeArray (Variables|Default|GetPlayer))))
      (Audio|PlaySoundatLocation 0 (Transformation|GetActorLocation))
      (CallFunction|DamageActorsServer (CallFunction|GetActorsfromHitArray _outhits)))))

(event Custom|AfterSlashEvents
  (params (New%20Velocity :type Vector))
  ;[1 1 1] Attack
  ;;[1 1 1] After Slash
  (Math|Integer|IncrementInt (Variables|Default|GetTimesAttackedSinceLanded)
    (:%20%20
      (CallFunction|EndLandingBlockAfterDelay)
      (if (Variables|ADS|GetADSPressed)
        (ADS|OnADSEnter true)))))

(event Custom|DamageActors_Server
  (params (Array :type Array%20of%20Actor%20Object%20References))
  ;[1 1 1] Attack
  ;;[1 1 1] Deal Damage
  (CallFunction|DamageActorsMulticast Array))

(event Custom|StartAttack
  ;[1 1 1] Attack
  ;;[1 1 1] Start Attack
  (if (Collision|BreakHitResult (CallFunction|GetTargetingResult false))
    (Variables|Cache|Attack|SetAttackOnNextAttach true)
    (Variables|Status|SetHasDetachedSinceFireWasPressed false)
    (|CallParentFunction)
    (else
      (Networking|SwitchHasAuthority
        (:Authority
          (if (not (Variables|Default|GetbServerWaitingForTargetData))
            (Variables|Default|SetbServerWaitingForTargetData true)))
        (:Remote
          _))
      (switch Utilities|FlowControl|Switch|SwitchonState_NyxGlass (Variables|Cache|GetState)
        (:NewEnumerator0
          (CallFunction|SetState "NewEnumerator0"))
        (:NewEnumerator2
          _)
        (:NewEnumerator7
          (FailedAction|UnabletoPerformAction (Variables|Default|GetPlayer) false (GameplayTags|MakeLiteralGameplayTagContainer "(GameplayTags=((TagName=\"Athena.Ability.Failed.NyxGlass.OutOfRange\")))")) ; show reason for failing on screen if one is given
        )
        (:NewEnumerator5
          _)
        (:NewEnumerator6
          _)
        (:NewEnumerator9
          _))
      (CallFunction|SetState "NewEnumerator2"))))

(event Custom|DamageActors_Multicast
  (params (Array :type Array%20of%20Actor%20Object%20References))
  ;[1 1 1] Attack
  ;;[1 1 1] Deal Damage
  (for _array_element Array
    (bind _aspawn (Utilities|Casting|CastToPawn _array_element))
    (Utilities|CheckLineOfSight _aspawn
      (:LOS
        (CallFunction|ApplyDamageGE _aspawn (Variables|Cache|Attack|GetGE_AttackDamage))
        (Audio|PlaySoundatLocation 0 (Transformation|GetActorLocation))
        (GameplayCue|ExecuteGameplayCue _array_element (Variables|Default|GetGC_AttackPlayer) (Class|AthenaFunctionLibrary|MakeSimpleFortEffectContext _array_element _array_element)))))
  (for _array_element_2 Array
    (Utilities|Casting|CastToPawn _array_element_2
      (:CastFailed
        (CallFunction|ApplyDamageGE _array_element_2 (Variables|Cache|Attack|GetGE_AttackDamage))
        (Audio|PlaySoundatLocation 0 (Transformation|GetActorLocation))))))

(event Custom|Play%20Attack%20Hit%20FX
  (params (Object :type Object%20Reference))
  ;[1 1 1] Attack
  ;;[1 1 1] Deal Damage
  (Networking|SwitchHasAuthority))

(event Custom|ForgivenessTimerStarted
  (Networking|SwitchHasAuthority
    (:Remote
      (ForgivenessTimer|StartForgivenessTimer))))

(event Custom|ForgivenessTimerFinished
  (Networking|SwitchHasAuthority
    (:Remote
      (ForgivenessTimer|OnNativeForgivenessTimerEnded))))

(event Custom|ForgivenessTimerCanceled
  (Networking|SwitchHasAuthority
    (:Remote
      (ForgivenessTimer|OnNativeForgivenessTimerCanceled))))

(event Custom|Server_OnTargetDataReceived
  (Networking|SwitchHasAuthority
    (:Authority
      (if (Variables|Default|GetbServerWaitingForTargetData)
        (CallFunction|StartAttack)
        (Variables|Default|SetbServerWaitingForTargetData false)
        (if (Variables|Default|GetServerTriedToAttachWhileWaiting)
          (CallFunction|AttachAfterServerFinishedWaitingforTargetData))))))

(event Custom|RemoteClients_SetState
  (params (New%20State :type State_NyxGlass%20Enum :default "NewEnumerator0"))
  (Networking|SwitchHasAuthority
    (:Remote
      (bind _player (Variables|Default|GetPlayer)
        (:then
          (if (not (Pawn|IsLocallyControlled _player))
            (CallFunction|SetState New%20State)))
        (:else)))))

(event Custom|Server_OutOfFuel
  (Networking|SwitchHasAuthority
    (:Authority
      (bind _player (Variables|Default|GetPlayer)
        (:then
          (Inventory|RemoveItemFromPlayer (Fortnite|GetFortPlayerControllerfromActor _player) "/DryBox/Items/NyxGlass/AGID_NyxGlass.AGID_NyxGlass")
          (CallFunction|MulticastOutOfFuel))
        (:else)))))

(event Custom|Multicast_OutOfFuel
  (bind _waistgear (Variables|Cache|GetWaistGear)
    (:then
      (Class|BGAWaistAttachmentNyxGlass|RopeVisualDestroy _waistgear))
    (:else
      _))
  (CallFunction|CancelAttackifAttacking true)
  (State|SetStateOff)
  (if (ScalableFloat|GetValueasBool (Variables|Default|GetExitSwingingOnItemDropped))
    (Utilities|FlowControl|DoOnce false
      (:Completed
        (CallFunction|ForceExitSwinging)))))

(fn ApplyDamageGE
  (params (Target :type Actor%20Object%20Reference) (DamageGE :type Gameplay%20Effect%20Class%20Reference))
  (Utilities|FlowControl|Sequence
    (:then_0
      (Utilities|IsValid Target
        (:Is%20Valid
          (Utilities|IsValid (Ability|GetFortAbilitySystemComponent Target)
            (:Is%20Valid
              (if (Utilities|IsValidClass DamageGE)
                (bind _player (Variables|Default|GetPlayer)
                  (:then
                    (Utilities|IsValid (Ability|GetFortAbilitySystemComponent _player)
                      (:Is%20Valid
                        (Ability|GameplayEffect|AddLooseGameplayTags (Variables|Default|GetPlayer) (Variables|Default|GetAttackDamageSourceTags)) ; Add the Abilities.Weapons.Melee tag to the attacking pawn so\nanalytics can determine what the source of the damage is.\nNOTE: This would typically be applied by a GA, but because this \ndamage isn't being applied in a GA, there isn't a great place to add it.
                        (GameplayEffects|ApplyGameplayEffectToTarget (Ability|GetFortAbilitySystemComponent (Variables|Default|GetPlayer)) (Variables|GetDamageGE) (Ability|GetFortAbilitySystemComponent (Variables|GetTarget)) 0.0 (GameplayEffects|MakeFortEffectContext (Ability|GetFortAbilitySystemComponent (Variables|Default|GetPlayer)) (Variables|Default|GetPlayer) self 0 0))
                        (Ability|GameplayEffect|RemoveLooseGameplayTags (Variables|Default|GetPlayer) (Variables|Default|GetAttackDamageSourceTags)) ; Immediately remove the damage source tags.
                        (return))))
                  (:else))))))))
    (:then_1
      ;[1 0.06198 0.09183] Error: Validation failed.
      (Fortnite|FortShippingLog self (Utilities|String|ToString(Text) (Utilities|Text|FormatText "NSLOCTEXT(\"[E2079C60299E575D39FFDE4CE19D00C1]\", \"5D30226B4953EE9B204E2787EC2F9914\", \"B_Weapon_NyxGlass::ApplyDamageGE - Failed to apply damage GE to target. Input data could not be validated.\")")) true))))

(fn BroadcastCooldownMessage
  (params (CooldownPercent :type Float%20%28double-precision%29))
  (Networking|SwitchHasAuthority
    (:Remote
      (GameplayEventRouter|BroadcastEvent (GameplayEventRouter|GetGlobalEventRouter) (Utilities|Struct|MakeDirectProgressUpdateData (Variables|GetCooldownPercent) "/DryBox/Items/NyxGlass/AGID_NyxGlass.AGID_NyxGlass") 0))))

(fn CalculateDetachmentForce
  (params (VectorToAttachPoint :type Vector :default "0, 0, 0") (Launch_ :type Boolean) (LaunchVector_ :type Vector :default "0, 0, 0"))
  (return (Math|Vector|SelectVector (|CallParentFunction VectorToAttachPoint Launch_ LaunchVector_) (Math|Vector|SelectVector (|CallParentFunction VectorToAttachPoint Launch_ LaunchVector_) (Math|Vector|MakeVector (.x (Utilities|Operators|Multiply (Utilities|Operators|Add (Math|Vector|SelectVector (Math|Vector|Normalize2D(Vector) (Transformation|GetVelocity (Variables|Default|GetPlayer))) (Math|Vector|Normalize (Transformation|GetVelocity (Variables|Default|GetPlayer))) (< (.z (Transformation|GetVelocity (Variables|Default|GetPlayer))) 0)) (Math|Vector|MakeVector 0.0 0.0 (ScalableFloat|GetValueatLevel (Variables|Default|GetSF_DetachmentUpForceMultiplier)))) (Math|Vector|VectorLength (|CallParentFunction VectorToAttachPoint Launch_ LaunchVector_)) 1.15)) (.y (Utilities|Operators|Multiply (Utilities|Operators|Add (Math|Vector|SelectVector (Math|Vector|Normalize2D(Vector) (Transformation|GetVelocity (Variables|Default|GetPlayer))) (Math|Vector|Normalize (Transformation|GetVelocity (Variables|Default|GetPlayer))) (< (.z (Transformation|GetVelocity (Variables|Default|GetPlayer))) 0)) (Math|Vector|MakeVector 0.0 0.0 (ScalableFloat|GetValueatLevel (Variables|Default|GetSF_DetachmentUpForceMultiplier)))) (Math|Vector|VectorLength (|CallParentFunction VectorToAttachPoint Launch_ LaunchVector_)) 1.15)) (.z (Utilities|Operators|Multiply (Utilities|Operators|Add (Math|Vector|SelectVector (Math|Vector|Normalize2D(Vector) (Transformation|GetVelocity (Variables|Default|GetPlayer))) (Math|Vector|Normalize (Transformation|GetVelocity (Variables|Default|GetPlayer))) (< (.z (Transformation|GetVelocity (Variables|Default|GetPlayer))) 0)) (Math|Vector|MakeVector 0.0 0.0 (ScalableFloat|GetValueatLevel (Variables|Default|GetSF_DetachmentUpForceMultiplier)))) (Math|Vector|VectorLength (|CallParentFunction VectorToAttachPoint Launch_ LaunchVector_)) 1.15))) (Variables|GetLaunch_)) (Class|AthenaFunctionLibrary|RowtoBool (Variables|Default|GetSF_UseOriginalWebShootersDetachmentForce)))))

(fn CheckWhichSidePointsAreOn
  (params (PointOne :type Vector) (PointTwo :type Vector) (PointOneValid :type Boolean) (PointTwoValid :type Boolean))
  (bind (_aimvector _startlocation _aimrightvector) (Utilities|CalculateAimingVector))
  (bind _returnvalue (- PointOne _startlocation))
  (bind _returnvalue_2 (Math|Vector|Normalize _returnvalue))
  (bind _returnvalue_3 (Math|Vector|DotProduct _returnvalue_2 _aimrightvector))
  (bind _returnvalue_4 (- PointTwo _startlocation))
  (bind _returnvalue_5 (Math|Vector|Normalize _returnvalue_4))
  (bind _returnvalue_6 (Math|Vector|DotProduct _aimrightvector _returnvalue_5))
  (if (and PointOneValid PointTwoValid)
    (if (> _returnvalue_3 _returnvalue_6)
      (return (Variables|GetPointTwo) (Variables|GetPointOne) true true)
      (else
        (return (Variables|GetPointOne) (Variables|GetPointTwo) true true)))
    (else
      (if (Variables|GetPointOneValid)
        (if (> _returnvalue_3 0)
          (return "0, 0, 0" (Variables|GetPointOne) false true)
          (else
            (return (Variables|GetPointOne) "0, 0, 0" true false)))
        (elif (> _returnvalue_6 0)
          (return "0, 0, 0" (Variables|GetPointTwo) false true)
          (else
            (return (Variables|GetPointTwo) "0, 0, 0" true false)))))))

(fn DelayBetweenRopeAttachAndMovement
  (return (or (not (CallFunction|IsAttackingBool)) (not (CallFunction|IsSwinging))) (Math|Float|SelectFloat (Utilities|ScalableFloattoFloat (Variables|Design|Attachment|GetSF_TimeToFullyAttachSwinging)) (Utilities|ScalableFloattoFloat (Variables|Design|Attachment|GetSF_TimeToFullyAttachGrounded)) (CallFunction|IsSwinging))))

(fn FuelIsUsable
  (return (or (and (not (Variables|Cache|FuelAndCooldown|GetRefueling)) (CallFunction|HasLocalFuel)) (CallFunction|IsAttackingBool))))

(fn GetActorsFromHitArray
  (params (Hits :type Array%20of%20Hit%20Result%20Structures))
  (for _array_element Hits
    (Utilities|Array|AddUnique (LocalVariables|GetActor%20Array) (.HitActor (Collision|BreakHitResult _array_element))))
  (return (LocalVariables|GetActor%20Array)))

(fn GetTargetingResult
  (params (GettingFireTarget :type Boolean))
  (bind _nyxglass_weapon_swinging_bp_comp (Variables|Default|GetNyxGlass%20Weapon%20Swinging%20BP%20Comp))
  (bind _curtargetingdata (Variables|/Script/DryBoxRuntime.NyxGlassWeaponComponent_Swinging|GetCurTargetingData _nyxglass_weapon_swinging_bp_comp))
  (bind (_leftactor _rightactor _leftattachpoint _rightattachpoint _bfoundleftattachpoint _bfoundrightattachpoint _busingdirectraycast) (Utilities|Struct|BreakNyxGlassTargetingData _curtargetingdata))
  (Utilities|FlowControl|Sequence
    (:then_0
      (if _busingdirectraycast
        (LocalVariables|SetFinalTargetingHitResult (Collision|MakeHitResult _bfoundleftattachpoint false 0.0 0.0 _leftattachpoint _leftattachpoint "0,0,1" "0,0,1" 0 _leftactor))
        (else
          (LocalVariables|SetFinalTargetingHitResult (Collision|MakeHitResult (or _bfoundleftattachpoint _bfoundrightattachpoint) false 0.0 0.0 (Math|Vector|SelectVector (+ (* (- _leftattachpoint _rightattachpoint) 0.5) _rightattachpoint) (Math|Vector|SelectVector _leftattachpoint _rightattachpoint _bfoundleftattachpoint) (and _bfoundleftattachpoint _bfoundrightattachpoint)) (Math|Vector|SelectVector (+ (* (- _leftattachpoint _rightattachpoint) 0.5) _rightattachpoint) (Math|Vector|SelectVector _leftattachpoint _rightattachpoint _bfoundleftattachpoint) (and _bfoundleftattachpoint _bfoundrightattachpoint)) "0,0,1" "0,0,1" 0 _leftactor)))))
    (:then_1
      (return (LocalVariables|GetFinalTargetingHitResult)))))

(fn HasLocalFuel
  (bind _player (Variables|Default|GetPlayer)
    (:then
      (return (> (.ReturnValue (Ability|Attribute|GetFloatAttribute _player "(AttributeName=\"LocalFuel\",Attribute=/Script/DryBoxRuntime.NyxGlassFuelAttributeSet:LocalFuel,AttributeOwner=\"/Script/CoreUObject.Class'/Script/DryBoxRuntime.NyxGlassFuelAttributeSet'\")")) 0)))
    (:else
      (return false))))

(fn HideBuiltInCapes
  (bind (_swaps _swapid) (Utilities|Struct|BreakFortCosmeticPropertySwapData (LocalVariables|GetBuiltInCapeSwap)))
  (bind _outproperty (Class|CosmeticDataComponent%28Experimental%29|FindProperty (Cosmetic|GetCosmeticDataComponent (Variables|Default|GetPlayer)) "(TagName=\"LoadoutSlot.Character\")" "(TagName=\"CosmeticProperty.Swap.HideBuiltInCape\")")
    (:True
      (bind _output_get (LocalVariables|SetBuiltInCapeSwap _outproperty))
      (if (Guid|IsValidGuid (.SwapId (Utilities|Struct|BreakFortCosmeticPropertySwapData _output_get)))
        (Variables|Cosmetics|SetCapeSwapId (.SwapId (Utilities|Struct|BreakFortCosmeticPropertySwapData _output_get)))
        (Class|FortPlayerPawn|SwapPawnCosmeticItemandVariant (Variables|Default|GetPlayer) _swapid _swaps true "Active" false)))))

(fn InitialAccelerationAlpha
  (if (<= (Variables|Cache|SpeedCalculation|GetTimeSpentInCurrentSwingingSession) 0)
    (return (Class|NyxGlassWeaponComponentSwinging|CalcInitialAccelerationAlpha (Variables|Default|GetNyxGlass%20Weapon%20Swinging%20BP%20Comp)))
    (else
      (return (Math|Curves|GetFloatValue (Variables|Cache|SpeedCalculation|GetInitialAccelerationCurve) (Utilities|NormalizeFloat (Variables|Cache|SpeedCalculation|GetTimeSpentInCurrentSwingingSession) 0 (Utilities|ScalableFloattoFloat (Variables|Design|GeneralMovement|AccelerationInitial|GetSF_InitialAccelerationTime)) true))))))

(fn InitialJumpMultiplier
  (return (Utilities|Select :Index (Variables|Cache|GetState) :NewEnumerator0 1.4 :NewEnumerator2 1.0 :NewEnumerator7 1.4 :NewEnumerator5 0.0 :NewEnumerator6 0.0 :NewEnumerator9 0.0)))

(fn IsAimingAttack
  (return (or (Utilities|Enum|Equal(Enum) (Variables|Cache|GetState) "NewEnumerator5") (Utilities|Enum|Equal(Enum) (Variables|Cache|GetState) "NewEnumerator7"))))

(fn IsAttackingBool
  (Utilities|IsAttacking
    (:Attacking
      (return true))
    (:NotAttacking
      (return false))))

(fn OnADSExit
  ;[1 0.09804 0.1882] Block Attack via HF
  (if (or (Class|AthenaFunctionLibrary|RowtoBool (Variables|Default|GetAttackEnabled)) (Class|AthenaFunctionLibrary|RowtoBool (Variables|Default|GetSF_AllowThunderSpearHoverEvenWhenAttackDisabled)))
    (Variables|ADS|SetADSPressed false)
    (Audio|PopSoundMixModifier "/DryBox/Items/NyxGlass/Audio/Settings/NyxGlass_ADS_Amb_Mix.NyxGlass_ADS_Amb_Mix")
    (switch Utilities|FlowControl|Switch|SwitchonState_NyxGlass (Variables|Cache|GetState)
      (:NewEnumerator7
        (CallFunction|SetState "NewEnumerator0"))
      (:NewEnumerator5
        (CallFunction|SetState "NewEnumerator2")
        (Variables|Default|SetLastStoppedHoveringTime (Utilities|Time|GetGameTimeinSeconds))))))

(fn OnAnyWeaponEquip
  (params (NewWeapon :type Fort%20Weapon%20Object%20Reference) (PrevWeapon :type Fort%20Weapon%20Object%20Reference))
  (bind (_asb_weapon_nyx_glass _bsuccess) (Utilities|Casting|CastToB_Weapon_NyxGlass (Variables|GetNewWeapon)))
  (bind (_asb_fey_crab_launcher _bsuccess_2) (Utilities|Casting|CastToB_FeyCrab_Launcher (Variables|GetNewWeapon)))
  (LocalVariables|SetbIsNyxGlass _bsuccess)
  (bind _output_get (LocalVariables|SetbIsFeyCrab _bsuccess_2))
  (LocalVariables|SetbNewWeapFeyCrabAndAllowed (and _output_get (Class|AthenaFunctionLibrary|RowtoBool (Variables|Default|GetSF_AllowUsingThunderSpears))))
  (Utilities|FlowControl|Sequence
    (:then_0
      (if (CallFunction|IsSwinging)
        (Weapon|SetHolsterWeapon (Variables|Default|GetPlayer) 0 (and (not (Variables|Default|GetBlockSwingEndOnLand)) (not (or (LocalVariables|GetbIsNyxGlass) (LocalVariables|GetbNewWeapFeyCrabAndAllowed)))) false "GrappleParent")
        (CallFunction|SetBuildModeBlocked true)))
    (:then_1
      (if (Class|AthenaFunctionLibrary|RowtoBool (Variables|Default|GetSF_HoverAfterEquippingThunderSpears))
        (if (LocalVariables|GetbNewWeapFeyCrabAndAllowed)
          (if (or (Variables|Cache|Attack|GetCached_SF_AttackEnabled) (Class|AthenaFunctionLibrary|RowtoBool (Variables|Default|GetSF_AllowThunderSpearHoverEvenWhenAttackDisabled)))
            (CallFunction|HoverfromThunderSpears)
            (Variables|ADS|SetForceADSForFeyCrabHover true)))))
    (:then_2
      (if (Class|AthenaFunctionLibrary|RowtoBool (Variables|Default|GetCancelSwingingMovementOnUnequip))
        (if (not (LocalVariables|GetbIsNyxGlass))
          (if (or (not (LocalVariables|GetbIsFeyCrab)) (Class|AthenaFunctionLibrary|RowtoBool (Variables|Default|GetCancelSwingingMovementOnUnequip_EvenIfThunderSpears)))
            (CallFunction|ForceExitSwinging)))))
    (:then_3
      ;[1 1 1] If switching back to NyxGlass from FeyCrab hover, forcefully turn off ADS
      (if (LocalVariables|GetbIsNyxGlass)
        (if (Variables|ADS|GetForceADSForFeyCrabHover)
          (ADS|OnADSExit)
          (else
            _))
        (Variables|ADS|SetForceADSForFeyCrabHover false)))))

(fn OnADSEnter
  (params (bUseFuel :type Boolean))
  ;[1 0.09804 0.1882] Block Attack via HF
  (if (Variables|Cache|Attack|GetCached_SF_AttackEnabled)
    ;[1 0.09804 0.1882] Blocks attack on mobile via HF
    (if (or (Class|AthenaFunctionLibrary|RowtoBool (Variables|Default|GetSF_AttackEnabledOnMobile)) (not (Variables|Default|GetbIsTouchControls)))
      (Variables|Default|SetConsume%20Fuel (Variables|GetbUseFuel))
      (Variables|ADS|SetADSPressed true)
      (Variables|Default|SetShouldShowAttackAgainPrompt false)
      (Audio|PushSoundMixModifier "/DryBox/Items/NyxGlass/Audio/Settings/NyxGlass_ADS_Amb_Mix.NyxGlass_ADS_Amb_Mix")
      (if (CallFunction|FuelIsUsable)
        (bind _returnvalue (Utilities|Time|InvalidateTimerHandle (Variables|Cache|Attach|GetAttachTimerHandle)))
        (Variables|Cache|Attach|SetAttachTimerHandle _returnvalue)
        (switch Utilities|FlowControl|Switch|SwitchonState_NyxGlass (Variables|Cache|GetState)
          (:NewEnumerator0
            (CallFunction|SetState "NewEnumerator7"))
          (:NewEnumerator2
            (if (Default|IsSwingAttached)
              (CallFunction|Detach false true)
              (else
                _))
            (State|SetStateHovering (Variables|Default|GetConsume%20Fuel))))))))

(fn OnNativeForgivenessTimerCanceled
  (Networking|SwitchHasAuthority
    (:Authority
      (CallFunction|ForgivenessTimerCanceled))
    (:Remote
      (CallFunction|HandleTimeoutAudioStopped)
      (Utilities|Time|ClearandInvalidateTimerbyHandle (Variables|Cache|ForgivenessTimer|GetForgivenessTimer_BP))
      (if (Variables|Cache|ForgivenessTimer|GetClientForgivenessTimerElapsed)
        (Variables|Cache|SetDisableFiring false)
        (Variables|Cache|ForgivenessTimer|SetClientForgivenessTimerElapsed false)))))

(fn OnNativeForgivenessTimerEnded
  (Networking|SwitchHasAuthority
    (:Authority
      (State|SetStateOff)
      (CallFunction|ApplyCooldownServer) ; The player has spent enough time "grounded" without swinging, start refueling.
      (CallFunction|ForgivenessTimerFinished))
    (:Remote
      (ForgivenessTimer|StopForgivenessTimerClient)
      (Variables|Default|SetShouldShowAttackAgainPrompt false))))

(fn OnNativizedVelocityCalculated
  (params (VelocityData :type Nyx%20Glass%20Velocity%20Metadata%20Structure))
  (bind _returnvalue (CallFunction|IsAimingAttack))
  (bind (_bsuccessfullyfoundattribute _returnvalue_2) (Ability|Attribute|GetFloatAttribute (Variables|Default|GetPlayer) "(AttributeName=\"LocalFuel\",Attribute=/Script/DryBoxRuntime.NyxGlassFuelAttributeSet:LocalFuel,AttributeOwner=\"/Script/CoreUObject.Class'/Script/DryBoxRuntime.NyxGlassFuelAttributeSet'\")"))
  (bind _waistgear (Variables|Cache|GetWaistGear))
  (bind _ns_gasparticles_nyxglass (Variables|/DryBox/Items/NyxGlass/BGA_WaistAttachment_NyxGlass.BGA_WaistAttachment_NyxGlass_C|GetNS_GasParticles_NyxGlass _waistgear))
  (Utilities|FlowControl|Sequence
    (:then_0
      (if (Utilities|Struct|BreakNyxGlassVelocityMetadata (Variables|GetVelocityData))
        (if (Utilities|Enum|Equal(Enum) (Variables|Cache|GetState) "NewEnumerator6")
          (CallFunction|SetState "NewEnumerator9")
          (Variables|Cache|Attack|SetSubState_Slashing "NewEnumerator0")
          (bind _attachpointbga (Variables|Attach|GetAttachPointBGA)
            (:then
              (Variables|Cache|Attack|SetSlashingTarget (Transformation|GetActorLocation _attachpointbga)))
            (:else
              _))
          (CallFunction|Detach true true)
          (Development|PrintString "NYXGLASS_DETACH_closetopoint_attack_native")
          (CallFunction|BlockSwingingEnd)
          (Variables|Default|SetAlreadyDidSlashDamage false)
          (else
            (CallFunction|Detach true)
            (Development|PrintString "NYXGLASS_DETACH_closetopoint_native")))))
    (:then_1
      (Networking|SwitchHasAuthority
        (:Remote
          (bind _ns_gasparticles_nyxglass_2 (Variables|/DryBox/Items/NyxGlass/BGA_WaistAttachment_NyxGlass.BGA_WaistAttachment_NyxGlass_C|GetNS_GasParticles_NyxGlass (Variables|Cache|GetWaistGear))
            (:then
              (Transformation|SetWorldRotation _ns_gasparticles_nyxglass_2 (Math|Rotator|MakeRotfromZ (Math|Vector|Normalize (Utilities|Struct|BreakNyxGlassVelocityMetadata (Variables|GetVelocityData)))) false false)
              (Effects|Components|ParticleSystem|SetFloatParameter _ns_gasparticles_nyxglass_2 "SpraySpeed" 25.0)
              (if (or (Default|IsSwingAttached) (Utilities|Enum|Equal(Enum) (Variables|Cache|GetState) "NewEnumerator5"))
                (if (not (Components|Activation|IsActive _ns_gasparticles_nyxglass))
                  (Components|Activation|Activate _ns_gasparticles_nyxglass true)
                  (Effects|Components|ParticleSystem|SetBoolParameter _ns_gasparticles_nyxglass "bIsAimAttacking" _returnvalue)
                  (if (not _returnvalue)
                    (Class|BGAWaistAttachmentNyxGlass|RopeVisualSpawnPushForwardFX _waistgear)
                    (bind _returnvalue_3 (Audio|SpawnSoundAttached (Variables|Audio|GetGasBoostSound) (Variables|/Script/Engine.Character|GetMesh (Variables|Default|GetPlayer))))
                    (Audio|Parameter|SetFloatParameter _returnvalue_3 "GasLevel" _returnvalue_2 :!class /Script/AudioExtensions.AudioParameterControllerInterface)))
                (else
                  (Components|Activation|Deactivate _ns_gasparticles_nyxglass)
                  (Class|BGAWaistAttachmentNyxGlass|RopeVisualDeactivePushForwardFX _waistgear))))
            (:else)))))
    (:then_2
      (switch Utilities|FlowControl|Switch|SwitchonState_NyxGlass (Variables|Cache|GetState)
        (:NewEnumerator5
          (if (not (Utilities|Struct|BreakNyxGlassVelocityMetadata (Variables|GetVelocityData)))
            (Class|BGAWaistAttachmentNyxGlass|RopeVisualDeactivePushForwardFX (Variables|Cache|GetWaistGear))
            (GameplayEffects|RemoveActiveEffectswithGrantedTags (Ability|GetFortAbilitySystemComponent (Variables|Default|GetPlayer)) "(GameplayTags=((TagName=\"Granted.Athena.NyxGlass.UsingGas.Hovering\")))")
            (Variables|Cache|SpeedCalculation|SetUsingGas false)))
        (:NewEnumerator6
          (Transformation|SetWorldRotation (Variables|/DryBox/Items/NyxGlass/BGA_WaistAttachment_NyxGlass.BGA_WaistAttachment_NyxGlass_C|GetNS_GasParticles_NyxGlass (Variables|Cache|GetWaistGear)) (Math|Rotator|MakeRotfromZ (* (Math|Vector|Normalize (Utilities|Struct|BreakNyxGlassVelocityMetadata (Variables|GetVelocityData))) -1.0)) false false)
          (Variables|Default|SetInSlowPartOfAttack (< (Math|Vector|VectorLength (Utilities|Struct|BreakNyxGlassVelocityMetadata (Variables|GetVelocityData))) (Utilities|ScalableFloattoFloat (Variables|Design|GetMinSpeedToNotDetachAtBadAngle))))
          (if (Utilities|Struct|BreakNyxGlassVelocityMetadata (Variables|GetVelocityData))
            (CallFunction|ForceExitSwinging)
            (State|SetStateOff)))
        (:NewEnumerator9
          (switch Utilities|FlowControl|Switch|SwitchonSubState_Slash_NyxGlass (Variables|Cache|Attack|GetSubState_Slashing)
            (:NewEnumerator0
              (if (Utilities|Struct|BreakNyxGlassVelocityMetadata (Variables|GetVelocityData))
                (CallFunction|ForceExitSwinging)
                (State|SetStateOff)
                (else
                  (if (Utilities|Struct|BreakNyxGlassVelocityMetadata (Variables|GetVelocityData))
                    (Variables|Cache|Attack|SetSubState_Slashing "NewEnumerator10")
                    (GameplayCue|RemoveGameplayCue (Variables|Actor|GetInstigator) (Variables|Default|GetGC_Attacking))
                    (else
                      (if (Utilities|Struct|BreakNyxGlassVelocityMetadata (Variables|GetVelocityData))
                        _)))
                  (if (not (Variables|Default|GetAlreadyDidSlashDamage))
                    (Variables|Default|SetAlreadyDidSlashDamage true)
                    (bind _attachpointbga_2 (Variables|Attach|GetAttachPointBGA)
                      (:then
                        (Variables|Cache|Attack|SetSlashingTarget (Transformation|GetActorLocation _attachpointbga_2)))
                      (:else
                        _))
                    (CallFunction|SlashDamageandFX) ; do this a bit earlier than actual slash so it's timed better\nwith replication
                  ))))
            (:NewEnumerator10
              (if (Utilities|Struct|BreakNyxGlassVelocityMetadata (Variables|GetVelocityData))
                (Variables|Cache|Attack|SetSubState_Slashing "NewEnumerator11")))
            (:NewEnumerator11
              (Networking|SwitchHasAuthority
                (:Authority
                  (GameplayEffects|ApplyGameplayEffectToSelf :self (Ability|GetFortAbilitySystemComponent (Variables|Default|GetPlayer)) :GameplayEffectClass "/DryBox/Items/NyxGlass/GE_NyxGlass_BlockAttackUntilReady.GE_NyxGlass_BlockAttackUntilReady_C" :Level 0.0))
                (:Remote
                  _))
              (if (Utilities|Struct|BreakNyxGlassVelocityMetadata (Variables|GetVelocityData))
                (if (and (Variables|ADS|GetADSPressed) (and (not (Variables|Cache|FuelAndCooldown|GetRefueling)) (CallFunction|HasLocalFuel)))
                  (CallFunction|StartHoveringandConsumeFuel)
                  (else
                    (CallFunction|SetState "NewEnumerator2")))
                (Variables|Cache|Attack|SetSubState_Slashing "NewEnumerator0")
                (Networking|SwitchHasAuthority
                  (:Authority
                    (GameplayEffects|RemoveActiveEffectswithGrantedTags (Ability|GetFortAbilitySystemComponent (Variables|Default|GetPlayer)) "(GameplayTags=((TagName=\"Granted.Athena.NyxGlass.BlockAttackUntilReady\")))"))
                  (:Remote
                    _))
                (CallFunction|AfterSlashEvents (Utilities|Struct|BreakNyxGlassVelocityMetadata (Variables|GetVelocityData)))
                (if (and (CallFunction|FuelIsUsable) (not (Variables|Default|GetPlayerLearnedToAttackFromAir)))
                  (Variables|Default|SetShouldShowAttackAgainPrompt true))))))))
    (:then_3
      (Variables|Cache|SpeedCalculation|SetCurrentMaxSpeedMultiplier (Utilities|Struct|BreakNyxGlassVelocityMetadata (Variables|GetVelocityData))) ; Used by cam modifier
    )))

(fn PlayReplenishedAudio
  (if (Pawn|IsLocallyViewed (Variables|Default|GetPlayer))
    (Audio|PlaySound2D (Variables|Audio|GetRefuelCompleteSound))))

(fn RemoveGearIfOutOfFuel
  (Networking|SwitchHasAuthority
    (:Authority
      (bind _player (Variables|Default|GetPlayer)
        (:then
          (if (<= (.ReturnValue (Ability|Attribute|GetFloatAttribute _player "(AttributeName=\"Fuel\",Attribute=/Script/DryBoxRuntime.NyxGlassFuelAttributeSet:Fuel,AttributeOwner=\"/Script/CoreUObject.Class'/Script/DryBoxRuntime.NyxGlassFuelAttributeSet'\")")) 0)
            (CallFunction|ServerOutOfFuel)))
        (:else)))))

(fn SetState
  (params (New%20State :type State_NyxGlass%20Enum :default "NewEnumerator0"))
  (bind _waistgear (Variables|Cache|GetWaistGear))
  (bind _ns_gasparticles_nyxglass (Variables|/DryBox/Items/NyxGlass/BGA_WaistAttachment_NyxGlass.BGA_WaistAttachment_NyxGlass_C|GetNS_GasParticles_NyxGlass _waistgear))
  (if (Utilities|Enum|NotEqual(Enum) New%20State (Variables|Cache|GetState))
    (LocalVariables|SetPreviousState (Variables|Cache|GetState))
    (Utilities|FlowControl|Sequence
      (:then_0
        (switch Utilities|FlowControl|Switch|SwitchonState_NyxGlass (LocalVariables|GetPreviousState)
          (:NewEnumerator2
            (GameplayCue|RemoveGameplayCue (Variables|Actor|GetInstigator) (Variables|Cache|GetGC_Swinging)))
          (:NewEnumerator7
            (if (Utilities|Enum|Equal(Enum) (Variables|GetNew%20State) "NewEnumerator6")
              (Variables|Default|SetAttackingFromGround true)))
          (:NewEnumerator5
            (GameplayEffects|RemoveActiveEffectswithGrantedTags (Ability|GetFortAbilitySystemComponent (Variables|Default|GetPlayer)) "(GameplayTags=((TagName=\"Granted.Athena.NyxGlass.UsingGas.Hovering\")))")
            (switch Utilities|FlowControl|Switch|SwitchonState_NyxGlass (Variables|GetNew%20State)
              (:NewEnumerator0
                _)
              (:NewEnumerator2
                _)
              (:NewEnumerator7
                _))
            (Variables|Cache|SpeedCalculation|SetUsingGas false))
          (:NewEnumerator6
            (Variables|Default|SetAttackingFromGround false)
            (Variables|Default|SetInSlowPartOfAttack false))))
      (:then_1
        (bind _output_get (Variables|Cache|SetState New%20State))
        (Default|CallOnStateChanged)
        (switch Utilities|FlowControl|Switch|SwitchonState_NyxGlass _output_get
          (:NewEnumerator0
            ;[1 1 1] Failsafe to remove GC's
            (GameplayCue|RemoveGameplayCue (Variables|Actor|GetInstigator) (Variables|Default|GetGC_Attacking))
            (GameplayCue|RemoveGameplayCue (Variables|Actor|GetInstigator) (Variables|Cache|GetGC_Swinging)))
          (:NewEnumerator2
            (GameplayCue|AddGameplayCue (Variables|Actor|GetInstigator) (Variables|Cache|GetGC_Swinging) (Class|AthenaFunctionLibrary|MakeSimpleFortEffectContext (Variables|Actor|GetInstigator) (Variables|Actor|GetInstigator))))
          (:NewEnumerator5
            (Variables|Cache|Attack|SetStartedHoveringTime (Utilities|Time|GetGameTimeinSeconds))
            (Variables|Cache|Attack|SetHoveringGoalLocation (Utilities|Operators|Add (Transformation|GetActorLocation (Variables|Default|GetPlayer)) (Math|Vector|MakeVector 0.0 0.0 (Utilities|RemapFloat (.z (Transformation|GetVelocity (Variables|Default|GetPlayer))) "(X=0.000000,Y=-1000.000000)" "(X=25.000000,Y=200.000000)" true)))))
          (:NewEnumerator6
            (GameplayCue|AddGameplayCue (Variables|Actor|GetInstigator) (Variables|Default|GetGC_Attacking) (Class|AthenaFunctionLibrary|MakeSimpleFortEffectContext (Variables|Actor|GetInstigator) (Variables|Actor|GetInstigator)))
            (Variables|Default|SetAttackDistance (Math|Vector|VectorLength (CallFunction|VectortoAttachPoint)))
            (Class|NyxGlassWeaponComponentSwinging|OnAttackStart (Variables|Default|GetNyxGlass%20Weapon%20Swinging%20BP%20Comp))
            (GameplayCue|ExecuteGameplayCueOnActor(Burst) (Variables|Actor|GetInstigator) (Variables|Default|GetGC_AttackHookImpact) (Ability|GameplayCue|MakeGameplayCueParameters :NormalizedMagnitude 0.0 :RawMagnitude 0.0 :Location (Transformation|GetActorLocation (Variables|Attach|GetAttachPointBGA)) :Normal "0,0,0" :Instigator (Variables|Actor|GetInstigator) :SourceObject (Variables|Actor|GetInstigator) :GameplayEffectLevel 1 :AbilityLevel 1 :bReplicateLocationWhenUsingMinimalRepProxy false))
            (Variables|Cache|Attack|SetStartedAttackingTime (Utilities|Time|GetGameTimeinSeconds))
            (Components|Activation|Activate _ns_gasparticles_nyxglass false)
            (Transformation|SetRelativeRotation _ns_gasparticles_nyxglass "90.000000,0.000000,0.000000" false false)
            (Effects|Components|ParticleSystem|SetFloatParameter _ns_gasparticles_nyxglass "SpraySpeed" 1e+03))))
      (:then_2
        (switch Utilities|FlowControl|Switch|SwitchonState_NyxGlass (LocalVariables|GetPreviousState)
          (:NewEnumerator2
            _)
          (:NewEnumerator5
            _)
          (:NewEnumerator6
            _)
          (:NewEnumerator9
            _))
        (switch Utilities|FlowControl|Switch|SwitchonState_NyxGlass (Variables|GetNew%20State)
          (:NewEnumerator0
            _)
          (:NewEnumerator7
            _))
        (Variables|Cache|Attack|SetAttackOnNextAttach false))
      (:then_3
        ;[1 1 1] Ensure Attack GC gets removed (wasn't getting removed when attack was cancelled early by different failsafes)
        (if (Utilities|Enum|Equal(Enum) (LocalVariables|GetPreviousState) "NewEnumerator6")
          (GameplayCue|RemoveGameplayCue (Variables|Actor|GetInstigator) (Variables|Default|GetGC_Attacking))))
      (:then_4
        (switch Utilities|FlowControl|Switch|SwitchonState_NyxGlass (Variables|Cache|GetState)
          (:NewEnumerator0
            _)
          (:NewEnumerator2
            _)
          (:NewEnumerator5
            (Variables|Default|SetLastStartedHoveringTime (Utilities|Time|GetGameTimeinSeconds)))
          (:NewEnumerator6
            (Variables|Default|SetLastStartedAttackingTime (Utilities|Time|GetGameTimeinSeconds)))
          (:NewEnumerator9
            (Variables|Default|SetLastStartedSlashingTime (Utilities|Time|GetGameTimeinSeconds))))
        (Networking|SwitchHasAuthority
          (:Authority
            (CallFunction|RemoveGearifOutOfFuel))))
      (:then_5
        (Class|NyxGlassWeaponComponentSwinging|OnNyxGlassStateChanged (Variables|Default|GetNyxGlass%20Weapon%20Swinging%20BP%20Comp) (Class|BComponentNyxGlassWeaponSwinging|BPNyxGlassStatetoNativeState (Variables|Default|GetNyxGlass%20Weapon%20Swinging%20BP%20Comp) (LocalVariables|GetPreviousState)) (Class|BComponentNyxGlassWeaponSwinging|BPNyxGlassStatetoNativeState (Variables|Default|GetNyxGlass%20Weapon%20Swinging%20BP%20Comp) (Variables|Cache|GetState))) ; Tell native code to change swinging control parms
        (Networking|SwitchHasAuthority
          (:Authority
            (CallFunction|RemoteClientsSetState (Variables|Cache|GetState))))))))

(fn SetStateHovering
  (params (bUseFuel :type Boolean))
  (if bUseFuel
    (GameplayEffects|ApplyGameplayEffectToSelf :self (Ability|GetFortAbilitySystemComponent (Variables|Default|GetPlayer)) :GameplayEffectClass "/DryBox/Items/NyxGlass/GE_UseGas_Hover_OverTime.GE_UseGas_Hover_OverTime_C" :Level 0.0)
    (Variables|Cache|SpeedCalculation|SetUsingGas true)
    (else
      _))
  (CallFunction|SetState "NewEnumerator5"))

(fn SetState_Attacking
  (params (New%20State :type State_NyxGlass%20Enum :default "NewEnumerator0"))
  (CallFunction|SetState "NewEnumerator6")
  (Variables|Cache|Attack|SetAutoAttacking true)
  (Variables|Cache|Attack|SetAttackOnNextAttach false))

(fn SetState_Off
  (CallFunction|SetState "NewEnumerator0")
  (Class|CameraModifierGrappleParent|TurnOff (Variables|Default|GetCameraModifier)))

(fn SetupCachedScalableFloats
  (Variables|Default|SetCached_SF_TimedFailsafe_Attack_Enabled (Class|AthenaFunctionLibrary|RowtoBool (Variables|Default|GetTimedFailSafe_Attack_Enabled)))
  (Variables|Default|SetCached_SF_TimedFailsafe_Slash_Enabled (Class|AthenaFunctionLibrary|RowtoBool (Variables|Default|GetTimedFailSafe_Slash_Enabled)))
  (Variables|Default|SetCached_SF_TimedFailsafe_SlashMoveTowardTarget_Enabled (Class|AthenaFunctionLibrary|RowtoBool (Variables|Default|GetTimedFailSafe_SlashMoveTowardTarget_Enabled)))
  (Variables|Default|SetCached_SF_TimedFailsafe_Attack_EnabledOnClient (Class|AthenaFunctionLibrary|RowtoBool (Variables|Default|GetTimedFailSafe_Attack_EnabledOnClient)))
  (Variables|Default|SetCached_SF_TimedFailsafe_Slash_EnabledOnClient (Class|AthenaFunctionLibrary|RowtoBool (Variables|Default|GetTimedFailSafe_Slash_EnabledOnClient)))
  (Variables|Default|SetCached_SF_TimedFailsafe_SlashMoveTowardTarget_EnabledOnClient (Class|AthenaFunctionLibrary|RowtoBool (Variables|Default|GetTimedFailSafe_SlashMoveTowardTarget_EnabledOnClient)))
  (Variables|Default|SetCachedFixedStoredAttackState (Class|AthenaFunctionLibrary|RowtoBool (Variables|Default|GetSF_FixedStoredAttackState)))
  (Variables|ADS|SetCached_SF_UseADSOverhaul (Class|AthenaFunctionLibrary|RowtoBool (Variables|ADS|GetSF_UseADSOverhaul)))
  (Variables|Cache|Attack|SetCached_SF_AttackEnabled (Class|AthenaFunctionLibrary|RowtoBool (Variables|Default|GetAttackEnabled))))

(fn ShouldBlockDetachment
  (return (Utilities|Enum|Equal(Enum) (Variables|Cache|GetState) "NewEnumerator9")))

(fn ShouldDetachIfTooSlow
  (return (Math|Boolean|ANDBoolean (|CallParentFunction) (> (CallFunction|InitialAccelerationAlpha) 0.8) (not (Variables|Default|GetInSlowPartOfAttack)))))

(fn ShouldDisableFiring
  (return (or (Variables|Cache|GetDisableFiring) (Variables|Default|GetDisableFiringAfterSlash))))

(fn ShouldShowGotShotWarning
  (return (< (- (Utilities|Time|GetGameTimeinSeconds) (Variables|Cache|UI|GetLastFreeFallTime)) 3.0) (or (> (- (Utilities|Time|GetGameTimeinSeconds) (Variables|Cache|UI|GetLastFreeFallTime)) 1.0) (> (Math|Trig|Sin(Radians) (* (- (Utilities|Time|GetGameTimeinSeconds) (Variables|Cache|UI|GetLastFreeFallTime)) 4e+01)) 0))))

(fn ShouldUpdateTargeting
  (Networking|SwitchHasAuthority
    (:Remote
      (if (Math|Boolean|ANDBoolean (not (Variables|Default|GetbWaitingforBGASpawn)) (not (Variables|Default|GetbJustFired)) (not (Variables|Status|GetHoldingFire)))
        (return true)
        (else
          (return false))))))

(fn StartForgivenessTimer
  (Utilities|FlowControl|Sequence
    (:then_0
      (Networking|SwitchHasAuthority
        (:Authority
          (Class|FortWeaponComponentSwinging|SetForgivenessTimer (Variables|Default|GetNyxGlass%20Weapon%20Swinging%20BP%20Comp) (ScalableFloat|GetValueatLevel (Variables|/Script/FortniteGame.FortWeaponComponent_Swinging|GetDefaultForgivenessTimer (Variables|Default|GetNyxGlass%20Weapon%20Swinging%20BP%20Comp))))
          (CallFunction|ForgivenessTimerStarted))
        (:Remote
          (CallFunction|HandleTimeoutAudio)
          (bind _returnvalue (Utilities|Time|SetTimerbyEvent (EventDispatchers|CreateEvent :!event OnForgivenessTimerElapsed_BPClientOnly) (ScalableFloat|GetValueatLevel (Variables|/Script/FortniteGame.FortWeaponComponent_Swinging|GetDefaultForgivenessTimer (Variables|Default|GetNyxGlass%20Weapon%20Swinging%20BP%20Comp))) true))
          (Variables|Cache|ForgivenessTimer|SetForgivenessTimer_BP _returnvalue)
          (Variables|Cache|ForgivenessTimer|SetClientForgivenessTimerElapsed false))))
    (:then_1
      (if (Variables|Default|GetCachedFixedStoredAttackState)
        (Variables|Cache|Attack|SetAttackOnNextAttach false)))))

(fn StopForgivenessTimer_Client
  (CallFunction|HandleTimeoutAudioStopped)
  (Utilities|Time|ClearandInvalidateTimerbyHandle (Variables|Cache|ForgivenessTimer|GetForgivenessTimer_BP))
  (Default|CallReticleBlockFiring)
  (Variables|Cache|SetDisableFiring true)
  (State|SetStateOff)
  (Variables|Cache|ForgivenessTimer|SetClientForgivenessTimerElapsed false))

(fn UnhideBuiltInCapes
  (if (Guid|IsValidGuid (Variables|Cosmetics|GetCapeSwapId))
    (Class|FortPlayerPawn|DisableCosmeticSwap (Variables|Default|GetPlayer) (Variables|Cosmetics|GetCapeSwapId))
    (Variables|Cosmetics|SetCapeSwapId (Utilities|Struct|MakeGuid))))

(fn UpdateRopeFX
  (bind _nyxglass_weapon_swinging_bp_comp (Variables|Default|GetNyxGlass%20Weapon%20Swinging%20BP%20Comp))
  (bind _curtargetingdata (Variables|/Script/DryBoxRuntime.NyxGlassWeaponComponent_Swinging|GetCurTargetingData _nyxglass_weapon_swinging_bp_comp))
  (bind (_leftactor _rightactor _leftattachpoint _rightattachpoint _bfoundleftattachpoint _bfoundrightattachpoint) (Utilities|Struct|BreakNyxGlassTargetingData _curtargetingdata))
  (bind _attacking (CallFunction|IsAttackingBool))
  (bind _waistgear (Variables|Cache|GetWaistGear))
  (bind _waistgear_2 (Variables|Cache|GetWaistGear))
  (Utilities|IsValid (Variables|Cache|GetWaistGear)
    (:Is%20Valid
      (if (Variables|Default|GetRopeAttached)
        (Utilities|FlowControl|Sequence
          (:then_0
            (if _bfoundleftattachpoint
              (Default|RopeVisualUpdateARopeVisual _waistgear _leftattachpoint self false _attacking)
              (else
                (Default|RopeVisualSetARopeVisibility _waistgear false false))))
          (:then_1
            (if _bfoundrightattachpoint
              (Default|RopeVisualUpdateARopeVisual _waistgear_2 _rightattachpoint self true _attacking)
              (else
                (Default|RopeVisualSetARopeVisibility _waistgear_2 true false)))))))))

(fn UpdateRefueling
  (bind _player (Variables|Default|GetPlayer))
  (bind _returnvalue (Fortnite|GetFortPlayerControllerfromActor _player))
  (Networking|SwitchHasAuthority
    (:Authority
      (bind _player_2 (Variables|Default|GetPlayer)
        (:then
          (if (and (not (Variables|Cache|FuelAndCooldown|GetRefueling)) (< (.ReturnValue (Ability|Attribute|GetFloatAttribute _player_2 "(AttributeName=\"LocalFuel\",Attribute=/Script/DryBoxRuntime.NyxGlassFuelAttributeSet:LocalFuel,AttributeOwner=\"/Script/CoreUObject.Class'/Script/DryBoxRuntime.NyxGlassFuelAttributeSet'\")")) (Utilities|ScalableFloattoFloat (Variables|Design|FuelandCooldown|GetSF_MaxLocalFuel))))
            ;[0.3177 0.2164 0.1207] If the player dropped the weapon, immediately begin refueling.
            (Utilities|IsValid _returnvalue
              (:Is%20Valid
                (bind _returnvalue_2 (InventoryManagement|FindExistingItemforDefinition _returnvalue "/DryBox/Items/NyxGlass/AGID_NyxGlass.AGID_NyxGlass"))
                (Utilities|IsValid _returnvalue_2
                  (:Is%20Not%20Valid
                    (CallFunction|ApplyCooldownServer)
                    (return)))))))
        (:else)))))

(fn ConstructionScript
  (Variables|Default|SetGrappleSwingingComp (Variables|Default|GetNyxGlass%20Weapon%20Swinging%20BP%20Comp)))
