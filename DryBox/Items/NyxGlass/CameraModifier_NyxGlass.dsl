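; Client-side swinging camera modifier for the NyxGlass grapple weapon that shifts camera position and FOV during use.
; Turning on refreshes forward-offset speed bounds from the swing's max horizontal speed, or an override while sliding.
; Each update interpolates forward, right and FOV offsets by aim-attack, FeyCrab launcher, weapon state and attach distance, adding a curve-driven slingshot pullback.
; Modifier alpha times the weapon's initial-acceleration alpha scales the result.
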(event EventTurnOn
  (bind (_offsetmodifier _baseoffset _speedscaledoffset) (Utilities|Struct|BreakSpeedScaledAxisOffsetData (Variables|Config|Offset|GetRightOffsetData)))
  ;[0.1035 1 0] Immediately update the forward speed bounds to align with the player's current speed maximum.
  (|CallParentFunction)
  (CallFunction|UpdateForwardOffsetData)
  (Variables|Default|SetStartingRightSpeedScaledOffset _speedscaledoffset))

(fn %5BDEPRECATED%5DGetAttackPercentage
  (return (Math|Float|SelectFloat (- 1.0 (/ (Math|Vector|VectorLength (- (Transformation|GetActorLocation (Variables|FortCameraModifier_Swinging|GetCachedPlayer)) (Variables|Default|GetAttachLocation))) (Variables|Default|GetStartingLength))) (+ (- (Utilities|Time|GetGameTimeinSeconds) (Variables|Default|GetStoppedAttackTime)) 1.0) (Variables|Default|GetIsUsingCameraAnchor))))

(fn BlueprintModifyCamera
  (params (DeltaTime :type Float%20%28single-precision%29) (ViewLocation :type Vector :default "0, 0, 0") (ViewRotation :type Rotator :default "0, 0, 0") (FOV :type Float%20%28single-precision%29))
  (bind (_offsetmodifier _baseoffset _speedscaledoffset) (Utilities|Struct|BreakSpeedScaledAxisOffsetData (Variables|Config|Offset|GetRightOffsetData)))
  (bind (_offsetmodifier_2 _baseoffset_2 _speedscaledoffset_2) (Utilities|Struct|BreakSpeedScaledAxisOffsetData (Variables|Config|Offset|GetRightOffsetData)))
  (bind _returnvalue (Utilities|Time|GetGameTimeinSeconds))
  (bind (_asb_fey_crab_launcher _bsuccess) (Utilities|Casting|CastToB_FeyCrab_Launcher (Variables|/Script/FortniteGame.FortPawn|GetCurrentWeapon (Variables|FortCameraModifier_Swinging|GetCachedPlayer))))
  (bind _weapon (Variables|Default|GetWeapon))
  (bind _lastslingshottime (Variables|/DryBox/Items/NyxGlass/B_Weapon_NyxGlass.B_Weapon_NyxGlass_C|GetLastSlingshotTime _weapon))
  (bind _returnvalue_2 (- _returnvalue _lastslingshottime))
  (bind _sf_slingshotboostduration (Variables|Default|GetSF_SlingshotBoostDuration))
  (if (Networking|IsServer)
    (return ViewLocation ViewRotation FOV)
    (else
      (Utilities|IsValid (Variables|Default|GetWeapon)
        (:Is%20Valid
          _)
        (:Is%20Not%20Valid
          (bind _asb_weapon_nyx_glass (Utilities|Casting|CastToB_Weapon_NyxGlass (Variables|/Script/FortniteGame.FortPawn|GetCurrentWeapon (Variables|FortCameraModifier_Swinging|GetCachedPlayer))))
          (Variables|Default|SetWeapon _asb_weapon_nyx_glass)))
      (Variables|Default|SetCameraCurrentPosition (+ (Variables|GetViewLocation) (Variables|Default|GetPosOffset)))
      (Utilities|FlowControl|Sequence
        (:then_0
          (Utilities|FlowControl|Sequence
            (:then_0
              (if (not (Variables|Default|GetIsUsingCameraAnchor))
                (Variables|Default|SetIsUsingCameraAnchor true)
                (Variables|Default|SetAnchorStartLocation (Variables|Default|GetCameraCurrentPosition))
                (Variables|Default|SetAttachLocation (Transformation|GetActorLocation (Variables|/GrappleParent/GrappleParent/B_Weapon_GrappleParent.B_Weapon_GrappleParent_C|GetAttachPointBGA (Variables|Default|GetWeapon))))
                (Variables|Default|SetStartingLength (- (Math|Vector|VectorLength (- (Transformation|GetActorLocation (Variables|FortCameraModifier_Swinging|GetCachedPlayer)) (Transformation|GetActorLocation (Variables|/GrappleParent/GrappleParent/B_Weapon_GrappleParent.B_Weapon_GrappleParent_C|GetAttachPointBGA (Variables|Default|GetWeapon))))) 1e+03))))
            (:then_1
              (Variables|Default|SetAnchorAlpha (Math|Float|Clamp(Float) (+ (Variables|Default|GetAnchorAlpha) (* (Variables|GetDeltaTime) (Math|Float|SelectFloat 1.0 -1.0 (Variables|Default|GetIsUsingCameraAnchor)))))))))
        (:then_1
          (if (< _returnvalue_2 _sf_slingshotboostduration)
            (bind _output_get (Variables|Default|SetCurrentSlingshotBoost (* (Math|Curves|GetFloatValue (Variables|Default|GetCurve_SlingshotBoost) (Math|Float|NormalizetoRange _returnvalue_2 0.0 _sf_slingshotboostduration)) (Variables|/DryBox/Items/NyxGlass/B_Weapon_NyxGlass.B_Weapon_NyxGlass_C|GetCurrentMaxSpeedMultiplier (Variables|Default|GetWeapon)))))
            (Variables|Default|SetSlingshotBoost_ForwardOffset (* _output_get -3e+02))
            (else
              (Variables|Default|SetCurrentSlingshotBoost 0.0)
              (Variables|Default|SetSlingshotBoost_ForwardOffset 0.0))))
        (:then_2
          (if (or (Class|BWeaponNyxGlass|IsAimingAttack (Variables|Default|GetWeapon)) _bsuccess)
            (Variables|Default|SetCurrForwardOffset (Math|Interpolation|FInterpTo (Variables|Default|GetCurrForwardOffset) -3.8e+02 (Variables|GetDeltaTime) 3.0))
            (Variables|Default|SetCurrRightOffset (Math|Interpolation|FInterpTo (Variables|Default|GetCurrRightOffset) 1.2e+02 (Variables|GetDeltaTime) 8.0))
            (Utilities|Struct|SetmembersinSpeedScaledAxisOffsetData (Variables|Config|Offset|GetRightOffsetData) (Math|Interpolation|FInterpTo _speedscaledoffset 0.0 (Variables|GetDeltaTime) 8.0))
            (Variables|Default|SetCurrFOVOffset (Math|Interpolation|FInterpTo (Variables|Default|GetCurrFOVOffset) -3e+01 (Variables|GetDeltaTime) 4.0))
            (else
              (switch Utilities|FlowControl|Switch|SwitchonState_NyxGlass (Variables|/DryBox/Items/NyxGlass/B_Weapon_NyxGlass.B_Weapon_NyxGlass_C|GetState (Variables|Default|GetWeapon))
                (:NewEnumerator0
                  _)
                (:NewEnumerator2
                  _)
                (:NewEnumerator7
                  _)
                (:NewEnumerator5
                  _)
                (:NewEnumerator6
                  (Variables|Default|SetCurrForwardOffset (Math|Interpolation|FInterpTo (Variables|Default|GetCurrForwardOffset) (Utilities|RemapFloat (Math|Vector|VectorLength (- (Transformation|GetActorLocation (Variables|FortCameraModifier_Swinging|GetCachedPlayer)) (Transformation|GetActorLocation (Variables|/GrappleParent/GrappleParent/B_Weapon_GrappleParent.B_Weapon_GrappleParent_C|GetAttachPointBGA (Variables|Default|GetWeapon))))) "(X=500.000000,Y=4000.000000)" "(X=100.000000,Y=-230.000000)" true) (Variables|GetDeltaTime) 3.0))
                  (Variables|Default|SetCurrFOVOffset (Math|Interpolation|FInterpTo (Variables|Default|GetCurrFOVOffset) 3e+01 (Variables|GetDeltaTime) 1.84))
                  (Variables|Default|SetCurrRightOffset (Math|Interpolation|FInterpTo (Variables|Default|GetCurrRightOffset) (Utilities|RemapFloat (Variables|Default|GetCurrForwardOffset) "(X=-500.000000,Y=0.000000)" "(X=150.000000,Y=90.000000)" true) (Variables|GetDeltaTime) 5.0))
                  (Utilities|Struct|SetmembersinSpeedScaledAxisOffsetData (Variables|Config|Offset|GetRightOffsetData) (Math|Interpolation|FInterpTo _speedscaledoffset_2 (Variables|Default|GetStartingRightSpeedScaledOffset) (Variables|GetDeltaTime) 5.0))))
              (Variables|Default|SetCurrForwardOffset (Math|Interpolation|FInterpTo (Variables|Default|GetCurrForwardOffset) (* (Math|Float|SelectFloat (Utilities|RemapFloat (- (Utilities|Time|GetGameTimeinSeconds) (Variables|/GrappleParent/GrappleParent/B_Weapon_GrappleParent.B_Weapon_GrappleParent_C|GetLastAttachedTime (Variables|Default|GetWeapon))) "(X=0.000000,Y=0.350000)" (Math|Vector2D|MakeVector2D 0.0 (Utilities|RemapFloat (Class|AbilityFunctions|GetDistanceBetweenActors (Variables|/GrappleParent/GrappleParent/B_Weapon_GrappleParent.B_Weapon_GrappleParent_C|GetAttachPointBGA (Variables|Default|GetWeapon)) (Variables|FortCameraModifier_Swinging|GetCachedPlayer)) "(X=3200.000000,Y=1600.000000)" "(X=1.000000,Y=0.000000)" true)) true) 0.2 (Default|IsSwingAttached (Variables|Default|GetWeapon))) -2e+02) (Variables|GetDeltaTime) 1.5))
              (Variables|Default|SetCurrFOVOffset (Math|Interpolation|FInterpTo (Variables|Default|GetCurrFOVOffset) 1e+01 (Variables|GetDeltaTime) 1.5)))))
        (:then_3
          (Utilities|Struct|SetmembersinSpeedScaledAxisOffsetData (Variables|Config|Offset|GetForwardOffsetData) (* (CallFunction|CalcAlpha (Variables|Default|GetWeapon)) -5e+01))
          (Variables|Default|SetAttachLocation (Transformation|GetActorLocation (Variables|/GrappleParent/GrappleParent/B_Weapon_GrappleParent.B_Weapon_GrappleParent_C|GetAttachPointBGA (Variables|Default|GetWeapon))))
          (Variables|Default|SetPosOffset (+ (* (Math|Vector|GetForwardVector (Variables|GetViewRotation)) (+ (Variables|Default|GetCurrForwardOffset) (Variables|Default|GetSlingshotBoost_ForwardOffset))) (* (Math|Vector|GetRightVector (Variables|GetViewRotation)) (Variables|Default|GetCurrRightOffset))))
          (return (+ (Variables|GetViewLocation) (* (Variables|Default|GetPosOffset) (CallFunction|CalcAlpha (Variables|Default|GetWeapon)))) (Variables|GetViewRotation) (+ (Variables|GetFOV) (* (Variables|Default|GetCurrFOVOffset) (CallFunction|CalcAlpha (Variables|Default|GetWeapon))))))))))

(fn CalcAlpha
  (params (Target :type B%20Weapon%20Nyx%20Glass%20Object%20Reference))
  (Utilities|IsValid Target
    (:Is%20Valid
      (return (* (Variables|CameraModifier|GetAlpha) (Class|BWeaponNyxGlass|InitialAccelerationAlpha Target))))
    (:Is%20Not%20Valid
      (return 0.0))))

(fn GetAttackPercentage
  (return (- 1.0 (/ (- (Math|Vector|VectorLength (- (Transformation|GetActorLocation (Variables|FortCameraModifier_Swinging|GetCachedPlayer)) (Variables|Default|GetAttachLocation))) 1e+03) (Variables|Default|GetStartingLength)))))

(fn GetCameraAnchorAlpha)

(fn UpdateForwardOffsetData
  (bind _speedbounds_x (Math|Vector2D|BreakVector2D (Utilities|Struct|BreakSpeedScaledAxisOffsetData (Variables|Config|Offset|GetForwardOffsetData))))
  (bind _cachedplayer (Variables|FortCameraModifier_Swinging|GetCachedPlayer)
    (:then
      (bind _asfort_movement_comp_character (Utilities|Casting|CastToFortMovementComp_Character (Variables|/Script/Engine.Character|GetCharacterMovement _cachedplayer)))
      (Utilities|Struct|SetmembersinSpeedScaledAxisOffsetData (Variables|Config|Offset|GetForwardOffsetData) (Variables|Default|GetForwardScaledOffsetAtNormalSpeed) (Math|Vector2D|MakeVector2D _speedbounds_x (Math|Float|SelectFloat (Variables|Default|GetSpeedBoundsWhileSlidingOverride) (* (ScalableFloat|GetValueatLevel (.MaxHorizontalSpeed (Utilities|Struct|BreakSwingingControlParams (Swinging|GetSwingingControlParameters _asfort_movement_comp_character)))) (Variables|Default|GetSpeedBoundsMultiplier)) (GameplayTags|HasMatchingGameplayTag _cachedplayer "(TagName=\"MovementMode.Sliding\")")))))
    (:else)))
