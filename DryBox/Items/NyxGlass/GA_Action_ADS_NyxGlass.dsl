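; Gameplay ability that handles aim-down-sights input for the NyxGlass weapon.
; On activation it caches the player and weapon, then, only when the ADS overhaul flag is off, waits on input press and release to call the weapon's ADSPressed and ADSReleased.
; With touch controls and mobile toggle targeting both on, press toggles aim (ignored while equipping) and release does nothing.
; The failed-activation and end events only call the parent.
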(event Ability|EventActivateAbility
  (bind (_owneractor _avataractor) (Utilities|Struct|BreakGameplayAbilityActorInfo (Ability|GetActorInfo)))
  (bind _sf_usetoggletargetingonmobile (Variables|Default|GetSF_UseToggleTargetingOnMobile))
  (bind _returnvalue (Class|AthenaFunctionLibrary|RowtoBool _sf_usetoggletargetingonmobile))
  (bind _weapon (Variables|Default|GetWeapon))
  (bind _bistouchcontrols (Variables|/DryBox/Items/NyxGlass/B_Weapon_NyxGlass.B_Weapon_NyxGlass_C|GetbIsTouchControls _weapon))
  (bind _returnvalue_2 (and _bistouchcontrols _returnvalue))
  (|CallParentFunction)
  (bind _asplayer_pawn_athena (Utilities|Casting|CastToPlayerPawn_Athena _avataractor))
  (bind _output_get (Variables|Default|SetPlayer _asplayer_pawn_athena))
  (bind _asb_weapon_nyx_glass (Utilities|Casting|CastToB_Weapon_NyxGlass (Variables|/Script/FortniteGame.FortPawn|GetCurrentWeapon _output_get)))
  (Variables|Default|SetWeapon _asb_weapon_nyx_glass)
  (if (not (ScalableFloat|GetValueasBool (Variables|Default|GetSF_UseADSOverhaul)))
    (Utilities|FlowControl|Sequence
      (:then_0
        (Ability|Tasks|WaitInputPress false
          (:OnPress
            (Utilities|FlowControl|Sequence
              (:then_0
                _)
              (:then_1
                _))
            (if _returnvalue_2
              (if (Class|FortWeapon|IsBeingEquipped (Variables|Default|GetWeapon))
                _
                (elif (Class|BWeaponNyxGlass|IsAimingAttack (Variables|Default|GetWeapon))
                  (Class|BWeaponNyxGlass|ADSReleased (Variables|Default|GetWeapon))
                  (else
                    (Class|BWeaponNyxGlass|ADSPressed (Variables|Default|GetWeapon) true))))
              (else
                (Class|BWeaponNyxGlass|ADSPressed (Variables|Default|GetWeapon) true))))))
      (:then_1
        (Ability|Tasks|WaitInputRelease false
          (:OnRelease
            (Utilities|FlowControl|Sequence
              (:then_0
                _)
              (:then_2
                _))
            (if (not _returnvalue_2)
              (Class|BWeaponNyxGlass|ADSReleased (Variables|Default|GetWeapon)))))))))

(event Ability|EventFailedToActivatePassiveAbility
  (params (ActorInfo :type Gameplay%20Ability%20Actor%20Info%20Structure))
  (|CallParentFunction))

(event Ability|EventOnEndAbility
  (params (bWasCancelled :type Boolean))
  (|CallParentFunction bWasCancelled))
