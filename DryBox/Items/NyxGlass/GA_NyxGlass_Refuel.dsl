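; Gameplay ability that refuels the Nyx Glass item's LocalFuel attribute over time and shows progress as a cooldown in the UI.
; After a server-side start delay, a self-rescheduling timer applies a SetByCaller effect, scaling speed by current fuel and timer elapsed time.
; The server removes refueling-tagged effects and ends when the tank fills or the tag is lost.
; Clients poll fuel on the same interval and broadcast inverted progress to the UI.
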(event Ability|EventActivateAbility
  (bind (_owneractor _avataractor) (Utilities|Struct|BreakGameplayAbilityActorInfo (Ability|GetActorInfo)))
  ;[0.08032 1 0] Event ActivateAbility()
  (bind _asplayer_pawn_athena (Utilities|Casting|CastToPlayerPawn_Athena _avataractor))
  (Variables|Default|SetPlayer _asplayer_pawn_athena)
  (Ability|HasAuthority
    (:True
      (Utilities|Time|SetTimerbyEvent (EventDispatchers|CreateEvent :!event Custom|ScheduleUpdate) (Math|Float|Max(Float) (ScalableFloat|GetValueatLevel (Variables|Default|GetSF_TimeBeforeRefuelingStarts)) 0.001)))
    (:False
      (CallFunction|ScheduleUpdate))))

(event Ability|EventOnEndAbility
  (params (bWasCancelled :type Boolean))
  ;[1 0.6127 0] OnEndAbility: update the UI to show that this weapon's cooldown is done.
  (|CallParentFunction bWasCancelled)
  (Networking|CanExecuteCosmeticEvents
    (:True
      (CallFunction|UpdateCooldownUI 0.0))))

(event Custom|ScheduleUpdate
  ;[0.1497 0.1633 1] ScheduleUpdate
  (bind _returnvalue (Utilities|Time|SetTimerbyEvent (EventDispatchers|CreateEvent :!event Custom|UpdateRefueling) (Math|Float|Max(Float) (ScalableFloat|GetValueatLevel (Variables|Default|GetSF_UpdateRefuelingInterval)) 0.01)))
  (Variables|Default|SetUpdateRefueling%20TimerHandle _returnvalue))

(event Custom|UpdateRefueling
  ;[0 1 0.9925] Tick Refueling
  (Ability|HasAuthority
    (:True
      ;;[0.191 0.2518 1] Refuel on the server.
      (bind _shouldcontinuerefuelingnexttick (CallFunction|TickRefuelingFunction))
      (if _shouldcontinuerefuelingnexttick
        (CallFunction|ScheduleUpdate)
        (else
          ;;;[1 0.4717 0] End refueling.
          (bind _player (Variables|Default|GetPlayer)
            (:then
              (GameplayEffects|RemoveActiveEffectswithGrantedTags (Ability|GetFortAbilitySystemComponent _player) (GameplayTags|MakeGameplayTagContainerfromTag (Variables|Default|GetRefueling%20Gameplay%20Tag))))
            (:else
              _))
          (Ability|EndAbility))))
    (:False
      ;;[1 0 0.7691] Update the client
      (bind _player_2 (Variables|Default|GetPlayer)
        (:then
          (CallFunction|UpdateCooldownUI (/ (.ReturnValue (Ability|Attribute|GetFloatAttribute _player_2 "(AttributeName=\"LocalFuel\",Attribute=/Script/DryBoxRuntime.NyxGlassFuelAttributeSet:LocalFuel,AttributeOwner=\"/Script/CoreUObject.Class'/Script/DryBoxRuntime.NyxGlassFuelAttributeSet'\")")) (Math|Float|Max(Float) (ScalableFloat|GetValueatLevel (Variables|Default|GetSF_MaxLocalFuel)) 0.01)))
          (CallFunction|ScheduleUpdate))
        (:else)))))

(fn ApplyVariableRefuel
  (params (Magnitude :type Float%20%28single-precision%29))
  (bind _player (Variables|Default|GetPlayer))
  (bind _returnvalue (Ability|GetFortAbilitySystemComponent _player))
  (bind _returnvalue_2 (Ability|GameplayEffect|AssignTagSetbyCallerMagnitude (GameplayEffects|MakeOutgoingSpec _returnvalue "/DryBox/Items/NyxGlass/GE_RefuelGas_NyxGlass.GE_RefuelGas_NyxGlass_C" 0.0 (GameplayEffects|MakeFortEffectContext _returnvalue _player 0 _player 0)) "(TagName=\"SetByCaller.StaminaRegen\")" Magnitude))
  (GameplayEffects|ApplyGameplayEffectSpecToTarget _returnvalue _returnvalue_2 _returnvalue))

(fn IsFuelFull
  (bind _player (Variables|Default|GetPlayer)
    (:then
      (return (>= (.ReturnValue (Ability|Attribute|GetFloatAttribute _player "(AttributeName=\"LocalFuel\",Attribute=/Script/DryBoxRuntime.NyxGlassFuelAttributeSet:LocalFuel,AttributeOwner=\"/Script/CoreUObject.Class'/Script/DryBoxRuntime.NyxGlassFuelAttributeSet'\")")) (ScalableFloat|GetValueatLevel (Variables|Default|GetSF_MaxLocalFuel)))))
    (:else
      ;[1 0.474 0.474] We canNOT find the player, act like fuel is full.
      (return true))))

(fn TickRefuelingFunction
  (Utilities|FlowControl|Sequence
    (:then_0
      (bind _player (Variables|Default|GetPlayer)
        (:then
          (if (GameplayTags|HasMatchingGameplayTag _player (Variables|Default|GetRefueling%20Gameplay%20Tag))
            (CallFunction|ApplyVariableRefuel (* (Utilities|RemapFloat (.ReturnValue (Ability|Attribute|GetFloatAttribute _player "(AttributeName=\"LocalFuel\",Attribute=/Script/DryBoxRuntime.NyxGlassFuelAttributeSet:LocalFuel,AttributeOwner=\"/Script/CoreUObject.Class'/Script/DryBoxRuntime.NyxGlassFuelAttributeSet'\")")) "(X=0.000000,Y=100.000000)" (Math|Vector2D|MakeVector2D (ScalableFloat|GetValueatLevel (Variables|Default|GetSF_RefuelingSpeedAtEmpty)) (ScalableFloat|GetValueatLevel (Variables|Default|GetSF_RefuelingSpeedAtFull))) true) (Utilities|Time|GetTimerElapsedTimebyHandle (Variables|Default|GetUpdateRefueling%20TimerHandle))))
            ;[1 0.474 0.474] Refueling successful. We should continue to refuel next tick IF our tank is not full.
            (return (not (CallFunction|IsFuelFull)))))
        (:else)))
    (:then_1
      ;[1 0.474 0.474] Refueling failed for some reason, return that we should NOT continue refueling.
      (return false))))

(fn UpdateCooldownUI
  (params (Local%20Fuel%20Percent :type Float%20%28double-precision%29))
  (bind (_owneractor _avataractor) (Utilities|Struct|BreakGameplayAbilityActorInfo (Ability|GetActorInfo)))
  (bind _asfort_player_pawn (Utilities|Casting|CastToFortPlayerPawn _avataractor))
  (GameplayEventRouter|BroadcastEvent (GameplayEventRouter|GetGlobalEventRouter) (Utilities|Struct|MakeDirectProgressUpdateData (- 1.0 Local%20Fuel%20Percent) "/DryBox/Items/NyxGlass/AGID_NyxGlass.AGID_NyxGlass" _asfort_player_pawn) 0))
