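# Physics asset for the NyxGlass cannister mesh, giving its mirrored left and right cannisters and their two-segment legs simulated capsule bodies around odm_base.
# Cannisters and upper legs hang off the base with narrow angular limits, the upper legs also locking one swing axis.
# Lower legs join the upper legs, then link to the cannisters with free swing and twist, closing a loop per side.
# Each free cannister-to-leg constraint appears twice.
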PhysicsAsset /DryBox/Items/NyxGlass/Meshes/NyxGlass_Cannister_Physics
bodies {
 cannister_l [{
   shapeType=Capsule
   length=29.2368
   radius=18.1155
   shapeName=cannister_l_capsule
   center=<7.7696 20.2396 -3.3267>
   rotation=<rot 0 169.0075 -89.6558>
  }]
 cannister_leg_01_l [{
   shapeType=Capsule
   length=23.952
   radius=1.3195
   shapeName=cannister_leg_01_l_capsule
   center=<23.175 -0.0155 0.279>
   rotation=<rot 0 90.0634 -89.767>
  }]
 cannister_leg_01_r [{
   shapeType=Capsule
   length=23.952
   radius=1.2308
   shapeName=cannister_leg_01_r_capsule
   center=<23.1748 0.0009 -0.3179>
   rotation=<rot 0 -90.0056 -89.7585>
  }]
 cannister_leg_02_l [{shapeType=Capsule, length=0.505, radius=0.505, shapeName=cannister_leg_02_l_capsule, rotation=<rot 0 -90 0>}]
 cannister_leg_02_r [{shapeType=Capsule, length=0.505, radius=0.505, shapeName=cannister_leg_02_r_capsule, rotation=<rot 0 -90 0>}]
 cannister_r [{
   shapeType=Capsule
   length=29.2337
   radius=17.7052
   shapeName=cannister_r_capsule
   center=<7.8573 20.5548 3.6>
   rotation=<rot 0 -10.9563 -89.6783>
  }]
 odm_base [{shapeType=Capsule, length=1, radius=1}]
}
constraints [
 {bone1Name=cannister_r, bone2Name=odm_base, swing1LimitDegrees=10, swing2LimitDegrees=10, twistLimitDegrees=4}
 {bone1Name=cannister_l, bone2Name=odm_base, swing1LimitDegrees=10, swing2LimitDegrees=10, twistLimitDegrees=4}
 {bone1Name=cannister_leg_01_l, bone2Name=odm_base, swing1LimitDegrees=5, swing1Motion=Locked, swing2LimitDegrees=5, twistLimitDegrees=5}
 {bone1Name=cannister_leg_01_r, bone2Name=odm_base, swing1LimitDegrees=5, swing1Motion=Locked, swing2LimitDegrees=5, twistLimitDegrees=5}
 {bone1Name=cannister_leg_02_r, bone2Name=cannister_leg_01_r, swing1LimitDegrees=5, swing2LimitDegrees=5, twistLimitDegrees=5}
 {bone1Name=cannister_leg_02_l, bone2Name=cannister_leg_01_l, swing1LimitDegrees=5, swing2LimitDegrees=5, twistLimitDegrees=5}
 {
  bone1Name=cannister_r
  bone2Name=cannister_leg_02_r
  swing1LimitDegrees=45
  swing1Motion=Free
  swing2LimitDegrees=45
  swing2Motion=Free
  twistLimitDegrees=45
  twistMotion=Free
 }
 {
  bone1Name=cannister_l
  bone2Name=cannister_leg_02_l
  swing1LimitDegrees=45
  swing1Motion=Free
  swing2LimitDegrees=45
  swing2Motion=Free
  twistLimitDegrees=45
  twistMotion=Free
 }
 {
  bone1Name=cannister_r
  bone2Name=cannister_leg_02_r
  swing1LimitDegrees=45
  swing1Motion=Free
  swing2LimitDegrees=45
  swing2Motion=Free
  twistLimitDegrees=45
  twistMotion=Free
 }
 {
  bone1Name=cannister_l
  bone2Name=cannister_leg_02_l
  swing1LimitDegrees=45
  swing1Motion=Free
  swing2LimitDegrees=45
  swing2Motion=Free
  twistLimitDegrees=45
  twistMotion=Free
 }
]
