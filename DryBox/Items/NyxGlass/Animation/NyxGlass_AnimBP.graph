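# Animation Blueprint for the NyxGlass grapple weapon, supplying full-body player poses for zipping, hovering, slashing, releases and falls.
# The event graph derives flags and a lean value from weapon state, velocity and animation curves, and sets pawn swing rotation on notifies.
# FullBodyOverride blends zip and attack state machines with inertialization, with free fall overriding both.
# Zipping adds per-variant lean blendspaces, and wall hits replace world-space root transform.
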AnimBlueprint /DryBox/Items/NyxGlass/Animation/NyxGlass_AnimBP
target_skeleton=&Fortnite_M_Avg_Player_Skeleton
graphs [
 AnimGraph {nodes [Output Pose(Result=Input Pose()) {layerGroup=None}]}
 CharacterSkeletalControlPostLegIK {nodes [Output Pose(Result=Input Pose() {name=InPosePostLegIK}) {layerGroup=PlayerOverride}]}
 EventGraph {nodes [
   n1 = JustLanded() {is=title K2Node_CustomEvent}
   > Set CalculateRotation(CalculateRotation:bool=false) -> (Output_Get:bool)
   Event BlueprintUpdateAnimation()
   > n2 = Sequence!(execute) -> (
    then_0
    then_1
    then_2:Exec
    then_3:Exec
    then_4:Exec
    then_5:Exec
    then_6:Exec
    then_7:Exec
    then_8:Exec
    then_9:Exec
    then_10:Exec
    then_11:Exec
    then_12:Exec
   )
   > n3 = Is Valid(InputObject=Get Weapon() -> (Weapon:B Weapon Grapple Parent Object Reference))
   n4 = Cast To B_Weapon_NyxGlass!(
    execute=@n3.Is Not Valid
    Object=Get CurrentWeapon!(self:Fort Pawn Object Reference=Get Player() -> (Player:Fort Player Pawn Athena Object Reference)) -> (CurrentWeapon:Fort Weapon Object Reference)
   ) -> (then, CastFailed, AsB Weapon Nyx Glass:B Weapon Nyx Glass Object Reference) {is=title K2Node_DynamicCast}
   > Set Weapon(Weapon:B Weapon Grapple Parent Object Reference=@n4) -> (Output_Get:B Weapon Grapple Parent Object Reference)
   n5 = Cast To B_Weapon_NyxGlass(execute=@n2.then_1, Object=Get Weapon())
   > Set isFreeFalling(isFreeFalling:bool=Get FreeFalling!(self:B Weapon Grapple Parent Object Reference=Get Weapon()) -> (FreeFalling:bool)) -> (Output_Get:bool)
   > Set IsRopeAttached(IsRopeAttached:bool=Get RopeAttached!(self:B Weapon Grapple Parent Object Reference=Get Weapon()) -> (RopeAttached:bool)) -> (Output_Get:bool)
   > Set State(State:State_NyxGlass Enum=Get State!(self:B Weapon Nyx Glass Object Reference=@n5) -> (State:State_NyxGlass Enum=NewEnumerator0)) -> (Output_Get:State_NyxGlass Enum=NewEnumerator0)
   n6 = Get Player()
   > Set ThunderSpearsEquipped(ThunderSpearsEquipped:bool=Cast To B_FeyCrab_Launcher!(Object=Get CurrentWeapon(self=@n6)) -> (AsB Fey Crab Launcher:B Fey Crab Launcher Object Reference, bSuccess:bool=~) {is=title K2Node_DynamicCast}.bSuccess) -> (Output_Get:bool)
   n7 = Get SubState_Slashing!(self:B Weapon Nyx Glass Object Reference=@n5) -> (SubState_Slashing:SubState_Slash_NyxGlass Enum=NewEnumerator0)
   > Set IsLeapingBack(IsLeapingBack:bool="Equal (Enum)"(A:SubState_Slash_NyxGlass Enum=@n7, B:SubState_Slash_NyxGlass Enum=NewEnumerator11)) -> (Output_Get:bool)
   > Set IsSlashingPastTarget(IsSlashingPastTarget:bool="Equal (Enum)"(A=@n7, B=NewEnumerator10)) -> (Output_Get:bool)
   > Set IsMovingToTarget(IsMovingToTarget:bool="Equal (Enum)"(A=@n7, B=NewEnumerator0)) -> (Output_Get:bool)
   n8 = Get Player()
   n9 = BreakVector(InVec=Get Velocity!(self:Movement Component Object Reference=Get CharacterMovement!(self:Character Object Reference=@n8) -> (CharacterMovement:Character Movement Component Object Reference)) -> (Velocity:Vector))
   n10 = MakeVector(X=@n9, Y=@n9.Y)
   > Set ZipLean(ZipLean:double=Dot Product!(
     A:Vector=Normalize!(A:Vector=@n10, Tolerance:float=0.0001) -> (ReturnValue:Vector) {is=title K2Node_CallFunction}
     B:Vector=GetActorRightVector!(self:Actor Object Reference=@n8) -> (ReturnValue:Vector)
    ) -> (ReturnValue:double)) -> (Output_Get:double)
   n11 = Get State!() -> (State=NewEnumerator0)
   > Set IsAttackRelease(IsAttackRelease:bool=AND Boolean(A="Equal (Enum)"(A=@n7, B=NewEnumerator11), B="Equal (Enum)"(A:State_NyxGlass Enum=@n11, B:State_NyxGlass Enum=NewEnumerator9))) -> (Output_Get:bool)
   Set bSwinging_UseProgrammaticRotation!(execute=@n2.then_2, bSwinging_UseProgrammaticRotation:bool=false, self:Fort Player Pawn Athena Object Reference=Get Player()) -> (then, Output_Get:bool)
   > n12 = Branch(Condition=Get Bound() -> (Bound:bool))
   Bind Event to On Pawn Exit Swinging(execute=@n12.else, self:Fort Player Pawn Object Reference=Get Player(), Delegate=@n1) {is=title K2Node_AddDelegate}
   > Set Bound(Bound:bool=true) -> (Output_Get:bool)
   Set IsHovering(execute=@n2.then_6, IsHovering:bool="Equal (Enum)"(A=Get State() -> (State=NewEnumerator0), B=NewEnumerator5)) -> (Output_Get:bool)
   n13 = NOT Boolean(A="Equal (Enum)"(A:SubState_Slash_NyxGlass Enum=@n7, B:SubState_Slash_NyxGlass Enum=NewEnumerator11))
   n14 = "Equal (Enum)"(A:State_NyxGlass Enum=Get State() -> (State=NewEnumerator0), B:State_NyxGlass Enum=NewEnumerator9)
   n15 = Get State() -> (State=NewEnumerator0)
   n16 = Get State() -> (State=NewEnumerator0)
   > Set IsZipAttack(IsZipAttack:bool=OR Boolean!(A="Equal (Enum)"(A=@n16, B=NewEnumerator6), B=AND Boolean(A=@n13, B=@n14), C:bool="Equal (Enum)"(A=@n15, B=NewEnumerator5)) -> (ReturnValue)) -> (Output_Get:bool)
   > Set IsAttackStart(IsAttackStart:bool="Equal (Enum)"(A=Get State() -> (State=NewEnumerator0), B=NewEnumerator6)) -> (Output_Get:bool)
   > Set ZipLeanAlpha(ZipLeanAlpha:double=GetCurveValue(CurveName=Moving)) -> (Output_Get:double)
   > Set IsHoverRelease(IsHoverRelease:bool="float > float"(A=GetCurveValue(CurveName=Hover), B=0.5)) -> (Output_Get:bool)
   > Set IsAttackFall(IsAttackFall:bool="float > float"(A=GetCurveValue(CurveName=SwingPhase), B=2.5)) -> (Output_Get:bool)
   n17 = GetCurveValue(CurveName=SwingPhase)
   n18 = "float < float"(A=@n17, B=2.5)
   n19 = "float >= float"(A=@n17, B=1.5)
   n20 = Branch(execute=@n2.then_4, Condition=OR Boolean!(A=Get IsFallRelease() -> (IsFallRelease:bool), B=AND Boolean(A=@n19, B=@n18)) -> (ReturnValue))
   > Set AllowNonRelease(AllowNonRelease:bool=true) -> (Output_Get:bool)
   Set AllowNonRelease(execute=@n20.else, AllowNonRelease=false)
   n21 = BreakVector(InVec=Get XYVelocity() -> (XYVelocity:Vector))
   n22 = "float < float"(A="Absolute (Float)"(A=@n21), B=350)
   n23 = "float < float"(A="Absolute (Float)"(A=@n21.Y), B=350)
   n24 = GetCurveValue(CurveName=SwingPhase)
   n25 = "float >= float"(A=@n24, B=0.3)
   n26 = "float < float"(A=@n24, B=0.7)
   n27 = Branch(execute=@n2.then_5, Condition=AND Boolean(A=AND Boolean(A=@n22, B=@n23), B=AND Boolean(A=@n26, B=@n25)))
   > Set IsHanging(IsHanging:bool=true) -> (Output_Get:bool)
   Set IsHanging(execute=@n27.else, IsHanging=false)
   AnimNotify_EnterWebCast() {is=title K2Node_Event}
   n28 = Get Player()
   > Set CurrentRootRotation(CurrentRootRotation:Rotator=Combine Rotators!(
     A:Rotator=Get Actor Rotation!(self:Actor Object Reference=@n28) -> (ReturnValue:Rotator) {is=title K2Node_CallFunction}
     B:Rotator="0,-90, 0"
    ) -> (ReturnValue:Rotator)) -> (Output_Get:Rotator)
   > Set CalculateRotation(CalculateRotation=true)
   > n29 = Branch(Condition=Get IsGroundSwing() -> (IsGroundSwing:bool))
   > Set bSwingingShouldRotateMesh!(execute, bSwingingShouldRotateMesh:bool=false, self:Fort Player Pawn Athena Object Reference=Get Player()) -> (then, Output_Get:bool)
   Set bSwingingShouldRotateMesh(execute=@n29.else, bSwingingShouldRotateMesh=true, self=Get Player())
   AnimNotify_EnterNearGroundToPull() {is=title K2Node_Event}
   > Set bSwingingShouldRotateMesh(bSwingingShouldRotateMesh=false, self=Get Player())
   AnimNotify_EnterSwingNearGround() {is=title K2Node_Event}
   > Set bSwingingShouldRotateMesh(bSwingingShouldRotateMesh=true, self=Get Player())
   > Set IsGroundSwing(IsGroundSwing:bool=false) -> (Output_Get:bool)
   AnimNotify_EnterSwingInAir() {is=title K2Node_Event}
   > Set bSwingingShouldRotateMesh(bSwingingShouldRotateMesh=true, self=Get Player())
   > Set IsGroundSwing(IsGroundSwing=false)
   n30 = Event BlueprintProcessAnimInputEvent()
   > Branch(Condition=IsNamed(self=@n30.AnimInputEvent, InEventName=aie.Player.Swinging.HitWallAndDetached))
   > Set IsWallHit(IsWallHit:bool=true) -> (Output_Get:bool)
   Event BlueprintPostEvaluateAnimation() {is=title K2Node_Event}
   > Set IsWallHit(IsWallHit=false)
  ]}
 FullBodyOverride {nodes [
   Input Pose() {name=PassThroughFullBodyPose}
   AnimGraphNode_StateMachine_0 = ZipStateMachine() {is=title AnimGraphNode_StateMachine, bAllowConduitEntryStates=true, bSkipFirstUpdateTransition=false, maxTransitionsPerFrame=0}
   AnimGraphNode_StateMachine_1 = AttackStateMachine() {is=title AnimGraphNode_StateMachine, bAllowConduitEntryStates=true}
   n1 = Blend Poses by bool(
    BlendPose_0=@AnimGraphNode_StateMachine_1
    BlendPose_1=@AnimGraphNode_StateMachine_0
    BlendTime_0=0.2
    bActiveValue=Get IsZipAttack() -> (IsZipAttack:bool)
   ) {blendType=Linear, transitionType=Inertialization}
   n2 = Sequence Player 'NyxGlass_Fall_FreeFall'() {bLoopAnimation=false, sequence=&/NyxGlass_Fall_FreeFall}
   n3 = Get isFreeFalling() -> (isFreeFalling:bool)
   Output Pose(Result=Modify Curve(
     SourcePose=Blend Poses by bool(BlendPose_0=@n2, BlendPose_1=@n1, BlendTime_0=0.8, BlendTime_1=0.6, bActiveValue=@n3) {blendType=Linear, transitionType=Inertialization}
     CurveValues_0=1
     Alpha=1
    ) {curveNames [BothLegFK], curveValues [0]}) {layerGroup=PlayerOverride}
  ]}
 ZipStateMachine {
  entry=Conduit
  path=FullBodyOverride.AnimGraphNode_StateMachine_0.ZipStateMachine
  ids=id
  nodes [
   AnimStateConduitNode_0 = Conduit()
   AnimStateNode_0 = WallHit() {is=title AnimStateNode, stateEntered {guid=#1, notifyName=EnterWallHit}}
   AnimStateNode_4 = Zipping() {
    is=title AnimStateNode
    stateEntered {guid=#2, notifyName=EnterSwinging}
    stateFullyBlended {guid=#3}
    stateLeft {guid=#4, notifyName=ExitSwinging}
   }
   AnimStateNode_5 = ZipRelease() {
    is=title AnimStateNode
    stateEntered {guid=#5}
    stateFullyBlended {guid=#6}
    stateLeft {guid=#7, notifyName=ExitSwingRelease}
   }
   AnimStateNode_6 = Fall() {
    is=title AnimStateNode
    stateEntered {guid=#8}
    stateFullyBlended {guid=#9}
    stateLeft {guid=#10}
   }
   AnimStateTransitionNode_0 = AnimStateTransitionNode() {bAutomaticRuleBasedOnSequencePlayerInState=true, crossfadeDuration=0.4}
   AnimStateTransitionNode_1 = AnimStateTransitionNode() {crossfadeDuration=0.25, logicType=TLT_Inertialization}
   AnimStateTransitionNode_15 = AnimStateTransitionNode()
   AnimStateTransitionNode_16 = AnimStateTransitionNode() {crossfadeDuration=0.35, logicType=TLT_Inertialization, priorityOrder=100}
   AnimStateTransitionNode_17 = AnimStateTransitionNode() {crossfadeDuration=0.4, logicType=TLT_Inertialization, priorityOrder=2}
   AnimStateTransitionNode_18 = AnimStateTransitionNode() {crossfadeDuration=0.15, logicType=TLT_Inertialization}
   AnimStateTransitionNode_19 = AnimStateTransitionNode()
   AnimStateTransitionNode_2 = AnimStateTransitionNode() {crossfadeDuration=0.25, logicType=TLT_Inertialization, priorityOrder=100}
   AnimStateTransitionNode_20 = AnimStateTransitionNode()
   AnimStateTransitionNode_4 = AnimStateTransitionNode() {bAutomaticRuleBasedOnSequencePlayerInState=true, logicType=TLT_Inertialization}
  ]
  links [
   {
    dst=Fall
    rule=FullBodyOverride.AnimGraphNode_StateMachine_0.ZipStateMachine.AnimStateTransitionNode_17.Transition
    src=Conduit
    via=AnimStateTransitionNode_17
   }
   {
    dst=ZipRelease
    rule=FullBodyOverride.AnimGraphNode_StateMachine_0.ZipStateMachine.AnimStateTransitionNode_16.Transition
    src=Conduit
    via=AnimStateTransitionNode_16
   }
   {
    dst=Zipping
    rule=FullBodyOverride.AnimGraphNode_StateMachine_0.ZipStateMachine.AnimStateTransitionNode_18.Transition
    src=Conduit
    via=AnimStateTransitionNode_18
   }
   {
    dst=Conduit
    rule=FullBodyOverride.AnimGraphNode_StateMachine_0.ZipStateMachine.AnimStateTransitionNode_20.Transition
    src=Fall
    via=AnimStateTransitionNode_20
   }
   {
    dst=Fall
    rule=FullBodyOverride.AnimGraphNode_StateMachine_0.ZipStateMachine.AnimStateTransitionNode_4.Transition
    src=WallHit
    via=AnimStateTransitionNode_4
   }
   {
    dst=Zipping
    rule=FullBodyOverride.AnimGraphNode_StateMachine_0.ZipStateMachine.AnimStateTransitionNode_2.Transition
    src=WallHit
    via=AnimStateTransitionNode_2
   }
   {
    dst=Conduit
    rule=FullBodyOverride.AnimGraphNode_StateMachine_0.ZipStateMachine.AnimStateTransitionNode_19.Transition
    src=ZipRelease
    via=AnimStateTransitionNode_19
   }
   {
    dst=Fall
    rule=FullBodyOverride.AnimGraphNode_StateMachine_0.ZipStateMachine.AnimStateTransitionNode_0.Transition
    src=ZipRelease
    via=AnimStateTransitionNode_0
   }
   {
    dst=Conduit
    rule=FullBodyOverride.AnimGraphNode_StateMachine_0.ZipStateMachine.AnimStateTransitionNode_15.Transition
    src=Zipping
    via=AnimStateTransitionNode_15
   }
   {
    dst=WallHit
    rule=FullBodyOverride.AnimGraphNode_StateMachine_0.ZipStateMachine.AnimStateTransitionNode_1.Transition
    src=Zipping
    via=AnimStateTransitionNode_1
   }
  ]
 }
 Conduit {path=FullBodyOverride.AnimGraphNode_StateMachine_0.ZipStateMachine.AnimStateConduitNode_0.Conduit, nodes [Result(bCanEnterTransition=true)]}
 WallHit {
  path=FullBodyOverride.AnimGraphNode_StateMachine_0.ZipStateMachine.AnimStateNode_0.WallHit
  nodes [
   n1 = "Transform (Modify) Bone - Bone: Root"!(
    ComponentPose=Local To Component(LocalPose=Sequence Player 'NyxGlass_Release_Wall_Contact'() {sequence=/DryBox/Items/NyxGlass/Animation/Release/NyxGlass_Release_Wall_Contact})
    bAlphaBoolEnabled:bool=True
    Translation=Get CurrentRootTranslation() -> (CurrentRootTranslation:Vector)
    Rotation=Get CurrentRootRotation() -> (CurrentRootRotation:Rotator)
    Scale="1,1,1"
   ) -> (Pose) {
    alphaInputType=Bool
    rotationMode=BMM_Replace
    rotationSpace=BCS_WorldSpace
    translationMode=BMM_Replace
    translationSpace=BCS_WorldSpace
    alphaBoolBlend {blendOption=Linear, blendOutTime=0.5}
    boneToModify {boneName=Root}
   }
   Output Animation Pose(Result=Component To Local(ComponentPose=@n1))
  ]
 }
 Zipping {
  path=FullBodyOverride.AnimGraphNode_StateMachine_0.ZipStateMachine.AnimStateNode_4.Zipping
  nodes [
   AnimGraphNode_StateMachine_2 = ZipSelectionLeanAdditive_B() {is=title AnimGraphNode_StateMachine, bAllowConduitEntryStates=true}
   AnimGraphNode_StateMachine_3 = ZipSelectionLeanAdditive_C() {is=title AnimGraphNode_StateMachine, bAllowConduitEntryStates=true}
   AnimGraphNode_StateMachine_4 = ZipSelectionLeanAdditive_A() {is=title AnimGraphNode_StateMachine, bAllowConduitEntryStates=true}
   n1 = Blend Poses by int!(
    BlendPose_0:Pose Link Structure=@AnimGraphNode_StateMachine_3
    BlendPose_1:Pose Link Structure=@AnimGraphNode_StateMachine_4
    BlendPose_2:Pose Link Structure=@AnimGraphNode_StateMachine_2
    BlendTime_0:float=0.1
    BlendTime_1:float=0.1
    BlendTime_2:float=0.1
    ActiveChildIndex:int=Get CurrentZipCastAnim() -> (CurrentZipCastAnim:int=0)
   ) -> (Pose:Pose Link Structure) {
    is=title AnimGraphNode_BlendListByInt
    blendType=Linear
    childUpateMode=ResetChildOnActivate
    transitionType=Inertialization
    blendPose [{linkId=-1, sourceLinkId=-1}, {linkId=-1, sourceLinkId=-1}, {linkId=-1, sourceLinkId=-1}]
    blendTime [0.1,0.1,0.1]
   }
   AnimGraphNode_StateMachine_1 = ZipSelection() {is=title AnimGraphNode_StateMachine, bSkipFirstUpdateTransition=false, maxTransitionsPerFrame=0}
   Output Animation Pose(Result=Apply Additive(Base=@AnimGraphNode_StateMachine_1, Additive=@n1, Alpha=Get ZipLeanAlpha() -> (ZipLeanAlpha:double)) {alphaBoolBlend {blendOption=Linear}})
  ]
 }
 ZipSelection {
  entry=ZipStart
  path=FullBodyOverride.AnimGraphNode_StateMachine_0.ZipStateMachine.AnimStateNode_4.Zipping.AnimGraphNode_StateMachine_1.ZipSelection
  ids=id
  nodes [
   AnimStateNode_3 = ZipStart() {is=title AnimStateNode}
   AnimStateNode_4 = ZipLoop() {is=title AnimStateNode}
   AnimStateTransitionNode_7 = AnimStateTransitionNode() {crossfadeDuration=0.4, logicType=TLT_Inertialization}
  ]
  links [{
    dst=ZipLoop
    rule=FullBodyOverride.AnimGraphNode_StateMachine_0.ZipStateMachine.AnimStateNode_4.Zipping.AnimGraphNode_StateMachine_1.ZipSelection.AnimStateTransitionNode_7.Transition
    src=ZipStart
    via=AnimStateTransitionNode_7
   }]
 }
 ZipStart {
  path=FullBodyOverride.AnimGraphNode_StateMachine_0.ZipStateMachine.AnimStateNode_4.Zipping.AnimGraphNode_StateMachine_1.ZipSelection.AnimStateNode_3.ZipStart
  nodes [
   n1 = Sequence Player() {
    becomeRelevantFunction {bSelfContext=true, memberGuid=#11, memberName=ZipCastSelection}
    node {bLoopAnimation=false, becomeRelevantFunction {functionName=ZipToCast}}
   }
   Output Animation Pose(Result=@n1)
  ]
 }
 ZipLoop {
  path=FullBodyOverride.AnimGraphNode_StateMachine_0.ZipStateMachine.AnimStateNode_4.Zipping.AnimGraphNode_StateMachine_1.ZipSelection.AnimStateNode_4.ZipLoop
  nodes [
   n1 = Blend Poses by int(
    BlendPose_0=Sequence Player 'NyxGlass_Zip_C_Idle'() {sequence=/DryBox/Items/NyxGlass/Animation/Cast/NyxGlass_Zip_C_Idle}
    BlendPose_1=Sequence Player 'NyxGlass_Zip_Idle_A'() {sequence=/DryBox/Items/NyxGlass/Animation/Cast/NyxGlass_Zip_Idle_A}
    BlendPose_2=Sequence Player 'NyxGlass_Zip_B_Idle'() {sequence=/DryBox/Items/NyxGlass/Animation/Swing/NyxGlass_Zip_B_Idle}
    BlendTime_0=0.1
    BlendTime_1=0.1
    BlendTime_2=0.1
    ActiveChildIndex=Get CurrentZipCastAnim() -> (CurrentZipCastAnim=0)
   ) {
    blendType=Linear
    childUpateMode=ResetChildOnActivate
    transitionType=Inertialization
    blendPose [{linkId=-1, sourceLinkId=-1}, {linkId=-1, sourceLinkId=-1}, {linkId=-1, sourceLinkId=-1}]
    blendTime [0.1,0.1,0.1]
   }
   Output Animation Pose(Result=Modify Curve!(SourcePose=@n1, CurveValues_0=1, CurveValues_1:float=1, CurveValues_2:float=1, Alpha=1) -> (Pose) {curveNames [Moving,BothLegFK,BothHandFK], curveValues [0,0,0]})
  ]
 }
 Transition {
  path=FullBodyOverride.AnimGraphNode_StateMachine_0.ZipStateMachine.AnimStateNode_4.Zipping.AnimGraphNode_StateMachine_1.ZipSelection.AnimStateTransitionNode_7.Transition
  nodes [Result(bCanEnterTransition="float >= float"(A=GetCurveValue(CurveName=BlendOut), B=0.5))]
 }
 ZipSelectionLeanAdditive_B {
  entry=ZipIdle_B
  path=FullBodyOverride.AnimGraphNode_StateMachine_0.ZipStateMachine.AnimStateNode_4.Zipping.AnimGraphNode_StateMachine_2.ZipSelectionLeanAdditive_B
  ids=id
  nodes [AnimStateNode_0 = ZipIdle_B() {is=title AnimStateNode}]
 }
 ZipIdle_B {
  path=FullBodyOverride.AnimGraphNode_StateMachine_0.ZipStateMachine.AnimStateNode_4.Zipping.AnimGraphNode_StateMachine_2.ZipSelectionLeanAdditive_B.AnimStateNode_0.ZipIdle_B
  nodes [
   n1 = Get ZipLean() -> (ZipLean:double)
   Output Animation Pose(Result=Apply Additive(
     Base="Sequence Evaluator (by time) 'NyxGlass_Zip_B_Neutral_Pose'"!(ExplicitTime:float=0) -> (Pose:Pose Link Structure) {is=title AnimGraphNode_SequenceEvaluator, sequence=/DryBox/Items/NyxGlass/Animation/Swing/NyxGlass_Zip_B_Neutral_Pose}
     Additive=Blendspace Player 'NyxGlass_Zip_Lean_B_BS'(X=@n1, Y=0) {blendSpace=/DryBox/Items/NyxGlass/Animation/Cast/NyxGlass_Zip_Lean_B_BS}
    ) {alphaBoolBlend {blendOption=Linear}})
  ]
 }
 ZipSelectionLeanAdditive_C {
  entry=ZipIdle_C
  path=FullBodyOverride.AnimGraphNode_StateMachine_0.ZipStateMachine.AnimStateNode_4.Zipping.AnimGraphNode_StateMachine_3.ZipSelectionLeanAdditive_C
  ids=id
  nodes [AnimStateNode_2 = ZipIdle_C() {is=title AnimStateNode}]
 }
 ZipIdle_C {
  path=FullBodyOverride.AnimGraphNode_StateMachine_0.ZipStateMachine.AnimStateNode_4.Zipping.AnimGraphNode_StateMachine_3.ZipSelectionLeanAdditive_C.AnimStateNode_2.ZipIdle_C
  nodes [
   n1 = Get ZipLean()
   Output Animation Pose(Result=Apply Additive(
     Base="Sequence Evaluator (by time) 'NyxGlass_Zip_C_Neutral_Pose'"!(ExplicitTime:float=0) -> (Pose:Pose Link Structure) {is=title AnimGraphNode_SequenceEvaluator, sequence=/DryBox/Items/NyxGlass/Animation/Cast/NyxGlass_Zip_C_Neutral_Pose}
     Additive=Blendspace Player 'NyxGlass_Zip_Lean_C_BS'(X=@n1, Y=0) {blendSpace=/DryBox/Items/NyxGlass/Animation/Cast/NyxGlass_Zip_Lean_C_BS}
    ) {alphaBoolBlend {blendOption=Linear}})
  ]
 }
 ZipSelectionLeanAdditive_A {
  entry=ZipIdle_A
  path=FullBodyOverride.AnimGraphNode_StateMachine_0.ZipStateMachine.AnimStateNode_4.Zipping.AnimGraphNode_StateMachine_4.ZipSelectionLeanAdditive_A
  ids=id
  nodes [AnimStateNode_1 = ZipIdle_A() {is=title AnimStateNode}]
 }
 ZipIdle_A {
  path=FullBodyOverride.AnimGraphNode_StateMachine_0.ZipStateMachine.AnimStateNode_4.Zipping.AnimGraphNode_StateMachine_4.ZipSelectionLeanAdditive_A.AnimStateNode_1.ZipIdle_A
  nodes [
   n1 = Get ZipLean()
   Output Animation Pose(Result=Apply Additive(
     Base="Sequence Evaluator (by time) 'NyxGlass_Zip_A_Neutral_Pose'"!(ExplicitTime:float=0) -> (Pose:Pose Link Structure) {is=title AnimGraphNode_SequenceEvaluator, sequence=/DryBox/Items/NyxGlass/Animation/Cast/NyxGlass_Zip_A_Neutral_Pose}
     Additive=Blendspace Player 'NyxGlass_Zip_Lean_A_BS'(X=@n1, Y=0) {blendSpace=/DryBox/Items/NyxGlass/Animation/Cast/NyxGlass_Zip_Lean_A_BS}
    ) {alphaBoolBlend {blendOption=Linear}})
  ]
 }
 ZipRelease {
  path=FullBodyOverride.AnimGraphNode_StateMachine_0.ZipStateMachine.AnimStateNode_5.ZipRelease
  nodes [
   AnimGraphNode_StateMachine_1 = SwingRleaseState() {is=title AnimGraphNode_StateMachine, bAllowConduitEntryStates=true}
   Output Animation Pose(Result=@AnimGraphNode_StateMachine_1)
  ]
 }
 SwingRleaseState {
  entry=Conduit
  path=FullBodyOverride.AnimGraphNode_StateMachine_0.ZipStateMachine.AnimStateNode_5.ZipRelease.AnimGraphNode_StateMachine_1.SwingRleaseState
  ids=id
  nodes [
   AnimStateConduitNode_0 = Conduit()
   AnimStateNode_0 = AttackRelease() {is=title AnimStateNode}
   AnimStateNode_1 = ZipRelease()
   AnimStateTransitionNode_2 = AnimStateTransitionNode() {logicType=TLT_Inertialization}
   AnimStateTransitionNode_3 = AnimStateTransitionNode() {crossfadeDuration=0.5, logicType=TLT_Inertialization, priorityOrder=100}
  ]
  links [
   {
    dst=AttackRelease
    rule=FullBodyOverride.AnimGraphNode_StateMachine_0.ZipStateMachine.AnimStateNode_5.ZipRelease.AnimGraphNode_StateMachine_1.SwingRleaseState.AnimStateTransitionNode_2.Transition
    src=Conduit
    via=AnimStateTransitionNode_2
   }
   {
    dst=ZipRelease
    rule=FullBodyOverride.AnimGraphNode_StateMachine_0.ZipStateMachine.AnimStateNode_5.ZipRelease.AnimGraphNode_StateMachine_1.SwingRleaseState.AnimStateTransitionNode_3.Transition
    src=Conduit
    via=AnimStateTransitionNode_3
   }
  ]
 }
 Conduit {
  path=FullBodyOverride.AnimGraphNode_StateMachine_0.ZipStateMachine.AnimStateNode_5.ZipRelease.AnimGraphNode_StateMachine_1.SwingRleaseState.AnimStateConduitNode_0.Conduit
  nodes [Result(bCanEnterTransition=true)]
 }
 AttackRelease {
  path=FullBodyOverride.AnimGraphNode_StateMachine_0.ZipStateMachine.AnimStateNode_5.ZipRelease.AnimGraphNode_StateMachine_1.SwingRleaseState.AnimStateNode_0.AttackRelease
  nodes [
   AnimGraphNode_StateMachine_0 = AttackReleaseSelectione() {is=title AnimGraphNode_StateMachine, bAllowConduitEntryStates=true}
   n1 = Blend Poses by int(
    BlendPose_0=Sequence Player 'NyxGlass_Attack_C_Release'() {bLoopAnimation=false, sequence=/DryBox/Items/NyxGlass/Animation/Attack/NyxGlass_Attack_C_Release}
    BlendPose_1=Sequence Player 'NyxGlass_Attack_Release'() {bLoopAnimation=false, sequence=/DryBox/Items/NyxGlass/Animation/Release/NyxGlass_Attack_Release}
    BlendPose_2=Sequence Player 'NyxGlass_Attack_B_Release'() {bLoopAnimation=false, sequence=/DryBox/Items/NyxGlass/Animation/Release/NyxGlass_Attack_B_Release}
    BlendTime_0=0.1
    BlendTime_1=0.1
    BlendTime_2=0.1
    ActiveChildIndex=Get CurrentAttackIntroAnim() -> (CurrentAttackIntroAnim:int=0)
   ) {
    blendType=Linear
    childUpateMode=ResetChildOnActivate
    transitionType=Inertialization
    blendPose [{linkId=-1, sourceLinkId=-1}, {linkId=-1, sourceLinkId=-1}, {linkId=-1, sourceLinkId=-1}]
    blendTime [0.1,0.1,0.1]
   }
   Output Animation Pose(Result=@n1)
  ]
 }
 AttackReleaseSelectione {
  entry=Conduit
  path=FullBodyOverride.AnimGraphNode_StateMachine_0.ZipStateMachine.AnimStateNode_5.ZipRelease.AnimGraphNode_StateMachine_1.SwingRleaseState.AnimStateNode_0.AttackRelease.AnimGraphNode_StateMachine_0.AttackReleaseSelectione
  ids=id
  nodes [
   AnimStateConduitNode_0 = Conduit()
   AnimStateNode_0 = Release_A() {is=title AnimStateNode}
   AnimStateNode_1 = Release_C() {is=title AnimStateNode}
   AnimStateNode_2 = Release_B() {is=title AnimStateNode}
   AnimStateTransitionNode_0 = AnimStateTransitionNode() {logicType=TLT_Inertialization, priorityOrder=100}
   AnimStateTransitionNode_1 = AnimStateTransitionNode() {logicType=TLT_Inertialization, priorityOrder=2}
   AnimStateTransitionNode_2 = AnimStateTransitionNode() {logicType=TLT_Inertialization}
  ]
  links [
   {
    dst=Release_A
    rule=FullBodyOverride.AnimGraphNode_StateMachine_0.ZipStateMachine.AnimStateNode_5.ZipRelease.AnimGraphNode_StateMachine_1.SwingRleaseState.AnimStateNode_0.AttackRelease.AnimGraphNode_StateMachine_0.AttackReleaseSelectione.AnimStateTransitionNode_0.AnimationTransitionGraph_1
    src=Conduit
    via=AnimStateTransitionNode_0
   }
   {
    dst=Release_B
    rule=FullBodyOverride.AnimGraphNode_StateMachine_0.ZipStateMachine.AnimStateNode_5.ZipRelease.AnimGraphNode_StateMachine_1.SwingRleaseState.AnimStateNode_0.AttackRelease.AnimGraphNode_StateMachine_0.AttackReleaseSelectione.AnimStateTransitionNode_2.Transition
    src=Conduit
    via=AnimStateTransitionNode_2
   }
   {
    dst=Release_C
    rule=FullBodyOverride.AnimGraphNode_StateMachine_0.ZipStateMachine.AnimStateNode_5.ZipRelease.AnimGraphNode_StateMachine_1.SwingRleaseState.AnimStateNode_0.AttackRelease.AnimGraphNode_StateMachine_0.AttackReleaseSelectione.AnimStateTransitionNode_1.AnimationTransitionGraph_1
    src=Conduit
    via=AnimStateTransitionNode_1
   }
  ]
 }
 Conduit {
  path=FullBodyOverride.AnimGraphNode_StateMachine_0.ZipStateMachine.AnimStateNode_5.ZipRelease.AnimGraphNode_StateMachine_1.SwingRleaseState.AnimStateNode_0.AttackRelease.AnimGraphNode_StateMachine_0.AttackReleaseSelectione.AnimStateConduitNode_0.Conduit
  nodes [Result(bCanEnterTransition=true)]
 }
 Release_A {
  path=FullBodyOverride.AnimGraphNode_StateMachine_0.ZipStateMachine.AnimStateNode_5.ZipRelease.AnimGraphNode_StateMachine_1.SwingRleaseState.AnimStateNode_0.AttackRelease.AnimGraphNode_StateMachine_0.AttackReleaseSelectione.AnimStateNode_0.Release_A
  nodes [Output Animation Pose(Result=Sequence Player 'NyxGlass_Attack_Release'() {bLoopAnimation=false, sequence=&NyxGlass_Attack_Release, startPosition=0.2})]
 }
 Release_C {
  path=FullBodyOverride.AnimGraphNode_StateMachine_0.ZipStateMachine.AnimStateNode_5.ZipRelease.AnimGraphNode_StateMachine_1.SwingRleaseState.AnimStateNode_0.AttackRelease.AnimGraphNode_StateMachine_0.AttackReleaseSelectione.AnimStateNode_1.Release_C
  nodes [Output Animation Pose(Result=Sequence Player 'NyxGlass_Attack_C_Release'() {bLoopAnimation=false, sequence=&NyxGlass_Attack_C_Release, startPosition=0.2})]
 }
 Release_B {
  path=FullBodyOverride.AnimGraphNode_StateMachine_0.ZipStateMachine.AnimStateNode_5.ZipRelease.AnimGraphNode_StateMachine_1.SwingRleaseState.AnimStateNode_0.AttackRelease.AnimGraphNode_StateMachine_0.AttackReleaseSelectione.AnimStateNode_2.Release_B
  nodes [Output Animation Pose(Result=Sequence Player 'NyxGlass_Attack_B_Release'() {bLoopAnimation=false, sequence=&NyxGlass_Attack_B_Release, startPosition=0.2})]
 }
 AnimationTransitionGraph_1 {
  path=FullBodyOverride.AnimGraphNode_StateMachine_0.ZipStateMachine.AnimStateNode_5.ZipRelease.AnimGraphNode_StateMachine_1.SwingRleaseState.AnimStateNode_0.AttackRelease.AnimGraphNode_StateMachine_0.AttackReleaseSelectione.AnimStateTransitionNode_0.AnimationTransitionGraph_1
  nodes [Result(bCanEnterTransition=true)]
 }
 AnimationTransitionGraph_1 {
  path=FullBodyOverride.AnimGraphNode_StateMachine_0.ZipStateMachine.AnimStateNode_5.ZipRelease.AnimGraphNode_StateMachine_1.SwingRleaseState.AnimStateNode_0.AttackRelease.AnimGraphNode_StateMachine_0.AttackReleaseSelectione.AnimStateTransitionNode_1.AnimationTransitionGraph_1
  nodes [Result(bCanEnterTransition="Equal (Integer)"(A=Get CurrentZipCastAnim() -> (CurrentZipCastAnim=0)))]
 }
 Transition {
  path=FullBodyOverride.AnimGraphNode_StateMachine_0.ZipStateMachine.AnimStateNode_5.ZipRelease.AnimGraphNode_StateMachine_1.SwingRleaseState.AnimStateNode_0.AttackRelease.AnimGraphNode_StateMachine_0.AttackReleaseSelectione.AnimStateTransitionNode_2.Transition
  nodes [Result(bCanEnterTransition="Equal (Integer)"(A=Get CurrentAttackIntroAnim() -> (CurrentAttackIntroAnim=0), B=2))]
 }
 ZipRelease {
  path=FullBodyOverride.AnimGraphNode_StateMachine_0.ZipStateMachine.AnimStateNode_5.ZipRelease.AnimGraphNode_StateMachine_1.SwingRleaseState.AnimStateNode_1.ZipRelease
  nodes [
   n1 = Sequence Player() {
    becomeRelevantFunction {bSelfContext=true, memberGuid=#12, memberName=ZipReleaseSelection}
    node {becomeRelevantFunction {functionName=ZipRelease_0}}
   }
   Output Animation Pose(Result=@n1)
  ]
 }
 Transition {
  path=FullBodyOverride.AnimGraphNode_StateMachine_0.ZipStateMachine.AnimStateNode_5.ZipRelease.AnimGraphNode_StateMachine_1.SwingRleaseState.AnimStateTransitionNode_2.Transition
  nodes [Result(bCanEnterTransition=Get IsAttackRelease() -> (IsAttackRelease:bool))]
 }
 Transition {
  path=FullBodyOverride.AnimGraphNode_StateMachine_0.ZipStateMachine.AnimStateNode_5.ZipRelease.AnimGraphNode_StateMachine_1.SwingRleaseState.AnimStateTransitionNode_3.Transition
  nodes [Result(bCanEnterTransition=true)]
 }
 Fall {
  path=FullBodyOverride.AnimGraphNode_StateMachine_0.ZipStateMachine.AnimStateNode_6.Fall
  nodes [
   AnimGraphNode_StateMachine_0 = FallState() {is=title AnimGraphNode_StateMachine, bAllowConduitEntryStates=true}
   Output Animation Pose(Result=@AnimGraphNode_StateMachine_0)
  ]
 }
 FallState {
  entry=Conduit
  path=FullBodyOverride.AnimGraphNode_StateMachine_0.ZipStateMachine.AnimStateNode_6.Fall.AnimGraphNode_StateMachine_0.FallState
  ids=id
  nodes [
   AnimStateConduitNode_0 = Conduit()
   AnimStateNode_0 = FallStart() {is=title AnimStateNode}
   AnimStateNode_1 = FallLoop() {is=title AnimStateNode}
   AnimStateTransitionNode_0 = AnimStateTransitionNode() {bAutomaticRuleBasedOnSequencePlayerInState=true, crossfadeDuration=0.5}
   AnimStateTransitionNode_1 = AnimStateTransitionNode() {crossfadeDuration=0.4, priorityOrder=100}
   AnimStateTransitionNode_2 = AnimStateTransitionNode() {crossfadeDuration=0.5}
  ]
  links [
   {
    dst=FallLoop
    rule=FullBodyOverride.AnimGraphNode_StateMachine_0.ZipStateMachine.AnimStateNode_6.Fall.AnimGraphNode_StateMachine_0.FallState.AnimStateTransitionNode_2.Transition
    src=Conduit
    via=AnimStateTransitionNode_2
   }
   {
    dst=FallStart
    rule=FullBodyOverride.AnimGraphNode_StateMachine_0.ZipStateMachine.AnimStateNode_6.Fall.AnimGraphNode_StateMachine_0.FallState.AnimStateTransitionNode_1.Transition
    src=Conduit
    via=AnimStateTransitionNode_1
   }
   {
    dst=FallLoop
    rule=FullBodyOverride.AnimGraphNode_StateMachine_0.ZipStateMachine.AnimStateNode_6.Fall.AnimGraphNode_StateMachine_0.FallState.AnimStateTransitionNode_0.Transition
    src=FallStart
    via=AnimStateTransitionNode_0
   }
  ]
 }
 Conduit {
  path=FullBodyOverride.AnimGraphNode_StateMachine_0.ZipStateMachine.AnimStateNode_6.Fall.AnimGraphNode_StateMachine_0.FallState.AnimStateConduitNode_0.Conduit
  nodes [Result(bCanEnterTransition=true)]
 }
 FallStart {
  path=FullBodyOverride.AnimGraphNode_StateMachine_0.ZipStateMachine.AnimStateNode_6.Fall.AnimGraphNode_StateMachine_0.FallState.AnimStateNode_0.FallStart
  nodes [Output Animation Pose(Result=Sequence Player 'NyxGlass_Fall_Start'() {bLoopAnimation=false, sequence=&/NyxGlass_Fall_Start})]
 }
 FallLoop {
  path=FullBodyOverride.AnimGraphNode_StateMachine_0.ZipStateMachine.AnimStateNode_6.Fall.AnimGraphNode_StateMachine_0.FallState.AnimStateNode_1.FallLoop
  nodes [Output Animation Pose(Result=Sequence Player 'NyxGlass_Fall_Loop'() {bLoopAnimation=false, sequence=&/NyxGlass_Fall_Loop})]
 }
 Transition {
  path=FullBodyOverride.AnimGraphNode_StateMachine_0.ZipStateMachine.AnimStateNode_6.Fall.AnimGraphNode_StateMachine_0.FallState.AnimStateTransitionNode_0.Transition
  nodes [Result(bCanEnterTransition=true)]
 }
 Transition {
  path=FullBodyOverride.AnimGraphNode_StateMachine_0.ZipStateMachine.AnimStateNode_6.Fall.AnimGraphNode_StateMachine_0.FallState.AnimStateTransitionNode_1.Transition
  nodes [Result(bCanEnterTransition=true)]
 }
 Transition {
  path=FullBodyOverride.AnimGraphNode_StateMachine_0.ZipStateMachine.AnimStateNode_6.Fall.AnimGraphNode_StateMachine_0.FallState.AnimStateTransitionNode_2.Transition
  nodes [Result(bCanEnterTransition=Get IsAttackFall() -> (IsAttackFall:bool))]
 }
 Transition {path=FullBodyOverride.AnimGraphNode_StateMachine_0.ZipStateMachine.AnimStateTransitionNode_0.Transition, nodes [Result()]}
 Transition {
  path=FullBodyOverride.AnimGraphNode_StateMachine_0.ZipStateMachine.AnimStateTransitionNode_1.Transition
  nodes [Result(bCanEnterTransition=Get IsWallHit() -> (IsWallHit:bool))]
 }
 Transition {
  path=FullBodyOverride.AnimGraphNode_StateMachine_0.ZipStateMachine.AnimStateTransitionNode_15.Transition
  nodes [Result(bCanEnterTransition=NOT Boolean(A=Get IsRopeAttached() -> (IsRopeAttached:bool)))]
 }
 Transition {
  path=FullBodyOverride.AnimGraphNode_StateMachine_0.ZipStateMachine.AnimStateTransitionNode_16.Transition
  nodes [Result(bCanEnterTransition=true)]
 }
 Transition {
  path=FullBodyOverride.AnimGraphNode_StateMachine_0.ZipStateMachine.AnimStateTransitionNode_17.Transition
  nodes [Result(bCanEnterTransition=OR Boolean(A=Get AllowNonRelease() -> (AllowNonRelease:bool), B=Get IsHoverRelease() -> (IsHoverRelease:bool)))]
 }
 Transition {
  path=FullBodyOverride.AnimGraphNode_StateMachine_0.ZipStateMachine.AnimStateTransitionNode_18.Transition
  nodes [Result(bCanEnterTransition=Get IsRopeAttached())]
 }
 Transition {
  path=FullBodyOverride.AnimGraphNode_StateMachine_0.ZipStateMachine.AnimStateTransitionNode_19.Transition
  nodes [
   Result(bCanEnterTransition=Get IsRopeAttached())
   Get IsAttached() -> (IsAttached:bool)
  ]
 }
 Transition {
  path=FullBodyOverride.AnimGraphNode_StateMachine_0.ZipStateMachine.AnimStateTransitionNode_2.Transition
  nodes [Result(bCanEnterTransition=Get IsAttached())]
 }
 Transition {
  path=FullBodyOverride.AnimGraphNode_StateMachine_0.ZipStateMachine.AnimStateTransitionNode_20.Transition
  nodes [Result(bCanEnterTransition=Get IsAttached())]
 }
 Transition {
  path=FullBodyOverride.AnimGraphNode_StateMachine_0.ZipStateMachine.AnimStateTransitionNode_4.Transition
  nodes [Result(bCanEnterTransition=true)]
 }
 AttackStateMachine {
  entry=Conduit
  path=FullBodyOverride.AnimGraphNode_StateMachine_1.AttackStateMachine
  ids=id
  nodes [
   AnimStateConduitNode_0 = Conduit()
   AnimStateNode_0 = Hover() {is=title AnimStateNode}
   AnimStateNode_1 = Attack() {is=title AnimStateNode}
   AnimStateNode_2 = Slash() {is=title AnimStateNode}
   AnimStateTransitionNode_0 = AnimStateTransitionNode() {logicType=TLT_Inertialization, priorityOrder=2}
   AnimStateTransitionNode_1 = AnimStateTransitionNode() {logicType=TLT_Inertialization}
   AnimStateTransitionNode_2 = AnimStateTransitionNode()
   AnimStateTransitionNode_3 = AnimStateTransitionNode()
   AnimStateTransitionNode_4 = AnimStateTransitionNode() {logicType=TLT_Inertialization, priorityOrder=100}
   AnimStateTransitionNode_5 = AnimStateTransitionNode()
  ]
  links [
   {
    dst=Conduit
    rule=FullBodyOverride.AnimGraphNode_StateMachine_1.AttackStateMachine.AnimStateTransitionNode_3.Transition
    src=Attack
    via=AnimStateTransitionNode_3
   }
   {
    dst=Attack
    rule=FullBodyOverride.AnimGraphNode_StateMachine_1.AttackStateMachine.AnimStateTransitionNode_1.Transition
    src=Conduit
    via=AnimStateTransitionNode_1
   }
   {
    dst=Hover
    rule=FullBodyOverride.AnimGraphNode_StateMachine_1.AttackStateMachine.AnimStateTransitionNode_0.Transition
    src=Conduit
    via=AnimStateTransitionNode_0
   }
   {
    dst=Slash
    rule=FullBodyOverride.AnimGraphNode_StateMachine_1.AttackStateMachine.AnimStateTransitionNode_4.Transition
    src=Conduit
    via=AnimStateTransitionNode_4
   }
   {
    dst=Conduit
    rule=FullBodyOverride.AnimGraphNode_StateMachine_1.AttackStateMachine.AnimStateTransitionNode_2.Transition
    src=Hover
    via=AnimStateTransitionNode_2
   }
   {
    dst=Conduit
    rule=FullBodyOverride.AnimGraphNode_StateMachine_1.AttackStateMachine.AnimStateTransitionNode_5.Transition
    src=Slash
    via=AnimStateTransitionNode_5
   }
  ]
 }
 Conduit {
  path=FullBodyOverride.AnimGraphNode_StateMachine_1.AttackStateMachine.AnimStateConduitNode_0.Conduit
  nodes [Result(bCanEnterTransition=true)]
 }
 Hover {
  path=FullBodyOverride.AnimGraphNode_StateMachine_1.AttackStateMachine.AnimStateNode_0.Hover
  nodes [Output Animation Pose(Result=Sequence Player 'NyxGlass_Attack_Hover_Pose'() {sequence=/DryBox/Items/NyxGlass/Animation/Attack/NyxGlass_Attack_Hover_Pose})]
 }
 Attack {
  path=FullBodyOverride.AnimGraphNode_StateMachine_1.AttackStateMachine.AnimStateNode_1.Attack
  nodes [
   n1 = Sequence Player() {
    becomeRelevantFunction {bSelfContext=true, memberGuid=#13, memberName=ZipAttackSelection}
    node {becomeRelevantFunction {functionName=ZipAttackSelection}}
   }
   Output Animation Pose(Result=@n1)
  ]
 }
 Slash {
  path=FullBodyOverride.AnimGraphNode_StateMachine_1.AttackStateMachine.AnimStateNode_2.Slash
  nodes [
   AnimGraphNode_StateMachine_0 = SlashSelection() {is=title AnimGraphNode_StateMachine, bAllowConduitEntryStates=true}
   Output Animation Pose(Result=@AnimGraphNode_StateMachine_0)
  ]
 }
 SlashSelection {
  entry=Conduit
  path=FullBodyOverride.AnimGraphNode_StateMachine_1.AttackStateMachine.AnimStateNode_2.Slash.AnimGraphNode_StateMachine_0.SlashSelection
  ids=id
  nodes [
   AnimStateConduitNode_0 = Conduit()
   AnimStateNode_0 = Strike_A() {is=title AnimStateNode}
   AnimStateNode_1 = Strike_C() {is=title AnimStateNode}
   AnimStateNode_2 = Strike_B() {is=title AnimStateNode}
   AnimStateTransitionNode_0 = AnimStateTransitionNode() {crossfadeDuration=0.3, logicType=TLT_Inertialization, priorityOrder=100}
   AnimStateTransitionNode_1 = AnimStateTransitionNode() {crossfadeDuration=0.3, logicType=TLT_Inertialization, priorityOrder=2}
   AnimStateTransitionNode_2 = AnimStateTransitionNode()
  ]
  links [
   {
    dst=Strike_A
    rule=FullBodyOverride.AnimGraphNode_StateMachine_1.AttackStateMachine.AnimStateNode_2.Slash.AnimGraphNode_StateMachine_0.SlashSelection.AnimStateTransitionNode_0.AnimationTransitionGraph_1
    src=Conduit
    via=AnimStateTransitionNode_0
   }
   {
    dst=Strike_B
    rule=FullBodyOverride.AnimGraphNode_StateMachine_1.AttackStateMachine.AnimStateNode_2.Slash.AnimGraphNode_StateMachine_0.SlashSelection.AnimStateTransitionNode_2.Transition
    src=Conduit
    via=AnimStateTransitionNode_2
   }
   {
    dst=Strike_C
    rule=FullBodyOverride.AnimGraphNode_StateMachine_1.AttackStateMachine.AnimStateNode_2.Slash.AnimGraphNode_StateMachine_0.SlashSelection.AnimStateTransitionNode_1.AnimationTransitionGraph_1
    src=Conduit
    via=AnimStateTransitionNode_1
   }
  ]
 }
 Conduit {
  path=FullBodyOverride.AnimGraphNode_StateMachine_1.AttackStateMachine.AnimStateNode_2.Slash.AnimGraphNode_StateMachine_0.SlashSelection.AnimStateConduitNode_0.Conduit
  nodes [Result(bCanEnterTransition=true)]
 }
 Strike_A {
  path=FullBodyOverride.AnimGraphNode_StateMachine_1.AttackStateMachine.AnimStateNode_2.Slash.AnimGraphNode_StateMachine_0.SlashSelection.AnimStateNode_0.Strike_A
  nodes [Output Animation Pose(Result=Sequence Player 'NyxGlass_Attack_A_Strike'() {bLoopAnimation=false, sequence=/DryBox/Items/NyxGlass/Animation/Attack/NyxGlass_Attack_A_Strike})]
 }
 Strike_C {
  path=FullBodyOverride.AnimGraphNode_StateMachine_1.AttackStateMachine.AnimStateNode_2.Slash.AnimGraphNode_StateMachine_0.SlashSelection.AnimStateNode_1.Strike_C
  nodes [Output Animation Pose(Result=Sequence Player 'NyxGlass_Attack_C_Strike'() {bLoopAnimation=false, sequence=/DryBox/Items/NyxGlass/Animation/Attack/NyxGlass_Attack_C_Strike})]
 }
 Strike_B {
  path=FullBodyOverride.AnimGraphNode_StateMachine_1.AttackStateMachine.AnimStateNode_2.Slash.AnimGraphNode_StateMachine_0.SlashSelection.AnimStateNode_2.Strike_B
  nodes [Output Animation Pose(Result=Sequence Player 'NyxGlass_Attack_B_Strike'() {bLoopAnimation=false, sequence=/DryBox/Items/NyxGlass/Animation/Attack/NyxGlass_Attack_B_Strike})]
 }
 AnimationTransitionGraph_1 {
  path=FullBodyOverride.AnimGraphNode_StateMachine_1.AttackStateMachine.AnimStateNode_2.Slash.AnimGraphNode_StateMachine_0.SlashSelection.AnimStateTransitionNode_0.AnimationTransitionGraph_1
  nodes [Result(bCanEnterTransition=true)]
 }
 AnimationTransitionGraph_1 {
  path=FullBodyOverride.AnimGraphNode_StateMachine_1.AttackStateMachine.AnimStateNode_2.Slash.AnimGraphNode_StateMachine_0.SlashSelection.AnimStateTransitionNode_1.AnimationTransitionGraph_1
  nodes [Result(bCanEnterTransition="Equal (Integer)"(A=Get CurrentAttackIntroAnim() -> (CurrentAttackIntroAnim=0)))]
 }
 Transition {
  path=FullBodyOverride.AnimGraphNode_StateMachine_1.AttackStateMachine.AnimStateNode_2.Slash.AnimGraphNode_StateMachine_0.SlashSelection.AnimStateTransitionNode_2.Transition
  nodes [Result(bCanEnterTransition="Equal (Integer)"(A=Get CurrentAttackIntroAnim() -> (CurrentAttackIntroAnim=0), B=2))]
 }
 Transition {
  path=FullBodyOverride.AnimGraphNode_StateMachine_1.AttackStateMachine.AnimStateTransitionNode_0.Transition
  nodes [Result(bCanEnterTransition=Get IsHovering() -> (IsHovering:bool))]
 }
 Transition {
  path=FullBodyOverride.AnimGraphNode_StateMachine_1.AttackStateMachine.AnimStateTransitionNode_1.Transition
  nodes [Result(bCanEnterTransition=Get IsAttackStart() -> (IsAttackStart:bool))]
 }
 Transition {
  path=FullBodyOverride.AnimGraphNode_StateMachine_1.AttackStateMachine.AnimStateTransitionNode_2.Transition
  nodes [Result(bCanEnterTransition=true)]
 }
 Transition {
  path=FullBodyOverride.AnimGraphNode_StateMachine_1.AttackStateMachine.AnimStateTransitionNode_3.Transition
  nodes [Result(bCanEnterTransition=true)]
 }
 Transition {
  path=FullBodyOverride.AnimGraphNode_StateMachine_1.AttackStateMachine.AnimStateTransitionNode_4.Transition
  nodes [
   Result(bCanEnterTransition=true)
   AND Boolean(A=Get IsSlashingPastTarget() -> (IsSlashingPastTarget:bool), B="Equal (Enum)"(A=Get State() -> (State=NewEnumerator0), B=NewEnumerator9))
   AND Boolean(A=NOT Boolean(A=Get IsHovering()), B=NOT Boolean(A=Get IsAttackStart()))
  ]
 }
 Transition {
  path=FullBodyOverride.AnimGraphNode_StateMachine_1.AttackStateMachine.AnimStateTransitionNode_5.Transition
  nodes [Result(bCanEnterTransition=true)]
 }
 PreLegIK {nodes [
   Input Pose()
   Output Pose() {layerGroup=PlayerOverride}
  ]}
 SplitVector {nodes [
   n1 = SplitVector!() -> (then, Vec:Vector=~) {is=title K2Node_FunctionEntry}
   n2 = Vector Length(A=@n1)
   > Return Node!(execute, Magnitude:double=@n2, Direction:Vector=vector / vector(A:Vector=@n1, B=@n2) -> (ReturnValue:Vector))
  ]}
 UpperBody {nodes [
   Input Pose() {name=InPoseUpperBody}
   Output Pose() {layerGroup=PlayerOverride}
  ]}
 ZipAttackSelection {nodes [
   n1 = ZipAttackSelection!() -> (then, Context:"Anim Update Context Structure (by ref)", Node:"Anim Node Reference Structure (by ref)") {is=title K2Node_FunctionEntry}
   > SetSequence!(
    execute
    SequencePlayer:"Sequence Player Reference Structure (by ref)"=Convert to Sequence Player!(Node:"Anim Node Reference Structure (by ref)"=@n1.Node) -> (SequencePlayer:Sequence Player Reference Structure, Result:bool)
    Sequence:Anim Sequence Base Object Reference="Get (a copy)"(
     Array:Array of Anim Sequence Base Object References=Get AttackIntroSelection() -> (AttackIntroSelection:Array of Anim Sequence Base Object References)
     Dimension 1=Get CurrentAttackIntroAnim() -> (CurrentAttackIntroAnim=0)
    ) -> (Output:Anim Sequence Base Object Reference)
   ) -> (then, ReturnValue:Sequence Player Reference Structure)
   > n2 = Increment Int(Value=Get CurrentAttackIntroAnim() -> (CurrentAttackIntroAnim=0))
   n3 = Get AttackIntroSelection()
   > Branch(Condition="integer >= integer"(A=@n2, B=Length(TargetArray:Array of Anim Sequence Base Object References=@n3)))
   > Set CurrentAttackIntroAnim(CurrentAttackIntroAnim:int=0) -> (Output_Get:int=0)
   > Return Node!(execute)
  ]}
 ZipCastSelection {nodes [
   n1 = ZipCastSelection!() -> (then, Context:"Anim Update Context Structure (by ref)", Node:"Anim Node Reference Structure (by ref)") {is=title K2Node_FunctionEntry}
   > SetSequence(
    SequencePlayer=Convert to Sequence Player(Node=@n1.Node)
    Sequence="Get (a copy)"(
     Array=Get ZipCast() -> (ZipCast:Array of Anim Sequence Base Object References)
     Dimension 1=Get CurrentZipCastAnim() -> (CurrentZipCastAnim=0)
    )
   )
   > n2 = Increment Int(Value=Get CurrentZipCastAnim() -> (CurrentZipCastAnim=0))
   n3 = Get ZipCast()
   > Branch(Condition="integer >= integer"(A=@n2, B=Length(TargetArray=@n3)))
   > Set CurrentZipCastAnim(CurrentZipCastAnim:int=0) -> (Output_Get:int=0)
   > Return Node()
  ]}
 ZipReleaseSelection {nodes [
   n1 = ZipReleaseSelection!() -> (then, Context:"Anim Update Context Structure (by ref)", Node:"Anim Node Reference Structure (by ref)") {is=title K2Node_FunctionEntry}
   > SetSequence(
    SequencePlayer=Convert to Sequence Player(Node=@n1.Node)
    Sequence="Get (a copy)"(
     Array=Get CurrentReleaseZipAnim() -> (CurrentReleaseZipAnim:Array of Anim Sequence Base Object References)
     Dimension 1=Get CurrentZipReleaseAnim() -> (CurrentZipReleaseAnim:int=0)
    )
   )
   > n2 = Increment Int(Value=Get CurrentZipReleaseAnim() -> (CurrentZipReleaseAnim=0))
   n3 = Get CurrentReleaseZipAnim()
   > Branch(Condition="integer >= integer"(A=@n2, B=Length(TargetArray=@n3)))
   > Set CurrentZipReleaseAnim(CurrentZipReleaseAnim:int=0) -> (Output_Get:int=0)
   > Return Node()
  ]}
]
variables [
 AllowNonRelease {type=boolean}
 AttackIntroSelection {
  type="array<AnimSequenceBase>"
  default [
   /DryBox/Items/NyxGlass/Animation/Attack/NyxGlass_Attack_A_Intro
   /DryBox/Items/NyxGlass/Animation/Attack/NyxGlass_Attack_B_Intro
   /DryBox/Items/NyxGlass/Animation/Attack/NyxGlass_Attack_C_Intro
  ]
 }
 Bound {type=boolean}
 CurrentAttackIntroAnim {type=integer}
 CurrentReleaseAnims_-90 {type=integer}
 CurrentReleaseAnims_0 {type=integer}
 CurrentReleaseAnims_45 {type=integer}
 CurrentReleaseAnims_90 {type=integer}
 CurrentReleaseZipAnim {
  type="array<AnimSequenceBase>"
  default [
   /DryBox/Items/NyxGlass/Animation/Release/NyxGlass_Release_A
   /DryBox/Items/NyxGlass/Animation/Release/NyxGlass_Release_B
   /DryBox/Items/NyxGlass/Animation/Release/NyxGlass_Release_C
   /DryBox/Items/NyxGlass/Animation/Release/NyxGlass_Release_D
   /DryBox/Items/NyxGlass/Animation/Release/NyxGlass_Release_E
   /DryBox/Items/NyxGlass/Animation/Release/NyxGlass_Release_F
   /DryBox/Items/NyxGlass/Animation/Release/NyxGlass_Release_G
   /DryBox/Items/NyxGlass/Animation/Release/NyxGlass_Release_H
   /DryBox/Items/NyxGlass/Animation/Release/NyxGlass_Release_I
  ]
 }
 CurrentZipCastAnim {type=integer}
 CurrentZipReleaseAnim {type=integer}
 IsAttackFall {type=boolean}
 IsAttackRelease {type=boolean}
 IsAttackStart {type=boolean}
 IsHovering {type=boolean}
 IsHoverRelease {type=boolean}
 IsLeapingBack {type=boolean}
 IsMovingToTarget {type=boolean}
 IsRopeAttached {type=boolean}
 IsSlashingPastTarget {type=boolean}
 IsWallHit {type=boolean}
 IsZipAttack {type=boolean}
 State {type=State_NyxGlass, default=Off}
 ThunderSpearsEquipped {type=boolean}
 Weapon {type=B_Weapon_GrappleParent_C, default=null}
 ZipCast {
  type="array<AnimSequenceBase>"
  default [
   /DryBox/Items/NyxGlass/Animation/Cast/NyxGlass_Cast_To_Zip
   /DryBox/Items/NyxGlass/Animation/Swing/NyxGlass_Zip_B_Cast_To_Zip
   /DryBox/Items/NyxGlass/Animation/Cast/NyxGlass_Zip_C_CasttoZip
  ]
 }
 ZipLean {type=number}
 ZipLeanAlpha {type=number}
]
