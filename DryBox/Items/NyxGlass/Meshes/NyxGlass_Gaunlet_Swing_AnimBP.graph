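# Animation Blueprint for the NyxGlass gauntlet mesh, animating it through zip swinging, falling and zip attacks.
# Each update copies state from the pawn's linked NyxGlass_AnimBP layer and reads local Hover and LeanCurve curves.
# IsZipAttack selects between a zip machine (cast/idle, release, fall) and an attack machine (hover, intro, slash), using inertialized blends.
# Integer indices pick animation variants; additive blendspaces add lean during zips.
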AnimBlueprint /DryBox/Items/NyxGlass/Meshes/NyxGlass_Gaunlet_Swing_AnimBP
target_skeleton=&/NyxGlass_Gauntlet_Skeleton
graphs [
 AnimGraph {nodes [
   AnimGraphNode_StateMachine_0 = ZipStateMachine() {is=title AnimGraphNode_StateMachine, bAllowConduitEntryStates=true}
   AnimGraphNode_StateMachine_1 = ZipAttackStateMachine() {is=title AnimGraphNode_StateMachine, bAllowConduitEntryStates=true}
   n1 = Get IsZipAttack() -> (IsZipAttack:bool)
   Output Pose(Result=Inertialization!(Source:Pose Link Structure=Blend Poses by bool(BlendPose_0=@AnimGraphNode_StateMachine_1, BlendPose_1=@AnimGraphNode_StateMachine_0, BlendTime_0=0.2, BlendTime_1=0.2, bActiveValue=@n1) {blendType=Linear, transitionType=Inertialization}) -> (Pose:Pose Link Structure) {is=title AnimGraphNode_Inertialization}) {layerGroup=None}
  ]}
 ZipStateMachine {
  entry=Conduit
  path=AnimGraph.AnimGraphNode_StateMachine_0.ZipStateMachine
  ids=id
  nodes [
   AnimStateConduitNode_0 = Conduit()
   AnimStateNode_0 = IsZipping() {is=title AnimStateNode}
   AnimStateNode_1 = IsZippingReleaseState() {is=title AnimStateNode}
   AnimStateNode_2 = Fall() {is=title AnimStateNode}
   AnimStateTransitionNode_0 = AnimStateTransitionNode() {crossfadeDuration=0.15, logicType=TLT_Inertialization}
   AnimStateTransitionNode_1 = AnimStateTransitionNode() {crossfadeDuration=0.5, logicType=TLT_Inertialization, priorityOrder=100}
   AnimStateTransitionNode_2 = AnimStateTransitionNode()
   AnimStateTransitionNode_3 = AnimStateTransitionNode() {crossfadeDuration=0.4, logicType=TLT_Inertialization, priorityOrder=2}
   AnimStateTransitionNode_4 = AnimStateTransitionNode() {bAutomaticRuleBasedOnSequencePlayerInState=true, crossfadeDuration=0.4}
   AnimStateTransitionNode_5 = AnimStateTransitionNode()
   AnimStateTransitionNode_6 = AnimStateTransitionNode()
  ]
  links [
   {
    dst=Fall
    rule=AnimGraph.AnimGraphNode_StateMachine_0.ZipStateMachine.AnimStateTransitionNode_3.Transition
    src=Conduit
    via=AnimStateTransitionNode_3
   }
   {
    dst=IsZipping
    rule=AnimGraph.AnimGraphNode_StateMachine_0.ZipStateMachine.AnimStateTransitionNode_0.Transition
    src=Conduit
    via=AnimStateTransitionNode_0
   }
   {
    dst=IsZippingReleaseState
    rule=AnimGraph.AnimGraphNode_StateMachine_0.ZipStateMachine.AnimStateTransitionNode_1.Transition
    src=Conduit
    via=AnimStateTransitionNode_1
   }
   {
    dst=Conduit
    rule=AnimGraph.AnimGraphNode_StateMachine_0.ZipStateMachine.AnimStateTransitionNode_6.Transition
    src=Fall
    via=AnimStateTransitionNode_6
   }
   {
    dst=Conduit
    rule=AnimGraph.AnimGraphNode_StateMachine_0.ZipStateMachine.AnimStateTransitionNode_5.Transition
    src=IsZipping
    via=AnimStateTransitionNode_5
   }
   {
    dst=Conduit
    rule=AnimGraph.AnimGraphNode_StateMachine_0.ZipStateMachine.AnimStateTransitionNode_2.Transition
    src=IsZippingReleaseState
    via=AnimStateTransitionNode_2
   }
   {
    dst=Fall
    rule=AnimGraph.AnimGraphNode_StateMachine_0.ZipStateMachine.AnimStateTransitionNode_4.Transition
    src=IsZippingReleaseState
    via=AnimStateTransitionNode_4
   }
  ]
 }
 Conduit {path=AnimGraph.AnimGraphNode_StateMachine_0.ZipStateMachine.AnimStateConduitNode_0.Conduit, nodes [Result(bCanEnterTransition=true)]}
 IsZipping {
  path=AnimGraph.AnimGraphNode_StateMachine_0.ZipStateMachine.AnimStateNode_0.IsZipping
  nodes [
   AnimGraphNode_StateMachine_1 = ZipCastState_B() {is=title AnimGraphNode_StateMachine}
   AnimGraphNode_StateMachine_2 = ZipCastStaste_C() {is=title AnimGraphNode_StateMachine}
   AnimGraphNode_StateMachine_3 = ZipCastState_A() {is=title AnimGraphNode_StateMachine}
   n1 = Blend Poses by int!(
    BlendPose_0:Pose Link Structure=@AnimGraphNode_StateMachine_2
    BlendPose_1:Pose Link Structure=@AnimGraphNode_StateMachine_3
    BlendPose_2:Pose Link Structure=@AnimGraphNode_StateMachine_1
    BlendTime_0:float=0.1
    BlendTime_1:float=0.1
    BlendTime_2:float=0.1
    ActiveChildIndex:int=Get CurrentZipCastAnim() -> (CurrentZipCastAnim:int=0)
   ) -> (Pose:Pose Link Structure) {
    is=title AnimGraphNode_BlendListByInt
    blendType=Linear
    childUpateMode=ResetChildOnActivate
    transitionType=Inertialization
    blendPose [{linkId=-1, sourceLinkId=-1}, {linkId=-1, sourceLinkId=-1}, {linkId=-1, sourceLinkId=-1}]
    blendTime [0.1,0.1,0.1]
   }
   Output Animation Pose(Result=@n1)
  ]
 }
 ZipCastState_B {
  entry=ZipCast_B
  path=AnimGraph.AnimGraphNode_StateMachine_0.ZipStateMachine.AnimStateNode_0.IsZipping.AnimGraphNode_StateMachine_1.ZipCastState_B
  ids=id
  nodes [
   AnimStateNode_1 = ZipCast_B() {is=title AnimStateNode}
   AnimStateNode_3 = ZipCastIdle_B() {is=title AnimStateNode}
   AnimStateTransitionNode_1 = AnimStateTransitionNode() {crossfadeDuration=0.4, logicType=TLT_Inertialization}
  ]
  links [{
    dst=ZipCastIdle_B
    rule=AnimGraph.AnimGraphNode_StateMachine_0.ZipStateMachine.AnimStateNode_0.IsZipping.AnimGraphNode_StateMachine_1.ZipCastState_B.AnimStateTransitionNode_1.AnimationTransitionGraph_0
    src=ZipCast_B
    via=AnimStateTransitionNode_1
   }]
 }
 ZipCast_B {
  path=AnimGraph.AnimGraphNode_StateMachine_0.ZipStateMachine.AnimStateNode_0.IsZipping.AnimGraphNode_StateMachine_1.ZipCastState_B.AnimStateNode_1.ZipCast_B
  nodes [
   n1 = Get Zip Lean() -> (Zip Lean:double)
   Output Animation Pose(Result=Apply Additive(
     Base="Sequence Player (additive) 'NyxGlass_Zip_B_CasttoZip_WEP'"() {bLoopAnimation=false, sequence=/DryBox/Items/NyxGlass/Animation/Cast/NyxGlass_Zip_B_CasttoZip_WEP}
     Additive=Blendspace Player 'NyxGlass_Zip_Idle_Lean_WEP_B_BS'(X=@n1, Y=0) {blendSpace=/DryBox/Items/NyxGlass/Animation/Cast/NyxGlass_Zip_Idle_Lean_WEP_B_BS}
     Alpha=Get ZipIdleLeanCurve() -> (ZipIdleLeanCurve:double)
    ) {alphaBoolBlend {blendOption=Linear}})
  ]
 }
 ZipCastIdle_B {
  path=AnimGraph.AnimGraphNode_StateMachine_0.ZipStateMachine.AnimStateNode_0.IsZipping.AnimGraphNode_StateMachine_1.ZipCastState_B.AnimStateNode_3.ZipCastIdle_B
  nodes [
   n1 = Get Zip Lean()
   Output Animation Pose(Result=Apply Additive(
     Base="Sequence Player (additive) 'NyxGlass_Zip_B_Idle_WEP'"() {sequence=/DryBox/Items/NyxGlass/Animation/Cast/NyxGlass_Zip_B_Idle_WEP}
     Additive=Blendspace Player 'NyxGlass_Zip_Idle_Lean_WEP_B_BS'(X=@n1, Y=0) {blendSpace=&NyxGlass_Zip_Idle_Lean_WEP_B_BS}
    ) {alphaBoolBlend {blendOption=Linear}})
  ]
 }
 AnimationTransitionGraph_0 {
  path=AnimGraph.AnimGraphNode_StateMachine_0.ZipStateMachine.AnimStateNode_0.IsZipping.AnimGraphNode_StateMachine_1.ZipCastState_B.AnimStateTransitionNode_1.AnimationTransitionGraph_0
  nodes [Result(bCanEnterTransition="float >= float"(A=GetCurveValue(CurveName=BlendOut), B=0.5))]
 }
 ZipCastStaste_C {
  entry=ZipCast_C
  path=AnimGraph.AnimGraphNode_StateMachine_0.ZipStateMachine.AnimStateNode_0.IsZipping.AnimGraphNode_StateMachine_2.ZipCastStaste_C
  ids=id
  nodes [
   AnimStateNode_4 = ZipCast_C() {is=title AnimStateNode}
   AnimStateNode_5 = ZipCastIdle_C() {is=title AnimStateNode}
   AnimStateTransitionNode_0 = AnimStateTransitionNode() {crossfadeDuration=0.4, logicType=TLT_Inertialization}
  ]
  links [{
    dst=ZipCastIdle_C
    rule=AnimGraph.AnimGraphNode_StateMachine_0.ZipStateMachine.AnimStateNode_0.IsZipping.AnimGraphNode_StateMachine_2.ZipCastStaste_C.AnimStateTransitionNode_0.Transition
    src=ZipCast_C
    via=AnimStateTransitionNode_0
   }]
 }
 ZipCast_C {
  path=AnimGraph.AnimGraphNode_StateMachine_0.ZipStateMachine.AnimStateNode_0.IsZipping.AnimGraphNode_StateMachine_2.ZipCastStaste_C.AnimStateNode_4.ZipCast_C
  nodes [
   n1 = Get Zip Lean()
   Output Animation Pose(Result=Apply Additive(
     Base="Sequence Player (additive) 'NyxGlass_Zip_C_Cast_To_Zip_WEP'"() {bLoopAnimation=false, sequence=/DryBox/Items/NyxGlass/Animation/Cast/NyxGlass_Zip_C_Cast_To_Zip_WEP}
     Additive=Blendspace Player 'NyxGlass_Zip_Idle_Lean_WEP_C_BS'(X=@n1, Y=0) {blendSpace=/DryBox/Items/NyxGlass/Animation/Cast/NyxGlass_Zip_Idle_Lean_WEP_C_BS}
     Alpha=Get ZipIdleLeanCurve()
    ) {alphaBoolBlend {blendOption=Linear}})
  ]
 }
 ZipCastIdle_C {
  path=AnimGraph.AnimGraphNode_StateMachine_0.ZipStateMachine.AnimStateNode_0.IsZipping.AnimGraphNode_StateMachine_2.ZipCastStaste_C.AnimStateNode_5.ZipCastIdle_C
  nodes [
   n1 = Get Zip Lean()
   Output Animation Pose(Result=Apply Additive(
     Base="Sequence Player (additive) 'NyxGlass_Zip_C_Idle_WEP'"() {sequence=/DryBox/Items/NyxGlass/Animation/Cast/NyxGlass_Zip_C_Idle_WEP}
     Additive=Blendspace Player 'NyxGlass_Zip_Idle_Lean_WEP_C_BS'(X=@n1, Y=0) {blendSpace=&NyxGlass_Zip_Idle_Lean_WEP_C_BS}
    ) {alphaBoolBlend {blendOption=Linear}})
  ]
 }
 Transition {
  path=AnimGraph.AnimGraphNode_StateMachine_0.ZipStateMachine.AnimStateNode_0.IsZipping.AnimGraphNode_StateMachine_2.ZipCastStaste_C.AnimStateTransitionNode_0.Transition
  nodes [Result(bCanEnterTransition="float >= float"(A=GetCurveValue(CurveName=BlendOut), B=0.5))]
 }
 ZipCastState_A {
  entry=ZipCast_A
  path=AnimGraph.AnimGraphNode_StateMachine_0.ZipStateMachine.AnimStateNode_0.IsZipping.AnimGraphNode_StateMachine_3.ZipCastState_A
  ids=id
  nodes [
   AnimStateNode_0 = ZipCast_A() {is=title AnimStateNode}
   AnimStateNode_2 = ZipCastIdle_A() {is=title AnimStateNode}
   AnimStateTransitionNode_0 = AnimStateTransitionNode() {crossfadeDuration=0.4, logicType=TLT_Inertialization}
  ]
  links [{
    dst=ZipCastIdle_A
    rule=AnimGraph.AnimGraphNode_StateMachine_0.ZipStateMachine.AnimStateNode_0.IsZipping.AnimGraphNode_StateMachine_3.ZipCastState_A.AnimStateTransitionNode_0.AnimationTransitionGraph_3
    src=ZipCast_A
    via=AnimStateTransitionNode_0
   }]
 }
 ZipCast_A {
  path=AnimGraph.AnimGraphNode_StateMachine_0.ZipStateMachine.AnimStateNode_0.IsZipping.AnimGraphNode_StateMachine_3.ZipCastState_A.AnimStateNode_0.ZipCast_A
  nodes [
   n1 = Get Zip Lean()
   Output Animation Pose(Result=Apply Additive(
     Base="Sequence Player (additive) 'NyxGlass_Cast_To_Zip_WEP'"() {bLoopAnimation=false, sequence=/DryBox/Items/NyxGlass/Animation/Cast/NyxGlass_Cast_To_Zip_WEP}
     Additive=Blendspace Player 'NyxGlass_Zip_Idle_Lean_WEP_A_BS'(X=@n1, Y=0) {bLoop=false, blendSpace=/DryBox/Items/NyxGlass/Animation/Cast/NyxGlass_Zip_Idle_Lean_WEP_A_BS}
     Alpha=Get ZipIdleLeanCurve()
    ) {alphaBoolBlend {blendOption=Linear}})
  ]
 }
 ZipCastIdle_A {
  path=AnimGraph.AnimGraphNode_StateMachine_0.ZipStateMachine.AnimStateNode_0.IsZipping.AnimGraphNode_StateMachine_3.ZipCastState_A.AnimStateNode_2.ZipCastIdle_A
  nodes [
   n1 = Get Zip Lean()
   Output Animation Pose(Result=Apply Additive(
     Base="Sequence Player (additive) 'NyxGlass_Zip_Idle_A_WEP'"() {sequence=/DryBox/Items/NyxGlass/Animation/Cast/NyxGlass_Zip_Idle_A_WEP}
     Additive=Blendspace Player 'NyxGlass_Zip_Idle_Lean_WEP_A_BS'(X=@n1, Y=0) {blendSpace=&NyxGlass_Zip_Idle_Lean_WEP_A_BS}
    ) {alphaBoolBlend {blendOption=Linear}})
  ]
 }
 AnimationTransitionGraph_3 {
  path=AnimGraph.AnimGraphNode_StateMachine_0.ZipStateMachine.AnimStateNode_0.IsZipping.AnimGraphNode_StateMachine_3.ZipCastState_A.AnimStateTransitionNode_0.AnimationTransitionGraph_3
  nodes [Result(bCanEnterTransition="float >= float"(A=GetCurveValue(CurveName=BlendOut), B=0.5))]
 }
 IsZippingReleaseState {
  path=AnimGraph.AnimGraphNode_StateMachine_0.ZipStateMachine.AnimStateNode_1.IsZippingReleaseState
  nodes [
   AnimGraphNode_StateMachine_0 = ZipRelease() {is=title AnimGraphNode_StateMachine, bAllowConduitEntryStates=true}
   Output Animation Pose(Result=@AnimGraphNode_StateMachine_0)
  ]
 }
 ZipRelease {
  entry=Conduit
  path=AnimGraph.AnimGraphNode_StateMachine_0.ZipStateMachine.AnimStateNode_1.IsZippingReleaseState.AnimGraphNode_StateMachine_0.ZipRelease
  ids=id
  nodes [
   AnimStateConduitNode_0 = Conduit()
   AnimStateNode_1 = AttackRelease() {is=title AnimStateNode}
   AnimStateNode_2 = ZipRelease() {is=title AnimStateNode}
   AnimStateTransitionNode_0 = AnimStateTransitionNode() {crossfadeDuration=0.5, logicType=TLT_Inertialization}
   AnimStateTransitionNode_1 = AnimStateTransitionNode() {crossfadeDuration=0.5, logicType=TLT_Inertialization, priorityOrder=100}
  ]
  links [
   {
    dst=AttackRelease
    rule=AnimGraph.AnimGraphNode_StateMachine_0.ZipStateMachine.AnimStateNode_1.IsZippingReleaseState.AnimGraphNode_StateMachine_0.ZipRelease.AnimStateTransitionNode_0.Transition
    src=Conduit
    via=AnimStateTransitionNode_0
   }
   {
    dst=ZipRelease
    rule=AnimGraph.AnimGraphNode_StateMachine_0.ZipStateMachine.AnimStateNode_1.IsZippingReleaseState.AnimGraphNode_StateMachine_0.ZipRelease.AnimStateTransitionNode_1.Transition
    src=Conduit
    via=AnimStateTransitionNode_1
   }
  ]
 }
 Conduit {
  path=AnimGraph.AnimGraphNode_StateMachine_0.ZipStateMachine.AnimStateNode_1.IsZippingReleaseState.AnimGraphNode_StateMachine_0.ZipRelease.AnimStateConduitNode_0.Conduit
  nodes [Result(bCanEnterTransition=true)]
 }
 AttackRelease {
  path=AnimGraph.AnimGraphNode_StateMachine_0.ZipStateMachine.AnimStateNode_1.IsZippingReleaseState.AnimGraphNode_StateMachine_0.ZipRelease.AnimStateNode_1.AttackRelease
  nodes [
   n1 = Blend Poses by int(
    BlendPose_0="Sequence Player (additive) 'NyxGlass_Attack_C_Release_WEP'"() {bLoopAnimation=false, sequence=/DryBox/Items/NyxGlass/Animation/Attack/NyxGlass_Attack_C_Release_WEP}
    BlendPose_1="Sequence Player (additive) 'NyxGlass_Attack_Release_WEP'"() {bLoopAnimation=false, sequence=/DryBox/Items/NyxGlass/Animation/Release/NyxGlass_Attack_Release_WEP}
    BlendPose_2="Sequence Player (additive) 'NyxGlass_Attack_B_Release_WEP'"() {bLoopAnimation=false, sequence=/DryBox/Items/NyxGlass/Animation/Release/NyxGlass_Attack_B_Release_WEP}
    BlendTime_0=0.1
    BlendTime_1=0.1
    BlendTime_2=0.1
    ActiveChildIndex=Get CurrentAttackIntroAnim() -> (CurrentAttackIntroAnim:int=0)
   ) {
    blendType=Linear
    childUpateMode=ResetChildOnActivate
    transitionType=Inertialization
    blendPose [{linkId=-1, sourceLinkId=-1}, {linkId=-1, sourceLinkId=-1}, {linkId=-1, sourceLinkId=-1}]
    blendTime [0.1,0.1,0.1]
   }
   Output Animation Pose(Result=@n1)
  ]
 }
 ZipRelease {
  path=AnimGraph.AnimGraphNode_StateMachine_0.ZipStateMachine.AnimStateNode_1.IsZippingReleaseState.AnimGraphNode_StateMachine_0.ZipRelease.AnimStateNode_2.ZipRelease
  nodes [
   n1 = Blend Poses by int!(
    BlendPose_0="Sequence Player (additive) 'NyxGlass_Release_I_WEP'"() {bLoopAnimation=false, sequence=/DryBox/Items/NyxGlass/Animation/Release/NyxGlass_Release_I_WEP}
    BlendPose_1="Sequence Player (additive) 'NyxGlass_Release_A_WEP'"() {bLoopAnimation=false, sequence=/DryBox/Items/NyxGlass/Animation/Release/NyxGlass_Release_A_WEP}
    BlendPose_2="Sequence Player (additive) 'NyxGlass_Release_B_WEP'"() {bLoopAnimation=false, sequence=/DryBox/Items/NyxGlass/Animation/Release/NyxGlass_Release_B_WEP}
    BlendPose_3:Pose Link Structure="Sequence Player (additive) 'NyxGlass_Release_C_WEP'"() {bLoopAnimation=false, sequence=/DryBox/Items/NyxGlass/Animation/Release/NyxGlass_Release_C_WEP}
    BlendPose_4:Pose Link Structure="Sequence Player (additive) 'NyxGlass_Release_D_WEP'"() {bLoopAnimation=false, sequence=/DryBox/Items/NyxGlass/Animation/Release/NyxGlass_Release_D_WEP}
    BlendPose_5:Pose Link Structure="Sequence Player (additive) 'NyxGlass_Release_E_WEP'"() {bLoopAnimation=false, sequence=/DryBox/Items/NyxGlass/Animation/Release/NyxGlass_Release_E_WEP}
    BlendPose_6:Pose Link Structure="Sequence Player (additive) 'NyxGlass_Release_F_WEP'"() {bLoopAnimation=false, sequence=/DryBox/Items/NyxGlass/Animation/Release/NyxGlass_Release_F_WEP}
    BlendPose_7:Pose Link Structure="Sequence Player (additive) 'NyxGlass_Release_G_WEP'"() {bLoopAnimation=false, sequence=/DryBox/Items/NyxGlass/Animation/Release/NyxGlass_Release_G_WEP}
    BlendPose_8:Pose Link Structure="Sequence Player (additive) 'NyxGlass_Release_H_WEP'"() {bLoopAnimation=false, sequence=/DryBox/Items/NyxGlass/Animation/Release/NyxGlass_Release_H_WEP}
    BlendTime_0=0.1
    BlendTime_1=0.1
    BlendTime_2=0.1
    BlendTime_3:float=0.1
    BlendTime_4:float=0.1
    BlendTime_5:float=0.1
    BlendTime_6:float=0.1
    BlendTime_7:float=0.1
    BlendTime_8:float=0.1
    ActiveChildIndex=Get CurrentZipReleaseAnim() -> (CurrentZipReleaseAnim:int=0)
   ) -> (Pose) {
    blendType=Linear
    childUpateMode=ResetChildOnActivate
    transitionType=Inertialization
    blendPose [
     {linkId=-1, sourceLinkId=-1}
     {linkId=-1, sourceLinkId=-1}
     {linkId=-1, sourceLinkId=-1}
     {linkId=-1, sourceLinkId=-1}
     {linkId=-1, sourceLinkId=-1}
     {linkId=-1, sourceLinkId=-1}
     {linkId=-1, sourceLinkId=-1}
     {linkId=-1, sourceLinkId=-1}
     {linkId=-1, sourceLinkId=-1}
    ]
    blendTime [0.1,0.1,0.1,0.1,0.1,0.1,0.1,0.1,0.1]
   }
   Output Animation Pose(Result=@n1)
  ]
 }
 Transition {
  path=AnimGraph.AnimGraphNode_StateMachine_0.ZipStateMachine.AnimStateNode_1.IsZippingReleaseState.AnimGraphNode_StateMachine_0.ZipRelease.AnimStateTransitionNode_0.Transition
  nodes [Result(bCanEnterTransition=Get IsAttackRelease() -> (IsAttackRelease:bool))]
 }
 Transition {
  path=AnimGraph.AnimGraphNode_StateMachine_0.ZipStateMachine.AnimStateNode_1.IsZippingReleaseState.AnimGraphNode_StateMachine_0.ZipRelease.AnimStateTransitionNode_1.Transition
  nodes [Result(bCanEnterTransition=true)]
 }
 Fall {
  path=AnimGraph.AnimGraphNode_StateMachine_0.ZipStateMachine.AnimStateNode_2.Fall
  nodes [
   AnimGraphNode_StateMachine_0 = FallState() {is=title AnimGraphNode_StateMachine, bAllowConduitEntryStates=true}
   Output Animation Pose(Result=@AnimGraphNode_StateMachine_0)
  ]
 }
 FallState {
  entry=Conduit
  path=AnimGraph.AnimGraphNode_StateMachine_0.ZipStateMachine.AnimStateNode_2.Fall.AnimGraphNode_StateMachine_0.FallState
  ids=id
  nodes [
   AnimStateConduitNode_0 = Conduit()
   AnimStateNode_0 = FallStart() {is=title AnimStateNode}
   AnimStateNode_1 = FallLoop() {is=title AnimStateNode}
   AnimStateTransitionNode_0 = AnimStateTransitionNode() {bAutomaticRuleBasedOnSequencePlayerInState=true, crossfadeDuration=0.5}
   AnimStateTransitionNode_1 = AnimStateTransitionNode() {crossfadeDuration=0.4, priorityOrder=100}
   AnimStateTransitionNode_2 = AnimStateTransitionNode() {crossfadeDuration=0.5}
  ]
  links [
   {
    dst=FallLoop
    rule=AnimGraph.AnimGraphNode_StateMachine_0.ZipStateMachine.AnimStateNode_2.Fall.AnimGraphNode_StateMachine_0.FallState.AnimStateTransitionNode_2.Transition
    src=Conduit
    via=AnimStateTransitionNode_2
   }
   {
    dst=FallStart
    rule=AnimGraph.AnimGraphNode_StateMachine_0.ZipStateMachine.AnimStateNode_2.Fall.AnimGraphNode_StateMachine_0.FallState.AnimStateTransitionNode_1.Transition
    src=Conduit
    via=AnimStateTransitionNode_1
   }
   {
    dst=FallLoop
    rule=AnimGraph.AnimGraphNode_StateMachine_0.ZipStateMachine.AnimStateNode_2.Fall.AnimGraphNode_StateMachine_0.FallState.AnimStateTransitionNode_0.Transition
    src=FallStart
    via=AnimStateTransitionNode_0
   }
  ]
 }
 Conduit {
  path=AnimGraph.AnimGraphNode_StateMachine_0.ZipStateMachine.AnimStateNode_2.Fall.AnimGraphNode_StateMachine_0.FallState.AnimStateConduitNode_0.Conduit
  nodes [Result(bCanEnterTransition=true)]
 }
 FallStart {
  path=AnimGraph.AnimGraphNode_StateMachine_0.ZipStateMachine.AnimStateNode_2.Fall.AnimGraphNode_StateMachine_0.FallState.AnimStateNode_0.FallStart
  nodes [Output Animation Pose(Result="Sequence Player (additive) 'NyxGlass_Fall_Start_WEP'"() {bLoopAnimation=false, sequence=/DryBox/Items/NyxGlass/Animation/NyxGlass_Fall_Start_WEP})]
 }
 FallLoop {
  path=AnimGraph.AnimGraphNode_StateMachine_0.ZipStateMachine.AnimStateNode_2.Fall.AnimGraphNode_StateMachine_0.FallState.AnimStateNode_1.FallLoop
  nodes [Output Animation Pose(Result="Sequence Player (additive) 'NyxGlass_Fall_Loop_WEP'"() {bLoopAnimation=false, sequence=/DryBox/Items/NyxGlass/Animation/NyxGlass_Fall_Loop_WEP})]
 }
 Transition {
  path=AnimGraph.AnimGraphNode_StateMachine_0.ZipStateMachine.AnimStateNode_2.Fall.AnimGraphNode_StateMachine_0.FallState.AnimStateTransitionNode_0.Transition
  nodes [Result(bCanEnterTransition=true)]
 }
 Transition {
  path=AnimGraph.AnimGraphNode_StateMachine_0.ZipStateMachine.AnimStateNode_2.Fall.AnimGraphNode_StateMachine_0.FallState.AnimStateTransitionNode_1.Transition
  nodes [Result(bCanEnterTransition=true)]
 }
 Transition {
  path=AnimGraph.AnimGraphNode_StateMachine_0.ZipStateMachine.AnimStateNode_2.Fall.AnimGraphNode_StateMachine_0.FallState.AnimStateTransitionNode_2.Transition
  nodes [Result(bCanEnterTransition=Get IsAttackFall() -> (IsAttackFall:bool))]
 }
 Transition {
  path=AnimGraph.AnimGraphNode_StateMachine_0.ZipStateMachine.AnimStateTransitionNode_0.Transition
  nodes [Result(bCanEnterTransition=Get Is Rope Attached() -> (Is Rope Attached:bool))]
 }
 Transition {path=AnimGraph.AnimGraphNode_StateMachine_0.ZipStateMachine.AnimStateTransitionNode_1.Transition, nodes [Result(bCanEnterTransition=true)]}
 Transition {path=AnimGraph.AnimGraphNode_StateMachine_0.ZipStateMachine.AnimStateTransitionNode_2.Transition, nodes [Result(bCanEnterTransition=true)]}
 Transition {
  path=AnimGraph.AnimGraphNode_StateMachine_0.ZipStateMachine.AnimStateTransitionNode_3.Transition
  nodes [Result(bCanEnterTransition=OR Boolean(A=Get IsHover() -> (IsHover:bool), B=Get Allow Non Release() -> (Allow Non Release:bool)))]
 }
 Transition {path=AnimGraph.AnimGraphNode_StateMachine_0.ZipStateMachine.AnimStateTransitionNode_4.Transition, nodes [Result()]}
 Transition {
  path=AnimGraph.AnimGraphNode_StateMachine_0.ZipStateMachine.AnimStateTransitionNode_5.Transition
  nodes [Result(bCanEnterTransition=NOT Boolean(A=Get Is Rope Attached()))]
 }
 Transition {
  path=AnimGraph.AnimGraphNode_StateMachine_0.ZipStateMachine.AnimStateTransitionNode_6.Transition
  nodes [Result(bCanEnterTransition=Get IsAttached() -> (IsAttached:bool))]
 }
 ZipAttackStateMachine {
  entry=Conduit
  path=AnimGraph.AnimGraphNode_StateMachine_1.ZipAttackStateMachine
  ids=id
  nodes [
   AnimStateConduitNode_0 = Conduit()
   AnimStateNode_0 = Hover() {is=title AnimStateNode}
   AnimStateNode_1 = Attack() {is=title AnimStateNode}
   AnimStateNode_2 = Slash() {is=title AnimStateNode}
   AnimStateTransitionNode_0 = AnimStateTransitionNode() {logicType=TLT_Inertialization, priorityOrder=2}
   AnimStateTransitionNode_1 = AnimStateTransitionNode()
   AnimStateTransitionNode_2 = AnimStateTransitionNode() {logicType=TLT_Inertialization}
   AnimStateTransitionNode_3 = AnimStateTransitionNode()
   AnimStateTransitionNode_4 = AnimStateTransitionNode() {logicType=TLT_Inertialization, priorityOrder=100}
   AnimStateTransitionNode_5 = AnimStateTransitionNode()
  ]
  links [
   {
    dst=Conduit
    rule=AnimGraph.AnimGraphNode_StateMachine_1.ZipAttackStateMachine.AnimStateTransitionNode_3.Transition
    src=Attack
    via=AnimStateTransitionNode_3
   }
   {
    dst=Attack
    rule=AnimGraph.AnimGraphNode_StateMachine_1.ZipAttackStateMachine.AnimStateTransitionNode_2.Transition
    src=Conduit
    via=AnimStateTransitionNode_2
   }
   {
    dst=Hover
    rule=AnimGraph.AnimGraphNode_StateMachine_1.ZipAttackStateMachine.AnimStateTransitionNode_0.Transition
    src=Conduit
    via=AnimStateTransitionNode_0
   }
   {
    dst=Slash
    rule=AnimGraph.AnimGraphNode_StateMachine_1.ZipAttackStateMachine.AnimStateTransitionNode_4.Transition
    src=Conduit
    via=AnimStateTransitionNode_4
   }
   {
    dst=Conduit
    rule=AnimGraph.AnimGraphNode_StateMachine_1.ZipAttackStateMachine.AnimStateTransitionNode_1.Transition
    src=Hover
    via=AnimStateTransitionNode_1
   }
   {
    dst=Conduit
    rule=AnimGraph.AnimGraphNode_StateMachine_1.ZipAttackStateMachine.AnimStateTransitionNode_5.Transition
    src=Slash
    via=AnimStateTransitionNode_5
   }
  ]
 }
 Conduit {path=AnimGraph.AnimGraphNode_StateMachine_1.ZipAttackStateMachine.AnimStateConduitNode_0.Conduit, nodes [Result(bCanEnterTransition=true)]}
 Hover {
  path=AnimGraph.AnimGraphNode_StateMachine_1.ZipAttackStateMachine.AnimStateNode_0.Hover
  nodes [Output Animation Pose(Result="Sequence Player (additive) 'NyxGlass_Attack_Hover_Pose_WEP'"() {sequence=/DryBox/Items/NyxGlass/Animation/Attack/NyxGlass_Attack_Hover_Pose_WEP})]
 }
 Attack {
  path=AnimGraph.AnimGraphNode_StateMachine_1.ZipAttackStateMachine.AnimStateNode_1.Attack
  nodes [
   n1 = Blend Poses by int!(
    BlendPose_0="Sequence Player (additive) 'NyxGlass_Attack_C_Intro_WEP'"() {bLoopAnimation=false, sequence=/DryBox/Items/NyxGlass/Animation/Attack/NyxGlass_Attack_C_Intro_WEP}
    BlendPose_1="Sequence Player (additive) 'NyxGlass_Attack_A_Intro_WEP'"() {bLoopAnimation=false, sequence=/DryBox/Items/NyxGlass/Animation/Attack/NyxGlass_Attack_A_Intro_WEP}
    BlendPose_2="Sequence Player (additive) 'NyxGlass_Attack_B_Intro_WEP'"() {bLoopAnimation=false, sequence=/DryBox/Items/NyxGlass/Animation/Attack/NyxGlass_Attack_B_Intro_WEP}
    BlendTime_0=0.1
    BlendTime_1=0.1
    BlendTime_2=0.1
    ActiveChildIndex=Get CurrentAttackIntroAnim() -> (CurrentAttackIntroAnim=0)
   ) -> (Pose) {
    blendType=Linear
    childUpateMode=ResetChildOnActivate
    transitionType=Inertialization
    blendPose [{linkId=-1, sourceLinkId=-1}, {linkId=-1, sourceLinkId=-1}, {linkId=-1, sourceLinkId=-1}]
    blendTime [0.1,0.1,0.1]
   }
   Output Animation Pose(Result=@n1)
  ]
 }
 Slash {
  path=AnimGraph.AnimGraphNode_StateMachine_1.ZipAttackStateMachine.AnimStateNode_2.Slash
  nodes [
   n1 = Blend Poses by int(
    BlendPose_0="Sequence Player (additive) 'NyxGlass_Attack_C_Strike_WEP'"() {bLoopAnimation=false, sequence=/DryBox/Items/NyxGlass/Animation/Attack/NyxGlass_Attack_C_Strike_WEP}
    BlendPose_1="Sequence Player (additive) 'NyxGlass_Attack_A_Strike_WEP'"() {bLoopAnimation=false, sequence=/DryBox/Items/NyxGlass/Animation/Attack/NyxGlass_Attack_A_Strike_WEP}
    BlendPose_2="Sequence Player (additive) 'NyxGlass_Attack_B_Strike_WEP'"() {bLoopAnimation=false, sequence=/DryBox/Items/NyxGlass/Animation/Attack/NyxGlass_Attack_B_Strike_WEP}
    BlendTime_0=0.1
    BlendTime_1=0.1
    BlendTime_2=0.1
    ActiveChildIndex=Get CurrentAttackIntroAnim() -> (CurrentAttackIntroAnim=0)
   ) {
    blendType=Linear
    childUpateMode=ResetChildOnActivate
    transitionType=Inertialization
    blendPose [{linkId=-1, sourceLinkId=-1}, {linkId=-1, sourceLinkId=-1}, {linkId=-1, sourceLinkId=-1}]
    blendTime [0.1,0.1,0.1]
   }
   Output Animation Pose(Result=@n1)
  ]
 }
 Transition {
  path=AnimGraph.AnimGraphNode_StateMachine_1.ZipAttackStateMachine.AnimStateTransitionNode_0.Transition
  nodes [Result(bCanEnterTransition=Get IsHovering() -> (IsHovering:bool))]
 }
 Transition {
  path=AnimGraph.AnimGraphNode_StateMachine_1.ZipAttackStateMachine.AnimStateTransitionNode_1.Transition
  nodes [Result(bCanEnterTransition=true)]
 }
 Transition {
  path=AnimGraph.AnimGraphNode_StateMachine_1.ZipAttackStateMachine.AnimStateTransitionNode_2.Transition
  nodes [Result(bCanEnterTransition=Get IsAttackStart() -> (IsAttackStart:bool))]
 }
 Transition {
  path=AnimGraph.AnimGraphNode_StateMachine_1.ZipAttackStateMachine.AnimStateTransitionNode_3.Transition
  nodes [Result(bCanEnterTransition=true)]
 }
 Transition {
  path=AnimGraph.AnimGraphNode_StateMachine_1.ZipAttackStateMachine.AnimStateTransitionNode_4.Transition
  nodes [Result(bCanEnterTransition=true)]
 }
 Transition {
  path=AnimGraph.AnimGraphNode_StateMachine_1.ZipAttackStateMachine.AnimStateTransitionNode_5.Transition
  nodes [Result(bCanEnterTransition=true)]
 }
 EventGraph {nodes [
   Event BlueprintUpdateAnimation()
   > n1 = Sequence()
   > n2 = Cast To FortPlayerPawn(Object=TryGetPawnOwner())
   > n3 = Cast To NyxGlass_AnimBP!(
    execute
    Object=GetLinkedAnimLayerInstanceByClass!(
     self:Skeletal Mesh Component Object Reference=Get Mesh(self=@n2)
     InClass:Anim Instance Class Reference=/DryBox/Items/NyxGlass/Animation/NyxGlass_AnimBP.NyxGlass_AnimBP_C
    ) -> (ReturnValue:Anim Instance Object Reference)
   ) -> (then, CastFailed, AsNyx Glass Anim BP:Nyx Glass Anim BP Object Reference) {is=title K2Node_DynamicCast}
   > Set Is Rope Attached(Is Rope Attached:bool=Get IsRopeAttached!(self:Nyx Glass Anim BP Object Reference=@n3) -> (IsRopeAttached:bool)) -> (Output_Get:bool)
   > Set CurrentZipCastAnim(CurrentZipCastAnim:int=Get CurrentZipCastAnim!(self:Nyx Glass Anim BP Object Reference=@n3) -> (CurrentZipCastAnim=0)) -> (Output_Get:int=0)
   > Set CurrentZipReleaseAnim(CurrentZipReleaseAnim:int=Get CurrentZipReleaseAnim!(self:Nyx Glass Anim BP Object Reference=@n3) -> (CurrentZipReleaseAnim=0)) -> (Output_Get:int=0)
   > Set CurrentAttackIntroAnim(CurrentAttackIntroAnim:int=Get CurrentAttackIntroAnim!(self:Nyx Glass Anim BP Object Reference=@n3) -> (CurrentAttackIntroAnim=0)) -> (Output_Get:int=0)
   > Set IsLeapingBack(IsLeapingBack:bool=Get IsLeapingBack!(self:Nyx Glass Anim BP Object Reference=@n3) -> (IsLeapingBack:bool)) -> (Output_Get:bool)
   > Set IsAttackRelease(IsAttackRelease:bool=Get IsAttackRelease!(self:Nyx Glass Anim BP Object Reference=@n3) -> (IsAttackRelease)) -> (Output_Get:bool)
   > Set Zip Lean(Zip Lean:double=Get ZipLean!(self:Nyx Glass Anim BP Object Reference=@n3) -> (ZipLean:double)) -> (Output_Get:double)
   > Set IsZipAttack(IsZipAttack:bool=Get IsZipAttack!(self:Nyx Glass Anim BP Object Reference=@n3) -> (IsZipAttack)) -> (Output_Get:bool)
   > Set IsHovering(IsHovering:bool=Get IsHovering!(self:Nyx Glass Anim BP Object Reference=@n3) -> (IsHovering)) -> (Output_Get:bool)
   > Set IsAttackStart(IsAttackStart:bool=Get IsAttackStart!(self:Nyx Glass Anim BP Object Reference=@n3) -> (IsAttackStart)) -> (Output_Get:bool)
   > Set IsAttached(IsAttached:bool=Get IsAttached!(self:Fort Swinging Layer Anim Instance Object Reference=@n3) -> (IsAttached)) -> (Output_Get:bool)
   > Set IsAttackFall(IsAttackFall:bool=Get IsAttackFall!(self:Nyx Glass Anim BP Object Reference=@n3) -> (IsAttackFall)) -> (Output_Get:bool)
   > Set ZipIdleLeanCurve(ZipIdleLeanCurve:double=GetCurveValue(CurveName=LeanCurve)) -> (Output_Get:double)
   > Set Allow Non Release(Allow Non Release:bool=Get AllowNonRelease!(self:Nyx Glass Anim BP Object Reference=@n3) -> (AllowNonRelease:bool)) -> (Output_Get:bool)
   Set IsHover(execute=@n1.then_1, IsHover:bool="float > float"(A=GetCurveValue(CurveName=Hover), B=0.5)) -> (Output_Get:bool)
  ]}
]
variables [
 Allow Non Release {type=boolean}
 CurrentAttackIntroAnim {type=integer}
 CurrentZipCastAnim {type=integer, default=5}
 CurrentZipReleaseAnim {type=integer}
 Is Rope Attached {type=boolean}
 IsAttached {type=boolean}
 IsAttackFall {type=boolean}
 IsAttackRelease {type=boolean}
 IsAttackStart {type=boolean}
 IsCast_C {type=boolean}
 IsHover {type=boolean}
 IsHovering {type=boolean}
 IsLeapingBack {type=boolean}
 IsZipAttack {type=boolean}
 Zip Lean {type=number}
 ZipIdleLeanCurve {type=number}
]
