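# Control Rig named for the NyxGlass gauntlet that fits a harness-like gear skeleton (base, cables, canisters, belts, hooks, back) to character body bones.
# The forward chain position-constrains odm_base to pelvis, then partially rotation-constrains back to spine_03, both maintaining offset.
# An aim constraint from back toward spine_02 is left off the execution chain, and the three hierarchy curves are unused by the graph.
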ControlRigBlueprint /DryBox/Items/NyxGlass/Meshes/NyxGlass_Gauntlet_CtrlRig
graphs [
 RigVMFunctionLibrary {links [], nodes []}
 RigVMModel {nodes [
   RigUnit_BeginExecution = &RigUnit_BeginExecution()
   /Script/ControlRig.RigUnit_PositionConstraintLocalSpaceOffset(
    Child=(Type=Bone,Name="odm_base")
    Filter=(bX=True,bY=True,bZ=True)
    Parents=((Item=(Type=Bone,Name="pelvis"),Weight=1))
    Weight=1
    bMaintainOffset=True
    ExecutePin=@RigUnit_BeginExecution.ExecutePin
   ) {is=struct RigVMUnitNode}
   > &RigUnit_RotationConstraintLocalSpaceOffset(
    AdvancedSettings=(RotationOrderForFilter=XZY)
    Child=(Type=Bone,Name="back")
    Filter=(bX=True,bY=True,bZ=True)
    Parents=((Item=(Type=Bone,Name="spine_03"),Weight=1))
    Weight=0.4
    bMaintainOffset=true
   )
   &RigUnit_AimConstraintLocalSpaceOffset(
    AdvancedSettings=(DebugSettings=(bEnabled=false,Scale=10,WorldOffset=(Rotation=(X=0,Y=0,Z=0,W=1),Translation=(X=0,Y=0,Z=0),Scale3D=(X=1,Y=1,Z=1))),RotationOrderForFilter=XZY)
    Child=(Name="back")
    ExecutePin=()
    Filter=(bX=True,bY=True,bZ=True)
    Parents=((Item=(Type=Bone,Name="spine_02"),Weight=1))
    Weight=1
    WorldUp=(Target=(X=0,Y=0,Z=1),Space=(Type=None,Name="None"))
   )
  ]}
]
hierarchy [
 root {loc=<0 0 0>, rot=<0 0 0 1>}
 odm_base {parent=root, loc=<0 -0.7071 93.1191>, rot=<0.0127 0.707 -0.0127 -0.707>}
 lower_cable_01_r {parent=odm_base, loc=<1.344 -17.8681 4.5796>, rot=<-0.0081 -0.9365 -0.0095 0.3503>}
 lower_cable_02_r {parent=lower_cable_01_r, loc=<10.4384 0 0>, rot=<-0.0343 -0.1525 -0.2345 -0.9595>}
 lower_cable_03_r {parent=lower_cable_02_r, loc=<11.2382 0 0>, rot=<-0.1218 0.0224 -0.1479 -0.9812>}
 lower_cable_01_l {parent=odm_base, loc=<1.4164 -17.8707 -4.6663>, rot=<-0.0146 -0.9384 0.007 -0.3451>}
 lower_cable_02_l {parent=lower_cable_01_l, loc=<10.7195 0 0>, rot=<0.0375 0.158 -0.2326 -0.9589>}
 lower_cable_03_l {parent=lower_cable_02_l, loc=<9.2056 0 0>, rot=<0.0421 0.0299 -0.0632 -0.9967>}
 cannister_leg_01_r {parent=odm_base, loc=<2.9884 -25.1596 4.7728>, rot=<0.2302 -0.951 0.0829 0.1891>}
 cannister_leg_02_r {parent=cannister_leg_01_r, loc=<46.8892 0 0>, rot=<0 -1 0 0>}
 cannister_leg_01_l {parent=odm_base, loc=<2.9884 -25.1596 -4.7728>, rot=<0.2006 -0.953 0.0547 -0.2202>, scale=<0.9994 1 1>}
 cannister_leg_02_l {parent=cannister_leg_01_l, loc=<46.3733 0 0>, rot=<0 -0.0001 0 -1>, scale=<1.0006 1 1>}
 cannister_l {parent=odm_base, loc=<-35.1062 -43.9856 -23.1974>, rot=<0 0 0 1>}
 cannister_r {parent=odm_base, loc=<-35.5784 -44.219 23.4258>, rot=<0 0 0 1>}
 belt_01_l {parent=odm_base, loc=<6.2697 -22.9798 -17.5984>, rot=<-0.4297 -0.5634 -0.4075 -0.5761>}
 belt_02_l {parent=belt_01_l, loc=<6.0261 0 0>, rot=<0.9765 0.215 -0.0014 -0.0166>}
 hook_l {parent=belt_02_l, loc=<24.3235 0 0>, rot=<0 0 -0.0649 -0.9979>}
 belt_01_r {parent=odm_base, loc=<6.2697 -22.9798 17.5984>, rot=<0.4283 0.566 -0.4047 -0.5766>}
 belt_02_r {parent=belt_01_r, loc=<3.2732 0 0>, rot=<0.98 0.198 0.0024 0.0212>}
 hook_r {parent=belt_02_r, loc=<26.8559 0 0>, rot=<0 0 -0.0438 -0.999>}
 back {parent=odm_base, loc=<14.2126 -16.2221 -0.0018>, rot=<-0.0009 -0.0008 -0.7125 -0.7017>}
 attach {parent=root, loc=<0 0 0>, rot=<0 0 0 1>}
 pelvis {parent=root, loc=<0 -0.7071 93.1191>, rot=<0.0127 0.707 -0.0127 -0.707>}
 spine_01 {parent=pelvis, loc=<6.7494 -0.6388 0>, rot=<0 0 0.034 -0.9994>}
 spine_02 {parent=spine_01, loc=<7.2518 -0.035 0>, rot=<0 0 0.0238 -0.9997>}
 spine_03 {parent=spine_02, loc=<6.0104 0 0>, rot=<0 0 0.0149 -0.9999>}
 spine_04 {parent=spine_03, loc=<8.1024 0 0>, rot=<0 0 -0.0132 -0.9999>}
 spine_05 {parent=spine_04, loc=<10.0146 0 0>, rot=<0 0 -0.0015 -1>}
 clavicle_l {parent=spine_05, loc=<4.3575 -0.1466 -2.3223>, rot=<0.0465 -0.7289 -0.009 -0.6829>}
 clavicle_r {parent=spine_05, loc=<4.3575 -0.1466 2.3223>, rot=<0.7289 0.0465 -0.6829 0.009>}
 neck_01 {parent=spine_05, loc=<11.3138 0.1529 0>, rot=<0 0 -0.2003 -0.9797>}
 neck_02 {parent=neck_01, loc=<4.5627 0 0>, rot=<0 0 0.0504 -0.9987>}
 head {parent=neck_02, loc=<5.3914 0 0>, rot=<0 0 0.115 -0.9934>}
 thigh_l {parent=pelvis, loc=<-2.5693 -0.0109 -7.7501>, rot=<0.0012 0.0572 0.0155 -0.9982>}
 calf_l {parent=thigh_l, loc=<-39.0484 0 0>, rot=<0.0006 -0.0161 -0.0391 -0.9991>}
 foot_l {parent=calf_l, loc=<-42.561 0.1077 0.4292>, rot=<0.0418 -0.0455 0.0343 -0.9975>}
 ball_l {parent=foot_l, loc=<-6.1599 13.8264 0.6806>, rot=<0.0001 -0.0001 -0.7063 -0.7079>}
 thigh_r {parent=pelvis, loc=<-2.5691 -0.0109 7.7501>, rot=<0.0572 -0.0012 0.9982 0.0155>}
 calf_r {parent=thigh_r, loc=<39.0487 0 0>, rot=<0.0006 -0.0161 -0.0391 -0.9991>}
 foot_r {parent=calf_r, loc=<42.5612 -0.1077 -0.4292>, rot=<0.0418 -0.0455 0.0343 -0.9975>}
 ball_r {parent=foot_r, loc=<6.1598 -13.8264 -0.6806>, rot=<0.0001 -0.0001 -0.7063 -0.7079>}
 DynamicAlpha {type=Curve, loc=<0 0 0>, rot=<0 0 0 1>}
 BlendOut {type=Curve, loc=<0 0 0>, rot=<0 0 0 1>}
 LeanCurve {type=Curve, loc=<0 0 0>, rot=<0 0 0 1>}
]
