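# Physics asset for the NyxGlass Gauntlet skeletal mesh, defining physics bodies for its left and right lower cables.
# Each side has three identical capsule segments, chained by constraints to a shared small capsule on the odm_base bone.
# Every joint has the same tight swing and twist limits, restricting each segment to slight bending relative to the one before it.
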PhysicsAsset /DryBox/Items/NyxGlass/Meshes/NyxGlass_Gauntlet_Physics
bodies {
 lower_cable_01_l [{shapeType=Capsule, length=5.25, radius=2.5, shapeName=lower_cable_01_l_capsule, center=<2 0 0>, rotation=<rot 90 0 0>}]
 lower_cable_01_r [{shapeType=Capsule, length=5.25, radius=2.5, shapeName=lower_cable_01_r_capsule, center=<2 0 0>, rotation=<rot 90 0 0>}]
 lower_cable_02_l [{shapeType=Capsule, length=5.25, radius=2.5, shapeName=lower_cable_02_l_capsule, center=<2 0 0>, rotation=<rot 90 0 0>}]
 lower_cable_02_r [{shapeType=Capsule, length=5.25, radius=2.5, shapeName=lower_cable_02_r_capsule, center=<2 0 0>, rotation=<rot 90 0 0>}]
 lower_cable_03_l [{shapeType=Capsule, length=5.25, radius=2.5, shapeName=lower_cable_03_l_capsule, center=<2 0 0>, rotation=<rot 90 0 0>}]
 lower_cable_03_r [{shapeType=Capsule, length=5.25, radius=2.5, shapeName=lower_cable_03_r_capsule, center=<2 0 0>, rotation=<rot 90 0 0>}]
 odm_base [{shapeType=Capsule, length=1, radius=1}]
}
constraints [
 {bone1Name=lower_cable_01_l, bone2Name=odm_base, swing1LimitDegrees=5, swing2LimitDegrees=5, twistLimitDegrees=5}
 {bone1Name=lower_cable_02_l, bone2Name=lower_cable_01_l, swing1LimitDegrees=5, swing2LimitDegrees=5, twistLimitDegrees=5}
 {bone1Name=lower_cable_03_l, bone2Name=lower_cable_02_l, swing1LimitDegrees=5, swing2LimitDegrees=5, twistLimitDegrees=5}
 {bone1Name=lower_cable_01_r, bone2Name=odm_base, swing1LimitDegrees=5, swing2LimitDegrees=5, twistLimitDegrees=5}
 {bone1Name=lower_cable_02_r, bone2Name=lower_cable_01_r, swing1LimitDegrees=5, swing2LimitDegrees=5, twistLimitDegrees=5}
 {bone1Name=lower_cable_03_r, bone2Name=lower_cable_02_r, swing1LimitDegrees=5, swing2LimitDegrees=5, twistLimitDegrees=5}
]
