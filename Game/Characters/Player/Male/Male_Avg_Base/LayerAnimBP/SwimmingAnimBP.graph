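# Swimming animation layer blueprint for the Fortnite player, filling MovementMode layer graphs for aim offsets, full-body poses, locomotion additives and final-pose override.
# State machines drive sprint, sprint stop, downed (DBNO) swimming and movement additives, blended over input poses by SwimmingLayerAlpha.
# State enter/exit notifies set delayed flags gating sprint and start-movement behaviour.
# The final override builds an aim-offset-adjusted dive jump, but its blend alpha is 0.
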AnimBlueprint /Game/Characters/Player/Male/Male_Avg_Base/LayerAnimBP/SwimmingAnimBP
target_skeleton=&Fortnite_M_Avg_Player_Skeleton
graphs [
 AnimGraph {nodes [Output Pose() {layerGroup=None}]}
 EventGraph {nodes [
   AnimNotify_SwimSprintStart_Entered_Layer() {is=title K2Node_Event}
   > Set bSwimmingAllowSlowSprint(bSwimmingAllowSlowSprint:bool=false) -> (Output_Get:bool)
   > Set AllowSprintStop(AllowSprintStop:bool=false) -> (Output_Get:bool)
   > Delay(Duration=0.8)
   > Set AllowSprintStop(AllowSprintStop=true)
   AnimNotify_SwimSprintStart_Exited_Layer() {is=title K2Node_Event}
   > Set bSwimmingAllowSlowSprint(bSwimmingAllowSlowSprint=false)
   > Delay(Duration=1)
   > Set bSwimmingAllowSlowSprint(bSwimmingAllowSlowSprint=true)
   AnimNotify_SwimSprint_Entered_Layer() {is=title K2Node_Event}
   > Set AllowSprintStop(AllowSprintStop=true)
   AnimNotify_SwimSprintSlow_Entered_Layer() {is=title K2Node_Event}
   > Set bSwimmingSprintSlowEnoughTimePassed(bSwimmingSprintSlowEnoughTimePassed:bool=false) -> (Output_Get:bool)
   > RetriggerableDelay(Duration=5)
   > Set bSwimmingSprintSlowEnoughTimePassed(bSwimmingSprintSlowEnoughTimePassed=true)
   AnimNotify_SwimStartAdditive_Entered_Layer() {is=title K2Node_Event}
   > n1 = Sequence()
   Set ChangeMovementToVelocityFromAccel(execute=@n1.then_1, ChangeMovementToVelocityFromAccel:bool=false) -> (Output_Get:bool)
   > Set LocalAccelYawAngleAtStart(LocalAccelYawAngleAtStart:double=Get LocalAccelYawAngle() -> (LocalAccelYawAngle:float)) -> (Output_Get:double)
   > RetriggerableDelay(Duration=0.4)
   > Gate!(Enter:Exec, Open:Exec=@n1.then_0, Close:Exec=@n2.then, Toggle:Exec, bStartClosed:bool=true) -> (Exit:Exec) {is=title K2Node_MacroInstance}
   > Set ChangeMovementToVelocityFromAccel(ChangeMovementToVelocityFromAccel=true)
   AnimNotify_SwimMovementAccelAdditive_Exited_Layer() {is=title K2Node_Event}
   > n2 = Set ChangeMovementToVelocityFromAccel(ChangeMovementToVelocityFromAccel=false)
  ]}
 MovementMode_AimOffset {nodes [
   n1 = Input Pose() {name=InSourcePosePreAimOffsets}
   n2 = Break Fort Anim Input Adjusted Aim!(FortAnimInput_AdjustedAim:"Fort Anim Input Adjusted Aim Structure (by ref)"=Get AdjustedAim() -> (AdjustedAim:Fort Anim Input Adjusted Aim Structure)) -> (ResultingYaw:float=~, ResultingPitch:float=~, ZeroOutPitchWeightCurveName:Name) {is=title K2Node_BreakStruct}
   n3 = Break Fort Anim Input Player Anim Asset!(FortAnimInput_PlayerAnimAsset:"Fort Anim Input Player Anim Asset Structure (by ref)"=Get CurrentAnimSet() -> (CurrentAnimSet:Fort Anim Input Player Anim Asset Structure)) -> (
    bPlayUpperBodySlotOnFullBodyInAir:bool=~
    bDisableFullBodyAimOffsetDuringMelee:bool=~
    bShouldApplyAimOffsetFullBody:bool=~
    FullBodyAimOffsetAlpha:float=~
    UpperBodyAimOffsetAlpha:float=~
    bOverrideDisableArmsHeadAdditive:bool=~
    DisableArmsHeadAdditiveOverride:float=~
    SprintPlayrateClampRange:Fort Anim Playrate Range Structure
    SprintTargetingPlayrateClampRange:Fort Anim Playrate Range Structure
    CrouchSprintPlayrateClampRange:Fort Anim Playrate Range Structure
    CrouchSprintTargetingPlayrateClampRange:Fort Anim Playrate Range Structure
    TargetingAimOffset:Aim Offset Blend Space Object Reference
    NonTargetingAimOffset:Aim Offset Blend Space Object Reference
    RelaxedAimOffset:Aim Offset Blend Space Object Reference
    SwimTargetingPose:Anim Sequence Object Reference
    SwimNonTargetingPose:Anim Sequence Object Reference
    SwimJogAdditiveBlendSpace:Blend Space Object Reference
    SwimTargetingAimOffset:Aim Offset Blend Space Object Reference
    SwimIdleNoise:Anim Sequence Object Reference
    SwimJumpSurfaceEndAdditve:Anim Sequence Object Reference
    SwimJumpSurfaceLoopAdditve:Anim Sequence Object Reference
    SwimJumpFallLoopAdditve:Anim Sequence Object Reference
    FlyModeLoopAdditive:Anim Sequence Object Reference
    FlyModeStartAdditive:Anim Sequence Object Reference
    ZipLineStartAdditve:Anim Sequence Object Reference
    JetPackJumpAdditve:Anim Sequence Object Reference
    JetPackStartAdditve:Anim Sequence Object Reference
    FallAdditive:Anim Sequence Object Reference
    JumpLoopAdditive:Anim Sequence Object Reference
    JumpUpAdditive:Anim Sequence Object Reference
    FullBodyJumpAdditive:Anim Sequence Object Reference
    WeaponChargeLoop:Anim Sequence Object Reference
    LandingAdditiveOverride:Anim Sequence Object Reference
    TurnInPlaceLeftAdditiveOverride:Anim Sequence Object Reference
    TurnInPlaceRightAdditiveOverride:Anim Sequence Object Reference
    bOverridePitchAndYawOffsets:bool=~
    AnimSetOffsets:Fort Anim Input Adjusted Aim Offset Structure
    BonesToHide:Array of Names
    bOverrideHandIKRetargetingWeight:bool=~
    HandIKRetargetingWeight:float=~
    bForceKneeSliding:bool=~
    TargetingPose:Anim Sequence Object Reference
    NonTargetingPose:Anim Sequence Object Reference
    IdleNoise:Anim Sequence Object Reference
   ) {is=title K2Node_BreakStruct}
   n4 = "Blend Poses (EFortWeaponCoreAnimation)"!(
    BlendPose_0:Pose Link Structure=Use cached pose 'AimOffsetsInputPose'()
    BlendPose_1:Pose Link Structure=Blend Poses by bool(
     BlendPose_0="AimOffset Player '(None)'"!(
      X:float=0
      Y:float=@n2.ResultingPitch
      BlendSpace:Blend Space Object Reference=@n3.SwimTargetingAimOffset
      BasePose:Pose Link Structure=@n1
      Alpha:float=@n3.UpperBodyAimOffsetAlpha
     ) -> (Pose:Pose Link Structure) {lODThreshold=3, alphaBoolBlend {blendOption=Linear}}
     BlendPose_1=Use cached pose 'AimOffsetsInputPose'()
     bActiveValue=Get bIsSurfaceSwimming()
    ) {blendType=Linear}
    BlendTime_0:float=0.1
    BlendTime_1:float=0.1
    ActiveEnumValue:Byte=Get WeaponCoreAnim!() -> (WeaponCoreAnim)
   ) -> (Pose:Pose Link Structure) {
    is=title AnimGraphNode_BlendListByEnum
    blendType=Linear
    blendPose [{linkId=-1, sourceLinkId=-1}, {linkId=-1, sourceLinkId=-1}]
    blendTime [0.1,0.1]
   }
   Output Pose(Result=@n4) {layerGroup=MovementMode}
   Save cached pose 'AimOffsetsInputPose'(Pose=Input Pose() {name=InAimOffsetsPose}) {cacheName=AimOffsetsInputPose}
  ]}
 MovementMode_BaseLocomotion {nodes [
   Input Pose() {name=InPoseBaseLocomotion}
   Output Pose() {layerGroup=MovementMode}
  ]}
 MovementMode_FinalPoseOverride {nodes [
   n1 = "Blend Poses (EFortPlayerAnimBodyType)"!(
    BlendPose_0:Pose Link Structure=Use cached pose 'DiveJumpPostAdditives'()
    BlendPose_1:Pose Link Structure=Apply Additive(
     Base=Use cached pose 'DiveJumpPostAdditives'()
     Additive="Sequence Player (additive) 'Large_Swim_DiveJump_Pose_Add'"() {sequence=/Game/Animation/Game/MainPlayer/Swimming/DiveJump/Large/Large_Swim_DiveJump_Pose_Add}
    ) {alphaBoolBlend {blendOption=Linear}}
    BlendTime_0:float=0.1
    BlendTime_1:float=0.1
    ActiveEnumValue:Byte=Get AnimBodyType() -> (AnimBodyType:EFortPlayerAnimBodyType Enum=Small)
   ) -> (Pose:Pose Link Structure) {
    is=title AnimGraphNode_BlendListByEnum
    blendType=Linear
    blendPose [{linkId=-1, sourceLinkId=-1}, {linkId=-1, sourceLinkId=-1}]
    blendTime [0.1,0.1]
   }
   n2 = Use cached pose 'PostDiveJump'()
   n3 = Get bEnableDiveJumpAimOffsets() -> (bEnableDiveJumpAimOffsets:bool)
   Output Pose(Result=Two Way Blend(A=Input Pose() {name=InFinalPose}, B=Blend Poses by bool(BlendPose_0=@n1, BlendPose_1=@n2, bActiveValue=@n3) {blendType=Linear}, Alpha=0) {blendNode {alphaBoolBlend {blendOption=Linear}}}) {layerGroup=MovementMode}
   AnimGraphNode_StateMachine_0 = DiveJump() {is=title AnimGraphNode_StateMachine, bSkipFirstUpdateTransition=false}
   Save cached pose 'PostDiveJump'(Pose=@AnimGraphNode_StateMachine_0) {cacheName=PostDiveJump}
   n4 = Sequence Player 'Swim_DiveJump_ArmsTucked_Pose'() {sequence=/Game/Animation/Game/MainPlayer/Swimming/DiveJump/Swim_DiveJump_ArmsTucked_Pose}
   n5 = Use cached pose 'PostDiveJump'()
   n6 = AimOffset Player 'AO_Swim_DiveJump_Pitch'!(
    X:float=0
    Y:float=Get SwimmingDivePitch() -> (SwimmingDivePitch:float)
    BasePose:Pose Link Structure=Blend Poses by bool(
     BlendPose_0=Layered blend per bone(BasePose=@n5, BlendPoses_0=@n4) {layerSetup [{branchFilters [{blendDepth=2, boneName=Spine_01}]}]}
     BlendPose_1=Use cached pose 'PostDiveJump'()
     bActiveValue=Get bIsHidingInProp() -> (bIsHidingInProp:bool)
    ) {blendType=Linear}
    Alpha:float=1
   ) -> (Pose:Pose Link Structure) {blendSpace=/Game/Animation/Game/MainPlayer/Swimming/DiveJump/AO_Swim_DiveJump_Pitch, alphaBoolBlend {blendOption=Linear}}
   n7 = Get SwimmingYawDeltaRatePerSecond() -> (SwimmingYawDeltaRatePerSecond:float)
   n8 = Get SwimmingDivePitchDeltaRatePerSec() -> (SwimmingDivePitchDeltaRatePerSec:float)
   Save cached pose 'DiveJumpPostAdditives'(Pose=Rotate Root Bone!(
     BasePose:Pose Link Structure=AimOffset Player 'AO_Swim_DiveJump_Pose'!(X:float=@n7, Y:float=@n8, BasePose:Pose Link Structure=@n6, Alpha:float=1) -> (Pose:Pose Link Structure) {blendSpace=/Game/Animation/Game/MainPlayer/Swimming/DiveJump/AO_Swim_DiveJump_Pose, alphaBoolBlend {blendOption=Linear}}
     Yaw:float=Get SwimmingLocalAccelerationYawAngleSmooth() -> (SwimmingLocalAccelerationYawAngleSmooth:float)
    ) -> (Pose:Pose Link Structure) {is=title AnimGraphNode_RotateRootBone}) {cacheName=DiveJumpPostAdditives}
  ]}
 DiveJump {
  entry=DiveJump_Up
  path=MovementMode_FinalPoseOverride.AnimGraphNode_StateMachine_0.DiveJump
  ids=id
  nodes [
   Alias() {is=title AnimStateAliasNode, also {aliases [DiveJump_Fall, DiveJump_Land, DiveJump_Up]}}
   AnimStateNode_0 = DiveJump_Up() {is=title AnimStateNode, bAlwaysResetOnEntry=true}
   AnimStateNode_2 = DiveJump_Fall() {is=title AnimStateNode, bAlwaysResetOnEntry=true}
   AnimStateNode_3 = DiveJump_Land() {is=title AnimStateNode, bAlwaysResetOnEntry=true}
   AnimStateNode_4 = DiveJump_GroundLand() {is=title AnimStateNode}
   AnimStateTransitionNode_0 = AnimStateTransitionNode() {bAutomaticRuleBasedOnSequencePlayerInState=true, crossfadeDuration=0.05}
   AnimStateTransitionNode_1 = AnimStateTransitionNode()
   AnimStateTransitionNode_2 = AnimStateTransitionNode() {crossfadeDuration=0}
   AnimStateTransitionNode_3 = AnimStateTransitionNode()
   AnimStateTransitionNode_4 = AnimStateTransitionNode()
  ]
  links [
   {
    dst=DiveJump_GroundLand
    rule=MovementMode_FinalPoseOverride.AnimGraphNode_StateMachine_0.DiveJump.AnimStateTransitionNode_4.Transition
    src=Alias
    via=AnimStateTransitionNode_4
   }
   {
    dst=DiveJump_Land
    rule=MovementMode_FinalPoseOverride.AnimGraphNode_StateMachine_0.DiveJump.AnimStateTransitionNode_1.Transition
    src=DiveJump_Fall
    via=AnimStateTransitionNode_1
   }
   {
    dst=DiveJump_Up
    rule=MovementMode_FinalPoseOverride.AnimGraphNode_StateMachine_0.DiveJump.AnimStateTransitionNode_2.Transition
    src=DiveJump_Land
    via=AnimStateTransitionNode_2
   }
   {
    dst=DiveJump_Fall
    rule=MovementMode_FinalPoseOverride.AnimGraphNode_StateMachine_0.DiveJump.AnimStateTransitionNode_0.Transition
    src=DiveJump_Up
    via=AnimStateTransitionNode_0
   }
   {
    dst=DiveJump_Land
    rule=MovementMode_FinalPoseOverride.AnimGraphNode_StateMachine_0.DiveJump.AnimStateTransitionNode_3.Transition
    src=DiveJump_Up
    via=AnimStateTransitionNode_3
   }
  ]
 }
 DiveJump_Up {
  path=MovementMode_FinalPoseOverride.AnimGraphNode_StateMachine_0.DiveJump.AnimStateNode_0.DiveJump_Up
  nodes [Output Animation Pose(Result=Sequence Player 'Swim_DiveJump_Up'() {bLoopAnimation=false, sequence=/Game/Animation/Game/MainPlayer/Swimming/DiveJump/Swim_DiveJump_Up})]
 }
 DiveJump_Fall {
  path=MovementMode_FinalPoseOverride.AnimGraphNode_StateMachine_0.DiveJump.AnimStateNode_2.DiveJump_Fall
  nodes [Output Animation Pose(Result=Sequence Player 'Swim_DiveJump_Fall'() {bLoopAnimation=false, sequence=/Game/Animation/Game/MainPlayer/Swimming/DiveJump/Swim_DiveJump_Fall})]
 }
 DiveJump_Land {
  path=MovementMode_FinalPoseOverride.AnimGraphNode_StateMachine_0.DiveJump.AnimStateNode_3.DiveJump_Land
  nodes [Output Animation Pose(Result=Sequence Player 'Swim_DiveJump_Land'() {bLoopAnimation=false, sequence=/Game/Animation/Game/MainPlayer/Swimming/DiveJump/Swim_DiveJump_Land})]
 }
 DiveJump_GroundLand {
  path=MovementMode_FinalPoseOverride.AnimGraphNode_StateMachine_0.DiveJump.AnimStateNode_4.DiveJump_GroundLand
  nodes [Output Animation Pose(Result=Sequence Player 'Swim_DiveJump_LandOnGroundRoll'() {bLoopAnimation=false, sequence=/Game/Animation/Game/MainPlayer/Swimming/DiveJump/Swim_DiveJump_LandOnGroundRoll}) {
    node {updateFunction {functionName=OnNewFunction}}
    updateFunction {bSelfContext=true, memberGuid=#1, memberName=On Dive Jump Ground Land Update}
   }]
 }
 Transition {path=MovementMode_FinalPoseOverride.AnimGraphNode_StateMachine_0.DiveJump.AnimStateTransitionNode_0.Transition, nodes [Result()]}
 Transition {
  path=MovementMode_FinalPoseOverride.AnimGraphNode_StateMachine_0.DiveJump.AnimStateTransitionNode_1.Transition
  nodes [Result(bCanEnterTransition=NOT Boolean(A=Get bIsWaterJump() -> (bIsWaterJump:bool)))]
 }
 Transition {
  path=MovementMode_FinalPoseOverride.AnimGraphNode_StateMachine_0.DiveJump.AnimStateTransitionNode_2.Transition
  nodes [Result(bCanEnterTransition=Get bIsWaterJump())]
 }
 Transition {
  path=MovementMode_FinalPoseOverride.AnimGraphNode_StateMachine_0.DiveJump.AnimStateTransitionNode_3.Transition
  nodes [Result(bCanEnterTransition=NOT Boolean(A=Get bIsWaterJump()))]
 }
 Transition {
  path=MovementMode_FinalPoseOverride.AnimGraphNode_StateMachine_0.DiveJump.AnimStateTransitionNode_4.Transition
  nodes [Result(bCanEnterTransition=Get bSwimmingDoveIntoGround() -> (bSwimmingDoveIntoGround:bool))]
 }
 MovementMode_FullBodyOverride {nodes [
   AnimGraphNode_StateMachine_1 = SwimmingFullBody() {is=title AnimGraphNode_StateMachine}
   Output Pose(Result=Two Way Blend(A=Use cached pose 'FullBody'(), B=@AnimGraphNode_StateMachine_1, Alpha=Get SwimmingLayerAlpha() -> (SwimmingLayerAlpha:float)) {blendNode {alphaBoolBlend {blendOption=Linear}}}) {layerGroup=MovementMode}
   Save cached pose 'FullBody'(Pose=Input Pose() {name=InputPoseFullBody}) {cacheName=FullBody, cachePoseName=FullBody}
   AnimGraphNode_StateMachine_6 = Swimming Sprint() {is=title AnimGraphNode_StateMachine}
   Save cached pose 'SprintingPose_PostMask'(Pose=Rotate Root Bone(BasePose=@AnimGraphNode_StateMachine_6, Yaw=Get SwimmingLocalAccelerationYawAngleSmooth())) {cacheName=SprintingPose_PostMask}
  ]}
 SwimmingFullBody {
  entry=SwimmingFullBody
  path=MovementMode_FullBodyOverride.AnimGraphNode_StateMachine_1.SwimmingFullBody
  ids=id
  nodes [
   AnimStateConduitNode_0 = Conduit()
   AnimStateNode_0 = SwimmingFullBody() {is=title AnimStateNode}
   AnimStateNode_1 = SwimmingSprint() {is=title AnimStateNode}
   AnimStateNode_3 = SwimSprint_Stop() {is=title AnimStateNode}
   AnimStateNode_4 = DBNO() {is=title AnimStateNode}
   AnimStateTransitionNode_0 = AnimStateTransitionNode()
   AnimStateTransitionNode_1 = AnimStateTransitionNode() {crossfadeDuration=0}
   AnimStateTransitionNode_11 = AnimStateTransitionNode() {crossfadeDuration=0.25}
   AnimStateTransitionNode_12 = AnimStateTransitionNode()
   AnimStateTransitionNode_13 = AnimStateTransitionNode() {crossfadeDuration=0.4}
   AnimStateTransitionNode_2 = AnimStateTransitionNode()
   AnimStateTransitionNode_3 = AnimStateTransitionNode() {crossfadeDuration=0}
   AnimStateTransitionNode_4 = AnimStateTransitionNode()
   AnimStateTransitionNode_49 = AnimStateTransitionNode() {crossfadeDuration=1}
   AnimStateTransitionNode_5 = AnimStateTransitionNode() {crossfadeDuration=0}
   AnimStateTransitionNode_50 = AnimStateTransitionNode()
   AnimStateTransitionNode_6 = AnimStateTransitionNode() {bAutomaticRuleBasedOnSequencePlayerInState=true, crossfadeDuration=0.25}
   AnimStateTransitionNode_7 = AnimStateTransitionNode() {crossfadeDuration=0}
   AnimStateTransitionNode_8 = AnimStateTransitionNode()
   AnimStateTransitionNode_9 = AnimStateTransitionNode() {crossfadeDuration=0}
  ]
  links [
   {
    dst=SwimSprint_Stop
    rule=MovementMode_FullBodyOverride.AnimGraphNode_StateMachine_1.SwimmingFullBody.AnimStateTransitionNode_49.AnimationTransitionGraph_1
    src=Conduit
    via=AnimStateTransitionNode_49
   }
   {
    dst=SwimSprint_Stop
    rule=MovementMode_FullBodyOverride.AnimGraphNode_StateMachine_1.SwimmingFullBody.AnimStateTransitionNode_50.AnimationTransitionGraph_1
    src=Conduit
    via=AnimStateTransitionNode_50
   }
   {
    dst=SwimmingFullBody
    rule=MovementMode_FullBodyOverride.AnimGraphNode_StateMachine_1.SwimmingFullBody.AnimStateTransitionNode_12.Transition
    src=DBNO
    via=AnimStateTransitionNode_12
   }
   {
    dst=SwimmingFullBody
    rule=MovementMode_FullBodyOverride.AnimGraphNode_StateMachine_1.SwimmingFullBody.AnimStateTransitionNode_13.Transition
    src=SwimSprint_Stop
    via=AnimStateTransitionNode_13
   }
   {
    dst=SwimmingFullBody
    rule=MovementMode_FullBodyOverride.AnimGraphNode_StateMachine_1.SwimmingFullBody.AnimStateTransitionNode_6.Transition
    src=SwimSprint_Stop
    via=AnimStateTransitionNode_6
   }
   {
    dst=SwimmingFullBody
    rule=MovementMode_FullBodyOverride.AnimGraphNode_StateMachine_1.SwimmingFullBody.AnimStateTransitionNode_7.Transition
    src=SwimSprint_Stop
    via=AnimStateTransitionNode_7
   }
   {
    dst=SwimmingFullBody
    rule=MovementMode_FullBodyOverride.AnimGraphNode_StateMachine_1.SwimmingFullBody.AnimStateTransitionNode_8.Transition
    src=SwimSprint_Stop
    via=AnimStateTransitionNode_8
   }
   {
    dst=SwimmingFullBody
    rule=MovementMode_FullBodyOverride.AnimGraphNode_StateMachine_1.SwimmingFullBody.AnimStateTransitionNode_9.Transition
    src=SwimSprint_Stop
    via=AnimStateTransitionNode_9
   }
   {
    dst=DBNO
    rule=MovementMode_FullBodyOverride.AnimGraphNode_StateMachine_1.SwimmingFullBody.AnimStateTransitionNode_11.Transition
    src=SwimmingFullBody
    via=AnimStateTransitionNode_11
   }
   {
    dst=SwimmingSprint
    rule=MovementMode_FullBodyOverride.AnimGraphNode_StateMachine_1.SwimmingFullBody.AnimStateTransitionNode_0.Transition
    src=SwimmingFullBody
    via=AnimStateTransitionNode_0
   }
   {
    dst=Conduit
    rule=MovementMode_FullBodyOverride.AnimGraphNode_StateMachine_1.SwimmingFullBody.AnimStateTransitionNode_4.Transition
    src=SwimmingSprint
    via=AnimStateTransitionNode_4
   }
   {
    dst=SwimSprint_Stop
    rule=MovementMode_FullBodyOverride.AnimGraphNode_StateMachine_1.SwimmingFullBody.AnimStateTransitionNode_5.Transition
    src=SwimmingSprint
    via=AnimStateTransitionNode_5
   }
   {
    dst=SwimmingFullBody
    rule=MovementMode_FullBodyOverride.AnimGraphNode_StateMachine_1.SwimmingFullBody.AnimStateTransitionNode_1.Transition
    src=SwimmingSprint
    via=AnimStateTransitionNode_1
   }
   {
    dst=SwimmingFullBody
    rule=MovementMode_FullBodyOverride.AnimGraphNode_StateMachine_1.SwimmingFullBody.AnimStateTransitionNode_2.Transition
    src=SwimmingSprint
    via=AnimStateTransitionNode_2
   }
   {
    dst=SwimmingFullBody
    rule=MovementMode_FullBodyOverride.AnimGraphNode_StateMachine_1.SwimmingFullBody.AnimStateTransitionNode_3.Transition
    src=SwimmingSprint
    via=AnimStateTransitionNode_3
   }
  ]
 }
 Conduit {
  path=MovementMode_FullBodyOverride.AnimGraphNode_StateMachine_1.SwimmingFullBody.AnimStateConduitNode_0.Conduit
  nodes [Result(bCanEnterTransition=true)]
 }
 SwimmingFullBody {
  path=MovementMode_FullBodyOverride.AnimGraphNode_StateMachine_1.SwimmingFullBody.AnimStateNode_0.SwimmingFullBody
  nodes [Output Animation Pose(Result=Use cached pose 'FullBody'())]
 }
 SwimmingSprint {
  path=MovementMode_FullBodyOverride.AnimGraphNode_StateMachine_1.SwimmingFullBody.AnimStateNode_1.SwimmingSprint
  nodes [Output Animation Pose(Result=Use cached pose 'SprintingPose_PostMask'())]
 }
 SwimSprint_Stop {
  path=MovementMode_FullBodyOverride.AnimGraphNode_StateMachine_1.SwimmingFullBody.AnimStateNode_3.SwimSprint_Stop
  nodes [Output Animation Pose(Result=Sequence Player 'Swim_Sprint_Stop_CMM'() {bLoopAnimation=false, sequence=/Game/Animation/Game/MainPlayer/Swimming/Sprint/Swim_Sprint_Stop_CMM})]
 }
 DBNO {
  path=MovementMode_FullBodyOverride.AnimGraphNode_StateMachine_1.SwimmingFullBody.AnimStateNode_4.DBNO
  nodes [
   AnimGraphNode_StateMachine_0 = DBNO Swimming() {is=title AnimGraphNode_StateMachine, bSkipFirstUpdateTransition=false}
   n1 = "Transform (Modify) Bone - Bone: pelvis"!(
    ComponentPose=Local To Component(LocalPose=@AnimGraphNode_StateMachine_0)
    bAlphaBoolEnabled:bool=Get bIsFemale() -> (bIsFemale:bool)
    Translation="0,0,5"
   ) -> (Pose) {alphaInputType=Bool, translationMode=BMM_Additive, alphaBoolBlend {blendOption=Linear}, boneToModify {boneName=pelvis}, translation {z=5}}
   Output Animation Pose(Result=Component To Local(ComponentPose=@n1))
  ]
 }
 DBNO Swimming {
  entry=DBNO Swim Idle
  path=MovementMode_FullBodyOverride.AnimGraphNode_StateMachine_1.SwimmingFullBody.AnimStateNode_4.DBNO.AnimGraphNode_StateMachine_0.DBNO Swimming
  ids=id
  nodes [
   AnimStateNode_0 = DBNO Swim Idle() {is=title AnimStateNode}
   AnimStateNode_1 = DBNO Swim Move() {is=title AnimStateNode}
   AnimStateNode_3 = DBNO Swim Turn() {is=title AnimStateNode}
   AnimStateTransitionNode_6 = AnimStateTransitionNode() {
    crossfadeDuration=0.6
    blendProfileWrapper {blendProfile=&Fortnite_M_Avg_Player_Skeleton:WeaponL_Slow}
   }
   AnimStateTransitionNode_7 = AnimStateTransitionNode() {
    crossfadeDuration=1
    blendProfileWrapper {blendProfile=&Fortnite_M_Avg_Player_Skeleton:WeaponL_Slow}
   }
   AnimStateTransitionNode_8 = AnimStateTransitionNode() {blendProfileWrapper {blendProfile=&Fortnite_M_Avg_Player_Skeleton:WeaponL_Slow}}
   AnimStateTransitionNode_9 = AnimStateTransitionNode() {blendProfileWrapper {blendProfile=&Fortnite_M_Avg_Player_Skeleton:WeaponL_Slow}}
  ]
  links [
   {
    dst=DBNO Swim Move
    rule=MovementMode_FullBodyOverride.AnimGraphNode_StateMachine_1.SwimmingFullBody.AnimStateNode_4.DBNO.AnimGraphNode_StateMachine_0.DBNO Swimming.AnimStateTransitionNode_6.AnimationTransitionGraph_0
    src=DBNO Swim Idle
    via=AnimStateTransitionNode_6
   }
   {
    dst=DBNO Swim Turn
    rule=MovementMode_FullBodyOverride.AnimGraphNode_StateMachine_1.SwimmingFullBody.AnimStateNode_4.DBNO.AnimGraphNode_StateMachine_0.DBNO Swimming.AnimStateTransitionNode_8.AnimationTransitionGraph_0
    src=DBNO Swim Idle
    via=AnimStateTransitionNode_8
   }
   {
    dst=DBNO Swim Idle
    rule=MovementMode_FullBodyOverride.AnimGraphNode_StateMachine_1.SwimmingFullBody.AnimStateNode_4.DBNO.AnimGraphNode_StateMachine_0.DBNO Swimming.AnimStateTransitionNode_7.AnimationTransitionGraph_0
    src=DBNO Swim Move
    via=AnimStateTransitionNode_7
   }
   {
    dst=DBNO Swim Idle
    rule=MovementMode_FullBodyOverride.AnimGraphNode_StateMachine_1.SwimmingFullBody.AnimStateNode_4.DBNO.AnimGraphNode_StateMachine_0.DBNO Swimming.AnimStateTransitionNode_9.AnimationTransitionGraph_0
    src=DBNO Swim Turn
    via=AnimStateTransitionNode_9
   }
  ]
 }
 DBNO Swim Idle {
  path=MovementMode_FullBodyOverride.AnimGraphNode_StateMachine_1.SwimmingFullBody.AnimStateNode_4.DBNO.AnimGraphNode_StateMachine_0.DBNO Swimming.AnimStateNode_0.DBNO Swim Idle
  nodes [Output Animation Pose(Result=Sequence Player 'DBNO_Swim_Idle'!(PlayRate:float=0.65) -> (Pose) {sequence=/Game/Animation/Game/MainPlayer/Locomotion/Medium/Male/DBNO/DBNO_Swim_Idle})]
 }
 DBNO Swim Move {
  path=MovementMode_FullBodyOverride.AnimGraphNode_StateMachine_1.SwimmingFullBody.AnimStateNode_4.DBNO.AnimGraphNode_StateMachine_0.DBNO Swimming.AnimStateNode_1.DBNO Swim Move
  nodes [Output Animation Pose(Result=Blendspace Player 'BS_DBNO_Swim_CMM'(X=Get SwimmingVelocityAngle() -> (SwimmingVelocityAngle:float), Y=105) {blendSpace=/Game/Animation/Game/MainPlayer/Locomotion/Medium/Male/DBNO/BS_DBNO_Swim_CMM})]
 }
 DBNO Swim Turn {
  path=MovementMode_FullBodyOverride.AnimGraphNode_StateMachine_1.SwimmingFullBody.AnimStateNode_4.DBNO.AnimGraphNode_StateMachine_0.DBNO Swimming.AnimStateNode_3.DBNO Swim Turn
  nodes [
   n1 = Get DBNOTurnPlayRateAbs() -> (DBNOTurnPlayRateAbs:float)
   n2 = Sequence Player 'DBNO_SwimTurn_E_CMM'!(PlayRate:float=@n1) -> (Pose) {groupName=DBNOSwimSync, method=SyncGroup, sequence=/Game/Animation/Game/MainPlayer/Locomotion/Medium/Male/DBNO/DBNO_SwimTurn_E_CMM}
   n3 = Sequence Player 'DBNO_SwimTurn_W_CMM'!(PlayRate:float=@n1) -> (Pose) {groupName=DBNOSwimSync, method=SyncGroup, sequence=/Game/Animation/Game/MainPlayer/Locomotion/Medium/Male/DBNO/DBNO_SwimTurn_W_CMM}
   Output Animation Pose(Result=Blend Poses by bool(
     BlendPose_0=@n2
     BlendPose_1=@n3
     BlendTime_0=0.5
     BlendTime_1=0.5
     bActiveValue=Get bSwimmingPlayDBNOTurnEast() -> (bSwimmingPlayDBNOTurnEast:bool)
    ) {blendType=Linear})
  ]
 }
 AnimationTransitionGraph_0 {
  path=MovementMode_FullBodyOverride.AnimGraphNode_StateMachine_1.SwimmingFullBody.AnimStateNode_4.DBNO.AnimGraphNode_StateMachine_0.DBNO Swimming.AnimStateTransitionNode_6.AnimationTransitionGraph_0
  nodes [Result(bCanEnterTransition=Get bIsAccelerating2D())]
 }
 AnimationTransitionGraph_0 {
  path=MovementMode_FullBodyOverride.AnimGraphNode_StateMachine_1.SwimmingFullBody.AnimStateNode_4.DBNO.AnimGraphNode_StateMachine_0.DBNO Swimming.AnimStateTransitionNode_7.AnimationTransitionGraph_0
  nodes [Result(bCanEnterTransition=NOT Boolean(A=Get bIsAccelerating2D()))]
 }
 AnimationTransitionGraph_0 {
  path=MovementMode_FullBodyOverride.AnimGraphNode_StateMachine_1.SwimmingFullBody.AnimStateNode_4.DBNO.AnimGraphNode_StateMachine_0.DBNO Swimming.AnimStateTransitionNode_8.AnimationTransitionGraph_0
  nodes [Result(bCanEnterTransition=Get bTransition_DBNOSwimIdle_to_DBNOSwimTurn() -> (bTransition_DBNOSwimIdle_to_DBNOSwimTurn:bool))]
 }
 AnimationTransitionGraph_0 {
  path=MovementMode_FullBodyOverride.AnimGraphNode_StateMachine_1.SwimmingFullBody.AnimStateNode_4.DBNO.AnimGraphNode_StateMachine_0.DBNO Swimming.AnimStateTransitionNode_9.AnimationTransitionGraph_0
  nodes [Result(bCanEnterTransition=NOT Boolean(A=Get bTransition_DBNOSwimIdle_to_DBNOSwimTurn()))]
 }
 Transition {
  path=MovementMode_FullBodyOverride.AnimGraphNode_StateMachine_1.SwimmingFullBody.AnimStateTransitionNode_0.Transition
  nodes [Result(bCanEnterTransition=Get bTransition_Swimming_FullBody_Sprinting() -> (bTransition_Swimming_FullBody_Sprinting:bool))]
 }
 Transition {
  path=MovementMode_FullBodyOverride.AnimGraphNode_StateMachine_1.SwimmingFullBody.AnimStateTransitionNode_1.Transition
  nodes [Result(bCanEnterTransition=Get bRecentlyFired() -> (bRecentlyFired:bool))]
 }
 Transition {
  path=MovementMode_FullBodyOverride.AnimGraphNode_StateMachine_1.SwimmingFullBody.AnimStateTransitionNode_11.Transition
  nodes [Result(bCanEnterTransition=Get bIsDBNO())]
 }
 Transition {
  path=MovementMode_FullBodyOverride.AnimGraphNode_StateMachine_1.SwimmingFullBody.AnimStateTransitionNode_12.Transition
  nodes [Result(bCanEnterTransition=NOT Boolean(A=Get bIsDBNO()))]
 }
 Transition {
  path=MovementMode_FullBodyOverride.AnimGraphNode_StateMachine_1.SwimmingFullBody.AnimStateTransitionNode_13.Transition
  nodes [Result(bCanEnterTransition=Get bIsAccelerating2D())]
 }
 Transition {
  path=MovementMode_FullBodyOverride.AnimGraphNode_StateMachine_1.SwimmingFullBody.AnimStateTransitionNode_2.Transition
  nodes [Result(bCanEnterTransition=Get bIsWaterJump())]
 }
 Transition {
  path=MovementMode_FullBodyOverride.AnimGraphNode_StateMachine_1.SwimmingFullBody.AnimStateTransitionNode_3.Transition
  nodes [Result(bCanEnterTransition=Get bTransition_InterruptSprintSwimming() -> (bTransition_InterruptSprintSwimming:bool))]
 }
 Transition {
  path=MovementMode_FullBodyOverride.AnimGraphNode_StateMachine_1.SwimmingFullBody.AnimStateTransitionNode_4.Transition
  nodes [Result(bCanEnterTransition=Get bTransition_Sprinting_to_Conduit() -> (bTransition_Sprinting_to_Conduit:bool))]
 }
 AnimationTransitionGraph_1 {
  path=MovementMode_FullBodyOverride.AnimGraphNode_StateMachine_1.SwimmingFullBody.AnimStateTransitionNode_49.AnimationTransitionGraph_1
  nodes [Result(bCanEnterTransition=NOT Boolean(A=Get AllowSprintStop() -> (AllowSprintStop:bool)))]
 }
 Transition {
  path=MovementMode_FullBodyOverride.AnimGraphNode_StateMachine_1.SwimmingFullBody.AnimStateTransitionNode_5.Transition
  nodes [Result(bCanEnterTransition=Get bRecentlyFired())]
 }
 AnimationTransitionGraph_1 {
  path=MovementMode_FullBodyOverride.AnimGraphNode_StateMachine_1.SwimmingFullBody.AnimStateTransitionNode_50.AnimationTransitionGraph_1
  nodes [Result(bCanEnterTransition=Get AllowSprintStop())]
 }
 Transition {path=MovementMode_FullBodyOverride.AnimGraphNode_StateMachine_1.SwimmingFullBody.AnimStateTransitionNode_6.Transition, nodes [Result()]}
 Transition {
  path=MovementMode_FullBodyOverride.AnimGraphNode_StateMachine_1.SwimmingFullBody.AnimStateTransitionNode_7.Transition
  nodes [Result(bCanEnterTransition=Get bRecentlyFired())]
 }
 Transition {
  path=MovementMode_FullBodyOverride.AnimGraphNode_StateMachine_1.SwimmingFullBody.AnimStateTransitionNode_8.Transition
  nodes [Result(bCanEnterTransition=Get bIsWaterJump())]
 }
 Transition {
  path=MovementMode_FullBodyOverride.AnimGraphNode_StateMachine_1.SwimmingFullBody.AnimStateTransitionNode_9.Transition
  nodes [Result(bCanEnterTransition=Get bTransition_InterruptSprintSwimming())]
 }
 Swimming Sprint {
  entry=SwimSprintStart
  path=MovementMode_FullBodyOverride.AnimGraphNode_StateMachine_6.Swimming Sprint
  ids=id
  nodes [
   AnimStateNode_1 = SwimSprintSlow() {is=title AnimStateNode, stateEntered {guid=#2, notifyName=SwimSprintSlow_Entered_Layer}}
   AnimStateNode_4 = SwimSprintStart() {
    is=title AnimStateNode
    bAlwaysResetOnEntry=true
    stateEntered {guid=#3, notifyName=SwimSprintStart_Entered_Layer}
    stateLeft {guid=#4, notifyName=SwimSprintStart_Exited_Layer}
   }
   AnimStateNode_5 = SwimSprint() {is=title AnimStateNode, stateEntered {guid=#5, notifyName=SwimSprint_Entered_Layer}}
   AnimStateTransitionNode_1 = AnimStateTransitionNode() {bAutomaticRuleBasedOnSequencePlayerInState=true, crossfadeDuration=0.1}
   AnimStateTransitionNode_10 = AnimStateTransitionNode()
   AnimStateTransitionNode_4 = AnimStateTransitionNode() {crossfadeDuration=0.5}
   AnimStateTransitionNode_9 = AnimStateTransitionNode() {crossfadeDuration=0.5}
  ]
  links [
   {
    dst=SwimSprintSlow
    rule=MovementMode_FullBodyOverride.AnimGraphNode_StateMachine_6.Swimming Sprint.AnimStateTransitionNode_9.Transition
    src=SwimSprint
    via=AnimStateTransitionNode_9
   }
   {
    dst=SwimSprint
    rule=MovementMode_FullBodyOverride.AnimGraphNode_StateMachine_6.Swimming Sprint.AnimStateTransitionNode_4.Transition
    src=SwimSprintSlow
    via=AnimStateTransitionNode_4
   }
   {
    dst=SwimSprintStart
    rule=MovementMode_FullBodyOverride.AnimGraphNode_StateMachine_6.Swimming Sprint.AnimStateTransitionNode_10.Transition
    src=SwimSprintSlow
    via=AnimStateTransitionNode_10
   }
   {
    dst=SwimSprint
    rule=MovementMode_FullBodyOverride.AnimGraphNode_StateMachine_6.Swimming Sprint.AnimStateTransitionNode_1.Transition
    src=SwimSprintStart
    via=AnimStateTransitionNode_1
   }
  ]
 }
 SwimSprintSlow {
  path=MovementMode_FullBodyOverride.AnimGraphNode_StateMachine_6.Swimming Sprint.AnimStateNode_1.SwimSprintSlow
  nodes [
   Output Animation Pose(Result=Blendspace Player 'BS_Swim_Sprint_Slow'(X=Get SwimmingSprintYawDeltaRatePerSecondSmooth() -> (SwimmingSprintYawDeltaRatePerSecondSmooth:float), Y=0) {blendSpace=/Game/Animation/Game/MainPlayer/Swimming/Sprint/BS_Swim_Sprint_Slow, groupName=SwimSprint, method=SyncGroup})
   Reroute Node(InputPin:Wildcard) -> (OutputPin:Wildcard)
  ]
 }
 SwimSprintStart {
  path=MovementMode_FullBodyOverride.AnimGraphNode_StateMachine_6.Swimming Sprint.AnimStateNode_4.SwimSprintStart
  nodes [
   n1 = Get SwimmingSprintYawDeltaRatePerSecondSmooth()
   n2 = "Blend Poses (EFortCustomGender)"!(
    BlendPose_0:Pose Link Structure=Blendspace Player 'BS_Swim_Sprint_Start'!(X=@n1, Y=0, StartPosition:float=0) -> (Pose) {blendSpace=/Game/Animation/Game/MainPlayer/Swimming/Sprint/BS_Swim_Sprint_Start}
    BlendPose_1:Pose Link Structure=Blendspace Player 'BS_Swim_Sprint_Start_CMF'!(X=@n1, Y=0, StartPosition:float=0) -> (Pose) {blendSpace=/Game/Animation/Game/MainPlayer/Swimming/Sprint/BS_Swim_Sprint_Start_CMF}
    BlendTime_0:float=0
    BlendTime_1:float=0
    ActiveEnumValue:Byte=Get Gender() -> (Gender:EFortCustomGender Enum=Invalid)
   ) -> (Pose:Pose Link Structure) {
    is=title AnimGraphNode_BlendListByEnum
    blendType=Linear
    blendPose [{linkId=-1, sourceLinkId=-1}, {linkId=-1, sourceLinkId=-1}]
    blendTime [0.1,0.1]
   }
   Output Animation Pose(Result=@n2)
  ]
 }
 SwimSprint {
  path=MovementMode_FullBodyOverride.AnimGraphNode_StateMachine_6.Swimming Sprint.AnimStateNode_5.SwimSprint
  nodes [
   n1 = Get SwimmingSprintYawDeltaRatePerSecondSmooth()
   n2 = "Blend Poses (EFortCustomGender)"(
    BlendPose_0=Blendspace Player 'BS_Swim_Sprint'(X=@n1, Y=0) {blendSpace=/Game/Animation/Game/MainPlayer/Swimming/Sprint/BS_Swim_Sprint, groupName=SwimSprint, method=SyncGroup}
    BlendPose_1=Blendspace Player 'BS_Swim_Sprint_CMF'(X=@n1, Y=0) {blendSpace=/Game/Animation/Game/MainPlayer/Swimming/Sprint/BS_Swim_Sprint_CMF, groupName=SwimSprint, method=SyncGroup}
    BlendTime_0=0
    BlendTime_1=0
    ActiveEnumValue=Get Gender() -> (Gender=Invalid)
   ) {blendType=Linear, blendPose [{linkId=-1, sourceLinkId=-1}, {linkId=-1, sourceLinkId=-1}], blendTime [0.1,0.1]}
   Output Animation Pose(Result=@n2)
  ]
 }
 Transition {path=MovementMode_FullBodyOverride.AnimGraphNode_StateMachine_6.Swimming Sprint.AnimStateTransitionNode_1.Transition, nodes [Result()]}
 Transition {
  path=MovementMode_FullBodyOverride.AnimGraphNode_StateMachine_6.Swimming Sprint.AnimStateTransitionNode_10.Transition
  nodes [Result(bCanEnterTransition=Get bTransition_SwimSprintSlow_to_SwimSprintStart() -> (bTransition_SwimSprintSlow_to_SwimSprintStart:bool))]
 }
 Transition {
  path=MovementMode_FullBodyOverride.AnimGraphNode_StateMachine_6.Swimming Sprint.AnimStateTransitionNode_4.Transition
  nodes [Result(bCanEnterTransition=Get bTransition_SwimSprintSlow_to_SwimSprint() -> (bTransition_SwimSprintSlow_to_SwimSprint:bool))]
 }
 Transition {
  path=MovementMode_FullBodyOverride.AnimGraphNode_StateMachine_6.Swimming Sprint.AnimStateTransitionNode_9.Transition
  nodes [Result(bCanEnterTransition=Get bTransition_SwimSprint_to_SwimSprintSlow() -> (bTransition_SwimSprint_to_SwimSprintSlow:bool))]
 }
 MovementMode_InAirOverride {nodes [
   Input Pose() {name=InSource_Pose}
   Input Pose() {name=InAirPose}
   Output Pose() {layerGroup=MovementMode}
  ]}
 MovementMode_LocomotionAdditives {nodes [
   AnimGraphNode_StateMachine_0 = MovementAdditivePose() {is=title AnimGraphNode_StateMachine}
   AnimGraphNode_StateMachine_2 = SwimmingAdditive() {is=title AnimGraphNode_StateMachine}
   n1 = Layered blend per bone(
    BasePose=Blend Poses by bool(BlendPose_0=@AnimGraphNode_StateMachine_0, BlendPose_1=@AnimGraphNode_StateMachine_2, BlendTime_1=0.2, bActiveValue=Get bIsTargeting()) {blendType=Linear}
    BlendPoses_0=Additive Identity Pose()
    BlendWeights_0=Get PlayMeleeAttackAO() -> (PlayMeleeAttackAO:float)
   ) {curveBlendOption=BlendByWeight, lODThreshold=2, layerSetup [{branchFilters [{blendDepth=2, boneName=Spine_01}, {boneName=Ik_hand_root}]}]}
   n2 = Layered blend per bone(
    BasePose=@n1
    BlendPoses_0=Additive Identity Pose()
    BlendWeights_0=Get DisableArmsHeadAdditiveCurveAlpha() -> (DisableArmsHeadAdditiveCurveAlpha:float)
   ) {
    bMeshSpaceRotationBlend=true
    lODThreshold=2
    layerSetup [{branchFilters [{boneName=clavicle_l}, {boneName=clavicle_r}, {boneName=ik_hand_root}]}]
   }
   n3 = Layered blend per bone(
    BasePose=Layered blend per bone(
     BasePose=@n2
     BlendPoses_0=Additive Identity Pose()
     BlendWeights_0=Get DisableIKRootAdditiveCurveAlpha() -> (DisableIKRootAdditiveCurveAlpha:float)
    ) {bMeshSpaceRotationBlend=true, lODThreshold=2, layerSetup [{branchFilters [{boneName=ik_hand_root}]}]}
    BlendPoses_0=Additive Identity Pose()
    BlendWeights_0=Get UpperBodyLocomotionAdditiveMaskAlpha() -> (UpperBodyLocomotionAdditiveMaskAlpha:float)
   ) {
    bMeshSpaceRotationBlend=true
    bMeshSpaceScaleBlend=true
    curveBlendOption=UseMaxValue
    layerSetup [{branchFilters [{blendDepth=4, boneName=spine_01}, {boneName=ik_hand_root}]}]
   }
   Output Pose(Result=Two Way Blend(A=Input Pose() {name=InLocomotionAdditivesPose}, B=@n3, Alpha=Get SwimmingLayerAlpha()) {blendNode {alphaBoolBlend {blendOption=Linear}}}) {layerGroup=MovementMode}
  ]}
 MovementAdditivePose {
  entry=State
  path=MovementMode_LocomotionAdditives.AnimGraphNode_StateMachine_0.MovementAdditivePose
  ids=id
  nodes [AnimStateNode_0 = State() {is=title AnimStateNode}]
 }
 State {
  path=MovementMode_LocomotionAdditives.AnimGraphNode_StateMachine_0.MovementAdditivePose.AnimStateNode_0.State
  nodes [
   n1 = Blendspace Player 'AR_Swim_Targeted_Additive_BS'(X=Get SwimmingVelocityAngle(), Y=Get SwimmingSpeedRelativeToFlow() -> (SwimmingSpeedRelativeToFlow:float)) {
    bResetPlayTimeWhenBlendSpaceChanges=false
    blendSpace=/Game/Animation/Game/MainPlayer/Swimming/AssaultRifle/DirectionalMovement/AR_Swim_Targeted_Additive_BS
    groupName=SwimTargetingLocomotion
    groupRole=AlwaysLeader
    method=SyncGroup
   }
   n2 = Break Fort Anim Input Player Anim Asset(FortAnimInput_PlayerAnimAsset=Get CurrentAnimSet()) -> (
    bPlayUpperBodySlotOnFullBodyInAir=~
    bDisableFullBodyAimOffsetDuringMelee=~
    bShouldApplyAimOffsetFullBody=~
    FullBodyAimOffsetAlpha=~
    UpperBodyAimOffsetAlpha=~
    bOverrideDisableArmsHeadAdditive=~
    DisableArmsHeadAdditiveOverride=~
    bOverridePitchAndYawOffsets=~
    bOverrideHandIKRetargetingWeight=~
    HandIKRetargetingWeight=~
    bForceKneeSliding=~
   )
   n3 = Get SwimmingVelocityAngle()
   Output Animation Pose(Result=Blend Poses by bool(
     BlendPose_0=@n1
     BlendPose_1=Blendspace Player!(X=@n3, Y=300, PlayRate:float=1, BlendSpace:Blend Space Object Reference=@n2.SwimJogAdditiveBlendSpace) -> (Pose) {bResetPlayTimeWhenBlendSpaceChanges=false, groupName=SwimLocomotion, groupRole=AlwaysLeader, method=SyncGroup}
     BlendTime_1=.2
     bActiveValue=Get bIsTargeting()
    ) {blendType=Linear})
  ]
 }
 SwimmingAdditive {
  entry=IdleAdditive
  path=MovementMode_LocomotionAdditives.AnimGraphNode_StateMachine_2.SwimmingAdditive
  ids=id
  nodes [
   AnimStateNode_0 = SwimMovementAccelAdditive() {is=title AnimStateNode, stateLeft {guid=#6, notifyName=SwimMovementAccelAdditive_Exited_Layer}}
   AnimStateNode_1 = SwimStopAdditive() {is=title AnimStateNode}
   AnimStateNode_4 = SwimStartAdditive() {
    is=title AnimStateNode
    bAlwaysResetOnEntry=true
    stateEntered {guid=#7, notifyName=SwimStartAdditive_Entered_Layer}
   }
   AnimStateNode_5 = IdleAdditive() {is=title AnimStateNode, bAlwaysResetOnEntry=true}
   AnimStateNode_7 = SwimMovementAdditive() {is=title AnimStateNode}
   AnimStateTransitionNode_0 = AnimStateTransitionNode() {crossfadeDuration=0.3}
   AnimStateTransitionNode_10 = AnimStateTransitionNode()
   AnimStateTransitionNode_14 = AnimStateTransitionNode() {crossfadeDuration=0.3}
   AnimStateTransitionNode_15 = AnimStateTransitionNode() {crossfadeDuration=0.4}
   AnimStateTransitionNode_2 = AnimStateTransitionNode() {crossfadeDuration=0.3}
   AnimStateTransitionNode_25 = AnimStateTransitionNode()
   AnimStateTransitionNode_27 = AnimStateTransitionNode()
   AnimStateTransitionNode_3 = AnimStateTransitionNode() {crossfadeDuration=0.4}
   AnimStateTransitionNode_4 = AnimStateTransitionNode() {crossfadeDuration=0.5}
   AnimStateTransitionNode_5 = AnimStateTransitionNode() {crossfadeDuration=0.3}
   AnimStateTransitionNode_6 = AnimStateTransitionNode() {bAutomaticRuleBasedOnSequencePlayerInState=true, crossfadeDuration=1.33}
   AnimStateTransitionNode_7 = AnimStateTransitionNode() {crossfadeDuration=0.3}
   AnimStateTransitionNode_8 = AnimStateTransitionNode()
  ]
  links [
   {
    dst=SwimStartAdditive
    rule=MovementMode_LocomotionAdditives.AnimGraphNode_StateMachine_2.SwimmingAdditive.AnimStateTransitionNode_0.Transition
    src=IdleAdditive
    via=AnimStateTransitionNode_0
   }
   {
    dst=IdleAdditive
    rule=MovementMode_LocomotionAdditives.AnimGraphNode_StateMachine_2.SwimmingAdditive.AnimStateTransitionNode_10.Transition
    src=SwimMovementAccelAdditive
    via=AnimStateTransitionNode_10
   }
   {
    dst=SwimMovementAdditive
    rule=MovementMode_LocomotionAdditives.AnimGraphNode_StateMachine_2.SwimmingAdditive.AnimStateTransitionNode_4.Transition
    src=SwimMovementAccelAdditive
    via=AnimStateTransitionNode_4
   }
   {
    dst=SwimStopAdditive
    rule=MovementMode_LocomotionAdditives.AnimGraphNode_StateMachine_2.SwimmingAdditive.AnimStateTransitionNode_14.Transition
    src=SwimMovementAccelAdditive
    via=AnimStateTransitionNode_14
   }
   {
    dst=SwimStopAdditive
    rule=MovementMode_LocomotionAdditives.AnimGraphNode_StateMachine_2.SwimmingAdditive.AnimStateTransitionNode_2.Transition
    src=SwimMovementAccelAdditive
    via=AnimStateTransitionNode_2
   }
   {
    dst=IdleAdditive
    rule=MovementMode_LocomotionAdditives.AnimGraphNode_StateMachine_2.SwimmingAdditive.AnimStateTransitionNode_10.Transition
    src=SwimMovementAdditive
    via=AnimStateTransitionNode_25
   }
   {
    dst=SwimStopAdditive
    rule=MovementMode_LocomotionAdditives.AnimGraphNode_StateMachine_2.SwimmingAdditive.AnimStateTransitionNode_5.Transition
    src=SwimMovementAdditive
    via=AnimStateTransitionNode_5
   }
   {
    dst=SwimStopAdditive
    rule=MovementMode_LocomotionAdditives.AnimGraphNode_StateMachine_2.SwimmingAdditive.AnimStateTransitionNode_7.Transition
    src=SwimMovementAdditive
    via=AnimStateTransitionNode_7
   }
   {
    dst=IdleAdditive
    rule=MovementMode_LocomotionAdditives.AnimGraphNode_StateMachine_2.SwimmingAdditive.AnimStateTransitionNode_10.Transition
    src=SwimStartAdditive
    via=AnimStateTransitionNode_27
   }
   {
    dst=SwimMovementAccelAdditive
    rule=MovementMode_LocomotionAdditives.AnimGraphNode_StateMachine_2.SwimmingAdditive.AnimStateTransitionNode_3.Transition
    src=SwimStartAdditive
    via=AnimStateTransitionNode_3
   }
   {
    dst=IdleAdditive
    rule=MovementMode_LocomotionAdditives.AnimGraphNode_StateMachine_2.SwimmingAdditive.AnimStateTransitionNode_6.Transition
    src=SwimStopAdditive
    via=AnimStateTransitionNode_6
   }
   {
    dst=IdleAdditive
    rule=MovementMode_LocomotionAdditives.AnimGraphNode_StateMachine_2.SwimmingAdditive.AnimStateTransitionNode_10.Transition
    src=SwimStopAdditive
    via=AnimStateTransitionNode_8
   }
   {
    dst=SwimStartAdditive
    rule=MovementMode_LocomotionAdditives.AnimGraphNode_StateMachine_2.SwimmingAdditive.AnimStateTransitionNode_15.Transition
    src=SwimStopAdditive
    via=AnimStateTransitionNode_15
   }
  ]
 }
 SwimMovementAccelAdditive {
  path=MovementMode_LocomotionAdditives.AnimGraphNode_StateMachine_2.SwimmingAdditive.AnimStateNode_0.SwimMovementAccelAdditive
  nodes [
   n1 = Get CurrentAnimSet()
   Output Animation Pose(Result=Blendspace Player(
     X=Get LocalAccelYawAngleAtStart() -> (LocalAccelYawAngleAtStart:double)
     Y=300
     PlayRate=1
     BlendSpace=Break Fort Anim Input Player Anim Asset(FortAnimInput_PlayerAnimAsset=@n1) -> (
      bPlayUpperBodySlotOnFullBodyInAir=~
      bDisableFullBodyAimOffsetDuringMelee=~
      bShouldApplyAimOffsetFullBody=~
      FullBodyAimOffsetAlpha=~
      UpperBodyAimOffsetAlpha=~
      bOverrideDisableArmsHeadAdditive=~
      DisableArmsHeadAdditiveOverride=~
      bOverridePitchAndYawOffsets=~
      bOverrideHandIKRetargetingWeight=~
      HandIKRetargetingWeight=~
      bForceKneeSliding=~
     ).SwimJogAdditiveBlendSpace
    ) {bResetPlayTimeWhenBlendSpaceChanges=false, groupName=SwimLocomotion, groupRole=AlwaysLeader, method=SyncGroup})
  ]
 }
 SwimStopAdditive {
  path=MovementMode_LocomotionAdditives.AnimGraphNode_StateMachine_2.SwimmingAdditive.AnimStateNode_1.SwimStopAdditive
  nodes [
   n1 = "Sequence Player (additive) 'AR_Swim_GenericStop_NonTargeted'"!(bLoopAnimation:bool=False) -> (Pose) {
    bLoopAnimation=false
    groupName=StopSync
    groupRole=AlwaysLeader
    method=SyncGroup
    sequence=/Game/Animation/Game/MainPlayer/Swimming/AssaultRifle/AdditiveTransitions/AR_Swim_GenericStop_NonTargeted
   }
   Output Animation Pose(Result=@n1)
  ]
 }
 SwimStartAdditive {
  path=MovementMode_LocomotionAdditives.AnimGraphNode_StateMachine_2.SwimmingAdditive.AnimStateNode_4.SwimStartAdditive
  nodes [
   n1 = "Sequence Player (additive) 'AR_Swim_GenericStart_NonTargeted'"!(PlayRate:float=1, bLoopAnimation:bool=False) -> (Pose) {
    bLoopAnimation=false
    groupName=SwimStartSync
    groupRole=AlwaysLeader
    method=SyncGroup
    sequence=/Game/Animation/Game/MainPlayer/Swimming/AssaultRifle/AdditiveTransitions/AR_Swim_GenericStart_NonTargeted
   }
   Output Animation Pose(Result=@n1)
  ]
 }
 IdleAdditive {
  path=MovementMode_LocomotionAdditives.AnimGraphNode_StateMachine_2.SwimmingAdditive.AnimStateNode_5.IdleAdditive
  nodes [
   n1 = Break Fort Anim Input Player Anim Asset(FortAnimInput_PlayerAnimAsset=Get CurrentAnimSet()) -> (
    bPlayUpperBodySlotOnFullBodyInAir=~
    bDisableFullBodyAimOffsetDuringMelee=~
    bShouldApplyAimOffsetFullBody=~
    FullBodyAimOffsetAlpha=~
    UpperBodyAimOffsetAlpha=~
    bOverrideDisableArmsHeadAdditive=~
    DisableArmsHeadAdditiveOverride=~
    bOverridePitchAndYawOffsets=~
    bOverrideHandIKRetargetingWeight=~
    HandIKRetargetingWeight=~
    bForceKneeSliding=~
   )
   n2 = "Sequence Player (additive) 'AR_Swim_Idle_Noise_Targeted'"() {
    groupName=SwimTargetingLocomotion
    groupRole=AlwaysLeader
    method=SyncGroup
    sequence=/Game/Animation/Game/MainPlayer/Swimming/AssaultRifle/DirectionalMovement/AR_Swim_Idle_Noise_Targeted
   }
   n3 = Blend Poses by bool(
    BlendPose_0=@n2
    BlendPose_1=Blend Poses by bool(
     BlendPose_0=Sequence Player!(Sequence:Anim Sequence Base Object Reference=@n1.SwimIdleNoise, StartPosition:float=0) -> (Pose)
     BlendPose_1=Sequence Player(Sequence=@n1.SwimIdleNoise, StartPosition=0)
     BlendTime_0=0.2
     BlendTime_1=0.2
     bActiveValue=Get bSwimIdleNoiseBlendTrigger() -> (bSwimIdleNoiseBlendTrigger:bool)
    ) {blendType=Linear, transitionType=Inertialization}
    bActiveValue=Get bIsTargeting()
   ) {blendType=Linear}
   n4 = Sequence Player!(Sequence=@n1.SwimIdleNoise, PlayRate:float=0, StartPosition=0) -> (Pose)
   n5 = Get bSwimmingIsWaterPaddlingToSurface() -> (bSwimmingIsWaterPaddlingToSurface:bool)
   Output Animation Pose(Result=Blend Poses by bool(
     BlendPose_0=Sequence Player(Sequence=@n1.SwimIdleNoise, PlayRate=0, StartPosition=0) {playRate=0}
     BlendPose_1=Blend Poses by bool(BlendPose_0=@n4, BlendPose_1=@n3, BlendTime_0=0.2, BlendTime_1=0.2, bActiveValue=@n5) {blendType=Linear}
     BlendTime_0=0.2
     BlendTime_1=0.2
     bActiveValue=Get bSwimmingIsWaterLanding() -> (bSwimmingIsWaterLanding:bool)
    ) {blendType=Linear})
  ]
 }
 SwimMovementAdditive {
  path=MovementMode_LocomotionAdditives.AnimGraphNode_StateMachine_2.SwimmingAdditive.AnimStateNode_7.SwimMovementAdditive
  nodes [
   Blendspace Player(
    X=Get SwimmingVelocityAngle()
    Y=300
    PlayRate=1
    BlendSpace=Break Fort Anim Input Player Anim Asset(FortAnimInput_PlayerAnimAsset=Get CurrentAnimSet()) -> (
     bPlayUpperBodySlotOnFullBodyInAir=~
     bDisableFullBodyAimOffsetDuringMelee=~
     bShouldApplyAimOffsetFullBody=~
     FullBodyAimOffsetAlpha=~
     UpperBodyAimOffsetAlpha=~
     bOverrideDisableArmsHeadAdditive=~
     DisableArmsHeadAdditiveOverride=~
     bOverridePitchAndYawOffsets=~
     bOverrideHandIKRetargetingWeight=~
     HandIKRetargetingWeight=~
     bForceKneeSliding=~
    ).SwimJogAdditiveBlendSpace
   ) {bResetPlayTimeWhenBlendSpaceChanges=false, groupName=SwimLocomotion, method=SyncGroup}
   n1 = Get CurrentAnimSet()
   Output Animation Pose(Result=Blendspace Player(
     X=Get SwimmingNonTargetingBlendspaceDir() -> (SwimmingNonTargetingBlendspaceDir:float)
     Y=300
     PlayRate=1
     BlendSpace=Break Fort Anim Input Player Anim Asset(FortAnimInput_PlayerAnimAsset=@n1) -> (
      bPlayUpperBodySlotOnFullBodyInAir=~
      bDisableFullBodyAimOffsetDuringMelee=~
      bShouldApplyAimOffsetFullBody=~
      FullBodyAimOffsetAlpha=~
      UpperBodyAimOffsetAlpha=~
      bOverrideDisableArmsHeadAdditive=~
      DisableArmsHeadAdditiveOverride=~
      bOverridePitchAndYawOffsets=~
      bOverrideHandIKRetargetingWeight=~
      HandIKRetargetingWeight=~
      bForceKneeSliding=~
     ).SwimJogAdditiveBlendSpace
    ) {bResetPlayTimeWhenBlendSpaceChanges=false, groupName=SwimLocomotion, method=SyncGroup})
  ]
 }
 Transition {
  path=MovementMode_LocomotionAdditives.AnimGraphNode_StateMachine_2.SwimmingAdditive.AnimStateTransitionNode_0.Transition
  nodes [Result(bCanEnterTransition=Get bTransition_Idles_to_SwimStart() -> (bTransition_Idles_to_SwimStart:bool))]
 }
 Transition {
  path=MovementMode_LocomotionAdditives.AnimGraphNode_StateMachine_2.SwimmingAdditive.AnimStateTransitionNode_10.Transition
  nodes [Result(bCanEnterTransition=Get bIsWaterJump())]
 }
 Transition {
  path=MovementMode_LocomotionAdditives.AnimGraphNode_StateMachine_2.SwimmingAdditive.AnimStateTransitionNode_14.Transition
  nodes [Result(bCanEnterTransition=Get bSwimmingSharpAccelerationChange() -> (bSwimmingSharpAccelerationChange:bool))]
 }
 Transition {
  path=MovementMode_LocomotionAdditives.AnimGraphNode_StateMachine_2.SwimmingAdditive.AnimStateTransitionNode_15.Transition
  nodes [Result(bCanEnterTransition=Get bIsAccelerating2D())]
 }
 Transition {
  path=MovementMode_LocomotionAdditives.AnimGraphNode_StateMachine_2.SwimmingAdditive.AnimStateTransitionNode_2.Transition
  nodes [Result(bCanEnterTransition=NOT Boolean(A=Get bIsAccelerating2D()))]
 }
 Transition {
  path=MovementMode_LocomotionAdditives.AnimGraphNode_StateMachine_2.SwimmingAdditive.AnimStateTransitionNode_3.Transition
  nodes [Result(bCanEnterTransition=Get bTransition_SwimStartAdditive_to_SwimMovementAccelAdditive() -> (bTransition_SwimStartAdditive_to_SwimMovementAccelAdditive:bool))]
 }
 Transition {
  path=MovementMode_LocomotionAdditives.AnimGraphNode_StateMachine_2.SwimmingAdditive.AnimStateTransitionNode_4.Transition
  nodes [Result(bCanEnterTransition=Get ChangeMovementToVelocityFromAccel() -> (ChangeMovementToVelocityFromAccel:bool))]
 }
 Transition {
  path=MovementMode_LocomotionAdditives.AnimGraphNode_StateMachine_2.SwimmingAdditive.AnimStateTransitionNode_5.Transition
  nodes [Result(bCanEnterTransition=NOT Boolean(A=Get bIsAccelerating2D()))]
 }
 Transition {path=MovementMode_LocomotionAdditives.AnimGraphNode_StateMachine_2.SwimmingAdditive.AnimStateTransitionNode_6.Transition, nodes [Result()]}
 Transition {
  path=MovementMode_LocomotionAdditives.AnimGraphNode_StateMachine_2.SwimmingAdditive.AnimStateTransitionNode_7.Transition
  nodes [Result(bCanEnterTransition=Get bSwimmingSharpAccelerationChange())]
 }
 MovementMode_LowerBodyOverride {nodes [
   n1 = Blendspace Player 'AR_Swim_Targeted_Core_BS'(X=Get SwimmingVelocityAngle(), Y=Get SwimmingSpeedRelativeToFlow()) {
    bResetPlayTimeWhenBlendSpaceChanges=false
    blendSpace=/Game/Animation/Game/MainPlayer/Swimming/AssaultRifle/DirectionalMovement/AR_Swim_Targeted_Core_BS
    groupName=SwimTargetingLocomotion
    groupRole=AlwaysFollower
    method=SyncGroup
   }
   AnimGraphNode_StateMachine_8 = SwimmingLowerBody() {is=title AnimGraphNode_StateMachine}
   n2 = Get bIsTargeting()
   Output Pose(Result=Two Way Blend(
     A=Input Pose() {name=InLowerBodyPose}
     B=Blend Poses by bool(BlendPose_0=@n1, BlendPose_1=@AnimGraphNode_StateMachine_8, BlendTime_1=.2, bActiveValue=@n2) {blendType=Linear}
     Alpha=Get SwimmingLayerAlpha()
    ) {blendNode {alphaBoolBlend {blendOption=Linear}}}) {layerGroup=MovementMode}
   Save cached pose 'BasePose'(Pose=Input Pose() {name=InSourcePose}) {cacheName=BasePose}
  ]}
 SwimmingLowerBody {
  entry=Idles
  path=MovementMode_LowerBodyOverride.AnimGraphNode_StateMachine_8.SwimmingLowerBody
  ids=id
  nodes [
   AnimStateNode_0 = SwimMovementAccel() {is=title AnimStateNode}
   AnimStateNode_1 = SwimStop() {is=title AnimStateNode}
   AnimStateNode_2 = Idles() {is=title AnimStateNode, bAlwaysResetOnEntry=true}
   AnimStateNode_3 = SwimMovement() {is=title AnimStateNode}
   AnimStateNode_6 = SwimStart() {is=title AnimStateNode, bAlwaysResetOnEntry=true}
   AnimStateTransitionNode_0 = AnimStateTransitionNode() {crossfadeDuration=0.3}
   AnimStateTransitionNode_1 = AnimStateTransitionNode() {crossfadeDuration=0.3}
   AnimStateTransitionNode_10 = AnimStateTransitionNode()
   AnimStateTransitionNode_15 = AnimStateTransitionNode() {crossfadeDuration=0.3}
   AnimStateTransitionNode_16 = AnimStateTransitionNode() {crossfadeDuration=0.4}
   AnimStateTransitionNode_2 = AnimStateTransitionNode() {crossfadeDuration=0.4}
   AnimStateTransitionNode_21 = AnimStateTransitionNode()
   AnimStateTransitionNode_23 = AnimStateTransitionNode()
   AnimStateTransitionNode_4 = AnimStateTransitionNode() {crossfadeDuration=0.5}
   AnimStateTransitionNode_5 = AnimStateTransitionNode() {crossfadeDuration=0.3}
   AnimStateTransitionNode_6 = AnimStateTransitionNode() {crossfadeDuration=0.3}
   AnimStateTransitionNode_7 = AnimStateTransitionNode() {bAutomaticRuleBasedOnSequencePlayerInState=true, crossfadeDuration=1.33}
   AnimStateTransitionNode_9 = AnimStateTransitionNode()
  ]
  links [
   {
    dst=SwimStart
    rule=MovementMode_LowerBodyOverride.AnimGraphNode_StateMachine_8.SwimmingLowerBody.AnimStateTransitionNode_0.Transition
    src=Idles
    via=AnimStateTransitionNode_0
   }
   {
    dst=Idles
    rule=MovementMode_LowerBodyOverride.AnimGraphNode_StateMachine_8.SwimmingLowerBody.AnimStateTransitionNode_9.Transition
    src=SwimMovement
    via=AnimStateTransitionNode_21
   }
   {
    dst=SwimStop
    rule=MovementMode_LowerBodyOverride.AnimGraphNode_StateMachine_8.SwimmingLowerBody.AnimStateTransitionNode_5.Transition
    src=SwimMovement
    via=AnimStateTransitionNode_5
   }
   {
    dst=SwimStop
    rule=MovementMode_LowerBodyOverride.AnimGraphNode_StateMachine_8.SwimmingLowerBody.AnimStateTransitionNode_6.Transition
    src=SwimMovement
    via=AnimStateTransitionNode_6
   }
   {
    dst=Idles
    rule=MovementMode_LowerBodyOverride.AnimGraphNode_StateMachine_8.SwimmingLowerBody.AnimStateTransitionNode_9.Transition
    src=SwimMovementAccel
    via=AnimStateTransitionNode_9
   }
   {
    dst=SwimMovement
    rule=MovementMode_LowerBodyOverride.AnimGraphNode_StateMachine_8.SwimmingLowerBody.AnimStateTransitionNode_4.Transition
    src=SwimMovementAccel
    via=AnimStateTransitionNode_4
   }
   {
    dst=SwimStop
    rule=MovementMode_LowerBodyOverride.AnimGraphNode_StateMachine_8.SwimmingLowerBody.AnimStateTransitionNode_1.Transition
    src=SwimMovementAccel
    via=AnimStateTransitionNode_1
   }
   {
    dst=SwimStop
    rule=MovementMode_LowerBodyOverride.AnimGraphNode_StateMachine_8.SwimmingLowerBody.AnimStateTransitionNode_15.Transition
    src=SwimMovementAccel
    via=AnimStateTransitionNode_15
   }
   {
    dst=Idles
    rule=MovementMode_LowerBodyOverride.AnimGraphNode_StateMachine_8.SwimmingLowerBody.AnimStateTransitionNode_9.Transition
    src=SwimStart
    via=AnimStateTransitionNode_23
   }
   {
    dst=SwimMovementAccel
    rule=MovementMode_LowerBodyOverride.AnimGraphNode_StateMachine_8.SwimmingLowerBody.AnimStateTransitionNode_2.Transition
    src=SwimStart
    via=AnimStateTransitionNode_2
   }
   {
    dst=Idles
    rule=MovementMode_LowerBodyOverride.AnimGraphNode_StateMachine_8.SwimmingLowerBody.AnimStateTransitionNode_9.Transition
    src=SwimStop
    via=AnimStateTransitionNode_10
   }
   {
    dst=Idles
    rule=MovementMode_LowerBodyOverride.AnimGraphNode_StateMachine_8.SwimmingLowerBody.AnimStateTransitionNode_7.Transition
    src=SwimStop
    via=AnimStateTransitionNode_7
   }
   {
    dst=SwimStart
    rule=MovementMode_LowerBodyOverride.AnimGraphNode_StateMachine_8.SwimmingLowerBody.AnimStateTransitionNode_16.Transition
    src=SwimStop
    via=AnimStateTransitionNode_16
   }
  ]
 }
 SwimMovementAccel {
  path=MovementMode_LowerBodyOverride.AnimGraphNode_StateMachine_8.SwimmingLowerBody.AnimStateNode_0.SwimMovementAccel
  nodes [
   n1 = Blendspace Player 'AR_Swim_NonTargeted_Core_BS'!(X=Get LocalAccelYawAngleAtStart(), Y=300, PlayRate:float=1) -> (Pose) {
    bResetPlayTimeWhenBlendSpaceChanges=false
    blendSpace=/Game/Animation/Game/MainPlayer/Swimming/AssaultRifle/DirectionalMovement/AR_Swim_NonTargeted_Core_BS
    groupName=SwimLocomotion
    groupRole=AlwaysFollower
    method=SyncGroup
   }
   n2 = "Transform (Modify) Bone - Bone: pelvis"!(
    ComponentPose=Local To Component(LocalPose=@n1)
    Alpha:float=1
    Translation="0,0,0"
    Rotation:Rotator=Get SwimmingMeleeTwistRotator() -> (SwimmingMeleeTwistRotator:Rotator)
    Scale:Vector="1,1,1"
   ) -> (Pose) {rotationMode=BMM_Additive, rotationSpace=BCS_BoneSpace, alphaBoolBlend {blendOption=Linear}, boneToModify {boneName=pelvis}}
   n3 = AimOffset Player 'AR_Swim_DirectionalZOffset_NonTargeted_AO'!(X:float=Get LocalAccelYawAngleAtStart(), Y:float=1, BasePose:Pose Link Structure=Component To Local(ComponentPose=@n2), Alpha:float=1) -> (Pose:Pose Link Structure) {
    blendSpace=/Game/Animation/Game/MainPlayer/Swimming/AssaultRifle/DirectionalMovement/AR_Swim_DirectionalZOffset_NonTargeted_AO
    alphaBoolBlend {blendOption=Linear}
   }
   Output Animation Pose(Result=@n3)
  ]
 }
 SwimStop {
  path=MovementMode_LowerBodyOverride.AnimGraphNode_StateMachine_8.SwimmingLowerBody.AnimStateNode_1.SwimStop
  nodes [
   n1 = Sequence Player 'AR_Swim_GenericStop_NonTargeted_Core'!(bLoopAnimation:bool=false) -> (Pose) {
    groupName=StopSync
    groupRole=AlwaysFollower
    method=SyncGroup
    sequence=/Game/Animation/Game/MainPlayer/Swimming/AssaultRifle/AdditiveTransitions/AR_Swim_GenericStop_NonTargeted_Core
   }
   Output Animation Pose(Result=@n1)
  ]
 }
 Idles {
  path=MovementMode_LowerBodyOverride.AnimGraphNode_StateMachine_8.SwimmingLowerBody.AnimStateNode_2.Idles
  nodes [
   n1 = "Transform (Modify) Bone - Bone: pelvis"(
    ComponentPose=Local To Component(LocalPose=Use cached pose 'BasePose'())
    Alpha=1
    Translation="0,0,0"
    Rotation=Get SwimmingMeleeTwistRotator()
    Scale="1,1,1"
   ) {rotationMode=BMM_Additive, rotationSpace=BCS_BoneSpace, alphaBoolBlend {blendOption=Linear}, boneToModify {boneName=pelvis}}
   n2 = "Transform (Modify) Bone - Bone: pelvis"(ComponentPose=@n1, Alpha=1, Translation="0,0,0", Rotation=Get SwimmingYawDeltaRotator() -> (SwimmingYawDeltaRotator:Rotator), Scale="1,1,1") {rotationMode=BMM_Additive, rotationSpace=BCS_WorldSpace, alphaBoolBlend {blendOption=Linear}, boneToModify {boneName=pelvis}}
   n3 = Local To Component(LocalPose=Component To Local(ComponentPose=@n2))
   n4 = Get IdlePelvisOffset() -> (IdlePelvisOffset:Vector)
   n5 = Get IdlePelvisOffsetAlpha() -> (IdlePelvisOffsetAlpha:float)
   Output Animation Pose(Result=Component To Local(ComponentPose="Transform (Modify) Bone - Bone: pelvis"(ComponentPose=@n3, Alpha=@n5, Translation=@n4, Rotation="0,0,0", Scale="1,1,1") {lODThreshold=3, translationMode=BMM_Additive, alphaBoolBlend {blendOption=Linear}, boneToModify {boneName=pelvis}}))
   Reroute Node()
  ]
 }
 SwimMovement {
  path=MovementMode_LowerBodyOverride.AnimGraphNode_StateMachine_8.SwimmingLowerBody.AnimStateNode_3.SwimMovement
  nodes [
   n1 = Get SwimmingVelocityAngle()
   n2 = Blendspace Player 'AR_Swim_NonTargeted_Core_BS'(X=@n1, Y=300, PlayRate=1) {
    bResetPlayTimeWhenBlendSpaceChanges=false
    blendSpace=&AR_Swim_NonTargeted_Core_BS
    groupName=SwimLocomotion
    groupRole=AlwaysFollower
    method=SyncGroup
   }
   n3 = Get MeleeTwistCurveValue() -> (MeleeTwistCurveValue:float)
   n4 = "Transform (Modify) Bone - Bone: pelvis"(
    ComponentPose=Local To Component(LocalPose=@n2)
    Alpha=1
    Translation="0,0,0"
    Rotation=MakeRotator(Roll=float * float(A=@n3, B=-1))
    Scale="1,1,1"
   ) {rotationMode=BMM_Additive, rotationSpace=BCS_BoneSpace, alphaBoolBlend {blendOption=Linear}, boneToModify {boneName=pelvis}}
   AimOffset Player 'AR_Swim_DirectionalZOffset_NonTargeted_AO'(X=@n1, Y=1, BasePose=Component To Local(ComponentPose=@n4), Alpha=1) {
    blendSpace=&AR_Swim_DirectionalZOffset_NonTargeted_AO
    alphaBoolBlend {blendOption=Linear}
   }
   n5 = Get SwimmingNonTargetingBlendspaceDir()
   n6 = Blendspace Player 'AR_Swim_NonTargeted_Core_BS'(X=@n5, Y=300, PlayRate=1) {
    bResetPlayTimeWhenBlendSpaceChanges=false
    blendSpace=&AR_Swim_NonTargeted_Core_BS
    groupName=SwimLocomotion
    groupRole=AlwaysFollower
    method=SyncGroup
   }
   n7 = "Transform (Modify) Bone - Bone: pelvis"(ComponentPose=Local To Component(LocalPose=@n6), Alpha=1, Translation="0,0,0", Rotation=Get SwimmingMeleeTwistRotator(), Scale="1,1,1") {rotationMode=BMM_Additive, rotationSpace=BCS_BoneSpace, alphaBoolBlend {blendOption=Linear}, boneToModify {boneName=pelvis}}
   n8 = AimOffset Player 'AR_Swim_DirectionalZOffset_NonTargeted_AO'(X=@n5, Y=1, BasePose=Component To Local(ComponentPose=@n7), Alpha=1) {
    blendSpace=&AR_Swim_DirectionalZOffset_NonTargeted_AO
    alphaBoolBlend {blendOption=Linear}
   }
   Output Animation Pose(Result=@n8)
   n9 = Get MeleeTwistCurveValue()
   Reroute Node(InputPin:Rotator=MakeRotator(Roll=float * float(A=@n9, B=-1))) -> (OutputPin:Rotator)
  ]
 }
 SwimStart {
  path=MovementMode_LowerBodyOverride.AnimGraphNode_StateMachine_8.SwimmingLowerBody.AnimStateNode_6.SwimStart
  nodes [
   n1 = Sequence Player 'AR_Swim_GenericStart_NonTargeted_Core'!(PlayRate:float=1, bLoopAnimation:bool=False) -> (Pose) {
    bLoopAnimation=false
    groupName=SwimStartSync
    groupRole=AlwaysFollower
    method=SyncGroup
    sequence=/Game/Animation/Game/MainPlayer/Swimming/AssaultRifle/AdditiveTransitions/AR_Swim_GenericStart_NonTargeted_Core
   }
   Output Animation Pose(Result=@n1)
  ]
 }
 Transition {
  path=MovementMode_LowerBodyOverride.AnimGraphNode_StateMachine_8.SwimmingLowerBody.AnimStateTransitionNode_0.Transition
  nodes [Result(bCanEnterTransition=Get bTransition_Idles_to_SwimStart())]
 }
 Transition {
  path=MovementMode_LowerBodyOverride.AnimGraphNode_StateMachine_8.SwimmingLowerBody.AnimStateTransitionNode_1.Transition
  nodes [Result(bCanEnterTransition=NOT Boolean(A=Get bIsAccelerating2D()))]
 }
 Transition {
  path=MovementMode_LowerBodyOverride.AnimGraphNode_StateMachine_8.SwimmingLowerBody.AnimStateTransitionNode_15.Transition
  nodes [Result(bCanEnterTransition=Get bSwimmingSharpAccelerationChange())]
 }
 Transition {
  path=MovementMode_LowerBodyOverride.AnimGraphNode_StateMachine_8.SwimmingLowerBody.AnimStateTransitionNode_16.Transition
  nodes [Result(bCanEnterTransition=Get bIsAccelerating2D())]
 }
 Transition {
  path=MovementMode_LowerBodyOverride.AnimGraphNode_StateMachine_8.SwimmingLowerBody.AnimStateTransitionNode_2.Transition
  nodes [Result(bCanEnterTransition=Get bTransition_SwimStart_to_SwimMovementAccel() -> (bTransition_SwimStart_to_SwimMovementAccel:bool))]
 }
 Transition {
  path=MovementMode_LowerBodyOverride.AnimGraphNode_StateMachine_8.SwimmingLowerBody.AnimStateTransitionNode_4.Transition
  nodes [Result(bCanEnterTransition=Get ChangeMovementToVelocityFromAccel())]
 }
 Transition {
  path=MovementMode_LowerBodyOverride.AnimGraphNode_StateMachine_8.SwimmingLowerBody.AnimStateTransitionNode_5.Transition
  nodes [Result(bCanEnterTransition=Get bSwimmingSharpAccelerationChange())]
 }
 Transition {
  path=MovementMode_LowerBodyOverride.AnimGraphNode_StateMachine_8.SwimmingLowerBody.AnimStateTransitionNode_6.Transition
  nodes [Result(bCanEnterTransition=NOT Boolean(A=Get bIsAccelerating2D()))]
 }
 Transition {path=MovementMode_LowerBodyOverride.AnimGraphNode_StateMachine_8.SwimmingLowerBody.AnimStateTransitionNode_7.Transition, nodes [Result()]}
 Transition {
  path=MovementMode_LowerBodyOverride.AnimGraphNode_StateMachine_8.SwimmingLowerBody.AnimStateTransitionNode_9.Transition
  nodes [Result(bCanEnterTransition=Get bIsWaterJump())]
 }
 MovementMode_OffScreen {nodes [
   n1 = Profile Blend!(
    BasePose:Pose Link Structure=Cached Sequence Evaluator!(Sequence:Anim Sequence Base Object Reference, ExplicitTime:float=0) -> (Pose:Pose Link Structure) {is=title FortAnimGraphNode_CachedSequenceEvaluator, bShouldLoop=false, cacheName=SourcePose}
    BlendPose:Pose Link Structure=Apply Additive(
     Base=Sequence Player 'AR_Swim_N_NonTargeted_Core'!(PlayRate:float=1) -> (Pose) {sequence=/Game/Animation/Game/MainPlayer/Swimming/AssaultRifle/DirectionalMovement/AR_Swim_N_NonTargeted_Core}
     Additive="Sequence Player (additive) 'AR_Swim_N_NonTargeted'"!(PlayRate:float=1) -> (Pose) {sequence=/Game/Animation/Game/MainPlayer/Swimming/AssaultRifle/DirectionalMovement/AR_Swim_N_NonTargeted}
    )
    BlendWeight:float=1
   ) -> (Pose:Pose Link Structure) {
    is=title AnimGraphNode_BlendProfileLayeredBlend
    blendProfileAsset=/Game/Characters/Player/BlendMasks/LowerBody_BlendMask
    blendPose {linkId=16842751}
   }
   n2 = Sequence Player 'Swim_Sprint'!(PlayRate:float=1) -> (Pose) {sequence=/Game/Animation/Game/MainPlayer/Swimming/Sprint/Swim_Sprint}
   n3 = Get bIsSprinting()
   Output Pose(Result=Two Way Blend(
     A=Input Pose() {name=InPoseOffscreen}
     B=Blend Poses by bool(BlendPose_0=@n2, BlendPose_1=@n1, bActiveValue=@n3) {transitionType=Inertialization}
     Alpha=Get SwimmingLayerAlpha()
    )) {layerGroup=MovementMode}
  ]}
 MovementMode_SourcePose {nodes [
   n1 = Break Fort Anim Input Player Anim Asset(FortAnimInput_PlayerAnimAsset=Get CurrentAnimSet()) -> (
    bPlayUpperBodySlotOnFullBodyInAir=~
    bDisableFullBodyAimOffsetDuringMelee=~
    bShouldApplyAimOffsetFullBody=~
    FullBodyAimOffsetAlpha=~
    UpperBodyAimOffsetAlpha=~
    bOverrideDisableArmsHeadAdditive=~
    DisableArmsHeadAdditiveOverride=~
    bOverridePitchAndYawOffsets=~
    bOverrideHandIKRetargetingWeight=~
    HandIKRetargetingWeight=~
    bForceKneeSliding=~
   )
   n2 = Break Fort Anim Input Player Anim Asset(FortAnimInput_PlayerAnimAsset=Get CurrentAnimSet()) -> (
    bPlayUpperBodySlotOnFullBodyInAir=~
    bDisableFullBodyAimOffsetDuringMelee=~
    bShouldApplyAimOffsetFullBody=~
    FullBodyAimOffsetAlpha=~
    UpperBodyAimOffsetAlpha=~
    bOverrideDisableArmsHeadAdditive=~
    DisableArmsHeadAdditiveOverride=~
    bOverridePitchAndYawOffsets=~
    bOverrideHandIKRetargetingWeight=~
    HandIKRetargetingWeight=~
    bForceKneeSliding=~
   )
   n3 = Two Way Blend(
    A=Use cached pose 'InSourcePoseBase'()
    B=Blend Poses by bool(
     BlendPose_0="Sequence Evaluator (by time)"!(Sequence:Anim Sequence Base Object Reference=@n2.SwimTargetingPose) -> (Pose:Pose Link Structure) {is=title AnimGraphNode_SequenceEvaluator}
     BlendPose_1="Sequence Evaluator (by time)"(Sequence=@n1.SwimNonTargetingPose)
     BlendTime_0=Get NonADSToADSBlendTime() -> (NonADSToADSBlendTime:float)
     BlendTime_1=Get ADSToNonADSBlendTime() -> (ADSToNonADSBlendTime:float)
     bActiveValue=Get bPlayUpperBodyTargeting() -> (bPlayUpperBodyTargeting:bool)
    ) {blendType=Linear, childUpateMode=ResetChildOnActivate}
    Alpha=Get SwimmingLayerAlpha()
   ) {blendNode {alphaBoolBlend {blendOption=Linear}}}
   n4 = Use cached pose 'InSourcePoseBase'()
   n5 = Get bIsBuilding() -> (bIsBuilding:bool)
   Output Pose(Result=Blend Poses by bool(
     BlendPose_0=Use cached pose 'InSourcePoseBase'()
     BlendPose_1=Blend Poses by bool(BlendPose_0=@n4, BlendPose_1=@n3, bActiveValue=@n5) {blendType=Linear}
     bActiveValue=Get bUseProceduralLayering() -> (bUseProceduralLayering:bool)
    ) {blendType=Linear}) {layerGroup=MovementMode}
   Save cached pose 'InSourcePoseBase'(Pose=Input Pose() {name=InSourcePoseBase}) {cacheName=InSourcePoseBase}
  ]}
 On Dive Jump Ground Land Update {nodes [
   n1 = On Dive Jump Ground Land Update!() -> (then, Context:"Anim Update Context Structure (by ref)", Node:"Anim Node Reference Structure (by ref)") {is=title K2Node_FunctionEntry}
   > n2 = Convert to Animation State!(execute, Node:"Anim Node Reference Structure (by ref)"=@n1.Node) -> (Succeeded:Exec, Failed:Exec, AnimationState:Animation State Reference Structure)
   > SetDiveGroundLandingRemainingTime!(
    execute
    self:Fort Swimming Layer Anim Instance Object Reference
    RemainingAnimTime:float=GetRelevantAnimTimeRemaining!(UpdateContext:"Anim Update Context Structure (by ref)"=@n1, Node:"Animation State Reference Structure (by ref)"=@n2) -> (ReturnValue:float)
   ) -> (then)
   > Return Node()
  ]}
]
variables [
 AllowSprintStop {type=boolean}
 ChangeMovementToVelocityFromAccel {type=boolean}
 LocalAccelYawAngleAtStart {type=number}
 PivotAngle {type=number}
]
