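# Animation Blueprint for the NyxGlass Gauntlet, ODM-style gear that follows the attached parent's pose through a Control Rig.
# An additive layer switches between a linked swing AnimBP and the conduit-hub ODMLocomotion state machine covering ground, air, swimming, riding, ziplines and grind rails.
# After inertialization, bone offsets adjust belts, hooks, base and back, then two RigidBody nodes add gauntlet and canister physics.
# Emotes scale odm_base to zero, hiding the gear.
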AnimBlueprint /DryBox/Items/NyxGlass/Meshes/NyxGlass_Gauntlet_AnimBP
target_skeleton=&/NyxGlass_Gauntlet_Skeleton
graphs [
 AnimGraph {nodes [
   n1 = NyxGlass_Gaunlet_Swing_AnimBP!() -> (Pose:Pose Link Structure) {
    is=title AnimGraphNode_LinkedAnimGraph
    bPropagateNotifiesToLinkedInstances=true
    bReceiveNotifiesFromLinkedInstances=true
    instanceClass=&/NyxGlass_Gaunlet_Swing_AnimBP.NyxGlass_Gaunlet_Swing_AnimBP_C
   }
   AnimGraphNode_StateMachine_0 = ODMLocomotion() {is=title AnimGraphNode_StateMachine, bAllowConduitEntryStates=true}
   n2 = Control Rig(Source=Copy Pose From Mesh() {bUseAttachedParent=true}) {
    alphaBoolBlend {blendOption=Linear}
    controlRigAssetReference {blueprintRigClass=&/NyxGlass_Gauntlet_CtrlRig.NyxGlass_Gauntlet_CtrlRig_C}
   }
   n3 = Apply Additive(
    Base=@n2
    Additive=Blend Poses by bool(
     BlendPose_0=@n1
     BlendPose_1=@AnimGraphNode_StateMachine_0
     BlendTime_0=0.2
     BlendTime_1=0.2
     bActiveValue=Get IsNyxGlass_Swing() -> (IsNyxGlass_Swing:bool)
    ) {blendType=Linear}
   ) {alphaBoolBlend {blendOption=Linear}}
   n4 = Get belt_adjustment_r() -> (belt_adjustment_r:Rotator)
   n5 = "Transform (Modify) Bone - Bone: belt_01_l"(
    ComponentPose=Local To Component(LocalPose=Inertialization!(Source:Pose Link Structure=@n3) -> (Pose:Pose Link Structure) {is=title AnimGraphNode_Inertialization})
    Alpha=1
    Translation="0,0,0"
    Rotation=@n4
    Scale="1,1,1"
   ) {
    rotationMode=BMM_Additive
    rotationSpace=BCS_BoneSpace
    alphaBoolBlend {blendOption=Linear}
    boneToModify {boneName=belt_01_l}
    rotation {yaw=5}
   }
   n6 = "Transform (Modify) Bone - Bone: hook_l"(ComponentPose=@n5, Alpha=1, Translation="0,0,0", Rotation=@n4, Scale="1,1,1") {
    rotationMode=BMM_Additive
    rotationSpace=BCS_BoneSpace
    alphaBoolBlend {blendOption=Linear}
    boneToModify {boneName=hook_l}
    rotation {yaw=5}
   }
   n7 = Get belt_adjustment_l() -> (belt_adjustment_l:Rotator)
   n8 = "Transform (Modify) Bone - Bone: belt_01_r"(ComponentPose=@n6, Alpha=1, Translation="0,0,0", Rotation=@n7, Scale="1,1,1") {
    rotationMode=BMM_Additive
    rotationSpace=BCS_BoneSpace
    alphaBoolBlend {blendOption=Linear}
    boneToModify {boneName=belt_01_r}
    rotation {yaw=-15}
   }
   n9 = "Transform (Modify) Bone - Bone: hook_r"(ComponentPose=@n8, Alpha=1, Translation="0,0,0", Rotation=@n7, Scale="1,1,1") {
    rotationMode=BMM_Additive
    rotationSpace=BCS_BoneSpace
    alphaBoolBlend {blendOption=Linear}
    boneToModify {boneName=hook_r}
    rotation {yaw=-15}
   }
   n10 = "Transform (Modify) Bone - Bone: odm_base"(ComponentPose=@n9, Alpha=1, Translation=Get nyx_position_adjustment() -> (nyx_position_adjustment:Vector), Rotation="0,0,0", Scale="1,1,1") {
    rotationMode=BMM_Additive
    rotationSpace=BCS_BoneSpace
    scaleSpace=BCS_BoneSpace
    translationMode=BMM_Additive
    translationSpace=BCS_BoneSpace
    alphaBoolBlend {blendOption=Linear}
    boneToModify {boneName=odm_base}
    translation {y=-5}
   }
   n11 = "Transform (Modify) Bone - Bone: odm_base"!(ComponentPose=@n10, bAlphaBoolEnabled:bool=Get bIsTargeting(), Translation="0,3,2", Rotation="-5,0,-30", Scale="1,1,1") -> (Pose) {
    alphaInputType=Bool
    rotationMode=BMM_Additive
    rotationSpace=BCS_BoneSpace
    scaleSpace=BCS_BoneSpace
    translationMode=BMM_Additive
    translationSpace=BCS_BoneSpace
    alphaBoolBlend {blendInTime=0.05, blendOption=Cubic, blendOutTime=0.25}
    boneToModify {boneName=odm_base}
    rotation {pitch=-5, roll=-30}
    translation {y=3, z=2}
   }
   n12 = "Transform (Modify) Bone - Bone: back"(ComponentPose=@n11, Alpha=0.5, Translation="0,0,0", Rotation=Get spine_05_twist() -> (spine_05_twist:Rotator), Scale="1,1,1") {
    rotationMode=BMM_Additive
    rotationSpace=BCS_BoneSpace
    scaleSpace=BCS_BoneSpace
    translationMode=BMM_Additive
    translationSpace=BCS_BoneSpace
    alphaBoolBlend {blendInTime=0.05, blendOption=Cubic, blendOutTime=0.25}
    boneToModify {boneName=back}
   }
   n13 = RigidBody(ComponentPose=@n12) {
    bOverrideWorldGravity=true
    overridePhysicsAsset=&/NyxGlass_Gauntlet_Physics
    simulationSpace=BaseBoneSpace
    alphaBoolBlend {blendOption=Linear}
    baseBoneRef {boneName=pelvis}
    componentAppliedLinearAccClamp=<2000 2000 2000>
    componentLinearAccScale=<0.4 0.4 0.4>
    componentLinearVelScale=<0.4 0.4 0.4>
    overrideWorldGravity {z=-980}
   }
   n14 = RigidBody(ComponentPose=@n13, Alpha=Get DynamicAlpha() -> (DynamicAlpha:float)) {
    bOverrideWorldGravity=true
    overridePhysicsAsset=&/NyxGlass_Cannister_Physics
    simulationSpace=BaseBoneSpace
    alphaBoolBlend {blendOption=Linear}
    baseBoneRef {boneName=pelvis}
    componentAppliedLinearAccClamp=<2000 2000 2000>
    componentLinearAccScale=<0.3 0.3 0.3>
    componentLinearVelScale=<0.3 0.3 0.3>
    overrideWorldGravity {z=-980}
   }
   n15 = "Transform (Modify) Bone - Bone: odm_base"(ComponentPose=@n14, bAlphaBoolEnabled=Get bIsPlayingEmote(), Translation="0,0,0", Rotation="0,0,0", Scale="0,0,0") {
    alphaInputType=Bool
    rotationMode=BMM_Additive
    rotationSpace=BCS_BoneSpace
    scaleMode=BMM_Replace
    scaleSpace=BCS_BoneSpace
    translationMode=BMM_Additive
    translationSpace=BCS_BoneSpace
    alphaBoolBlend {blendInTime=0.05, blendOption=Cubic, blendOutTime=0.1}
    boneToModify {boneName=odm_base}
    scale=<0 0 0>
   }
   Output Pose(Result=Component To Local(ComponentPose=@n15)) {layerGroup=None}
  ]}
 ODMLocomotion {
  entry=Conduit
  path=AnimGraph.AnimGraphNode_StateMachine_0.ODMLocomotion
  ids=id
  nodes [
   GroundLand() {is=title AnimStateAliasNode, also {aliases [JumpLand]}}
   GrindrailLand() {is=title AnimStateAliasNode, also {aliases [GrindLand]}}
   ToSkiDive() {is=title AnimStateAliasNode, also {aliases [Falling,FreeFalling,Jump,JumpLand]}}
   AnimStateConduitNode_0 = Conduit()
   AnimStateNode_0 = Idle() {
    stateEntered {guid=#1}
    stateFullyBlended {guid=#2}
    stateLeft {guid=#3}
   }
   AnimStateNode_1 = Locomotion() {is=title AnimStateNode}
   AnimStateNode_10 = Riding() {is=title AnimStateNode}
   AnimStateNode_11 = FreeFalling() {
    is=title AnimStateNode
    stateEntered {guid=#4}
    stateFullyBlended {guid=#5}
    stateLeft {guid=#6}
   }
   AnimStateNode_12 = ZiplinePivot() {is=title AnimStateNode}
   AnimStateNode_13 = DiveJumping() {
    is=title AnimStateNode
    stateEntered {guid=#7}
    stateFullyBlended {guid=#8}
    stateLeft {guid=#9}
   }
   AnimStateNode_14 = Grindrail() {is=title AnimStateNode}
   AnimStateNode_15 = GrindJump() {
    is=title AnimStateNode
    stateEntered {guid=#10}
    stateFullyBlended {guid=#11}
    stateLeft {guid=#12}
   }
   AnimStateNode_16 = GrindLand() {
    is=title AnimStateNode
    stateEntered {guid=#13}
    stateFullyBlended {guid=#14}
    stateLeft {guid=#15}
   }
   AnimStateNode_17 = GrindFallLoop() {
    is=title AnimStateNode
    stateEntered {guid=#16}
    stateFullyBlended {guid=#17}
    stateLeft {guid=#18}
   }
   AnimStateNode_18 = Equip() {is=title AnimStateNode}
   AnimStateNode_19 = GrindSprintBoost() {is=title AnimStateNode}
   AnimStateNode_2 = Sliding() {
    is=title AnimStateNode
    stateEntered {guid=#19}
    stateFullyBlended {guid=#20}
    stateLeft {guid=#21}
   }
   AnimStateNode_20 = IsSkyDive() {is=title AnimStateNode}
   AnimStateNode_3 = Jump() {
    is=title AnimStateNode
    stateEntered {guid=#22}
    stateFullyBlended {guid=#23}
    stateLeft {guid=#24}
   }
   AnimStateNode_4 = JumpLand() {
    is=title AnimStateNode
    stateEntered {guid=#25}
    stateFullyBlended {guid=#26}
    stateLeft {guid=#27}
   }
   AnimStateNode_5 = Swimming() {
    is=title AnimStateNode
    stateEntered {guid=#28}
    stateFullyBlended {guid=#29}
    stateLeft {guid=#30}
   }
   AnimStateNode_6 = Falling() {
    is=title AnimStateNode
    stateEntered {guid=#31}
    stateFullyBlended {guid=#32}
    stateLeft {guid=#33}
   }
   AnimStateNode_7 = Crouch() {
    is=title AnimStateNode
    stateEntered {guid=#34}
    stateFullyBlended {guid=#35}
    stateLeft {guid=#36}
   }
   AnimStateNode_8 = ZiplineEntry() {is=title AnimStateNode}
   AnimStateNode_9 = Ziplining() {is=title AnimStateNode}
   AnimStateTransitionNode_0 = AnimStateTransitionNode() {crossfadeDuration=0.4, logicType=TLT_Inertialization, priorityOrder=100}
   AnimStateTransitionNode_1 = AnimStateTransitionNode() {logicType=TLT_Inertialization}
   AnimStateTransitionNode_10 = AnimStateTransitionNode() {logicType=TLT_Inertialization, priorityOrder=4}
   AnimStateTransitionNode_11 = AnimStateTransitionNode() {logicType=TLT_Inertialization, priorityOrder=3}
   AnimStateTransitionNode_12 = AnimStateTransitionNode()
   AnimStateTransitionNode_13 = AnimStateTransitionNode() {logicType=TLT_Inertialization, priorityOrder=5}
   AnimStateTransitionNode_14 = AnimStateTransitionNode()
   AnimStateTransitionNode_15 = AnimStateTransitionNode()
   AnimStateTransitionNode_16 = AnimStateTransitionNode() {crossfadeDuration=0.5}
   AnimStateTransitionNode_17 = AnimStateTransitionNode() {crossfadeDuration=0.5}
   AnimStateTransitionNode_18 = AnimStateTransitionNode() {logicType=TLT_Inertialization, priorityOrder=6}
   AnimStateTransitionNode_19 = AnimStateTransitionNode() {priorityOrder=8}
   AnimStateTransitionNode_2 = AnimStateTransitionNode()
   AnimStateTransitionNode_20 = AnimStateTransitionNode()
   AnimStateTransitionNode_21 = AnimStateTransitionNode()
   AnimStateTransitionNode_22 = AnimStateTransitionNode() {logicType=TLT_Inertialization, priorityOrder=7}
   AnimStateTransitionNode_23 = AnimStateTransitionNode()
   AnimStateTransitionNode_24 = AnimStateTransitionNode() {bAutomaticRuleBasedOnSequencePlayerInState=true, crossfadeDuration=0.4}
   AnimStateTransitionNode_25 = AnimStateTransitionNode()
   AnimStateTransitionNode_26 = AnimStateTransitionNode() {
    transitionEnd {guid=#37}
    transitionInterrupt {guid=#38}
    transitionStart {guid=#39}
   }
   AnimStateTransitionNode_27 = AnimStateTransitionNode()
   AnimStateTransitionNode_28 = AnimStateTransitionNode()
   AnimStateTransitionNode_29 = AnimStateTransitionNode() {priorityOrder=9}
   AnimStateTransitionNode_3 = AnimStateTransitionNode()
   AnimStateTransitionNode_30 = AnimStateTransitionNode()
   AnimStateTransitionNode_31 = AnimStateTransitionNode() {crossfadeDuration=0.033, logicType=TLT_Inertialization}
   AnimStateTransitionNode_32 = AnimStateTransitionNode() {crossfadeDuration=0.1, logicType=TLT_Inertialization}
   AnimStateTransitionNode_33 = AnimStateTransitionNode() {crossfadeDuration=0.033, logicType=TLT_Inertialization, priorityOrder=2}
   AnimStateTransitionNode_34 = AnimStateTransitionNode() {bAutomaticRuleBasedOnSequencePlayerInState=true, crossfadeDuration=0.1}
   AnimStateTransitionNode_35 = AnimStateTransitionNode() {bAutomaticRuleBasedOnSequencePlayerInState=true}
   AnimStateTransitionNode_36 = AnimStateTransitionNode()
   AnimStateTransitionNode_37 = AnimStateTransitionNode()
   AnimStateTransitionNode_38 = AnimStateTransitionNode()
   AnimStateTransitionNode_39 = AnimStateTransitionNode() {
    crossfadeDuration=0.1
    priorityOrder=2
    transitionEnd {guid=#40}
    transitionInterrupt {guid=#41}
    transitionStart {guid=#42}
   }
   AnimStateTransitionNode_4 = AnimStateTransitionNode() {crossfadeDuration=0.5, logicType=TLT_Inertialization, priorityOrder=2}
   AnimStateTransitionNode_40 = AnimStateTransitionNode() {bAutomaticRuleBasedOnSequencePlayerInState=true}
   AnimStateTransitionNode_41 = AnimStateTransitionNode() {crossfadeDuration=0.033}
   AnimStateTransitionNode_42 = AnimStateTransitionNode()
   AnimStateTransitionNode_43 = AnimStateTransitionNode()
   AnimStateTransitionNode_44 = AnimStateTransitionNode() {
    crossfadeDuration=0.1
    priorityOrder=2
    transitionEnd {guid=#43}
    transitionInterrupt {guid=#44}
    transitionStart {guid=#45}
   }
   AnimStateTransitionNode_45 = AnimStateTransitionNode() {logicType=TLT_Inertialization}
   AnimStateTransitionNode_46 = AnimStateTransitionNode()
   AnimStateTransitionNode_47 = AnimStateTransitionNode() {logicType=TLT_Inertialization}
   AnimStateTransitionNode_5 = AnimStateTransitionNode()
   AnimStateTransitionNode_6 = AnimStateTransitionNode() {logicType=TLT_Inertialization}
   AnimStateTransitionNode_7 = AnimStateTransitionNode() {logicType=TLT_Inertialization}
   AnimStateTransitionNode_8 = AnimStateTransitionNode() {bAutomaticRuleBasedOnSequencePlayerInState=true}
   AnimStateTransitionNode_9 = AnimStateTransitionNode() {crossfadeDuration=0.4, logicType=TLT_Inertialization}
  ]
  links [
   {
    dst=Crouch
    rule=AnimGraph.AnimGraphNode_StateMachine_0.ODMLocomotion.AnimStateTransitionNode_13.Transition
    src=Conduit
    via=AnimStateTransitionNode_13
   }
   {
    dst=Equip
    rule=AnimGraph.AnimGraphNode_StateMachine_0.ODMLocomotion.AnimStateTransitionNode_32.Transition
    src=Conduit
    via=AnimStateTransitionNode_32
   }
   {
    dst=Grindrail
    rule=AnimGraph.AnimGraphNode_StateMachine_0.ODMLocomotion.AnimStateTransitionNode_29.Transition
    src=Conduit
    via=AnimStateTransitionNode_29
   }
   {
    dst=Idle
    rule=AnimGraph.AnimGraphNode_StateMachine_0.ODMLocomotion.AnimStateTransitionNode_0.Transition
    src=Conduit
    via=AnimStateTransitionNode_0
   }
   {
    dst=IsSkyDive
    rule=AnimGraph.AnimGraphNode_StateMachine_0.ODMLocomotion.AnimStateTransitionNode_45.Transition
    src=Conduit
    via=AnimStateTransitionNode_45
   }
   {
    dst=Jump
    rule=AnimGraph.AnimGraphNode_StateMachine_0.ODMLocomotion.AnimStateTransitionNode_10.Transition
    src=Conduit
    via=AnimStateTransitionNode_10
   }
   {
    dst=Locomotion
    rule=AnimGraph.AnimGraphNode_StateMachine_0.ODMLocomotion.AnimStateTransitionNode_9.Transition
    src=Conduit
    via=AnimStateTransitionNode_9
   }
   {
    dst=Riding
    rule=AnimGraph.AnimGraphNode_StateMachine_0.ODMLocomotion.AnimStateTransitionNode_19.Transition
    src=Conduit
    via=AnimStateTransitionNode_19
   }
   {
    dst=Sliding
    rule=AnimGraph.AnimGraphNode_StateMachine_0.ODMLocomotion.AnimStateTransitionNode_4.Transition
    src=Conduit
    via=AnimStateTransitionNode_4
   }
   {
    dst=Swimming
    rule=AnimGraph.AnimGraphNode_StateMachine_0.ODMLocomotion.AnimStateTransitionNode_11.Transition
    src=Conduit
    via=AnimStateTransitionNode_11
   }
   {
    dst=ZiplineEntry
    rule=AnimGraph.AnimGraphNode_StateMachine_0.ODMLocomotion.AnimStateTransitionNode_18.Transition
    src=Conduit
    via=AnimStateTransitionNode_18
   }
   {
    dst=Ziplining
    rule=AnimGraph.AnimGraphNode_StateMachine_0.ODMLocomotion.AnimStateTransitionNode_22.Transition
    src=Conduit
    via=AnimStateTransitionNode_22
   }
   {
    dst=Conduit
    rule=AnimGraph.AnimGraphNode_StateMachine_0.ODMLocomotion.AnimStateTransitionNode_14.Transition
    src=Crouch
    via=AnimStateTransitionNode_14
   }
   {
    dst=Swimming
    rule=AnimGraph.AnimGraphNode_StateMachine_0.ODMLocomotion.AnimStateTransitionNode_27.Transition
    src=DiveJumping
    via=AnimStateTransitionNode_27
   }
   {
    dst=Conduit
    rule=AnimGraph.AnimGraphNode_StateMachine_0.ODMLocomotion.AnimStateTransitionNode_35.Transition
    src=Equip
    via=AnimStateTransitionNode_35
   }
   {
    dst=Conduit
    rule=AnimGraph.AnimGraphNode_StateMachine_0.ODMLocomotion.AnimStateTransitionNode_38.Transition
    src=Equip
    via=AnimStateTransitionNode_38
   }
   {
    dst=GrindrailLand
    rule=AnimGraph.AnimGraphNode_StateMachine_0.ODMLocomotion.AnimStateTransitionNode_43.Transition
    src=Falling
    via=AnimStateTransitionNode_43
   }
   {
    dst=JumpLand
    rule=AnimGraph.AnimGraphNode_StateMachine_0.ODMLocomotion.AnimStateTransitionNode_7.Transition
    src=Falling
    via=AnimStateTransitionNode_7
   }
   {
    dst=GrindrailLand
    rule=AnimGraph.AnimGraphNode_StateMachine_0.ODMLocomotion.AnimStateTransitionNode_42.Transition
    src=FreeFalling
    via=AnimStateTransitionNode_42
   }
   {
    dst=JumpLand
    rule=AnimGraph.AnimGraphNode_StateMachine_0.ODMLocomotion.AnimStateTransitionNode_28.Transition
    src=FreeFalling
    via=AnimStateTransitionNode_28
   }
   {
    dst=GrindLand
    rule=AnimGraph.AnimGraphNode_StateMachine_0.ODMLocomotion.AnimStateTransitionNode_41.Transition
    src=GrindFallLoop
    via=AnimStateTransitionNode_41
   }
   {
    dst=GroundLand
    rule=AnimGraph.AnimGraphNode_StateMachine_0.ODMLocomotion.AnimStateTransitionNode_44.Transition
    src=GrindFallLoop
    via=AnimStateTransitionNode_44
   }
   {
    dst=GrindFallLoop
    rule=AnimGraph.AnimGraphNode_StateMachine_0.ODMLocomotion.AnimStateTransitionNode_40.Transition
    src=GrindJump
    via=AnimStateTransitionNode_40
   }
   {
    dst=GrindLand
    rule=AnimGraph.AnimGraphNode_StateMachine_0.ODMLocomotion.AnimStateTransitionNode_33.Transition
    src=GrindJump
    via=AnimStateTransitionNode_33
   }
   {
    dst=GroundLand
    rule=AnimGraph.AnimGraphNode_StateMachine_0.ODMLocomotion.AnimStateTransitionNode_39.Transition
    src=GrindJump
    via=AnimStateTransitionNode_39
   }
   {
    dst=Grindrail
    rule=AnimGraph.AnimGraphNode_StateMachine_0.ODMLocomotion.AnimStateTransitionNode_34.Transition
    src=GrindLand
    via=AnimStateTransitionNode_34
   }
   {
    dst=Grindrail
    rule=AnimGraph.AnimGraphNode_StateMachine_0.ODMLocomotion.AnimStateTransitionNode_37.Transition
    src=GrindSprintBoost
    via=AnimStateTransitionNode_37
   }
   {
    dst=Conduit
    rule=AnimGraph.AnimGraphNode_StateMachine_0.ODMLocomotion.AnimStateTransitionNode_30.Transition
    src=Grindrail
    via=AnimStateTransitionNode_30
   }
   {
    dst=GrindJump
    rule=AnimGraph.AnimGraphNode_StateMachine_0.ODMLocomotion.AnimStateTransitionNode_31.Transition
    src=Grindrail
    via=AnimStateTransitionNode_31
   }
   {
    dst=GrindSprintBoost
    rule=AnimGraph.AnimGraphNode_StateMachine_0.ODMLocomotion.AnimStateTransitionNode_36.Transition
    src=Grindrail
    via=AnimStateTransitionNode_36
   }
   {
    dst=Conduit
    rule=AnimGraph.AnimGraphNode_StateMachine_0.ODMLocomotion.AnimStateTransitionNode_3.Transition
    src=Idle
    via=AnimStateTransitionNode_3
   }
   {
    dst=Conduit
    rule=AnimGraph.AnimGraphNode_StateMachine_0.ODMLocomotion.AnimStateTransitionNode_46.Transition
    src=IsSkyDive
    via=AnimStateTransitionNode_46
   }
   {
    dst=Falling
    rule=AnimGraph.AnimGraphNode_StateMachine_0.ODMLocomotion.AnimStateTransitionNode_6.Transition
    src=Jump
    via=AnimStateTransitionNode_6
   }
   {
    dst=FreeFalling
    rule=AnimGraph.AnimGraphNode_StateMachine_0.ODMLocomotion.AnimStateTransitionNode_25.Transition
    src=Jump
    via=AnimStateTransitionNode_25
   }
   {
    dst=JumpLand
    rule=AnimGraph.AnimGraphNode_StateMachine_0.ODMLocomotion.AnimStateTransitionNode_1.Transition
    src=Jump
    via=AnimStateTransitionNode_1
   }
   {
    dst=Conduit
    rule=AnimGraph.AnimGraphNode_StateMachine_0.ODMLocomotion.AnimStateTransitionNode_8.Transition
    src=JumpLand
    via=AnimStateTransitionNode_8
   }
   {
    dst=Locomotion
    rule=AnimGraph.AnimGraphNode_StateMachine_0.ODMLocomotion.AnimStateTransitionNode_17.Transition
    src=JumpLand
    via=AnimStateTransitionNode_17
   }
   {
    dst=Conduit
    rule=AnimGraph.AnimGraphNode_StateMachine_0.ODMLocomotion.AnimStateTransitionNode_2.Transition
    src=Locomotion
    via=AnimStateTransitionNode_2
   }
   {
    dst=Conduit
    rule=AnimGraph.AnimGraphNode_StateMachine_0.ODMLocomotion.AnimStateTransitionNode_20.Transition
    src=Riding
    via=AnimStateTransitionNode_20
   }
   {
    dst=Conduit
    rule=AnimGraph.AnimGraphNode_StateMachine_0.ODMLocomotion.AnimStateTransitionNode_5.Transition
    src=Sliding
    via=AnimStateTransitionNode_5
   }
   {
    dst=Conduit
    rule=AnimGraph.AnimGraphNode_StateMachine_0.ODMLocomotion.AnimStateTransitionNode_12.Transition
    src=Swimming
    via=AnimStateTransitionNode_12
   }
   {
    dst=DiveJumping
    rule=AnimGraph.AnimGraphNode_StateMachine_0.ODMLocomotion.AnimStateTransitionNode_26.Transition
    src=Swimming
    via=AnimStateTransitionNode_26
   }
   {
    dst=IsSkyDive
    rule=AnimGraph.AnimGraphNode_StateMachine_0.ODMLocomotion.AnimStateTransitionNode_47.Transition
    src=ToSkiDive
    via=AnimStateTransitionNode_47
   }
   {
    dst=Ziplining
    rule=AnimGraph.AnimGraphNode_StateMachine_0.ODMLocomotion.AnimStateTransitionNode_16.Transition
    src=ZiplineEntry
    via=AnimStateTransitionNode_16
   }
   {
    dst=Ziplining
    rule=AnimGraph.AnimGraphNode_StateMachine_0.ODMLocomotion.AnimStateTransitionNode_23.Transition
    src=ZiplinePivot
    via=AnimStateTransitionNode_23
   }
   {
    dst=Ziplining
    rule=AnimGraph.AnimGraphNode_StateMachine_0.ODMLocomotion.AnimStateTransitionNode_24.Transition
    src=ZiplinePivot
    via=AnimStateTransitionNode_24
   }
   {
    dst=Conduit
    rule=AnimGraph.AnimGraphNode_StateMachine_0.ODMLocomotion.AnimStateTransitionNode_15.Transition
    src=Ziplining
    via=AnimStateTransitionNode_15
   }
   {
    dst=ZiplinePivot
    rule=AnimGraph.AnimGraphNode_StateMachine_0.ODMLocomotion.AnimStateTransitionNode_21.Transition
    src=Ziplining
    via=AnimStateTransitionNode_21
   }
  ]
 }
 Conduit {path=AnimGraph.AnimGraphNode_StateMachine_0.ODMLocomotion.AnimStateConduitNode_0.Conduit, nodes [Result(bCanEnterTransition=true)]}
 Idle {
  path=AnimGraph.AnimGraphNode_StateMachine_0.ODMLocomotion.AnimStateNode_0.Idle
  nodes [Output Animation Pose(Result="Sequence Player (additive) 'NyxGlass_Pose_CMM_WEP'"() {sequence=/DryBox/Items/NyxGlass/Animation/Gauntlet/NyxGlass_Pose_CMM_WEP})]
 }
 Locomotion {
  path=AnimGraph.AnimGraphNode_StateMachine_0.ODMLocomotion.AnimStateNode_1.Locomotion
  nodes [
   Modify Curve(CurveValues_0=0.5, Alpha=1) {curveNames [DynamicAlpha], curveValues [0]}
   n1 = Get PawnSpeed2D()
   n2 = Get PawnMovementDirectionAzimuth() -> (PawnMovementDirectionAzimuth:float)
   Output Animation Pose(Result=Blend Poses by bool(
     BlendPose_0="Sequence Player (additive) 'NyxGlass_Sprint_WEP'"() {sequence=/DryBox/Items/NyxGlass/Animation/Gauntlet/NyxGlass_Sprint_WEP}
     BlendPose_1=Blendspace Player 'NyxGlass_Loco_BS'(X=@n2, Y=@n1) {blendSpace=/DryBox/Items/NyxGlass/Animation/Gauntlet/NyxGlass_Loco_BS}
     BlendTime_0=0.25
     BlendTime_1=0.25
     bActiveValue="float > float"(A=@n1, B=400)
    ) {blendType=Linear})
  ]
 }
 Riding {
  path=AnimGraph.AnimGraphNode_StateMachine_0.ODMLocomotion.AnimStateNode_10.Riding
  nodes [
   n1 = Get RidingSpeed2D() -> (RidingSpeed2D:float)
   Output Animation Pose(Result=Blend Poses by bool(
     BlendPose_0="Sequence Player (additive) 'NyxGlass_HardSaddleWolf_Player_Sprint_N_WEP'"() {sequence=/DryBox/Items/NyxGlass/Animation/Gauntlet/NyxGlass_HardSaddleWolf_Player_Sprint_N_WEP}
     BlendPose_1=Blendspace Player 'NyxGlass_Hardsaddle_BS'(X=@n1, Y=0) {blendSpace=/DryBox/Items/NyxGlass/Animation/Gauntlet/NyxGlass_Hardsaddle_BS}
     bActiveValue=Get bIsTacticalSprinting()
    ) {blendType=Linear})
  ]
 }
 FreeFalling {
  path=AnimGraph.AnimGraphNode_StateMachine_0.ODMLocomotion.AnimStateNode_11.FreeFalling
  nodes [Output Animation Pose(Result="Sequence Player (additive) 'NyxGlass_FreeFalling_WEP'"() {sequence=/DryBox/Items/NyxGlass/Animation/Gauntlet/NyxGlass_FreeFalling_WEP})]
 }
 ZiplinePivot {
  path=AnimGraph.AnimGraphNode_StateMachine_0.ODMLocomotion.AnimStateNode_12.ZiplinePivot
  nodes [
   n1 = "Blend Poses (EFortCardinalDirection)"!(
    BlendPose_0:Pose Link Structure="Sequence Player (additive) 'NyxGlass_ZipLine_LowerBody_Pivot_StoN_WEP'"() {bLoopAnimation=false, sequence=/DryBox/Items/NyxGlass/Animation/Gauntlet/NyxGlass_ZipLine_LowerBody_Pivot_StoN_WEP}
    BlendPose_1:Pose Link Structure="Sequence Player (additive) 'NyxGlass_ZipLine_LowerBody_Pivot_WtoE_WEP'"() {bLoopAnimation=false, sequence=/DryBox/Items/NyxGlass/Animation/Gauntlet/NyxGlass_ZipLine_LowerBody_Pivot_WtoE_WEP}
    BlendPose_2:Pose Link Structure="Sequence Player (additive) 'NyxGlass_ZipLine_LowerBody_Pivot_NtoS_WEP'"() {bLoopAnimation=false, sequence=/DryBox/Items/NyxGlass/Animation/Gauntlet/NyxGlass_ZipLine_LowerBody_Pivot_NtoS_WEP}
    BlendPose_3:Pose Link Structure="Sequence Player (additive) 'NyxGlass_ZipLine_LowerBody_Pivot_EtoW_WEP'"() {bLoopAnimation=false, sequence=/DryBox/Items/NyxGlass/Animation/Gauntlet/NyxGlass_ZipLine_LowerBody_Pivot_EtoW_WEP}
    BlendTime_0:float=0.25
    BlendTime_1:float=0.25
    BlendTime_2:float=0.25
    BlendTime_3:float=0.25
    ActiveEnumValue:Byte=Get Pivot Cardinal Direction() -> (Pivot Cardinal Direction:EFortCardinalDirection Enum=North)
   ) -> (Pose:Pose Link Structure) {
    is=title AnimGraphNode_BlendListByEnum
    blendType=Linear
    blendPose [{linkId=-1, sourceLinkId=-1}, {linkId=-1, sourceLinkId=-1}, {linkId=-1, sourceLinkId=-1}, {linkId=-1, sourceLinkId=-1}]
    blendTime [0.1,0.1,0.1,0.1]
   }
   Output Animation Pose(Result=@n1)
  ]
 }
 DiveJumping {
  path=AnimGraph.AnimGraphNode_StateMachine_0.ODMLocomotion.AnimStateNode_13.DiveJumping
  nodes [
   n1 = "Sequence Player (additive) 'NyxGlass_Swim_DiveJump_Up_WEP'"() {sequence=/DryBox/Items/NyxGlass/Animation/Gauntlet/NyxGlass_Swim_DiveJump_Up_WEP}
   n2 = Get Swimming Dive Pitch() -> (Swimming Dive Pitch:float)
   Output Animation Pose(Result=Blend Poses by bool(
     BlendPose_0=AimOffset Player 'NyxGlass_AO_Swim_DiveJump_Pitch'!(X:float=0, Y:float=@n2, BasePose:Pose Link Structure=@n1, Alpha:float=1) -> (Pose:Pose Link Structure) {blendSpace=/DryBox/Items/NyxGlass/Animation/Gauntlet/NyxGlass_AO_Swim_DiveJump_Pitch, alphaBoolBlend {blendOption=Linear}}
     BlendPose_1="Sequence Player (additive) 'NyxGlass_Jump_CMM_WEP'"() {bLoopAnimation=false, sequence=/DryBox/Items/NyxGlass/Animation/Gauntlet/NyxGlass_Jump_CMM_WEP}
     bActiveValue=Get Is Dive Jumping() -> (Is Dive Jumping:bool)
    ) {blendType=Linear})
  ]
 }
 Grindrail {
  path=AnimGraph.AnimGraphNode_StateMachine_0.ODMLocomotion.AnimStateNode_14.Grindrail
  nodes [
   n1 = Blendspace Player 'NyxGlass_Grindrail_Leans_BS'!(X=Get Grind Lean Direction() -> (Grind Lean Direction:float)) -> (Pose) {blendSpace=/DryBox/Items/NyxGlass/Animation/Gauntlet/NyxGlass_Grindrail_Leans_BS}
   n2 = Blendspace Player 'NyxGlass_Grindrail_SpeedNoise_BS'!(Y=Get Grind Speed() -> (Grind Speed:double), PlayRate:float=1) -> (Pose) {blendSpace=/DryBox/Items/NyxGlass/Animation/Gauntlet/NyxGlass_Grindrail_SpeedNoise_BS}
   Output Animation Pose(Result=Blend Poses by bool(
     BlendPose_0="Sequence Player (additive) 'NyxGlass_GrindRail_Lean_Backward_WEP'"() {bLoopAnimation=false, sequence=/DryBox/Items/NyxGlass/Animation/Gauntlet/NyxGlass_GrindRail_Lean_Backward_WEP}
     BlendPose_1=Apply Additive(Base=@n2, Additive=@n1) {alphaBoolBlend {blendOption=Linear}}
     BlendTime_0=0.5
     BlendTime_1=0.2
     bActiveValue=Get Is Lean Bwd() -> (Is Lean Bwd:bool)
    ) {blendType=Linear})
  ]
 }
 GrindJump {
  path=AnimGraph.AnimGraphNode_StateMachine_0.ODMLocomotion.AnimStateNode_15.GrindJump
  nodes [Output Animation Pose(Result="Sequence Player (additive) 'NyxGlass_GrindRail_Jump_Start_WEP'"() {bLoopAnimation=false, sequence=/DryBox/Items/NyxGlass/Animation/Gauntlet/NyxGlass_GrindRail_Jump_Start_WEP})]
 }
 GrindLand {
  path=AnimGraph.AnimGraphNode_StateMachine_0.ODMLocomotion.AnimStateNode_16.GrindLand
  nodes [Output Animation Pose(Result="Sequence Player (additive) 'NyxGlass_GrindRail_Jump_Land_WEP'"() {bLoopAnimation=false, sequence=/DryBox/Items/NyxGlass/Animation/Gauntlet/NyxGlass_GrindRail_Jump_Land_WEP})]
 }
 GrindFallLoop {
  path=AnimGraph.AnimGraphNode_StateMachine_0.ODMLocomotion.AnimStateNode_17.GrindFallLoop
  nodes [Output Animation Pose(Result="Sequence Player (additive) 'NyxGlass_GrindRail_Jump_Loop_WEP'"() {sequence=/DryBox/Items/NyxGlass/Animation/Gauntlet/NyxGlass_GrindRail_Jump_Loop_WEP})]
 }
 Equip {
  path=AnimGraph.AnimGraphNode_StateMachine_0.ODMLocomotion.AnimStateNode_18.Equip
  nodes [Output Animation Pose(Result="Sequence Player (additive) 'NyxGlass_Equip_CMM_WEP'"() {bLoopAnimation=false, sequence=/DryBox/Items/NyxGlass/Animation/Gauntlet/NyxGlass_Equip_CMM_WEP})]
 }
 GrindSprintBoost {
  path=AnimGraph.AnimGraphNode_StateMachine_0.ODMLocomotion.AnimStateNode_19.GrindSprintBoost
  nodes [Output Animation Pose(Result="Sequence Player (additive) 'NyxGlass_Swim_N_NonTargeted_WEP'"() {bLoopAnimation=false, sequence=/DryBox/Items/NyxGlass/Animation/Gauntlet/NyxGlass_Swim_N_NonTargeted_WEP})]
 }
 Sliding {
  path=AnimGraph.AnimGraphNode_StateMachine_0.ODMLocomotion.AnimStateNode_2.Sliding
  nodes [
   n1 = Get KneeSlideSlopeAngle() -> (KneeSlideSlopeAngle:float)
   n2 = Get PawnSpeed2D()
   Output Animation Pose(Result=Apply Additive(
     Base=Blendspace Player 'NyxGlass_Sliding_BS'(X=@n1, Y=0) {blendSpace=/DryBox/Items/NyxGlass/Animation/Gauntlet/NyxGlass_Sliding_BS}
     Additive="Sequence Player (additive) 'NyxGlass_KneeSliding_Noise_WEP'"() {sequence=/DryBox/Items/NyxGlass/Animation/Gauntlet/NyxGlass_KneeSliding_Noise_WEP}
     Alpha="Clamp (Float)"(Value=@n2, Min=0, Max=1)
    ) {alphaBoolBlend {blendOption=Linear}})
  ]
 }
 IsSkyDive {
  path=AnimGraph.AnimGraphNode_StateMachine_0.ODMLocomotion.AnimStateNode_20.IsSkyDive
  nodes [
   n1 = Get Skydive Aim Yaw() -> (Skydive Aim Yaw:float)
   Output Animation Pose(Result=Blend Poses by bool(
     BlendPose_0=Blendspace Player 'NyxGlass_ZipLine_LowerBody_Leans_BS'(X=@n1, Y=0) {blendSpace=/DryBox/Items/NyxGlass/Animation/Gauntlet/NyxGlass_ZipLine_LowerBody_Leans_BS}
     BlendPose_1="Sequence Player (additive) 'NyxGlass_FreeFalling_WEP'"() {sequence=&NyxGlass_FreeFalling_WEP}
     BlendTime_0=0.3
     BlendTime_1=0.3
     bActiveValue=Get Is Parachute Open() -> (Is Parachute Open:bool)
    ) {blendType=Linear, transitionType=Inertialization})
  ]
 }
 Jump {
  path=AnimGraph.AnimGraphNode_StateMachine_0.ODMLocomotion.AnimStateNode_3.Jump
  nodes [Output Animation Pose(Result="Sequence Player (additive) 'NyxGlass_Jump_CMM_WEP'"() {bLoopAnimation=false, sequence=&NyxGlass_Jump_CMM_WEP})]
 }
 JumpLand {
  path=AnimGraph.AnimGraphNode_StateMachine_0.ODMLocomotion.AnimStateNode_4.JumpLand
  nodes [Output Animation Pose(Result="Sequence Player (additive) 'NyxGlass_Land_CMM_WEP'"() {bLoopAnimation=false, sequence=/DryBox/Items/NyxGlass/Animation/Gauntlet/NyxGlass_Land_CMM_WEP})]
 }
 Swimming {
  path=AnimGraph.AnimGraphNode_StateMachine_0.ODMLocomotion.AnimStateNode_5.Swimming
  nodes [
   n1 = "Sequence Player (additive) 'NyxGlass_Swim_Sprint_Slow_WEP'"() {sequence=/DryBox/Items/NyxGlass/Animation/Gauntlet/NyxGlass_Swim_Sprint_Slow_WEP}
   n2 = "Sequence Player (additive) 'NyxGlass_Swim_Sprint_WEP'"() {sequence=/DryBox/Items/NyxGlass/Animation/Gauntlet/NyxGlass_Swim_Sprint_WEP}
   n3 = Get SwimTopSpeed() -> (SwimTopSpeed:bool)
   n4 = Get PawnSpeed2D()
   n5 = Get PawnMovementDirectionAzimuth()
   Output Animation Pose(Result=Blend Poses by bool(
     BlendPose_0=Blend Poses by bool(BlendPose_0=@n2, BlendPose_1=@n1, BlendTime_0=0.25, BlendTime_1=0.25, bActiveValue=@n3) {blendType=Linear}
     BlendPose_1=Blendspace Player 'NyxGlass_Swim_BS'(X=@n5, Y=@n4) {blendSpace=/DryBox/Items/NyxGlass/Animation/Gauntlet/NyxGlass_Swim_BS}
     BlendTime_0=0.25
     BlendTime_1=0.25
     bActiveValue=Get bIsSwimSprinting() -> (bIsSwimSprinting:bool)
    ) {blendType=Linear})
  ]
 }
 Falling {
  path=AnimGraph.AnimGraphNode_StateMachine_0.ODMLocomotion.AnimStateNode_6.Falling
  nodes [Output Animation Pose(Result="Sequence Player (additive) 'NyxGlass_JumpFall_CMM_WEP'"() {sequence=/DryBox/Items/NyxGlass/Animation/Gauntlet/NyxGlass_JumpFall_CMM_WEP})]
 }
 Crouch {
  path=AnimGraph.AnimGraphNode_StateMachine_0.ODMLocomotion.AnimStateNode_7.Crouch
  nodes [
   n1 = Get PawnMovementDirectionAzimuth()
   n2 = Get PawnSpeed2D()
   Output Animation Pose(Result=Blend Poses by bool(
     BlendPose_0="Sequence Player (additive) 'NyxGlass_Crouch_Sprint_WEP'"() {sequence=/DryBox/Items/NyxGlass/Animation/Gauntlet/NyxGlass_Crouch_Sprint_WEP}
     BlendPose_1=Blendspace Player 'NyxGlass_Crouch_BS'(X=@n1, Y=@n2) {blendSpace=/DryBox/Items/NyxGlass/Animation/Gauntlet/NyxGlass_Crouch_BS}
     BlendTime_0=0.25
     BlendTime_1=0.25
     bActiveValue="float > float"(A=@n2, B=400)
    ) {blendType=Linear})
   Get bIsCrouchSprinting()
  ]
 }
 ZiplineEntry {
  path=AnimGraph.AnimGraphNode_StateMachine_0.ODMLocomotion.AnimStateNode_8.ZiplineEntry
  nodes [Output Animation Pose(Result="Sequence Player (additive) 'NyxGlass_Zipline_Entry_WEP'"() {bLoopAnimation=false, sequence=/DryBox/Items/NyxGlass/Animation/Gauntlet/NyxGlass_Zipline_Entry_WEP})]
 }
 Ziplining {
  path=AnimGraph.AnimGraphNode_StateMachine_0.ODMLocomotion.AnimStateNode_9.Ziplining
  nodes [
   n1 = Get!() -> (PawnVelocity_X:double, PawnVelocity_Y:double, PawnVelocity_Z:double)
   Blend Poses by bool(
    BlendPose_0=Blend Poses by bool(
     BlendPose_0="Sequence Player (additive) 'NyxGlass_Ascender_Up_Loop_WEP'"() {bLoopAnimation=false, sequence=/DryBox/Items/NyxGlass/Animation/Gauntlet/NyxGlass_Ascender_Up_Loop_WEP}
     BlendPose_1="Sequence Player (additive) 'NyxGlass_Ascender_Down_Loop_WEP'"() {bLoopAnimation=false, sequence=/DryBox/Items/NyxGlass/Animation/Gauntlet/NyxGlass_Ascender_Down_Loop_WEP}
     bActiveValue="float > float"(A=@n1.PawnVelocity_Z)
    ) {blendType=Linear}
    bActiveValue="Equal (Float)"(A=@n1.PawnVelocity_Y)
   ) {blendType=Linear}
   n2 = Get ZipLeanYaw() -> (ZipLeanYaw:float)
   Output Animation Pose(Result=Apply Additive(
     Base="Sequence Player (additive) 'NyxGlass_Zipline_Apex_Loop_WEP'"() {playRate=1.5, sequence=/DryBox/Items/NyxGlass/Animation/Gauntlet/NyxGlass_Zipline_Apex_Loop_WEP}
     Additive=Blendspace Player 'NyxGlass_ZipLine_LowerBody_Leans_BS'(X=@n2, Y=0) {blendSpace=&NyxGlass_ZipLine_LowerBody_Leans_BS}
    ) {alphaBoolBlend {blendOption=Linear}})
  ]
 }
 Transition {path=AnimGraph.AnimGraphNode_StateMachine_0.ODMLocomotion.AnimStateTransitionNode_0.Transition, nodes [Result(bCanEnterTransition=true)]}
 Transition {
  path=AnimGraph.AnimGraphNode_StateMachine_0.ODMLocomotion.AnimStateTransitionNode_1.Transition
  nodes [Result(bCanEnterTransition=Get Is Nyx Landing() -> (Is Nyx Landing:bool))]
 }
 Transition {
  path=AnimGraph.AnimGraphNode_StateMachine_0.ODMLocomotion.AnimStateTransitionNode_10.Transition
  nodes [
   n1 = Get Is Skydiving() -> (Is Skydiving:bool)
   Result(bCanEnterTransition=AND Boolean(A=Get Is Nyx Jumping() -> (Is Nyx Jumping:bool), B=NOT Boolean(A=@n1)))
  ]
 }
 Transition {
  path=AnimGraph.AnimGraphNode_StateMachine_0.ODMLocomotion.AnimStateTransitionNode_11.Transition
  nodes [Result(bCanEnterTransition=Get bIsSurfaceSwimming())]
 }
 Transition {path=AnimGraph.AnimGraphNode_StateMachine_0.ODMLocomotion.AnimStateTransitionNode_12.Transition, nodes [Result(bCanEnterTransition=true)]}
 Transition {
  path=AnimGraph.AnimGraphNode_StateMachine_0.ODMLocomotion.AnimStateTransitionNode_13.Transition
  nodes [Result(bCanEnterTransition=Get IsNyxCrouching() -> (IsNyxCrouching:bool))]
 }
 Transition {path=AnimGraph.AnimGraphNode_StateMachine_0.ODMLocomotion.AnimStateTransitionNode_14.Transition, nodes [Result(bCanEnterTransition=true)]}
 Transition {
  path=AnimGraph.AnimGraphNode_StateMachine_0.ODMLocomotion.AnimStateTransitionNode_15.Transition
  nodes [Result(bCanEnterTransition=NOT Boolean(A=Get IsZiplining() -> (IsZiplining:bool)))]
 }
 Transition {
  path=AnimGraph.AnimGraphNode_StateMachine_0.ODMLocomotion.AnimStateTransitionNode_16.Transition
  nodes [
   Result(bCanEnterTransition=Get BlendOut() -> (BlendOut:bool))
   "float > float"(A=GetCurveValue(CurveName=BlendOut), B=0.5)
  ]
 }
 Transition {
  path=AnimGraph.AnimGraphNode_StateMachine_0.ODMLocomotion.AnimStateTransitionNode_17.Transition
  nodes [Result(bCanEnterTransition=Get bIsSprinting())]
 }
 Transition {
  path=AnimGraph.AnimGraphNode_StateMachine_0.ODMLocomotion.AnimStateTransitionNode_18.Transition
  nodes [Result(bCanEnterTransition=Get ZiplineEnter() -> (ZiplineEnter:bool))]
 }
 Transition {
  path=AnimGraph.AnimGraphNode_StateMachine_0.ODMLocomotion.AnimStateTransitionNode_19.Transition
  nodes [Result(bCanEnterTransition=Get bIsRiding() -> (bIsRiding:bool))]
 }
 Transition {path=AnimGraph.AnimGraphNode_StateMachine_0.ODMLocomotion.AnimStateTransitionNode_2.Transition, nodes [Result(bCanEnterTransition=true)]}
 Transition {path=AnimGraph.AnimGraphNode_StateMachine_0.ODMLocomotion.AnimStateTransitionNode_20.Transition, nodes [Result(bCanEnterTransition=true)]}
 Transition {
  path=AnimGraph.AnimGraphNode_StateMachine_0.ODMLocomotion.AnimStateTransitionNode_21.Transition
  nodes [Result(bCanEnterTransition=Get ShouldPlayZiplinePivot() -> (ShouldPlayZiplinePivot:bool))]
 }
 Transition {
  path=AnimGraph.AnimGraphNode_StateMachine_0.ODMLocomotion.AnimStateTransitionNode_22.Transition
  nodes [
   n1 = Get IsZiplining()
   Result(bCanEnterTransition=@n1)
   AND Boolean(A=@n1, B=Get IsJumpingState() -> (IsJumpingState:bool))
  ]
 }
 Transition {
  path=AnimGraph.AnimGraphNode_StateMachine_0.ODMLocomotion.AnimStateTransitionNode_23.Transition
  nodes [Result(bCanEnterTransition=NOT Boolean(A=Get IsZiplining()))]
 }
 Transition {path=AnimGraph.AnimGraphNode_StateMachine_0.ODMLocomotion.AnimStateTransitionNode_24.Transition, nodes [Result()]}
 Transition {
  path=AnimGraph.AnimGraphNode_StateMachine_0.ODMLocomotion.AnimStateTransitionNode_25.Transition
  nodes [Result(bCanEnterTransition=Get Free Falling() -> (Free Falling:bool))]
 }
 Transition {
  path=AnimGraph.AnimGraphNode_StateMachine_0.ODMLocomotion.AnimStateTransitionNode_26.Transition
  nodes [Result(bCanEnterTransition=Get Is Dive Jumping())]
 }
 Transition {
  path=AnimGraph.AnimGraphNode_StateMachine_0.ODMLocomotion.AnimStateTransitionNode_27.Transition
  nodes [Result(bCanEnterTransition=NOT Boolean(A=Get bIsWaterJump() -> (bIsWaterJump:bool)))]
 }
 Transition {
  path=AnimGraph.AnimGraphNode_StateMachine_0.ODMLocomotion.AnimStateTransitionNode_28.Transition
  nodes [Result(bCanEnterTransition=Get Is Nyx Landing())]
 }
 Transition {
  path=AnimGraph.AnimGraphNode_StateMachine_0.ODMLocomotion.AnimStateTransitionNode_29.Transition
  nodes [Result(bCanEnterTransition=Get bIsOnGrindRail())]
 }
 Transition {path=AnimGraph.AnimGraphNode_StateMachine_0.ODMLocomotion.AnimStateTransitionNode_3.Transition, nodes [Result(bCanEnterTransition=true)]}
 Transition {path=AnimGraph.AnimGraphNode_StateMachine_0.ODMLocomotion.AnimStateTransitionNode_30.Transition, nodes [Result(bCanEnterTransition=true)]}
 Transition {
  path=AnimGraph.AnimGraphNode_StateMachine_0.ODMLocomotion.AnimStateTransitionNode_31.Transition
  nodes [Result(bCanEnterTransition=NOT Boolean(A=Get bIsOnGrindRail()))]
 }
 Transition {
  path=AnimGraph.AnimGraphNode_StateMachine_0.ODMLocomotion.AnimStateTransitionNode_32.Transition
  nodes [Result(bCanEnterTransition=Get IsEquip() -> (IsEquip:bool))]
 }
 Transition {
  path=AnimGraph.AnimGraphNode_StateMachine_0.ODMLocomotion.AnimStateTransitionNode_33.Transition
  nodes [Result(bCanEnterTransition=Get bIsOnGrindRail())]
 }
 Transition {
  path=AnimGraph.AnimGraphNode_StateMachine_0.ODMLocomotion.AnimStateTransitionNode_34.Transition
  nodes [
   Result()
   Get BlendOut()
  ]
 }
 Transition {path=AnimGraph.AnimGraphNode_StateMachine_0.ODMLocomotion.AnimStateTransitionNode_35.Transition, nodes [Result()]}
 Transition {
  path=AnimGraph.AnimGraphNode_StateMachine_0.ODMLocomotion.AnimStateTransitionNode_36.Transition
  nodes [Result(bCanEnterTransition=Get IsSprintBoosting() -> (IsSprintBoosting:bool))]
 }
 Transition {
  path=AnimGraph.AnimGraphNode_StateMachine_0.ODMLocomotion.AnimStateTransitionNode_37.Transition
  nodes [Result(bCanEnterTransition=NOT Boolean(A=Get IsSprintBoosting()))]
 }
 Transition {
  path=AnimGraph.AnimGraphNode_StateMachine_0.ODMLocomotion.AnimStateTransitionNode_38.Transition
  nodes [Result(bCanEnterTransition=Get IsMoving() -> (IsMoving:bool))]
 }
 Transition {
  path=AnimGraph.AnimGraphNode_StateMachine_0.ODMLocomotion.AnimStateTransitionNode_39.Transition
  nodes [Result(bCanEnterTransition=Get Is Nyx Landing())]
 }
 Transition {
  path=AnimGraph.AnimGraphNode_StateMachine_0.ODMLocomotion.AnimStateTransitionNode_4.Transition
  nodes [Result(bCanEnterTransition=Get Is Nyx Sliding() -> (Is Nyx Sliding:bool))]
 }
 Transition {path=AnimGraph.AnimGraphNode_StateMachine_0.ODMLocomotion.AnimStateTransitionNode_40.Transition, nodes [Result()]}
 Transition {
  path=AnimGraph.AnimGraphNode_StateMachine_0.ODMLocomotion.AnimStateTransitionNode_41.Transition
  nodes [Result(bCanEnterTransition=Get bIsOnGrindRail())]
 }
 Transition {
  path=AnimGraph.AnimGraphNode_StateMachine_0.ODMLocomotion.AnimStateTransitionNode_42.Transition
  nodes [Result(bCanEnterTransition=Get bIsOnGrindRail())]
 }
 Transition {
  path=AnimGraph.AnimGraphNode_StateMachine_0.ODMLocomotion.AnimStateTransitionNode_43.Transition
  nodes [Result(bCanEnterTransition=Get bIsOnGrindRail())]
 }
 Transition {
  path=AnimGraph.AnimGraphNode_StateMachine_0.ODMLocomotion.AnimStateTransitionNode_44.Transition
  nodes [Result(bCanEnterTransition=Get Is Nyx Landing())]
 }
 Transition {
  path=AnimGraph.AnimGraphNode_StateMachine_0.ODMLocomotion.AnimStateTransitionNode_45.Transition
  nodes [Result(bCanEnterTransition=Get Is Skydiving())]
 }
 Transition {path=AnimGraph.AnimGraphNode_StateMachine_0.ODMLocomotion.AnimStateTransitionNode_46.Transition, nodes [Result(bCanEnterTransition=true)]}
 Transition {
  path=AnimGraph.AnimGraphNode_StateMachine_0.ODMLocomotion.AnimStateTransitionNode_47.Transition
  nodes [Result(bCanEnterTransition=Get Is Skydiving())]
 }
 Transition {path=AnimGraph.AnimGraphNode_StateMachine_0.ODMLocomotion.AnimStateTransitionNode_5.Transition, nodes [Result(bCanEnterTransition=true)]}
 Transition {
  path=AnimGraph.AnimGraphNode_StateMachine_0.ODMLocomotion.AnimStateTransitionNode_6.Transition
  nodes [Result(bCanEnterTransition=Get Is Nyx Falling() -> (Is Nyx Falling:bool))]
 }
 Transition {
  path=AnimGraph.AnimGraphNode_StateMachine_0.ODMLocomotion.AnimStateTransitionNode_7.Transition
  nodes [Result(bCanEnterTransition=Get Is Nyx Landing())]
 }
 Transition {path=AnimGraph.AnimGraphNode_StateMachine_0.ODMLocomotion.AnimStateTransitionNode_8.Transition, nodes [Result()]}
 Transition {
  path=AnimGraph.AnimGraphNode_StateMachine_0.ODMLocomotion.AnimStateTransitionNode_9.Transition
  nodes [Result(bCanEnterTransition=Get IsLocomotion() -> (IsLocomotion:bool))]
 }
 EventGraph {nodes [
   Event BlueprintUpdateAnimation()
   > n1 = Sequence!(execute) -> (then_0, then_1, then_2:Exec, then_3:Exec)
   > n2 = Cast To FortPlayerPawn(Object=TryGetPawnOwner())
   > Set As Fort Player Pawn(As Fort Player Pawn:Fort Player Pawn Object Reference=@n2) -> (Output_Get:Fort Player Pawn Object Reference)
   n3 = GetAnimationInterface!(self:Fort Player Pawn Object Reference=@n2) -> (ReturnValue:Fort Player Animation Interface Object Reference)
   n4 = Break Fort Anim Input Zipline!(FortAnimInput_Zipline:"Fort Anim Input Zipline Structure (by ref)"=GetZiplineInput!(self:Fort Player Animation Interface Object Reference=@n3) -> (ReturnValue:"Fort Anim Input Zipline Structure (by ref)")) -> (
    bIsZiplining:bool=~
    bIsOnAscender:bool=~
    bShouldPlayPivotTransition:bool=~
    LeanYaw:float=~
    PivotBlendDelayRemaining:float=~
    PivotBlendDelay:float=~
    PivotCardinalDirection:EFortCardinalDirection Enum
    WorldVelocityLastTick:Vector=~
   ) {is=title K2Node_BreakStruct}
   > Set ShouldPlayZiplinePivot(ShouldPlayZiplinePivot:bool=@n4.bShouldPlayPivotTransition) -> (Output_Get:bool)
   > Set ZipLeanYaw(ZipLeanYaw:float=@n4.LeanYaw) -> (Output_Get:float)
   > Set Pivot Cardinal Direction(Pivot Cardinal Direction:EFortCardinalDirection Enum=@n4.PivotCardinalDirection) -> (Output_Get:EFortCardinalDirection Enum=North)
   n5 = Break Fort Anim Input Skydiving!(FortAnimInput_Skydiving:"Fort Anim Input Skydiving Structure (by ref)"=GetSkydivingInput!(self:Fort Player Animation Interface Object Reference=@n3) -> (ReturnValue:"Fort Anim Input Skydiving Structure (by ref)")) -> (
    ParaGilderRootOffset:Vector=~
    ParaGliderRootRotationOverride:Rotator=~
    bUseParaGlideRootModifier:bool=~
    bIsSkydivingFromLaunchPad:bool=~
    bIsSkydivingFromBus:bool=~
    bIsInVortex:bool=~
    bIsDiving:bool=~
    bIsParachuteOpen:bool=~
    bIsSkydiving:bool=~
    bIsParachuteLeaning:bool=~
    bIsSkydiveLeaning:bool=~
    bIsLeaning:bool=~
    bIsSkydiveDiveMode:bool=~
    bPlaySkydiveDrift:bool=~
    bSkydiveDriftDelayActive:bool=~
    bSkydiveDriftAnimAllowed:bool=~
    bIsGliderRight:bool=~
    bIsGliderCenter:bool=~
    bIsGliderLeft:bool=~
    bIsGliderForward:bool=~
    bIsGliderBack:bool=~
    bHasDeployedAndIsStandingUpright:bool=~
    LocalAccelForward:float=~
    LocalAccelRight:float=~
    LocalVelocityRight:float=~
    SkydiveAimPitch:float=~
    SkydiveAimYaw:float=~
    DeployChuteAnimRate:float=~
    SkydiveDriftAnimRate:float=~
    SkydiveDriftAnimRateCurrent:float=~
    SkydiveFidgetAnimRate:float=~
    SkydiveFidgetAnimRateCurrent:float=~
    SkydiveAdditiveAlpha:float=~
    SkydiveDriftDelay:float=~
    SkydiveDriftAnim:int
    SkydiveDriftAnimMax:int
    LaunchpadAnim:int
    LocalAccelDir:ESkydivingDirection Enum
    DirectionLast:ESkydivingDirection Enum
    PoseAdditive:Anim Sequence Object Reference
   ) {is=title K2Node_BreakStruct}
   > Set Is Parachute Open(Is Parachute Open:bool=@n5.bIsParachuteOpen) -> (Output_Get:bool)
   > Set Is Skydiving(Is Skydiving:bool=@n5.bIsSkydiving) -> (Output_Get:bool)
   > Set Skydive Aim Yaw(Skydive Aim Yaw:float=@n5.SkydiveAimYaw) -> (Output_Get:float)
   > n6 = Sequence!(execute) -> (then_0, then_1, then_2, then_3, then_4:Exec, then_5:Exec, then_6:Exec)
   n7 = Get As Fort Player Pawn() -> (As Fort Player Pawn:Fort Player Pawn Object Reference)
   > n8 = Cast To SlidingLayerAnimBP!(
    execute
    Object=GetLinkedAnimLayerInstanceByClass!(
     self:Skeletal Mesh Component Object Reference=Get Mesh(self=@n7)
     InClass:Anim Instance Class Reference=/Game/Characters/Player/Male/Male_Avg_Base/LayerAnimBP/SlidingLayerAnimBP.SlidingLayerAnimBP_C
    ) -> (ReturnValue:Anim Instance Object Reference)
   ) -> (then, CastFailed, AsSliding Layer Anim BP:Sliding Layer Anim BP Object Reference) {is=title K2Node_DynamicCast}
   > Set KneeSlideSlopeAngle(KneeSlideSlopeAngle:float=Get KneeSlideSlopeAngle!(self:Fort Sliding Layer Anim Instance Object Reference=@n8) -> (KneeSlideSlopeAngle)) -> (Output_Get:float)
   n9 = Get As Fort Player Pawn()
   n10 = Cast To SwimmingAnimBP!(
    execute=@n6.then_1
    Object=GetLinkedAnimLayerInstanceByClass(self=Get Mesh(self=@n9), InClass=/Game/Characters/Player/Male/Male_Avg_Base/LayerAnimBP/SwimmingAnimBP.SwimmingAnimBP_C)
   ) -> (then, CastFailed, AsSwimming Anim BP:Swimming Anim BP Object Reference) {is=title K2Node_DynamicCast}
   > Set Swimming Dive Pitch(Swimming Dive Pitch:float=Get SwimmingDivePitch!(self:Fort Swimming Layer Anim Instance Object Reference=@n10) -> (SwimmingDivePitch:float)) -> (Output_Get:float)
   > Set Is Dive Jumping(Is Dive Jumping:bool=AND Boolean(A=Get bIsDiveJumping!(self:Fort Player Anim Instance Proxy Object Reference=@n10) -> (bIsDiveJumping:bool), B=Get bIsWaterJump())) -> (Output_Get:bool)
   n11 = Get PawnSpeed2D()
   > Set SwimTopSpeed(SwimTopSpeed:bool=AND Boolean(A="float > float"(A=@n11, B=450), B=Get bIsSwimSprinting())) -> (Output_Get:bool)
   n12 = Cast To B_Weapon_GrappleParent!(execute=@n6.then_2, Object=Get CurrentWeapon() -> (CurrentWeapon:Fort Weapon Object Reference)) -> (then, CastFailed, AsB Weapon Grapple Parent:B Weapon Grapple Parent Object Reference) {is=title K2Node_DynamicCast}
   > Set Free Falling(Free Falling:bool=Get FreeFalling!(self:B Weapon Grapple Parent Object Reference=@n12) -> (FreeFalling:bool)) -> (Output_Get:bool)
   n13 = IsZiplining!(execute=@n6.then_3, self:Fort Player Pawn Object Reference=Get As Fort Player Pawn()) -> (then, ReturnValue:bool)
   > Set IsZiplining(IsZiplining:bool=@n13) -> (Output_Get:bool)
   n14 = Get IsJumpingState()
   > Set ZiplineEnter(ZiplineEnter:bool=AND Boolean(A=@n13, B=NOT Boolean(A=@n14))) -> (Output_Get:bool)
   > Set Is Nyx Sliding(Is Nyx Sliding:bool=Get bIsSliding!(self:Fort Pawn Object Reference=Get As Fort Player Pawn()) -> (bIsSliding)) -> (Output_Get:bool)
   n15 = Get As Fort Player Pawn()
   n16 = Get Free Falling()
   > Set IsNyxGlass_Swing(IsNyxGlass_Swing:bool=AND Boolean(A=GetIsSwinging!(self:Fort Pawn Object Reference=@n15) -> (ReturnValue:bool), B=NOT Boolean(A=@n16))) -> (Output_Get:bool)
   n17 = Get bIsJumping()
   n18 = Get bIsOnGrindRail()
   n19 = Get bIsFalling()
   > Set Is Nyx Landing(Is Nyx Landing:bool=AND Boolean!(A=NOT Boolean(A=@n17), B=NOT Boolean(A=@n19), C:bool=NOT Boolean(A=@n18)) -> (ReturnValue)) -> (Output_Get:bool)
   n20 = Get Free Falling()
   > Set Is Nyx Falling(Is Nyx Falling:bool=AND Boolean!(A=Get bIsFalling(), B=NOT Boolean(A=@n20)) -> (ReturnValue)) -> (Output_Get:bool)
   > Set Is Nyx Jumping(Is Nyx Jumping:bool=OR Boolean!(A=Get bIsJumping(), B=Get Free Falling(), C:bool=Get bIsFalling()) -> (ReturnValue)) -> (Output_Get:bool)
   n21 = Get bIsRiding()
   > Set IsNyxCrouching(IsNyxCrouching:bool=AND Boolean(A=Get bIsCrouching(), B=NOT Boolean(A=@n21))) -> (Output_Get:bool)
   n22 = Get bIsSurfaceSwimming()
   n23 = Get bIsFalling()
   n24 = Get IsZiplining()
   n25 = Get bIsJumping()
   n26 = Get Is Nyx Sliding()
   n27 = Get bIsCrouching()
   n28 = Get bIsOnGrindRail()
   n29 = Get PawnSpeed2D()
   Set IsLocomotion(
    execute=@n6.then_4
    IsLocomotion:bool=AND Boolean!(
     A="float > float"(A=@n29)
     B=NOT Boolean(A=@n26)
     C:bool=NOT Boolean(A=@n23)
     D:bool=NOT Boolean(A=@n25)
     E:bool=NOT Boolean(A=@n22)
     F:bool=NOT Boolean(A=@n24)
     G:bool=NOT Boolean(A=@n27)
     H:bool=NOT Boolean(A=@n28)
    ) -> (ReturnValue)
   ) -> (Output_Get:bool)
   > Set DynamicAlpha(DynamicAlpha:float=GetCurveValue(CurveName=DynamicAlpha)) -> (Output_Get:float)
   > Set IsJumpingState(IsJumpingState:bool="float >= float"(A=GetCurveValue(CurveName=Jump), B=0.5)) -> (Output_Get:bool)
   > Set BlendOut(BlendOut:bool="float > float"(A=GetCurveValue(CurveName=BlendOut), B=0.5)) -> (Output_Get:bool)
   n30 = BreakTransform(InTransform=GetDeltaTransformFromRefPose(self=GetOwnerSkeletalMeshComponent(), BoneName=spine_05, BaseName=pelvis))
   Set spine_05_twist(execute=@n1.then_1, spine_05_twist:Rotator=MakeRotator(Pitch=BreakRotator(InRot=@n30.Rotation))) -> (Output_Get:Rotator)
   Branch(execute=@n1.then_2, Condition="Equal (Float)"(A=GetCurveValue(CurveName=BlendOut), B=1))
   > Set IsEquip(IsEquip:bool=false) -> (Output_Get:bool)
   Set IsMoving(execute=@n1.then_3, IsMoving:bool="float > float"(A=Get PawnSpeed2D())) -> (Output_Get:bool)
   Event OnCharacterCustomizationLoaded()
   > n31 = Sequence!(execute) -> (then_0, then_1, then_2)
   > n32 = Cast To FortPlayerPawn(Object=TryGetPawnOwner())
   n33 = GetDisplayName(Object=GetSkeletalMeshForPartType(self=@n32))
   n34 = FindSubstring(SearchIn=@n33, Substring=F_MED)
   > n35 = Branch(Condition="integer >= integer"(A=@n34))
   > Set player_skin_name(player_skin_name:String=GetSubstring!(SourceString:String=@n33, StartIndex:int=@n34, Length:int=99) -> (ReturnValue:String)) -> (Output_Get:String)
   n36 = FindSubstring(SearchIn=@n33, Substring=M_MED)
   n37 = Branch(execute=@n35.else, Condition="integer >= integer"(A=@n36))
   > Set player_skin_name(player_skin_name=GetSubstring(SourceString=@n33, StartIndex=@n36, Length=99))
   n38 = FindSubstring(SearchIn=@n33, Substring=F_LRG)
   n39 = Branch(execute=@n37.else, Condition="integer >= integer"(A=@n38))
   > Set player_skin_name(player_skin_name=GetSubstring(SourceString=@n33, StartIndex=@n38, Length=99))
   n40 = FindSubstring(SearchIn=@n33, Substring=M_LRG)
   Branch(execute=@n39.else, Condition="integer >= integer"(A=@n40))
   > Set player_skin_name(player_skin_name=GetSubstring(SourceString=@n33, StartIndex=@n40, Length=99))
   n41 = Get player_skin_name() -> (player_skin_name:String)
   n42 = Find(TargetMap:Map of Strings to Rotators=Get character_edge_cases_belt_r() -> (character_edge_cases_belt_r:Map of Strings to Rotators), Key=@n41) -> (Value:Rotator)
   n43 = Find(
    TargetMap:Map of Strings to Vectors=Get character_edge_cases_translate() -> (character_edge_cases_translate:Map of Strings to Vectors)
    Key=@n41
   ) -> (Value:Vector)
   n44 = Find(TargetMap:Map of Strings to Rotators=Get character_edge_cases_belt_l() -> (character_edge_cases_belt_l:Map of Strings to Rotators), Key=@n41) -> (Value:Rotator)
   n45 = Branch(execute=@n31.then_1, Condition=OR Boolean(A=@n42.ReturnValue, B=@n44.ReturnValue, C=@n43.ReturnValue))
   > Set belt_adjustment_r(belt_adjustment_r:Rotator=@n42) -> (Output_Get:Rotator)
   > Set belt_adjustment_l(belt_adjustment_l:Rotator=@n44) -> (Output_Get:Rotator)
   > Set nyx_position_adjustment(nyx_position_adjustment:Vector=@n43) -> (Output_Get:Vector)
   n46 = Get player_skin_name()
   n47 = Branch(execute=@n45.else, Condition=MatchesWildcard(SourceString=@n46, Wildcard=*F_MED*))
   > Set belt_adjustment_r(belt_adjustment_r="0,10,0")
   > Set belt_adjustment_l(belt_adjustment_l="0,10,0")
   > Set nyx_position_adjustment(nyx_position_adjustment="0,5.5,0")
   n48 = Branch(execute=@n47.else, Condition=MatchesWildcard(SourceString=@n46, Wildcard=*M_MED*))
   > Set belt_adjustment_r(belt_adjustment_r="0, 0, 0")
   > Set belt_adjustment_l(belt_adjustment_l="0, 0, 0")
   > Set nyx_position_adjustment(nyx_position_adjustment="0, 0, 0")
   n49 = Branch(execute=@n48.else, Condition=CharacterItemDefinitionHasAnyMetaTag(Tags=Make Array("[0]"=(TagName="Cosmetics.MetaTags.Special.XLarge"))))
   > Set belt_adjustment_r(belt_adjustment_r="0,-15,0")
   > Set belt_adjustment_l(belt_adjustment_l="0,-15,0")
   > Set nyx_position_adjustment(nyx_position_adjustment="0,-5,0")
   n50 = Get player_skin_name()
   n51 = Branch(execute=@n49.else, Condition=MatchesWildcard(SourceString=@n50, Wildcard=*F_LRG*))
   > Set belt_adjustment_r(belt_adjustment_r="0,-7.5,0")
   > Set belt_adjustment_l(belt_adjustment_l="0,-7.5,0")
   > Set nyx_position_adjustment(nyx_position_adjustment="0, 0, 0")
   Branch(execute=@n51.else, Condition=MatchesWildcard(SourceString=@n50, Wildcard=*M_LRG*))
   > Set belt_adjustment_r(belt_adjustment_r="0,-5,0")
   > Set belt_adjustment_l(belt_adjustment_l="0,-5,0")
   > Set nyx_position_adjustment(nyx_position_adjustment="0,-1,0")
   Event BlueprintInitializeAnimation()
   > Set IsEquip(IsEquip=true)
  ]}
]
variables [
 As Fort Player Pawn {type=FortPlayerPawn}
 belt_adjustment_l {type=Rotator}
 belt_adjustment_r {type=Rotator}
 BlendOut {type=boolean}
 character_edge_cases_belt_l {
  type="map<Rotator>"
  default {
   Armin=<rot 0 0 0>
   F_MED_SplitDiamond=<rot 0 5 0>
   Levi=<rot 0 0 0>
   M_MED_ClearRadius=<rot 0 5 0>
   PlayerPawn_Athena0.CharacterPartSkelMesh_Body_1 M_LRG_Comet=<rot 0 0 0>
  }
 }
 character_edge_cases_belt_r {
  type="map<Rotator>"
  default {
   Armin=<rot 0 0 0>
   F_MED_SplitDiamond=<rot 0 5 0>
   Levi=<rot 0 0 0>
   M_MED_ClearRadius=<rot 0 5 0>
   PlayerPawn_Athena0.CharacterPartSkelMesh_Body_1 M_LRG_Comet=<rot 0 0 0>
  }
 }
 character_edge_cases_translate {
  type="map<Vector>"
  default {
   Armin=<1 1 1>
   F_MED_SplitDiamond=<0 5.5 0>
   Levi=<1 1 1>
   M_MED_ClearRadius=<0 3 0>
   PlayerPawn_Athena0.CharacterPartSkelMesh_Body_1 M_LRG_Comet=<1.2 1.2 1.2>
  }
 }
 DynamicAlpha {type=number}
 Fort Anim Input Skydiving {type=FortAnimInput_Skydiving, default {skydiveAimPitchInterpSpeed=5}}
 Free Falling {type=boolean}
 Grind Lean Direction {type=number}
 Grind Speed {type=number}
 Is Dive Jumping {type=boolean}
 Is Lean Bwd {type=boolean}
 Is Nyx Falling {type=boolean}
 Is Nyx Jumping {type=boolean}
 Is Nyx Landing {type=boolean}
 Is Nyx Sliding {type=boolean}
 Is Parachute Open {type=boolean}
 Is Skydiving {type=boolean}
 IsEquip {type=boolean}
 IsJumpingState {type=boolean}
 IsLocomotion {type=boolean}
 IsMoving {type=boolean}
 IsNyxCrouching {type=boolean}
 IsNyxGlass_Swing {type=boolean}
 IsSprintBoosting {type=boolean}
 IsZiplining {type=boolean}
 KneeSlideSlopeAngle {type=number}
 NoisePlayRate {type=number}
 nyx_position_adjustment {type=Vector, default=<1 1 1>}
 Pivot Cardinal Direction {type=EFortCardinalDirection, default=North}
 player_skin_name {type=string, default=""}
 ShouldPlayZiplinePivot {type=boolean}
 Skydive Aim Yaw {type=number}
 spine_05_twist {type=Rotator}
 Swimming Dive Pitch {type=number}
 SwimTopSpeed {type=boolean}
 ZipLeanYaw {type=number}
 ZiplineEnter {type=boolean}
 ZipLocalVelocityYawAngle {type=number}
]
